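; Anim notify state for ButterCake that pauses AI behavior over a window of an animation.
; On begin, it gets the mesh owner's Ability System Component and applies the GE_ButterCake_PauseAIBehavior effect to it.
; On end, it removes active effects tagged Event.ButterCake.PauseAIBehavior from that component.
; Both events do nothing to the effect state if the owner or component is invalid.
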(fn Received_NotifyBegin
  (params (MeshComp :type Skeletal%20Mesh%20Component%20Object%20Reference) (Animation :type Anim%20Sequence%20Base%20Object%20Reference) (TotalDuration :type Float%20%28single-precision%29) (EventReference :type Anim%20Notify%20Event%20Reference%20Structure%20%28by%20ref%29))
  (bind _returnvalue (Components|GetOwner MeshComp))
  (bind _returnvalue_2 (Ability|GetAbilitySystemComponent _returnvalue))
  (Utilities|IsValid _returnvalue
    (:Is%20Valid
      (Utilities|IsValid _returnvalue_2
        (:Is%20Valid
          (GameplayEffects|ApplyGameplayEffectToSelf :self _returnvalue_2 :GameplayEffectClass "/ButterCake/Abilities/GE_ButterCake_PauseAIBehavior.GE_ButterCake_PauseAIBehavior_C" :Level 0.0)
          (return false))
        (:Is%20Not%20Valid
          _)))
    (:Is%20Not%20Valid
      _))
  (return false))

(fn Received_NotifyEnd
  (params (MeshComp :type Skeletal%20Mesh%20Component%20Object%20Reference) (Animation :type Anim%20Sequence%20Base%20Object%20Reference) (EventReference :type Anim%20Notify%20Event%20Reference%20Structure%20%28by%20ref%29))
  (bind _returnvalue (Components|GetOwner MeshComp))
  (bind _returnvalue_2 (Ability|GetAbilitySystemComponent _returnvalue))
  (Utilities|IsValid _returnvalue
    (:Is%20Valid
      (Utilities|IsValid _returnvalue_2
        (:Is%20Valid
          (GameplayEffects|RemoveActiveEffectswithTags _returnvalue_2 "(GameplayTags=((TagName=\"Event.ButterCake.PauseAIBehavior\")))")
          (return false))
        (:Is%20Not%20Valid
          _)))
    (:Is%20Not%20Valid
      _))
  (return false))
