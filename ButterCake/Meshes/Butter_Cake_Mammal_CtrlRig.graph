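# Control Rig for the six-legged Butter Cake mammal that adapts its feet, tail and chin to the ground.
# Custom ButterCakeRuntime units update ground traces, then supply effector transforms to three PBIK passes run in order: feet, tail, chin.
# The feet pass solves full-body from the root, with per-foot strength alphas.
# Tail and chin passes solve short chains from tail base and chest, only when their unit flags IK as needed.
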ControlRigBlueprint /ButterCake/Meshes/Butter_Cake_Mammal_CtrlRig
graphs [
 ContainedGraph {nodes [
   RigVMFunctionEntryNode()
   > n1 = /Script/ButterCakeRuntime.RigUnit_MoveTailOutOfGround() {is=struct RigVMUnitNode}
   n2 = /Script/ControlRig.RigUnit_SequenceExecution(ExecuteContext=@n1.ExecutePin) {is=struct RigVMUnitNode}
   Branch_RigVMUnitNode = &RigVMFunction_ControlFlowBranch(Condition=@n1.bShouldDoIK, ExecuteContext=@n2.A)
   PBIK = /Script/PBIK.RigUnit_PBIK(
    BoneSettings="((Bone=\"C_Tail_A_Tail1_Jnt\",RotationStiffness=0.6,X=Free,MaxX=0,Y=Free,MinY=0,MaxY=0,Z=Free,MinZ=0,MaxZ=0,PreferredAngles=(X=0,Y=0,Z=0)),(Bone=\"C_Tail_A_Tail2_Jnt\",RotationStiffness=0.6,X=Free,Y=Free,MinY=0,MaxY=0,Z=Free,MinZ=0,MaxZ=0,PreferredAngles=(X=0, ...(truncated)"
    Debug=(DrawScale=1)
    Effectors=((Bone="C_Tail_A_Tail6_Jnt",Transform=(Rotation=(X=0,Y=0,Z=0,W=1),Translation=(X=0,Y=0,Z=0),Scale3D=(X=1,Y=1,Z=1)),PositionAlpha=1,RotationAlpha=1,StrengthAlpha=1,PullChainAlpha=1,PinRotation=1,ChainDepth=0))
    ExcludedBones=()
    Root=C_Tail_A_TailBase_Jnt
    Settings="(Iterations=10,SubIterations=0,MassMultiplier=3,PrePullRootSettings=(RotationAlpha=0,RotationAlphaX=1,RotationAlphaY=1,RotationAlphaZ=1,PositionAlpha=1,PositionAlphaX=1,PositionAlphaY=1,PositionAlphaZ=1),GlobalPullChainAlpha=1,MaxAngle=30,OverRelaxation=1. ...(truncated)"
    ExecutePin=@Branch_RigVMUnitNode.True
    Effectors.0.Transform=@n1.TailTipEffectorTransform
   ) {is=struct RigVMUnitNode}
   RigVMFunctionReturnNode(ExecuteContext=@n2.B)
  ]}
 ContainedGraph {nodes [
   RigVMFunctionEntryNode()
   > n1 = /Script/ButterCakeRuntime.RigUnit_MoveFeetToGround() {is=struct RigVMUnitNode}
   > PBIK = &RigUnit_PBIK(
    BoneSettings=()
    Debug=(DrawScale=1,bDrawDebug=false)
    Effectors="((Bone=\"L_PawedArm_A_Ball_Jnt\",Transform=(Rotation=(X=0,Y=0,Z=0,W=1),Translation=(X=0,Y=0,Z=0),Scale3D=(X=1,Y=1,Z=1)),PositionAlpha=1,RotationAlpha=1,StrengthAlpha=1,PullChainAlpha=1,PinRotation=1,ChainDepth=0),(Bone=\"R_PawedArm_A_Ball_Jnt\",Transform=(Rota ...(truncated)"
    ExcludedBones=("C_Leaf_FaceAttach_Leaf_Jnt","L_Leaf_Eye_Leaf_Jnt","L_Leaf_Eyelids_Leaf_Jnt","L_Leaf_EyelidUpper_Leaf_Jnt","L_Leaf_EyelidLower_Leaf_Jnt","R_Leaf_Eye_Leaf_Jnt","R_Leaf_Eyelids_Leaf_Jnt","R_Leaf_EyelidLower_Leaf_Jnt","R_Leaf_EyelidUpper_Leaf_Jnt")
    Root=C_Root_Main_Root_Jnt
    Settings="(Iterations=20,SubIterations=0,MassMultiplier=3.5,bAllowStretch=False,PrePullRootSettings=(RotationAlpha=0,RotationAlphaX=1,RotationAlphaY=1,RotationAlphaZ=1,PositionAlpha=1,PositionAlphaX=1,PositionAlphaY=1,PositionAlphaZ=1),GlobalPullChainAlpha=1,MaxAngl ...(truncated)"
    Effectors.0.Transform=RigVMDispatchNode(Index=0, Array=@n1.FootEffectorTransformList) {template="DISPATCH_RigVMDispatch_ArrayGetAtIndex(in Array,in Index,out Element)"}.Element
    Effectors.1.Transform=RigVMDispatchNode(Index=1, Array=@n1.FootEffectorTransformList) {template="DISPATCH_RigVMDispatch_ArrayGetAtIndex(in Array,in Index,out Element)"}.Element
    Effectors.2.Transform=RigVMDispatchNode(Index=2, Array=@n1.FootEffectorTransformList) {template="DISPATCH_RigVMDispatch_ArrayGetAtIndex(in Array,in Index,out Element)"}.Element
    Effectors.3.Transform=RigVMDispatchNode(Index=3, Array=@n1.FootEffectorTransformList) {template="DISPATCH_RigVMDispatch_ArrayGetAtIndex(in Array,in Index,out Element)"}.Element
    Effectors.4.Transform=RigVMDispatchNode(Index=4, Array=@n1.FootEffectorTransformList) {template="DISPATCH_RigVMDispatch_ArrayGetAtIndex(in Array,in Index,out Element)"}.Element
    Effectors.5.Transform=RigVMDispatchNode(Index=5, Array=@n1.FootEffectorTransformList) {template="DISPATCH_RigVMDispatch_ArrayGetAtIndex(in Array,in Index,out Element)"}.Element
    Effectors.0.StrengthAlpha=RigVMDispatchNode(Index=0, Array=@n1.FootEffectorStrengthAlphaList) {template="DISPATCH_RigVMDispatch_ArrayGetAtIndex(in Array,in Index,out Element)"}.Element
    Effectors.1.StrengthAlpha=RigVMDispatchNode(Index=1, Array=@n1.FootEffectorStrengthAlphaList) {template="DISPATCH_RigVMDispatch_ArrayGetAtIndex(in Array,in Index,out Element)"}.Element
    Effectors.2.StrengthAlpha=RigVMDispatchNode(Index=2, Array=@n1.FootEffectorStrengthAlphaList) {template="DISPATCH_RigVMDispatch_ArrayGetAtIndex(in Array,in Index,out Element)"}.Element
    Effectors.3.StrengthAlpha=RigVMDispatchNode(Index=3, Array=@n1.FootEffectorStrengthAlphaList) {template="DISPATCH_RigVMDispatch_ArrayGetAtIndex(in Array,in Index,out Element)"}.Element
    Effectors.4.StrengthAlpha=RigVMDispatchNode(Index=4, Array=@n1.FootEffectorStrengthAlphaList) {template="DISPATCH_RigVMDispatch_ArrayGetAtIndex(in Array,in Index,out Element)"}.Element
    Effectors.5.StrengthAlpha=RigVMDispatchNode(Index=5, Array=@n1.FootEffectorStrengthAlphaList) {template="DISPATCH_RigVMDispatch_ArrayGetAtIndex(in Array,in Index,out Element)"}.Element
   )
   > RigVMFunctionReturnNode()
  ]}
 ContainedGraph {nodes [
   RigVMFunctionEntryNode()
   > n1 = /Script/ButterCakeRuntime.RigUnit_MoveChinOutOfGround() {is=struct RigVMUnitNode}
   n2 = &RigUnit_SequenceExecution(ExecuteContext=@n1.ExecutePin)
   Branch_RigVMUnitNode = &RigVMFunction_ControlFlowBranch(Condition=@n1.bShouldDoIK, ExecuteContext=@n2.A)
   PBIK = &RigUnit_PBIK(
    BoneSettings="((Bone=\"C_QuadSpine_A_Chest_Jnt\",X=Locked,Y=Locked,Z=Limited,MinZ=-10,MaxZ=10,PreferredAngles=(X=0,Y=0,Z=0)),(Bone=\"C_BipedHeadNeck_A_NeckBase_Jnt\",RotationStiffness=0.4,X=Locked,Y=Locked,Z=Limited,MinZ=-10,MaxZ=10,PreferredAngles=(X=0,Y=0,Z=0)),(Bone=\"C_B ...(truncated)"
    Debug=(DrawScale=1)
    Effectors=((Bone="C_Mouth_Upper_Jnt",Transform=(Rotation=(X=0,Y=0,Z=0,W=1),Translation=(X=0,Y=0,Z=0),Scale3D=(X=1,Y=1,Z=1)),PositionAlpha=1,RotationAlpha=1,StrengthAlpha=1,PullChainAlpha=1,PinRotation=1,ChainDepth=0))
    ExcludedBones=("C_Leaf_FaceAttach_Leaf_Jnt")
    Root=C_QuadSpine_A_Chest_Jnt
    Settings="(Iterations=10,SubIterations=0,MassMultiplier=3,PrePullRootSettings=(RotationAlpha=0,RotationAlphaX=1,RotationAlphaY=1,RotationAlphaZ=1,PositionAlpha=1,PositionAlphaX=1,PositionAlphaY=1,PositionAlphaZ=1),GlobalPullChainAlpha=1,MaxAngle=30,OverRelaxation=1. ...(truncated)"
    ExecutePin=@Branch_RigVMUnitNode.True
    Effectors.0.Transform=@n1.ChinEffectorTransform
   )
   RigVMFunctionReturnNode(ExecuteContext=@n2.B)
  ]}
 RigVMFunctionLibrary {nodes [
   ChinOutOfGroundIK = RigVMCollapseNode(ExecuteContext=())
   FeetToGroundIK = RigVMCollapseNode(ExecuteContext=())
   TailOutOfGroundIK = RigVMCollapseNode(ExecuteContext=())
  ]}
 RigVMModel {nodes [
   n1 = &RigUnit_PrepareForExecution()
   /Script/ButterCakeRuntime.RigUnit_UpdateAllGroundTraces(ExecutePin=@n1.ExecutePin) {is=struct RigVMUnitNode}
   RigUnit_BeginExecution = &RigUnit_BeginExecution()
   /Script/ButterCakeRuntime.RigUnit_UpdateSomeGroundTraces(ExecutePin=@RigUnit_BeginExecution.ExecutePin) {is=struct RigVMUnitNode}
   > FeetToGroundIK = RigVMFunctionReferenceNode()
   > TailOutOfGroundIK = RigVMFunctionReferenceNode()
   > ChinOutOfGroundIK = RigVMFunctionReferenceNode()
  ]}
]
hierarchy [
 C_Root_Main_Root_Jnt {loc=<0 0 0>, rot=<0 0 0 1>}
 C_QuadSpine_A_Pelvis_Jnt {parent=C_Root_Main_Root_Jnt, loc=<0 -390.9342 468.0512>, rot=<0.5 0.5 -0.5 -0.5>}
 C_QuadSpine_A_Spine1_Jnt {parent=C_QuadSpine_A_Pelvis_Jnt, loc=<175.0004 0 0>, rot=<0 0 0 -1>}
 C_QuadSpine_A_Spine2_Jnt {parent=C_QuadSpine_A_Spine1_Jnt, loc=<203.8745 0 0>, rot=<0 0 0 -1>}
 C_QuadSpine_A_Spine3_Jnt {parent=C_QuadSpine_A_Spine2_Jnt, loc=<146.1248 0 0>, rot=<0 0 0 -1>}
 C_QuadSpine_A_Chest_Jnt {parent=C_QuadSpine_A_Spine3_Jnt, loc=<250.0003 0 0>, rot=<0 0 0 1>}
 C_BipedHeadNeck_A_NeckBase_Jnt {parent=C_QuadSpine_A_Chest_Jnt, loc=<121.9002 0 0>, rot=<0 0 0 -1>}
 C_BipedHeadNeck_A_NeckMid_Jnt {parent=C_BipedHeadNeck_A_NeckBase_Jnt, loc=<50 0 0>, rot=<0 0 0 -1>}
 C_BipedHeadNeck_A_Head_Jnt {parent=C_BipedHeadNeck_A_NeckMid_Jnt, loc=<50 0 0>, rot=<0 0 0 1>}
 L_Leaf_FinA_Leaf_Jnt {parent=C_BipedHeadNeck_A_Head_Jnt, loc=<206.8053 -185.6912 -102.9522>, rot=<-0.6884 0.3198 0.2746 -0.5902>}
 L_Leaf_FinE_Leaf_Jnt {parent=C_BipedHeadNeck_A_Head_Jnt, loc=<156.5514 -174.095 -242.5345>, rot=<0.8106 -0.252 -0.3184 0.422>}
 L_Leaf_FinI_Leaf_Jnt {parent=C_BipedHeadNeck_A_Head_Jnt, loc=<53.8854 -66.8336 -310.8317>, rot=<0.8203 -0.2612 -0.5087 -0.0104>}
 C_Leaf_FaceAttach_Leaf_Jnt {parent=C_BipedHeadNeck_A_Head_Jnt, loc=<0 0 0>, rot=<-0.7071 0 0 -0.7071>}
 C_Leaf_Jaw_Leaf_Jnt {parent=C_Leaf_FaceAttach_Leaf_Jnt, loc=<21.1096 0 98.816>, rot=<0 0 0 1>}
 C_Leaf_ThroatBulge_Leaf_Jnt {parent=C_Leaf_Jaw_Leaf_Jnt, loc=<246.0163 0 -281.8344>, rot=<0 -0.7071 0 -0.7071>}
 C_Leaf_TeethLower_Leaf_Jnt {parent=C_Leaf_Jaw_Leaf_Jnt, loc=<522.4631 0 -162.5481>, rot=<0 0 0 1>}
 C_Leaf_Tongue_Leaf1_Jnt {parent=C_Leaf_Jaw_Leaf_Jnt, loc=<-125.5273 -0.004 -195.7775>, rot=<0 0.0667 0 -0.9978>}
 C_Leaf_Tongue_Leaf2_Jnt {parent=C_Leaf_Tongue_Leaf1_Jnt, loc=<301.2814 0 -11.2893>, rot=<0 -0.1739 0 -0.9848>}
 C_Leaf_Tongue_Leaf3_Jnt {parent=C_Leaf_Tongue_Leaf2_Jnt, loc=<190 0 0>, rot=<0 -0.0071 0 -1>}
 C_Leaf_Tongue_Leaf4_Jnt {parent=C_Leaf_Tongue_Leaf3_Jnt, loc=<88.5719 0 0>, rot=<0 -0.0085 0 -1>}
 C_Mouth_Lower_Jnt {parent=C_Leaf_Jaw_Leaf_Jnt, loc=<550.957 0.3565 -160.3993>, rot=<0 0 0 1>}
 L_Mouth_Lower_Inner_Jnt {parent=C_Leaf_Jaw_Leaf_Jnt, loc=<491.8998 -249.7995 -147.5454>, rot=<0 0 0.3827 -0.9239>}
 R_Mouth_Lower_Inner_Jnt {parent=C_Leaf_Jaw_Leaf_Jnt, loc=<491.8998 249.7995 -147.5454>, rot=<0 0 0.9239 -0.3827>}
 L_Mouth_Lower_Outer_Jnt {parent=C_Leaf_Jaw_Leaf_Jnt, loc=<248.0483 -353.8648 -107.1007>, rot=<0 0 0.7071 -0.7071>}
 R_Mouth_Lower_Outer_Jnt {parent=C_Leaf_Jaw_Leaf_Jnt, loc=<248.1365 354.1917 -106.9978>, rot=<0 0 -0.7071 -0.7071>}
 L_Mouth_Lower_Inner_Minor_Jnt {parent=C_Leaf_Jaw_Leaf_Jnt, loc=<535.5602 -169.9231 -156.9607>, rot=<0 0 0.0872 -0.9962>}
 L_Mouth_Lower_Outer_Minor_Jnt {parent=C_Leaf_Jaw_Leaf_Jnt, loc=<385.0327 -311.1514 -128.8075>, rot=<0 0 0.5736 -0.8192>}
 R_Mouth_Lower_Inner_Minor_Jnt {parent=C_Leaf_Jaw_Leaf_Jnt, loc=<531.5394 168.3297 -156.3335>, rot=<0 0 0.9962 -0.0872>}
 R_Mouth_Lower_Outer_Minor_Jnt {parent=C_Leaf_Jaw_Leaf_Jnt, loc=<384.7246 310.6763 -128.92>, rot=<0 0 0.8192 -0.5736>}
 R_Mouth_Corner_Lower_Minor_Jnt {parent=C_Leaf_Jaw_Leaf_Jnt, loc=<158.8382 365.372 -102.3783>, rot=<0 0 0.7071 -0.7071>}
 L_Mouth_Corner_Lower_Minor_Jnt {parent=C_Leaf_Jaw_Leaf_Jnt, loc=<158.8382 -365.372 -102.3783>, rot=<0 0 0.7071 -0.7071>}
 C_Leaf_TeethUpper_Leaf_Jnt {parent=C_Leaf_FaceAttach_Leaf_Jnt, loc=<543.5727 0 -63.7321>, rot=<0 0 0 1>}
 L_Mouth_Corner_Jnt {parent=C_Leaf_FaceAttach_Leaf_Jnt, loc=<105.3495 -369.4177 -4.8126>, rot=<0 0 0.7071 -0.7071>}
 L_Mouth_Upper_Inner_Jnt {parent=C_Leaf_FaceAttach_Leaf_Jnt, loc=<509.9842 -247.3537 -48.1415>, rot=<0 0 0.3827 -0.9239>}
 R_Mouth_Upper_Inner_Jnt {parent=C_Leaf_FaceAttach_Leaf_Jnt, loc=<509.9842 247.3537 -48.1415>, rot=<0 0 0.9239 -0.3827>}
 C_Mouth_Upper_Jnt {parent=C_Leaf_FaceAttach_Leaf_Jnt, loc=<572.4091 0.0008 -61.6993>, rot=<0 0 0 1>}
 R_Mouth_Upper_Outer_Jnt {parent=C_Leaf_FaceAttach_Leaf_Jnt, loc=<269.1656 353.8067 -8.0952>, rot=<0 0 -0.7071 -0.7071>}
 L_Mouth_Upper_Outer_Jnt {parent=C_Leaf_FaceAttach_Leaf_Jnt, loc=<269.1656 -353.8067 -8.0952>, rot=<0 0 0.7071 -0.7071>}
 R_Mouth_Corner_Jnt {parent=C_Leaf_FaceAttach_Leaf_Jnt, loc=<105.3495 369.4177 -4.8126>, rot=<0 0 -0.7071 -0.7071>}
 L_Mouth_Upper_Inner_Minor_Jnt {parent=C_Leaf_FaceAttach_Leaf_Jnt, loc=<552.649 -168.3297 -57.5175>, rot=<0 0 0.0872 -0.9962>}
 L_Mouth_Upper_Outer_Minor_Jnt {parent=C_Leaf_FaceAttach_Leaf_Jnt, loc=<404.3691 -310.2137 -29.6692>, rot=<0 0 0.5736 -0.8192>}
 R_Mouth_Upper_Inner_Minor_Jnt {parent=C_Leaf_FaceAttach_Leaf_Jnt, loc=<556.6699 169.9231 -58.1447>, rot=<0 0 0.9962 -0.0872>}
 R_Mouth_Upper_Outer_Minor_Jnt {parent=C_Leaf_FaceAttach_Leaf_Jnt, loc=<404.3691 310.2137 -29.6692>, rot=<0 0 0.8192 -0.5736>}
 R_Mouth_Corner_Upper_Minor_Jnt {parent=C_Leaf_FaceAttach_Leaf_Jnt, loc=<179.9462 365.3268 -3.3692>, rot=<0 0 0.7071 -0.7071>}
 L_Mouth_Corner_Upper_Minor_Jnt {parent=C_Leaf_FaceAttach_Leaf_Jnt, loc=<179.9462 -365.3268 -3.3692>, rot=<0 0 0.7071 -0.7071>}
 L_Leaf_Nostril_Leaf_Jnt {parent=C_Leaf_FaceAttach_Leaf_Jnt, loc=<526.5497 -45.3282 125.8848>, rot=<0 0.3827 0 -0.9239>}
 R_Leaf_Nostril_Leaf_Jnt {parent=C_Leaf_FaceAttach_Leaf_Jnt, loc=<526.5497 45.3282 125.8848>, rot=<0 -0.9239 0 -0.3827>}
 L_Leaf_EyeMaster_Leaf_Jnt {parent=C_Leaf_FaceAttach_Leaf_Jnt, loc=<445.1629 -178.8664 67.322>, rot=<0 0 0.7071 -0.7071>}
 L_Leaf_Eye_Leaf_Jnt {parent=L_Leaf_EyeMaster_Leaf_Jnt, loc=<0 0 0>, rot=<0 0 0 1>}
 L_Leaf_Eyelids_Leaf_Jnt {parent=L_Leaf_EyeMaster_Leaf_Jnt, loc=<0 0 0>, rot=<0 0 0 1>}
 L_Leaf_EyelidUpper_Leaf_Jnt {parent=L_Leaf_Eyelids_Leaf_Jnt, loc=<0 0 0>, rot=<0.1949 0.283 -0.3251 -0.881>}
 L_Leaf_EyelidUpperInner_Leaf_Jnt {parent=L_Leaf_EyelidUpper_Leaf_Jnt, loc=<78.7101 44.0739 35.4165>, rot=<0.2454 -0.1623 0.0459 -0.9546>}
 L_Leaf_EyelidUpperOuter_Leaf_Jnt {parent=L_Leaf_EyelidUpper_Leaf_Jnt, loc=<70.5231 -55.3145 40.6278>, rot=<-0.2535 -0.0222 0.0911 -0.9628>}
 L_Leaf_EyelidLower_Leaf_Jnt {parent=L_Leaf_Eyelids_Leaf_Jnt, loc=<0 0 0>, rot=<0.1949 0.283 -0.3251 -0.881>}
 L_Leaf_EyelidLowerOuter_Leaf_Jnt {parent=L_Leaf_EyelidLower_Leaf_Jnt, loc=<72.1267 -52.9773 -30.7227>, rot=<0.2539 0.3992 0.0811 -0.8773>}
 L_Leaf_EyelidLowerInner_Leaf_Jnt {parent=L_Leaf_EyelidLower_Leaf_Jnt, loc=<80.0627 40.9839 -36.606>, rot=<-0.1456 0.4054 -0.0397 -0.9016>}
 L_Leaf_EyelidCornerInner_Leaf_Jnt {parent=L_Leaf_Eyelids_Leaf_Jnt, loc=<0.9903 103.416 -16.5794>, rot=<0 0.2663 0 -0.9639>}
 L_Leaf_EyelidCornerOuter_Leaf_Jnt {parent=L_Leaf_Eyelids_Leaf_Jnt, loc=<73.4561 -25.2948 73.1669>, rot=<0 0.3156 0 -0.9489>}
 R_Leaf_EyeMaster_Leaf_Jnt {parent=C_Leaf_FaceAttach_Leaf_Jnt, loc=<445.1629 178.8664 67.322>, rot=<0.7071 -0.7071 0 0>}
 R_Leaf_Eye_Leaf_Jnt {parent=R_Leaf_EyeMaster_Leaf_Jnt, loc=<0 0 0>, rot=<1 0 0 0>}
 R_Leaf_Eyelids_Leaf_Jnt {parent=R_Leaf_EyeMaster_Leaf_Jnt, loc=<0 0 0>, rot=<0 0 0 1>}
 R_Leaf_EyelidLower_Leaf_Jnt {parent=R_Leaf_Eyelids_Leaf_Jnt, loc=<0 0 0>, rot=<0.1949 0.283 -0.3251 -0.881>}
 R_Leaf_EyelidLowerOuter_Leaf_Jnt {parent=R_Leaf_EyelidLower_Leaf_Jnt, loc=<-72.1267 52.9773 30.7227>, rot=<0.2539 0.3992 0.0811 -0.8773>}
 R_Leaf_EyelidLowerInner_Leaf_Jnt {parent=R_Leaf_EyelidLower_Leaf_Jnt, loc=<-80.0627 -40.9839 36.606>, rot=<-0.1456 0.4054 -0.0397 -0.9016>}
 R_Leaf_EyelidUpper_Leaf_Jnt {parent=R_Leaf_Eyelids_Leaf_Jnt, loc=<0 0 0>, rot=<0.1949 0.283 -0.3251 -0.881>}
 R_Leaf_EyelidUpperOuter_Leaf_Jnt {parent=R_Leaf_EyelidUpper_Leaf_Jnt, loc=<-70.5231 55.3145 -40.6278>, rot=<-0.2535 -0.0222 0.0911 -0.9628>}
 R_Leaf_EyelidUpperInner_Leaf_Jnt {parent=R_Leaf_EyelidUpper_Leaf_Jnt, loc=<-78.7101 -44.0739 -35.4165>, rot=<0.2454 -0.1623 0.0459 -0.9546>}
 R_Leaf_EyelidCornerInner_Leaf_Jnt {parent=R_Leaf_Eyelids_Leaf_Jnt, loc=<-1.8287 -103.1703 20.3467>, rot=<0 0.2663 0 -0.9639>}
 R_Leaf_EyelidCornerOuter_Leaf_Jnt {parent=R_Leaf_Eyelids_Leaf_Jnt, loc=<-73.4561 25.2948 -73.1669>, rot=<0 0.3156 0 -0.9489>}
 L_Leaf_FinG_Leaf_Jnt {parent=C_BipedHeadNeck_A_Head_Jnt, loc=<52.2917 -169.7823 -161.7566>, rot=<-0.5892 0.2642 0.5418 -0.538>}
 L_Leaf_FinB_Leaf_Jnt {parent=C_BipedHeadNeck_A_Head_Jnt, loc=<201.2787 -190.2342 -198.7223>, rot=<-0.5801 0.2497 0.3894 -0.6705>}
 L_Leaf_FinF_Leaf_Jnt {parent=C_BipedHeadNeck_A_Head_Jnt, loc=<90.6227 -130.2298 -308.4832>, rot=<0.9069 -0.2377 -0.3188 0.1395>}
 L_Leaf_FinH_Leaf_Jnt {parent=C_BipedHeadNeck_A_Head_Jnt, loc=<41.2119 -146.3819 -244.3867>, rot=<0.8105 -0.2376 -0.4584 0.2765>}
 L_Leaf_FinD_Leaf_Jnt {parent=C_BipedHeadNeck_A_Head_Jnt, loc=<101.4122 -189.201 -161.9618>, rot=<-0.713 0.2441 0.3642 -0.5472>}
 L_Leaf_FinC_Leaf_Jnt {parent=C_BipedHeadNeck_A_Head_Jnt, loc=<153.1849 -126.8326 -331.3252>, rot=<0.784 0.0433 -0.5209 0.3348>}
 R_Leaf_FinD_Leaf_Jnt {parent=C_BipedHeadNeck_A_Head_Jnt, loc=<101.4122 -189.201 161.9618>, rot=<-0.2441 -0.713 -0.5472 -0.3642>}
 R_Leaf_FinC_Leaf_Jnt {parent=C_BipedHeadNeck_A_Head_Jnt, loc=<153.1849 -126.8326 331.3252>, rot=<0.0433 -0.784 -0.3348 -0.5209>}
 R_Leaf_FinG_Leaf_Jnt {parent=C_BipedHeadNeck_A_Head_Jnt, loc=<52.2917 -169.7823 161.7566>, rot=<-0.2642 -0.5892 -0.538 -0.5418>}
 R_Leaf_FinF_Leaf_Jnt {parent=C_BipedHeadNeck_A_Head_Jnt, loc=<90.6226 -130.2298 308.4832>, rot=<-0.2377 -0.9069 -0.1395 -0.3188>}
 R_Leaf_FinA_Leaf_Jnt {parent=C_BipedHeadNeck_A_Head_Jnt, loc=<206.8053 -185.6912 102.9522>, rot=<-0.3198 -0.6884 -0.5902 -0.2746>}
 R_Leaf_FinH_Leaf_Jnt {parent=C_BipedHeadNeck_A_Head_Jnt, loc=<41.2119 -146.3819 244.3867>, rot=<-0.2376 -0.8105 -0.2765 -0.4584>}
 R_Leaf_FinI_Leaf_Jnt {parent=C_BipedHeadNeck_A_Head_Jnt, loc=<53.8854 -66.8336 310.8317>, rot=<-0.2612 -0.8203 0.0104 -0.5087>}
 R_Leaf_FinE_Leaf_Jnt {parent=C_BipedHeadNeck_A_Head_Jnt, loc=<156.5515 -174.095 242.5345>, rot=<-0.252 -0.8106 -0.422 -0.3184>}
 R_Leaf_FinB_Leaf_Jnt {parent=C_BipedHeadNeck_A_Head_Jnt, loc=<201.2787 -190.2342 198.7223>, rot=<0.2497 0.5801 0.6705 0.3894>}
 C_Leaf_Blowhole_Leaf1_Jnt {parent=C_BipedHeadNeck_A_Head_Jnt, loc=<269.9177 -171.2768 0>, rot=<-0.7071 0 0 -0.7071>}
 C_Leaf_Blowhole_Leaf2_Jnt {parent=C_Leaf_Blowhole_Leaf1_Jnt, loc=<0 0 25>, rot=<0 0 0 1>}
 C_Leaf_BlowholeMid_Jnt {parent=C_BipedHeadNeck_A_Head_Jnt, loc=<269.9177 -183.7768 0>, rot=<0.5 0.5 -0.5 0.5>}
 C_Leaf_Uvula_Leaf_Jnt {parent=C_BipedHeadNeck_A_Head_Jnt, loc=<-20.5785 -58.7735 0>, rot=<0 0 -0.6487 -0.761>}
 L_PawedArm_A_Shoulder_Jnt {parent=C_QuadSpine_A_Chest_Jnt, loc=<4.5826 15.3278 -374.6284>, rot=<0.0026 -0.0212 -0.8372 -0.5465>}
 L_PawedArm_A_Elbow_Jnt {parent=L_PawedArm_A_Shoulder_Jnt, loc=<160.6606 0 0>, rot=<0 0 0.4176 -0.9087>}
 L_PawedArm_A_Wrist_Jnt {parent=L_PawedArm_A_Elbow_Jnt, loc=<179.0961 0 0>, rot=<0.954 0.2883 0.0734 -0.0359>}
 L_PawedArm_A_Ball_Jnt {parent=L_PawedArm_A_Wrist_Jnt, loc=<96.0208 0 0>, rot=<0.0537 0.0415 -0.5063 -0.8597>}
 L_PawedArm_A_Toe_Jnt {parent=L_PawedArm_A_Ball_Jnt, loc=<50 0 0>, rot=<0 0 -0.0898 -0.996>}
 L_PawedArm_A_ToeTip_Jnt {parent=L_PawedArm_A_Toe_Jnt, loc=<44.7214 0 0>, rot=<0 0 -0.2298 -0.9732>}
 L_PawedArm_A_Scapula_Jnt {parent=L_PawedArm_A_Shoulder_Jnt, loc=<-94.1862 94.4174 72.2364>, rot=<-0.371 -0.8955 -0.0922 0.2277>}
 R_PawedArm_A_Shoulder_Jnt {parent=C_QuadSpine_A_Chest_Jnt, loc=<4.5826 15.3278 374.6284>, rot=<-0.0212 -0.0026 0.5465 -0.8372>}
 R_PawedArm_A_Elbow_Jnt {parent=R_PawedArm_A_Shoulder_Jnt, loc=<-160.6607 0 0>, rot=<0 0 0.4176 -0.9087>}
 R_PawedArm_A_Wrist_Jnt {parent=R_PawedArm_A_Elbow_Jnt, loc=<-179.0961 0 0>, rot=<0.954 0.2883 0.0734 -0.0359>}
 R_PawedArm_A_Ball_Jnt {parent=R_PawedArm_A_Wrist_Jnt, loc=<-96.0208 0 0>, rot=<0.0537 0.0415 -0.5063 -0.8597>}
 R_PawedArm_A_Toe_Jnt {parent=R_PawedArm_A_Ball_Jnt, loc=<-50 0 0>, rot=<0 0 -0.0898 -0.996>}
 R_PawedArm_A_ToeTip_Jnt {parent=R_PawedArm_A_Toe_Jnt, loc=<-44.7214 0 0>, rot=<0 0 -0.2298 -0.9732>}
 R_PawedArm_A_Scapula_Jnt {parent=R_PawedArm_A_Shoulder_Jnt, loc=<94.1862 -94.4174 -72.2364>, rot=<0.2277 -0.0922 0.8955 0.371>}
 C_Leaf_ChestPuff_Leaf_Jnt {parent=C_QuadSpine_A_Spine3_Jnt, loc=<178.7132 311.5601 0>, rot=<-0.5 -0.5 -0.5 -0.5>}
 L_PawedArm_B_Shoulder_Jnt {parent=C_QuadSpine_A_Spine2_Jnt, loc=<65.5905 -9.8096 -416.7953>, rot=<-0.0002 0.001 -0.8281 -0.5605>}
 L_PawedArm_B_Elbow_Jnt {parent=L_PawedArm_B_Shoulder_Jnt, loc=<153.0108 0 0>, rot=<0 0 0.3841 -0.9233>}
 L_PawedArm_B_Wrist_Jnt {parent=L_PawedArm_B_Elbow_Jnt, loc=<209.8503 0 0>, rot=<0.9422 0.335 -0.0052 -0.0027>}
 L_PawedArm_B_Ball_Jnt {parent=L_PawedArm_B_Wrist_Jnt, loc=<96.8154 0 0>, rot=<0.0011 -0.0048 -0.5658 -0.8245>}
 L_PawedArm_B_Toe_Jnt {parent=L_PawedArm_B_Ball_Jnt, loc=<50 0 0>, rot=<0 0 -0.0898 -0.996>}
 L_PawedArm_B_ToeTip_Jnt {parent=L_PawedArm_B_Toe_Jnt, loc=<44.7214 0 0>, rot=<0 0 -0.2298 -0.9732>}
 L_PawedArm_B_Scapula_Jnt {parent=L_PawedArm_B_Shoulder_Jnt, loc=<-98.2375 98.6881 100.806>, rot=<-0.3651 -0.8785 -0.1183 0.2846>}
 R_PawedArm_B_Shoulder_Jnt {parent=C_QuadSpine_A_Spine2_Jnt, loc=<65.5905 -9.8096 416.7953>, rot=<0.001 0.0002 0.5605 -0.8281>}
 R_PawedArm_B_Elbow_Jnt {parent=R_PawedArm_B_Shoulder_Jnt, loc=<-153.0108 0 0>, rot=<0 0 0.3841 -0.9233>}
 R_PawedArm_B_Wrist_Jnt {parent=R_PawedArm_B_Elbow_Jnt, loc=<-209.8503 0 0>, rot=<0.9422 0.335 -0.0052 -0.0027>}
 R_PawedArm_B_Ball_Jnt {parent=R_PawedArm_B_Wrist_Jnt, loc=<-96.8154 0 0>, rot=<0.0011 -0.0048 -0.5658 -0.8245>}
 R_PawedArm_B_Toe_Jnt {parent=R_PawedArm_B_Ball_Jnt, loc=<-50 0 0>, rot=<0 0 -0.0898 -0.996>}
 R_PawedArm_B_ToeTip_Jnt {parent=R_PawedArm_B_Toe_Jnt, loc=<-44.7214 0 0>, rot=<0 0 -0.2298 -0.9732>}
 R_PawedArm_B_Scapula_Jnt {parent=R_PawedArm_B_Shoulder_Jnt, loc=<98.2375 -98.6881 -100.806>, rot=<0.2846 -0.1183 0.8785 0.3651>}
 L_Leaf_SpineDupe_Leaf_Jnt {parent=C_QuadSpine_A_Spine1_Jnt, loc=<-0.0004 0 0>, rot=<-0.5 -0.5 0.5 -0.5>}
 R_Leaf_SpineDupe_Leaf_Jnt {parent=C_QuadSpine_A_Spine1_Jnt, loc=<-0.0004 0 0>, rot=<0.5 -0.5 -0.5 -0.5>}
 L_PawedLeg_C_Thigh_Jnt {parent=C_QuadSpine_A_Pelvis_Jnt, loc=<25.3917 -0.4966 -370.3429>, rot=<0.8669 0.498 -0.0047 -0.0228>}
 L_PawedLeg_C_Knee_Jnt {parent=L_PawedLeg_C_Thigh_Jnt, loc=<240.0406 0 0>, rot=<0 0 0.6121 -0.7908>}
 L_PawedLeg_C_Ankle_Jnt {parent=L_PawedLeg_C_Knee_Jnt, loc=<175.4367 0 0>, rot=<-0.0273 0.0625 -0.3994 -0.9142>}
 L_PawedLeg_C_Ball_Jnt {parent=L_PawedLeg_C_Ankle_Jnt, loc=<113.4883 0 0>, rot=<-0.0077 -0.0532 -0.4322 -0.9002>}
 L_PawedLeg_C_Toe_Jnt {parent=L_PawedLeg_C_Ball_Jnt, loc=<25 0 0>, rot=<0 0 -0.0277 -0.9996>}
 L_PawedLeg_C_ToeTip_Jnt {parent=L_PawedLeg_C_Toe_Jnt, loc=<18.0278 0 0>, rot=<0 0 -0.2898 -0.9571>}
 C_Tail_A_TailBase_Jnt {parent=C_QuadSpine_A_Pelvis_Jnt, loc=<-126.2513 -175 0>, rot=<0 -0.7071 0.7071 0>}
 C_Tail_A_Tail1_Jnt {parent=C_Tail_A_TailBase_Jnt, loc=<137.6812 0 0>, rot=<0 0 0 1>}
 C_Tail_A_Tail2_Jnt {parent=C_Tail_A_Tail1_Jnt, loc=<137.6824 0 0>, rot=<0 0 0 1>}
 C_Tail_A_Tail3_Jnt {parent=C_Tail_A_Tail2_Jnt, loc=<137.6809 0 0>, rot=<0 0 0 1>}
 C_Tail_A_Tail4_Jnt {parent=C_Tail_A_Tail3_Jnt, loc=<137.6824 0 0>, rot=<0 0 0 1>}
 C_Tail_A_Tail5_Jnt {parent=C_Tail_A_Tail4_Jnt, loc=<137.6814 0 0>, rot=<0 0 0 1>}
 C_Tail_A_Tail6_Jnt {parent=C_Tail_A_Tail5_Jnt, loc=<137.6824 0 0>, rot=<0 0 0 1>}
 R_PawedLeg_C_Thigh_Jnt {parent=C_QuadSpine_A_Pelvis_Jnt, loc=<25.3917 -0.4966 370.3429>, rot=<0.498 -0.8669 0.0228 -0.0047>}
 R_PawedLeg_C_Knee_Jnt {parent=R_PawedLeg_C_Thigh_Jnt, loc=<-240.0406 0 0>, rot=<0 0 0.6121 -0.7908>}
 R_PawedLeg_C_Ankle_Jnt {parent=R_PawedLeg_C_Knee_Jnt, loc=<-175.4367 0 0>, rot=<-0.0273 0.0625 -0.3994 -0.9142>}
 R_PawedLeg_C_Ball_Jnt {parent=R_PawedLeg_C_Ankle_Jnt, loc=<-113.4883 0 0>, rot=<-0.0077 -0.0532 -0.4322 -0.9002>}
 R_PawedLeg_C_Toe_Jnt {parent=R_PawedLeg_C_Ball_Jnt, loc=<-25 0 0>, rot=<0 0 -0.0277 -0.9996>}
 R_PawedLeg_C_ToeTip_Jnt {parent=R_PawedLeg_C_Toe_Jnt, loc=<-18.0277 0 0>, rot=<0 0 -0.2898 -0.9571>}
 C_Leaf_AbdomenPuff_Leaf_Jnt {parent=C_QuadSpine_A_Pelvis_Jnt, loc=<50.7778 286.5381 0>, rot=<-0.5 -0.5 -0.5 -0.5>}
 IK_Foot_Root {parent=C_Root_Main_Root_Jnt, loc=<0 0 0>, rot=<0 0 0 1>}
 IK_Foot_A_L {parent=IK_Foot_Root, loc=<374.6284 388.6483 50>, rot=<-0.3162 -0.6325 -0.3162 -0.6325>}
 IK_Foot_B_L {parent=IK_Foot_Root, loc=<416.7953 53.5311 50.0008>, rot=<-0.3162 -0.6325 -0.3162 -0.6325>}
 IK_Foot_C_L {parent=IK_Foot_Root, loc=<370.3426 -365.5415 24.9991>, rot=<-0.3162 -0.6325 -0.3162 -0.6325>}
 IK_Foot_A_R {parent=IK_Foot_Root, loc=<-374.6284 388.6483 50>, rot=<0.6325 -0.3162 0.6325 -0.3162>}
 IK_Foot_B_R {parent=IK_Foot_Root, loc=<-416.7953 53.5311 50.0008>, rot=<0.6325 -0.3162 0.6325 -0.3162>}
 IK_Foot_C_R {parent=IK_Foot_Root, loc=<-370.3426 -365.5415 24.9991>, rot=<0.6325 -0.3162 0.6325 -0.3162>}
 IK_Tail_Root {parent=C_Root_Main_Root_Jnt, loc=<0 0 0>, rot=<0 0 0 1>}
 IK_Tail {parent=IK_Tail_Root, loc=<0 -517.1855 643.0512>, rot=<0 0 0.7071 -0.7071>}
 L_PawedArm_A_Ball_Jnt_Ctrl {controlType=EULER_TRANSFORM, type=Control, loc=<445.0131 481.5271 50>, rot=<0.3162 0.6325 0.3162 0.6325>}
 R_PawedArm_A_Ball_Jnt_Ctrl {controlType=EULER_TRANSFORM, type=Control, loc=<-445.0132 481.527 50.0001>, rot=<-0.6325 0.3162 -0.6325 0.3162>}
 L_PawedArm_B_Ball_Jnt_Ctrl {controlType=EULER_TRANSFORM, type=Control, loc=<449.7149 138.342 50>, rot=<0.3162 0.6325 0.3162 0.6325>}
 R_PawedArm_B_Ball_Jnt_Ctrl {controlType=EULER_TRANSFORM, type=Control, loc=<-449.7149 138.342 50.0001>, rot=<-0.6325 0.3162 -0.6325 0.3162>}
 L_PawedLeg_C_Ball_Jnt_Ctrl {controlType=EULER_TRANSFORM, type=Control, loc=<440.6634 -554.7944 25>, rot=<0.3697 0.6027 0.2604 0.6574>}
 R_PawedLeg_C_Ball_Jnt_Ctrl {controlType=EULER_TRANSFORM, type=Control, loc=<-440.6634 -554.7944 25>, rot=<-0.6574 0.2604 -0.6027 0.3697>}
 RootRotation {type=Curve, loc=<0 0 0>, rot=<0 0 0 1>}
 FBIK_ArmALeft {type=Curve, loc=<0 0 0>, rot=<0 0 0 1>}
 FBIK_ArmARight {type=Curve, loc=<0 0 0>, rot=<0 0 0 1>}
 FBIK_ArmBLeft {type=Curve, loc=<0 0 0>, rot=<0 0 0 1>}
 FBIK_ArmBRight {type=Curve, loc=<0 0 0>, rot=<0 0 0 1>}
 FBIK_LegCLeft {type=Curve, loc=<0 0 0>, rot=<0 0 0 1>}
 FBIK_LegCRight {type=Curve, loc=<0 0 0>, rot=<0 0 0 1>}
 ApplyBreathing {type=Curve, loc=<0 0 0>, rot=<0 0 0 1>}
 ApplyAimOffset {type=Curve, loc=<0 0 0>, rot=<0 0 0 1>}
 DisableHostileAdditive {type=Curve, loc=<0 0 0>, rot=<0 0 0 1>}
 PupilBlend {type=Curve, loc=<0 0 0>, rot=<0 0 0 1>}
 DisableFacialAdditive {type=Curve, loc=<0 0 0>, rot=<0 0 0 1>}
 IrisScale {type=Curve, loc=<0 0 0>, rot=<0 0 0 1>}
 FootPhase {type=Curve, loc=<0 0 0>, rot=<0 0 0 1>}
]
