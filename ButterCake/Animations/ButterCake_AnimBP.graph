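# Animation Blueprint for the ButterCake creature, driving locomotion, facial expression and gaze with neutral, hostile and lured variants.
# A Locomotion state machine covers idle with turn-in-place additives, walk, and stops picked by mood and foot phase.
# Its pose passes through montage slots, a curve-gated facial additive machine (neutral, angry, happy), aim and look-at offsets, Inertialization, then a Control Rig.
# State transitions use inertialization instead of standard crossfades.
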AnimBlueprint /ButterCake/Animations/ButterCake_AnimBP
target_skeleton=/ButterCake/Meshes/Butter_Cake_Mammal_Invis_Skeleton
graphs [
 AnimGraph {nodes [
   Apply Additive!(
    Base
    Additive="Sequence Player (additive) 'ButterCake_Hostile_Additive_Pose'"() {sequence=&/ButterCake_Hostile_Additive_Pose}
    bAlphaBoolEnabled:bool=Get bShouldApplyHostileAdditive() -> (bShouldApplyHostileAdditive:bool)
   ) -> (Pose) {
    alphaInputType=Bool
    alphaBoolBlend {blendInTime=0.3, blendOption=Linear, blendOutTime=0.3}
    alphaScaleBiasClamp {bClampResult=true, bias=1, scale=-1}
   }
   AnimGraphNode_StateMachine_0 = Locomotion() {is=title AnimGraphNode_StateMachine, bAllowConduitEntryStates=true, maxTransitionsPerFrame=1}
   AnimGraphNode_StateMachine_1 = Facial Additive() {is=title AnimGraphNode_StateMachine, bAllowConduitEntryStates=true}
   n1 = Apply Additive!(
    Base=Slot 'FullBodyAdditive'(Source=Slot 'DefaultSlot'(Source=@AnimGraphNode_StateMachine_0)) {slotName=FullBodyAdditive}
    Additive=@AnimGraphNode_StateMachine_1
    AlphaCurveName:Name=DisableFacialAdditive
   ) -> (Pose) {
    alphaInputType=Curve
    alphaBoolBlend {blendInTime=0.3, blendOption=Linear, blendOutTime=0.3}
    alphaScaleBiasClamp {bClampResult=true, bInterpResult=true, bias=1, scale=-1}
   }
   n2 = AimOffset Player 'ButterCake_AO'!(
    X:float=Get AimYaw() -> (AimYaw:float)
    Y:float=0
    BasePose:Pose Link Structure=@n1
    AlphaCurveName:Name=Get AimOffsetAlpha() -> (AimOffsetAlpha:Name=None)
   ) -> (Pose:Pose Link Structure) {
    alphaCurveName=ApplyAimOffset
    alphaInputType=Curve
    blendSpace=/ButterCake/Animations/AO/ButterCake_AO
    alphaBoolBlend {blendInTime=0.4, blendOption=Linear, blendOutTime=0.4}
    alphaScaleBiasClamp {bInterpResult=true}
   }
   n3 = AimOffset Player 'ButterCake_LookAt_AO'!(
    X:float=Get LookAtYaw() -> (LookAtYaw:float)
    Y:float=Get LookAtPitch() -> (LookAtPitch:float)
    BasePose:Pose Link Structure=@n2
    AlphaCurveName:Name=DisableLookAt
   ) -> (Pose:Pose Link Structure) {
    alphaInputType=Curve
    blendSpace=/ButterCake/Animations/AO/ButterCake_LookAt_AO
    alphaBoolBlend {blendOption=Linear}
    alphaScaleBiasClamp {bClampResult=true, bInterpResult=true, bias=1, scale=-1}
   }
   n4 = Control Rig(Source=Inertialization!(Source:Pose Link Structure=@n3) -> (Pose:Pose Link Structure) {is=title AnimGraphNode_Inertialization, filteredCurves [RootRotation]}) {
    alphaBoolBlend {blendOption=Linear}
    controlRigAssetReference {blueprintRigClass=/ButterCake/Meshes/Butter_Cake_Mammal_CtrlRig.Butter_Cake_Mammal_CtrlRig_C}
   }
   Output Pose(Result=@n4) {layerGroup=None}
  ]}
 Locomotion {
  entry=Conduit
  path=AnimGraph.AnimGraphNode_StateMachine_0.Locomotion
  ids=id
  nodes [
   AnimStateConduitNode_0 = Conduit()
   AnimStateNode_0 = Idle() {bAlwaysResetOnEntry=true}
   AnimStateNode_1 = Walk() {is=title AnimStateNode, bAlwaysResetOnEntry=true}
   AnimStateNode_2 = WalkToIdle() {is=title AnimStateNode}
   AnimStateTransitionNode_0 = AnimStateTransitionNode() {crossfadeDuration=0.4, logicType=TLT_Inertialization}
   AnimStateTransitionNode_1 = AnimStateTransitionNode() {crossfadeDuration=0.4, logicType=TLT_Inertialization}
   AnimStateTransitionNode_2 = AnimStateTransitionNode() {bAutomaticRuleBasedOnSequencePlayerInState=true, crossfadeDuration=0.4, logicType=TLT_Inertialization, priorityOrder=100}
   AnimStateTransitionNode_3 = AnimStateTransitionNode() {crossfadeDuration=0.4, logicType=TLT_Inertialization}
   AnimStateTransitionNode_4 = AnimStateTransitionNode() {crossfadeDuration=0.5, logicType=TLT_Inertialization, priorityOrder=100}
   AnimStateTransitionNode_5 = AnimStateTransitionNode() {crossfadeDuration=0.3, logicType=TLT_Inertialization}
  ]
  links [
   {
    dst=Idle
    rule=AnimGraph.AnimGraphNode_StateMachine_0.Locomotion.AnimStateTransitionNode_4.Transition
    src=Conduit
    via=AnimStateTransitionNode_4
   }
   {
    dst=Walk
    rule=AnimGraph.AnimGraphNode_StateMachine_0.Locomotion.AnimStateTransitionNode_5.Transition
    src=Conduit
    via=AnimStateTransitionNode_5
   }
   {
    dst=Walk
    rule=AnimGraph.AnimGraphNode_StateMachine_0.Locomotion.AnimStateTransitionNode_0.Transition
    src=Idle
    via=AnimStateTransitionNode_0
   }
   {
    dst=WalkToIdle
    rule=AnimGraph.AnimGraphNode_StateMachine_0.Locomotion.AnimStateTransitionNode_1.Transition
    src=Walk
    via=AnimStateTransitionNode_1
   }
   {
    dst=Idle
    rule=AnimGraph.AnimGraphNode_StateMachine_0.Locomotion.AnimStateTransitionNode_2.Transition
    src=WalkToIdle
    via=AnimStateTransitionNode_2
   }
   {
    dst=Walk
    rule=AnimGraph.AnimGraphNode_StateMachine_0.Locomotion.AnimStateTransitionNode_3.Transition
    src=WalkToIdle
    via=AnimStateTransitionNode_3
   }
  ]
 }
 Conduit {path=AnimGraph.AnimGraphNode_StateMachine_0.Locomotion.AnimStateConduitNode_0.Conduit, nodes [Result(bCanEnterTransition=true)]}
 Idle {
  path=AnimGraph.AnimGraphNode_StateMachine_0.Locomotion.AnimStateNode_0.Idle
  nodes [
   n1 = Sequence Player 'ButterCake_IdleHostile'() {sequence=&/ButterCake_IdleHostile}
   n2 = Sequence Player 'ButterCake_Idle'() {sequence=&/ButterCake_Idle}
   n3 = Get bIsHostile() -> (bIsHostile:bool)
   n4 = "Sequence Player (additive) 'ButterCake_IdleTurnLeft_OnSpot'"() {sequence=&/ButterCake_IdleTurnLeft_OnSpot}
   n5 = "Sequence Player (additive) 'ButterCake_IdleTurnRight_OnSpot'"() {sequence=&/ButterCake_IdleTurnRight_OnSpot}
   n6 = Get bIdleTurnRight() -> (bIdleTurnRight:bool)
   Output Animation Pose(Result=Apply Additive!(
     Base=Blend Poses by bool(BlendPose_0=@n1, BlendPose_1=@n2, BlendTime_0=0.6, BlendTime_1=0.6, bActiveValue=@n3) {blendType=Linear, childUpateMode=ResetChildOnActivate, transitionType=Inertialization}
     Additive=Blend Poses by bool(BlendPose_0=@n5, BlendPose_1=@n4, BlendTime_0=0.6, BlendTime_1=0.6, bActiveValue=@n6) {blendType=Linear, childUpateMode=ResetChildOnActivate, transitionType=Inertialization}
     bAlphaBoolEnabled:bool=Get bIsIdleTurn() -> (bIsIdleTurn:bool)
    ) -> (Pose) {alphaInputType=Bool, alphaBoolBlend {blendInTime=0.1, blendOption=Linear, blendOutTime=0.1}})
  ]
 }
 Walk {
  path=AnimGraph.AnimGraphNode_StateMachine_0.Locomotion.AnimStateNode_1.Walk
  nodes [
   Apply Additive(
    Additive=Blendspace Player 'ButterCake_Walk_Lean_BS'!(X=Get LeanAngle() -> (LeanAngle:float)) -> (Pose) {blendSpace=&/ButterCake_Walk_Lean_BS}
    bAlphaBoolEnabled=Get bIsMoving() -> (bIsMoving:bool)
   ) {alphaInputType=Bool, alphaBoolBlend {blendInTime=0.3, blendOutTime=0.3}}
   n1 = Sequence Player 'ButterCake_Walk_N'() {sequence=&/ButterCake_Walk_N}
   n2 = Sequence Player 'ButterCake_WalkHostile_N'() {sequence=&/ButterCake_WalkHostile_N}
   n3 = Get bIsHostile()
   Output Animation Pose(Result=Blend Poses by bool(
     BlendPose_0=Sequence Player 'ButterCake_Walk_Lured'() {sequence=&/ButterCake_Walk_Lured}
     BlendPose_1=Blend Poses by bool(BlendPose_0=@n2, BlendPose_1=@n1, BlendTime_0=0.4, BlendTime_1=0.4, bActiveValue=@n3) {blendType=Linear, transitionType=Inertialization}
     BlendTime_0=0.4
     BlendTime_1=0.4
     bActiveValue=Get bIsLured() -> (bIsLured:bool)
    ) {blendType=Linear, transitionType=Inertialization})
  ]
 }
 WalkToIdle {
  path=AnimGraph.AnimGraphNode_StateMachine_0.Locomotion.AnimStateNode_2.WalkToIdle
  nodes [
   AnimGraphNode_StateMachine_0 = Stop() {is=title AnimGraphNode_StateMachine, bAllowConduitEntryStates=true}
   Output Animation Pose(Result=@AnimGraphNode_StateMachine_0)
  ]
 }
 Stop {
  entry=Conduit
  path=AnimGraph.AnimGraphNode_StateMachine_0.Locomotion.AnimStateNode_2.WalkToIdle.AnimGraphNode_StateMachine_0.Stop
  ids=id
  nodes [
   AnimStateConduitNode_0 = Conduit()
   AnimStateNode_1 = Angry() {is=title AnimStateNode}
   AnimStateNode_7 = Neutral() {is=title AnimStateNode}
   AnimStateTransitionNode_5 = AnimStateTransitionNode() {crossfadeDuration=0.4, logicType=TLT_Inertialization}
   AnimStateTransitionNode_9 = AnimStateTransitionNode() {crossfadeDuration=0.4, logicType=TLT_Inertialization, priorityOrder=100}
  ]
  links [
   {
    dst=Angry
    rule=AnimGraph.AnimGraphNode_StateMachine_0.Locomotion.AnimStateNode_2.WalkToIdle.AnimGraphNode_StateMachine_0.Stop.AnimStateTransitionNode_5.Transition
    src=Conduit
    via=AnimStateTransitionNode_5
   }
   {
    dst=Neutral
    rule=AnimGraph.AnimGraphNode_StateMachine_0.Locomotion.AnimStateNode_2.WalkToIdle.AnimGraphNode_StateMachine_0.Stop.AnimStateTransitionNode_9.Transition
    src=Conduit
    via=AnimStateTransitionNode_9
   }
  ]
 }
 Conduit {
  path=AnimGraph.AnimGraphNode_StateMachine_0.Locomotion.AnimStateNode_2.WalkToIdle.AnimGraphNode_StateMachine_0.Stop.AnimStateConduitNode_0.Conduit
  nodes [Result(bCanEnterTransition=true)]
 }
 Angry {
  path=AnimGraph.AnimGraphNode_StateMachine_0.Locomotion.AnimStateNode_2.WalkToIdle.AnimGraphNode_StateMachine_0.Stop.AnimStateNode_1.Angry
  nodes [
   AnimGraphNode_StateMachine_0 = AngryStop() {is=title AnimGraphNode_StateMachine, bAllowConduitEntryStates=true}
   Output Animation Pose(Result=@AnimGraphNode_StateMachine_0)
  ]
 }
 AngryStop {
  entry=Conduit
  path=AnimGraph.AnimGraphNode_StateMachine_0.Locomotion.AnimStateNode_2.WalkToIdle.AnimGraphNode_StateMachine_0.Stop.AnimStateNode_1.Angry.AnimGraphNode_StateMachine_0.AngryStop
  ids=id
  nodes [
   AnimStateConduitNode_0 = Conduit()
   AnimStateNode_1 = LeftFootPass() {is=title AnimStateNode}
   AnimStateNode_2 = RightFootPass() {is=title AnimStateNode}
   AnimStateNode_3 = LeftFootPlant() {is=title AnimStateNode}
   AnimStateNode_4 = RightFootPlant() {is=title AnimStateNode}
   AnimStateTransitionNode_10 = AnimStateTransitionNode() {crossfadeDuration=0.4, logicType=TLT_Inertialization, priorityOrder=100}
   AnimStateTransitionNode_7 = AnimStateTransitionNode() {crossfadeDuration=0.4, logicType=TLT_Inertialization}
   AnimStateTransitionNode_8 = AnimStateTransitionNode() {crossfadeDuration=0.4, logicType=TLT_Inertialization, priorityOrder=3}
   AnimStateTransitionNode_9 = AnimStateTransitionNode() {crossfadeDuration=0.4, logicType=TLT_Inertialization, priorityOrder=2}
  ]
  links [
   {
    dst=LeftFootPass
    rule=AnimGraph.AnimGraphNode_StateMachine_0.Locomotion.AnimStateNode_2.WalkToIdle.AnimGraphNode_StateMachine_0.Stop.AnimStateNode_1.Angry.AnimGraphNode_StateMachine_0.AngryStop.AnimStateTransitionNode_7.Transition
    src=Conduit
    via=AnimStateTransitionNode_7
   }
   {
    dst=LeftFootPlant
    rule=AnimGraph.AnimGraphNode_StateMachine_0.Locomotion.AnimStateNode_2.WalkToIdle.AnimGraphNode_StateMachine_0.Stop.AnimStateNode_1.Angry.AnimGraphNode_StateMachine_0.AngryStop.AnimStateTransitionNode_9.Transition
    src=Conduit
    via=AnimStateTransitionNode_9
   }
   {
    dst=RightFootPass
    rule=AnimGraph.AnimGraphNode_StateMachine_0.Locomotion.AnimStateNode_2.WalkToIdle.AnimGraphNode_StateMachine_0.Stop.AnimStateNode_1.Angry.AnimGraphNode_StateMachine_0.AngryStop.AnimStateTransitionNode_8.Transition
    src=Conduit
    via=AnimStateTransitionNode_8
   }
   {
    dst=RightFootPlant
    rule=AnimGraph.AnimGraphNode_StateMachine_0.Locomotion.AnimStateNode_2.WalkToIdle.AnimGraphNode_StateMachine_0.Stop.AnimStateNode_1.Angry.AnimGraphNode_StateMachine_0.AngryStop.AnimStateTransitionNode_10.Transition
    src=Conduit
    via=AnimStateTransitionNode_10
   }
  ]
 }
 Conduit {
  path=AnimGraph.AnimGraphNode_StateMachine_0.Locomotion.AnimStateNode_2.WalkToIdle.AnimGraphNode_StateMachine_0.Stop.AnimStateNode_1.Angry.AnimGraphNode_StateMachine_0.AngryStop.AnimStateConduitNode_0.Conduit
  nodes [Result(bCanEnterTransition=true)]
 }
 LeftFootPass {
  path=AnimGraph.AnimGraphNode_StateMachine_0.Locomotion.AnimStateNode_2.WalkToIdle.AnimGraphNode_StateMachine_0.Stop.AnimStateNode_1.Angry.AnimGraphNode_StateMachine_0.AngryStop.AnimStateNode_1.LeftFootPass
  nodes [Output Animation Pose(Result=Sequence Player 'ButterCake_Walk_Hostile_Stop_LeftFootPassing'() {bLoopAnimation=false, sequence=&/ButterCake_Walk_Hostile_Stop_LeftFootPassing})]
 }
 RightFootPass {
  path=AnimGraph.AnimGraphNode_StateMachine_0.Locomotion.AnimStateNode_2.WalkToIdle.AnimGraphNode_StateMachine_0.Stop.AnimStateNode_1.Angry.AnimGraphNode_StateMachine_0.AngryStop.AnimStateNode_2.RightFootPass
  nodes [Output Animation Pose(Result=Sequence Player 'ButterCake_Walk_Hostile_Stop_RightFootPassing'() {bLoopAnimation=false, sequence=&/ButterCake_Walk_Hostile_Stop_RightFootPassing})]
 }
 LeftFootPlant {
  path=AnimGraph.AnimGraphNode_StateMachine_0.Locomotion.AnimStateNode_2.WalkToIdle.AnimGraphNode_StateMachine_0.Stop.AnimStateNode_1.Angry.AnimGraphNode_StateMachine_0.AngryStop.AnimStateNode_3.LeftFootPlant
  nodes [Output Animation Pose(Result=Sequence Player 'ButterCake_Walk_Hostile_Stop_LeftFootPlant'() {bLoopAnimation=false, sequence=&/ButterCake_Walk_Hostile_Stop_LeftFootPlant})]
 }
 RightFootPlant {
  path=AnimGraph.AnimGraphNode_StateMachine_0.Locomotion.AnimStateNode_2.WalkToIdle.AnimGraphNode_StateMachine_0.Stop.AnimStateNode_1.Angry.AnimGraphNode_StateMachine_0.AngryStop.AnimStateNode_4.RightFootPlant
  nodes [Output Animation Pose(Result=Sequence Player 'ButterCake_Walk_Hostile_Stop_RightFootPlanting'() {bLoopAnimation=false, sequence=&/ButterCake_Walk_Hostile_Stop_RightFootPlanting})]
 }
 Transition {
  path=AnimGraph.AnimGraphNode_StateMachine_0.Locomotion.AnimStateNode_2.WalkToIdle.AnimGraphNode_StateMachine_0.Stop.AnimStateNode_1.Angry.AnimGraphNode_StateMachine_0.AngryStop.AnimStateTransitionNode_10.Transition
  nodes [Result(bCanEnterTransition=Get bFootPhaseIsHostileRightPlant() -> (bFootPhaseIsHostileRightPlant:bool))]
 }
 Transition {
  path=AnimGraph.AnimGraphNode_StateMachine_0.Locomotion.AnimStateNode_2.WalkToIdle.AnimGraphNode_StateMachine_0.Stop.AnimStateNode_1.Angry.AnimGraphNode_StateMachine_0.AngryStop.AnimStateTransitionNode_7.Transition
  nodes [Result(bCanEnterTransition=Get bFootPhaseIsHostileLeftPass() -> (bFootPhaseIsHostileLeftPass:bool))]
 }
 Transition {
  path=AnimGraph.AnimGraphNode_StateMachine_0.Locomotion.AnimStateNode_2.WalkToIdle.AnimGraphNode_StateMachine_0.Stop.AnimStateNode_1.Angry.AnimGraphNode_StateMachine_0.AngryStop.AnimStateTransitionNode_8.Transition
  nodes [Result(bCanEnterTransition=Get bFootPhaseIsHostileRightPass() -> (bFootPhaseIsHostileRightPass:bool))]
 }
 Transition {
  path=AnimGraph.AnimGraphNode_StateMachine_0.Locomotion.AnimStateNode_2.WalkToIdle.AnimGraphNode_StateMachine_0.Stop.AnimStateNode_1.Angry.AnimGraphNode_StateMachine_0.AngryStop.AnimStateTransitionNode_9.Transition
  nodes [Result(bCanEnterTransition=Get bFootPhaseIsHostileLeftPlant() -> (bFootPhaseIsHostileLeftPlant:bool))]
 }
 Neutral {
  path=AnimGraph.AnimGraphNode_StateMachine_0.Locomotion.AnimStateNode_2.WalkToIdle.AnimGraphNode_StateMachine_0.Stop.AnimStateNode_7.Neutral
  nodes [
   AnimGraphNode_StateMachine_0 = NeutralStop() {is=title AnimGraphNode_StateMachine, bAllowConduitEntryStates=true}
   Output Animation Pose(Result=@AnimGraphNode_StateMachine_0)
  ]
 }
 NeutralStop {
  entry=Conduit
  path=AnimGraph.AnimGraphNode_StateMachine_0.Locomotion.AnimStateNode_2.WalkToIdle.AnimGraphNode_StateMachine_0.Stop.AnimStateNode_7.Neutral.AnimGraphNode_StateMachine_0.NeutralStop
  ids=id
  nodes [
   AnimStateConduitNode_0 = Conduit()
   AnimStateNode_1 = LeftFootPlantB() {is=title AnimStateNode}
   AnimStateNode_10 = BothPlantC() {is=title AnimStateNode}
   AnimStateNode_11 = BothPlantD() {is=title AnimStateNode}
   AnimStateNode_2 = RightFootPlantC() {is=title AnimStateNode}
   AnimStateNode_3 = RightFootPlantB() {is=title AnimStateNode}
   AnimStateNode_4 = LeftFootPlantA() {is=title AnimStateNode}
   AnimStateNode_5 = RightFootPlantA() {is=title AnimStateNode}
   AnimStateNode_6 = LeftFootPlantC() {is=title AnimStateNode}
   AnimStateNode_7 = BothPlant() {is=title AnimStateNode}
   AnimStateNode_8 = BothPlantA() {is=title AnimStateNode}
   AnimStateNode_9 = BothPlantB() {is=title AnimStateNode}
   AnimStateTransitionNode_0 = AnimStateTransitionNode() {crossfadeDuration=0.45, logicType=TLT_Inertialization, priorityOrder=4}
   AnimStateTransitionNode_1 = AnimStateTransitionNode() {crossfadeDuration=0.45, logicType=TLT_Inertialization}
   AnimStateTransitionNode_10 = AnimStateTransitionNode() {crossfadeDuration=0.45, logicType=TLT_Inertialization, priorityOrder=9}
   AnimStateTransitionNode_11 = AnimStateTransitionNode() {crossfadeDuration=0.45, logicType=TLT_Inertialization, priorityOrder=10}
   AnimStateTransitionNode_2 = AnimStateTransitionNode() {crossfadeDuration=0.45, logicType=TLT_Inertialization, priorityOrder=3}
   AnimStateTransitionNode_3 = AnimStateTransitionNode() {crossfadeDuration=0.45, logicType=TLT_Inertialization, priorityOrder=8}
   AnimStateTransitionNode_4 = AnimStateTransitionNode() {crossfadeDuration=0.45, logicType=TLT_Inertialization, priorityOrder=2}
   AnimStateTransitionNode_5 = AnimStateTransitionNode() {crossfadeDuration=0.45, logicType=TLT_Inertialization, priorityOrder=7}
   AnimStateTransitionNode_6 = AnimStateTransitionNode() {crossfadeDuration=0.45, logicType=TLT_Inertialization, priorityOrder=3}
   AnimStateTransitionNode_7 = AnimStateTransitionNode() {crossfadeDuration=0.45, logicType=TLT_Inertialization, priorityOrder=100}
   AnimStateTransitionNode_9 = AnimStateTransitionNode() {crossfadeDuration=0.45, logicType=TLT_Inertialization, priorityOrder=5}
  ]
  links [
   {
    dst=BothPlant
    rule=AnimGraph.AnimGraphNode_StateMachine_0.Locomotion.AnimStateNode_2.WalkToIdle.AnimGraphNode_StateMachine_0.Stop.AnimStateNode_7.Neutral.AnimGraphNode_StateMachine_0.NeutralStop.AnimStateTransitionNode_7.Transition
    src=Conduit
    via=AnimStateTransitionNode_7
   }
   {
    dst=BothPlantA
    rule=AnimGraph.AnimGraphNode_StateMachine_0.Locomotion.AnimStateNode_2.WalkToIdle.AnimGraphNode_StateMachine_0.Stop.AnimStateNode_7.Neutral.AnimGraphNode_StateMachine_0.NeutralStop.AnimStateTransitionNode_0.Transition
    src=Conduit
    via=AnimStateTransitionNode_0
   }
   {
    dst=BothPlantB
    rule=AnimGraph.AnimGraphNode_StateMachine_0.Locomotion.AnimStateNode_2.WalkToIdle.AnimGraphNode_StateMachine_0.Stop.AnimStateNode_7.Neutral.AnimGraphNode_StateMachine_0.NeutralStop.AnimStateTransitionNode_9.Transition
    src=Conduit
    via=AnimStateTransitionNode_9
   }
   {
    dst=BothPlantC
    rule=AnimGraph.AnimGraphNode_StateMachine_0.Locomotion.AnimStateNode_2.WalkToIdle.AnimGraphNode_StateMachine_0.Stop.AnimStateNode_7.Neutral.AnimGraphNode_StateMachine_0.NeutralStop.AnimStateTransitionNode_10.Transition
    src=Conduit
    via=AnimStateTransitionNode_10
   }
   {
    dst=BothPlantD
    rule=AnimGraph.AnimGraphNode_StateMachine_0.Locomotion.AnimStateNode_2.WalkToIdle.AnimGraphNode_StateMachine_0.Stop.AnimStateNode_7.Neutral.AnimGraphNode_StateMachine_0.NeutralStop.AnimStateTransitionNode_11.Transition
    src=Conduit
    via=AnimStateTransitionNode_11
   }
   {
    dst=LeftFootPlantA
    rule=AnimGraph.AnimGraphNode_StateMachine_0.Locomotion.AnimStateNode_2.WalkToIdle.AnimGraphNode_StateMachine_0.Stop.AnimStateNode_7.Neutral.AnimGraphNode_StateMachine_0.NeutralStop.AnimStateTransitionNode_1.AnimationTransitionGraph_0
    src=Conduit
    via=AnimStateTransitionNode_1
   }
   {
    dst=LeftFootPlantB
    rule=AnimGraph.AnimGraphNode_StateMachine_0.Locomotion.AnimStateNode_2.WalkToIdle.AnimGraphNode_StateMachine_0.Stop.AnimStateNode_7.Neutral.AnimGraphNode_StateMachine_0.NeutralStop.AnimStateTransitionNode_4.AnimationTransitionGraph_0
    src=Conduit
    via=AnimStateTransitionNode_4
   }
   {
    dst=LeftFootPlantC
    rule=AnimGraph.AnimGraphNode_StateMachine_0.Locomotion.AnimStateNode_2.WalkToIdle.AnimGraphNode_StateMachine_0.Stop.AnimStateNode_7.Neutral.AnimGraphNode_StateMachine_0.NeutralStop.AnimStateTransitionNode_6.AnimationTransitionGraph_0
    src=Conduit
    via=AnimStateTransitionNode_6
   }
   {
    dst=RightFootPlantA
    rule=AnimGraph.AnimGraphNode_StateMachine_0.Locomotion.AnimStateNode_2.WalkToIdle.AnimGraphNode_StateMachine_0.Stop.AnimStateNode_7.Neutral.AnimGraphNode_StateMachine_0.NeutralStop.AnimStateTransitionNode_2.AnimationTransitionGraph_0
    src=Conduit
    via=AnimStateTransitionNode_2
   }
   {
    dst=RightFootPlantB
    rule=AnimGraph.AnimGraphNode_StateMachine_0.Locomotion.AnimStateNode_2.WalkToIdle.AnimGraphNode_StateMachine_0.Stop.AnimStateNode_7.Neutral.AnimGraphNode_StateMachine_0.NeutralStop.AnimStateTransitionNode_5.AnimationTransitionGraph_0
    src=Conduit
    via=AnimStateTransitionNode_5
   }
   {
    dst=RightFootPlantC
    rule=AnimGraph.AnimGraphNode_StateMachine_0.Locomotion.AnimStateNode_2.WalkToIdle.AnimGraphNode_StateMachine_0.Stop.AnimStateNode_7.Neutral.AnimGraphNode_StateMachine_0.NeutralStop.AnimStateTransitionNode_3.AnimationTransitionGraph_0
    src=Conduit
    via=AnimStateTransitionNode_3
   }
  ]
 }
 Conduit {
  path=AnimGraph.AnimGraphNode_StateMachine_0.Locomotion.AnimStateNode_2.WalkToIdle.AnimGraphNode_StateMachine_0.Stop.AnimStateNode_7.Neutral.AnimGraphNode_StateMachine_0.NeutralStop.AnimStateConduitNode_0.Conduit
  nodes [Result(bCanEnterTransition=true)]
 }
 LeftFootPlantB {
  path=AnimGraph.AnimGraphNode_StateMachine_0.Locomotion.AnimStateNode_2.WalkToIdle.AnimGraphNode_StateMachine_0.Stop.AnimStateNode_7.Neutral.AnimGraphNode_StateMachine_0.NeutralStop.AnimStateNode_1.LeftFootPlantB
  nodes [Output Animation Pose(Result=Sequence Player 'ButterCake_Walk_Stop_Right_51'() {bLoopAnimation=false, sequence=&/ButterCake_Walk_Stop_Right_51})]
 }
 BothPlantC {
  path=AnimGraph.AnimGraphNode_StateMachine_0.Locomotion.AnimStateNode_2.WalkToIdle.AnimGraphNode_StateMachine_0.Stop.AnimStateNode_7.Neutral.AnimGraphNode_StateMachine_0.NeutralStop.AnimStateNode_10.BothPlantC
  nodes [Output Animation Pose(Result=Sequence Player 'ButterCake_Walk_Stop_Right_72'() {bLoopAnimation=false, sequence=&/ButterCake_Walk_Stop_Right_72})]
 }
 BothPlantD {
  path=AnimGraph.AnimGraphNode_StateMachine_0.Locomotion.AnimStateNode_2.WalkToIdle.AnimGraphNode_StateMachine_0.Stop.AnimStateNode_7.Neutral.AnimGraphNode_StateMachine_0.NeutralStop.AnimStateNode_11.BothPlantD
  nodes [Output Animation Pose(Result=Sequence Player 'ButterCake_Walk_Stop_Right_85'() {bLoopAnimation=false, sequence=&/ButterCake_Walk_Stop_Right_85})]
 }
 RightFootPlantC {
  path=AnimGraph.AnimGraphNode_StateMachine_0.Locomotion.AnimStateNode_2.WalkToIdle.AnimGraphNode_StateMachine_0.Stop.AnimStateNode_7.Neutral.AnimGraphNode_StateMachine_0.NeutralStop.AnimStateNode_2.RightFootPlantC
  nodes [Output Animation Pose(Result=Sequence Player 'ButterCake_Walk_Stop_Left_14'() {bLoopAnimation=false, sequence=&/ButterCake_Walk_Stop_Left_14})]
 }
 RightFootPlantB {
  path=AnimGraph.AnimGraphNode_StateMachine_0.Locomotion.AnimStateNode_2.WalkToIdle.AnimGraphNode_StateMachine_0.Stop.AnimStateNode_7.Neutral.AnimGraphNode_StateMachine_0.NeutralStop.AnimStateNode_3.RightFootPlantB
  nodes [Output Animation Pose(Result=Sequence Player 'ButterCake_Walk_Stop_Left_7'() {bLoopAnimation=false, sequence=&/ButterCake_Walk_Stop_Left_7})]
 }
 LeftFootPlantA {
  path=AnimGraph.AnimGraphNode_StateMachine_0.Locomotion.AnimStateNode_2.WalkToIdle.AnimGraphNode_StateMachine_0.Stop.AnimStateNode_7.Neutral.AnimGraphNode_StateMachine_0.NeutralStop.AnimStateNode_4.LeftFootPlantA
  nodes [Output Animation Pose(Result=Sequence Player 'ButterCake_Walk_Stop_Right_42'() {bLoopAnimation=false, sequence=&/ButterCake_Walk_Stop_Right_42})]
 }
 RightFootPlantA {
  path=AnimGraph.AnimGraphNode_StateMachine_0.Locomotion.AnimStateNode_2.WalkToIdle.AnimGraphNode_StateMachine_0.Stop.AnimStateNode_7.Neutral.AnimGraphNode_StateMachine_0.NeutralStop.AnimStateNode_5.RightFootPlantA
  nodes [Output Animation Pose(Result=Sequence Player 'ButterCake_Walk_Stop_Left_0'() {bLoopAnimation=false, sequence=&/ButterCake_Walk_Stop_Left_0})]
 }
 LeftFootPlantC {
  path=AnimGraph.AnimGraphNode_StateMachine_0.Locomotion.AnimStateNode_2.WalkToIdle.AnimGraphNode_StateMachine_0.Stop.AnimStateNode_7.Neutral.AnimGraphNode_StateMachine_0.NeutralStop.AnimStateNode_6.LeftFootPlantC
  nodes [Output Animation Pose(Result=Sequence Player 'ButterCake_Walk_Stop_Right_63'!(StartPosition:float=0) -> (Pose) {bLoopAnimation=false, sequence=&/ButterCake_Walk_Stop_Right_63})]
 }
 BothPlant {
  path=AnimGraph.AnimGraphNode_StateMachine_0.Locomotion.AnimStateNode_2.WalkToIdle.AnimGraphNode_StateMachine_0.Stop.AnimStateNode_7.Neutral.AnimGraphNode_StateMachine_0.NeutralStop.AnimStateNode_7.BothPlant
  nodes [Output Animation Pose(Result=Sequence Player 'ButterCake_Walk_Stop_25'() {bLoopAnimation=false, sequence=&/ButterCake_Walk_Stop_25})]
 }
 BothPlantA {
  path=AnimGraph.AnimGraphNode_StateMachine_0.Locomotion.AnimStateNode_2.WalkToIdle.AnimGraphNode_StateMachine_0.Stop.AnimStateNode_7.Neutral.AnimGraphNode_StateMachine_0.NeutralStop.AnimStateNode_8.BothPlantA
  nodes [Output Animation Pose(Result=Sequence Player 'ButterCake_Walk_Stop_Right_31'() {bLoopAnimation=false, sequence=&/ButterCake_Walk_Stop_Right_31})]
 }
 BothPlantB {
  path=AnimGraph.AnimGraphNode_StateMachine_0.Locomotion.AnimStateNode_2.WalkToIdle.AnimGraphNode_StateMachine_0.Stop.AnimStateNode_7.Neutral.AnimGraphNode_StateMachine_0.NeutralStop.AnimStateNode_9.BothPlantB
  nodes [Output Animation Pose(Result=Sequence Player 'ButterCake_Walk_Stop_Right_37'() {bLoopAnimation=false, sequence=&/ButterCake_Walk_Stop_Right_37})]
 }
 Transition {
  path=AnimGraph.AnimGraphNode_StateMachine_0.Locomotion.AnimStateNode_2.WalkToIdle.AnimGraphNode_StateMachine_0.Stop.AnimStateNode_7.Neutral.AnimGraphNode_StateMachine_0.NeutralStop.AnimStateTransitionNode_0.Transition
  nodes [Result(bCanEnterTransition=Get bFootPhaseIsBothPlantA() -> (bFootPhaseIsBothPlantA:bool))]
 }
 AnimationTransitionGraph_0 {
  path=AnimGraph.AnimGraphNode_StateMachine_0.Locomotion.AnimStateNode_2.WalkToIdle.AnimGraphNode_StateMachine_0.Stop.AnimStateNode_7.Neutral.AnimGraphNode_StateMachine_0.NeutralStop.AnimStateTransitionNode_1.AnimationTransitionGraph_0
  nodes [Result(bCanEnterTransition=Get bFootPhaseIsLeftPlantA() -> (bFootPhaseIsLeftPlantA:bool))]
 }
 Transition {
  path=AnimGraph.AnimGraphNode_StateMachine_0.Locomotion.AnimStateNode_2.WalkToIdle.AnimGraphNode_StateMachine_0.Stop.AnimStateNode_7.Neutral.AnimGraphNode_StateMachine_0.NeutralStop.AnimStateTransitionNode_10.Transition
  nodes [Result(bCanEnterTransition=Get bFootPhaseIsBothPlantC() -> (bFootPhaseIsBothPlantC:bool))]
 }
 Transition {
  path=AnimGraph.AnimGraphNode_StateMachine_0.Locomotion.AnimStateNode_2.WalkToIdle.AnimGraphNode_StateMachine_0.Stop.AnimStateNode_7.Neutral.AnimGraphNode_StateMachine_0.NeutralStop.AnimStateTransitionNode_11.Transition
  nodes [Result(bCanEnterTransition=Get bFootPhaseIsBothPlantD() -> (bFootPhaseIsBothPlantD:bool))]
 }
 AnimationTransitionGraph_0 {
  path=AnimGraph.AnimGraphNode_StateMachine_0.Locomotion.AnimStateNode_2.WalkToIdle.AnimGraphNode_StateMachine_0.Stop.AnimStateNode_7.Neutral.AnimGraphNode_StateMachine_0.NeutralStop.AnimStateTransitionNode_2.AnimationTransitionGraph_0
  nodes [Result(bCanEnterTransition=Get bFootPhaseIsRightPlantA() -> (bFootPhaseIsRightPlantA:bool))]
 }
 AnimationTransitionGraph_0 {
  path=AnimGraph.AnimGraphNode_StateMachine_0.Locomotion.AnimStateNode_2.WalkToIdle.AnimGraphNode_StateMachine_0.Stop.AnimStateNode_7.Neutral.AnimGraphNode_StateMachine_0.NeutralStop.AnimStateTransitionNode_3.AnimationTransitionGraph_0
  nodes [Result(bCanEnterTransition=Get bFootPhaseIsRightPlantC() -> (bFootPhaseIsRightPlantC:bool))]
 }
 AnimationTransitionGraph_0 {
  path=AnimGraph.AnimGraphNode_StateMachine_0.Locomotion.AnimStateNode_2.WalkToIdle.AnimGraphNode_StateMachine_0.Stop.AnimStateNode_7.Neutral.AnimGraphNode_StateMachine_0.NeutralStop.AnimStateTransitionNode_4.AnimationTransitionGraph_0
  nodes [Result(bCanEnterTransition=Get bFootPhaseIsLeftPlantB() -> (bFootPhaseIsLeftPlantB:bool))]
 }
 AnimationTransitionGraph_0 {
  path=AnimGraph.AnimGraphNode_StateMachine_0.Locomotion.AnimStateNode_2.WalkToIdle.AnimGraphNode_StateMachine_0.Stop.AnimStateNode_7.Neutral.AnimGraphNode_StateMachine_0.NeutralStop.AnimStateTransitionNode_5.AnimationTransitionGraph_0
  nodes [Result(bCanEnterTransition=Get bFootPhaseIsRightPlantB() -> (bFootPhaseIsRightPlantB:bool))]
 }
 AnimationTransitionGraph_0 {
  path=AnimGraph.AnimGraphNode_StateMachine_0.Locomotion.AnimStateNode_2.WalkToIdle.AnimGraphNode_StateMachine_0.Stop.AnimStateNode_7.Neutral.AnimGraphNode_StateMachine_0.NeutralStop.AnimStateTransitionNode_6.AnimationTransitionGraph_0
  nodes [Result(bCanEnterTransition=Get bFootPhaseIsLeftPlantC() -> (bFootPhaseIsLeftPlantC:bool))]
 }
 Transition {
  path=AnimGraph.AnimGraphNode_StateMachine_0.Locomotion.AnimStateNode_2.WalkToIdle.AnimGraphNode_StateMachine_0.Stop.AnimStateNode_7.Neutral.AnimGraphNode_StateMachine_0.NeutralStop.AnimStateTransitionNode_7.Transition
  nodes [Result(bCanEnterTransition=Get bFootPhaseIsBothPlant() -> (bFootPhaseIsBothPlant:bool))]
 }
 Transition {
  path=AnimGraph.AnimGraphNode_StateMachine_0.Locomotion.AnimStateNode_2.WalkToIdle.AnimGraphNode_StateMachine_0.Stop.AnimStateNode_7.Neutral.AnimGraphNode_StateMachine_0.NeutralStop.AnimStateTransitionNode_9.Transition
  nodes [Result(bCanEnterTransition=Get bFootPhaseIsBothPlantB() -> (bFootPhaseIsBothPlantB:bool))]
 }
 Transition {
  path=AnimGraph.AnimGraphNode_StateMachine_0.Locomotion.AnimStateNode_2.WalkToIdle.AnimGraphNode_StateMachine_0.Stop.AnimStateTransitionNode_5.Transition
  nodes [Result(bCanEnterTransition=Get bIsHostile())]
 }
 Transition {
  path=AnimGraph.AnimGraphNode_StateMachine_0.Locomotion.AnimStateNode_2.WalkToIdle.AnimGraphNode_StateMachine_0.Stop.AnimStateTransitionNode_9.Transition
  nodes [Result(bCanEnterTransition=true)]
 }
 Transition {
  path=AnimGraph.AnimGraphNode_StateMachine_0.Locomotion.AnimStateTransitionNode_0.Transition
  nodes [Result(bCanEnterTransition=Get bIsMoving())]
 }
 Transition {
  path=AnimGraph.AnimGraphNode_StateMachine_0.Locomotion.AnimStateTransitionNode_1.Transition
  nodes [Result(bCanEnterTransition=Get bLocomotion_Conduit_NOTIsMovingANDNOTIsLocomotion() -> (bLocomotion_Conduit_NOTIsMovingANDNOTIsLocomotion:bool))]
 }
 Transition {path=AnimGraph.AnimGraphNode_StateMachine_0.Locomotion.AnimStateTransitionNode_2.Transition, nodes [Result()]}
 Transition {
  path=AnimGraph.AnimGraphNode_StateMachine_0.Locomotion.AnimStateTransitionNode_3.Transition
  nodes [Result(bCanEnterTransition=Get bIsMoving())]
 }
 Transition {path=AnimGraph.AnimGraphNode_StateMachine_0.Locomotion.AnimStateTransitionNode_4.Transition, nodes [Result(bCanEnterTransition=true)]}
 Transition {
  path=AnimGraph.AnimGraphNode_StateMachine_0.Locomotion.AnimStateTransitionNode_5.Transition
  nodes [Result(bCanEnterTransition=Get bLocomotion_Conduit_IsMovingORIsLocomotion() -> (bLocomotion_Conduit_IsMovingORIsLocomotion:bool))]
 }
 Facial Additive {
  entry=Conduit
  path=AnimGraph.AnimGraphNode_StateMachine_1.Facial Additive
  ids=id
  nodes [
   AnimStateConduitNode_0 = Conduit()
   AnimStateNode_0 = Neutral() {
    bAlwaysResetOnEntry=true
    stateEntered {guid=#1}
    stateFullyBlended {guid=#2}
    stateLeft {guid=#3}
   }
   AnimStateNode_1 = Angry() {
    bAlwaysResetOnEntry=true
    stateEntered {guid=#4}
    stateFullyBlended {guid=#5}
    stateLeft {guid=#6}
   }
   AnimStateNode_5 = Happy() {is=title AnimStateNode, bAlwaysResetOnEntry=true}
   AnimStateTransitionNode_0 = AnimStateTransitionNode() {crossfadeDuration=0.5, logicType=TLT_Inertialization}
   AnimStateTransitionNode_1 = AnimStateTransitionNode() {
    crossfadeDuration=0.5
    logicType=TLT_Inertialization
    transitionEnd {guid=#7}
    transitionInterrupt {guid=#8}
    transitionStart {guid=#9}
   }
   AnimStateTransitionNode_3 = AnimStateTransitionNode() {crossfadeDuration=0.5, logicType=TLT_Inertialization}
   AnimStateTransitionNode_4 = AnimStateTransitionNode() {crossfadeDuration=0.5, logicType=TLT_Inertialization, priorityOrder=100}
   AnimStateTransitionNode_6 = AnimStateTransitionNode() {crossfadeDuration=0.5, logicType=TLT_Inertialization}
   AnimStateTransitionNode_7 = AnimStateTransitionNode() {crossfadeDuration=0.4, logicType=TLT_Inertialization, priorityOrder=2}
   AnimStateTransitionNode_9 = AnimStateTransitionNode() {crossfadeDuration=0.5, logicType=TLT_Inertialization}
  ]
  links [
   {
    dst=Neutral
    rule=AnimGraph.AnimGraphNode_StateMachine_1.Facial Additive.AnimStateTransitionNode_1.Transition
    src=Angry
    via=AnimStateTransitionNode_1
   }
   {
    dst=Angry
    rule=AnimGraph.AnimGraphNode_StateMachine_1.Facial Additive.AnimStateTransitionNode_6.Transition
    src=Conduit
    via=AnimStateTransitionNode_6
   }
   {
    dst=Happy
    rule=AnimGraph.AnimGraphNode_StateMachine_1.Facial Additive.AnimStateTransitionNode_7.Transition
    src=Conduit
    via=AnimStateTransitionNode_7
   }
   {
    dst=Neutral
    rule=AnimGraph.AnimGraphNode_StateMachine_1.Facial Additive.AnimStateTransitionNode_4.Transition
    src=Conduit
    via=AnimStateTransitionNode_4
   }
   {
    dst=Neutral
    rule=AnimGraph.AnimGraphNode_StateMachine_1.Facial Additive.AnimStateTransitionNode_9.Transition
    src=Happy
    via=AnimStateTransitionNode_9
   }
   {
    dst=Angry
    rule=AnimGraph.AnimGraphNode_StateMachine_1.Facial Additive.AnimStateTransitionNode_0.Transition
    src=Neutral
    via=AnimStateTransitionNode_0
   }
   {
    dst=Happy
    rule=AnimGraph.AnimGraphNode_StateMachine_1.Facial Additive.AnimStateTransitionNode_3.Transition
    src=Neutral
    via=AnimStateTransitionNode_3
   }
  ]
 }
 Conduit {path=AnimGraph.AnimGraphNode_StateMachine_1.Facial Additive.AnimStateConduitNode_0.Conduit, nodes [Result(bCanEnterTransition=true)]}
 Neutral {
  path=AnimGraph.AnimGraphNode_StateMachine_1.Facial Additive.AnimStateNode_0.Neutral
  nodes [
   AnimGraphNode_StateMachine_0 = NeutralFacialAdditive() {is=title AnimGraphNode_StateMachine}
   Output Animation Pose(Result=@AnimGraphNode_StateMachine_0)
  ]
 }
 NeutralFacialAdditive {
  entry=Idle
  path=AnimGraph.AnimGraphNode_StateMachine_1.Facial Additive.AnimStateNode_0.Neutral.AnimGraphNode_StateMachine_0.NeutralFacialAdditive
  ids=id
  nodes [
   AnimStateNode_0 = Idle()
   AnimStateNode_1 = Moving() {is=title AnimStateNode}
   AnimStateTransitionNode_2 = AnimStateTransitionNode() {crossfadeDuration=0.5, logicType=TLT_Inertialization}
   AnimStateTransitionNode_3 = AnimStateTransitionNode() {crossfadeDuration=0.5, logicType=TLT_Inertialization}
  ]
  links [
   {
    dst=Moving
    rule=AnimGraph.AnimGraphNode_StateMachine_1.Facial Additive.AnimStateNode_0.Neutral.AnimGraphNode_StateMachine_0.NeutralFacialAdditive.AnimStateTransitionNode_2.Transition
    src=Idle
    via=AnimStateTransitionNode_2
   }
   {
    dst=Idle
    rule=AnimGraph.AnimGraphNode_StateMachine_1.Facial Additive.AnimStateNode_0.Neutral.AnimGraphNode_StateMachine_0.NeutralFacialAdditive.AnimStateTransitionNode_3.Transition
    src=Moving
    via=AnimStateTransitionNode_3
   }
  ]
 }
 Idle {
  path=AnimGraph.AnimGraphNode_StateMachine_1.Facial Additive.AnimStateNode_0.Neutral.AnimGraphNode_StateMachine_0.NeutralFacialAdditive.AnimStateNode_0.Idle
  nodes [Output Animation Pose(Result="Sequence Player (additive) 'ButterCake_Facial_Neutral_Idle'"() {sequence=&/ButterCake_Facial_Neutral_Idle})]
 }
 Moving {
  path=AnimGraph.AnimGraphNode_StateMachine_1.Facial Additive.AnimStateNode_0.Neutral.AnimGraphNode_StateMachine_0.NeutralFacialAdditive.AnimStateNode_1.Moving
  nodes [Output Animation Pose(Result="Sequence Player (additive) 'ButterCake_Facial_Neutral_Moving'"() {sequence=&/ButterCake_Facial_Neutral_Moving})]
 }
 Transition {
  path=AnimGraph.AnimGraphNode_StateMachine_1.Facial Additive.AnimStateNode_0.Neutral.AnimGraphNode_StateMachine_0.NeutralFacialAdditive.AnimStateTransitionNode_2.Transition
  nodes [Result(bCanEnterTransition=Get bIsMoving())]
 }
 Transition {
  path=AnimGraph.AnimGraphNode_StateMachine_1.Facial Additive.AnimStateNode_0.Neutral.AnimGraphNode_StateMachine_0.NeutralFacialAdditive.AnimStateTransitionNode_3.Transition
  nodes [Result(bCanEnterTransition=NOT Boolean(A=Get bIsMoving()))]
 }
 Angry {
  path=AnimGraph.AnimGraphNode_StateMachine_1.Facial Additive.AnimStateNode_1.Angry
  nodes [
   AnimGraphNode_StateMachine_0 = AngryFacialAdditive() {is=title AnimGraphNode_StateMachine}
   Output Animation Pose(Result=@AnimGraphNode_StateMachine_0)
  ]
 }
 AngryFacialAdditive {
  entry=Idle
  path=AnimGraph.AnimGraphNode_StateMachine_1.Facial Additive.AnimStateNode_1.Angry.AnimGraphNode_StateMachine_0.AngryFacialAdditive
  ids=id
  nodes [
   AnimStateNode_0 = Idle()
   AnimStateNode_1 = Moving()
   AnimStateTransitionNode_4 = AnimStateTransitionNode() {crossfadeDuration=0.5, logicType=TLT_Inertialization}
   AnimStateTransitionNode_5 = AnimStateTransitionNode() {crossfadeDuration=0.5, logicType=TLT_Inertialization}
  ]
  links [
   {
    dst=Moving
    rule=AnimGraph.AnimGraphNode_StateMachine_1.Facial Additive.AnimStateNode_1.Angry.AnimGraphNode_StateMachine_0.AngryFacialAdditive.AnimStateTransitionNode_4.Transition
    src=Idle
    via=AnimStateTransitionNode_4
   }
   {
    dst=Idle
    rule=AnimGraph.AnimGraphNode_StateMachine_1.Facial Additive.AnimStateNode_1.Angry.AnimGraphNode_StateMachine_0.AngryFacialAdditive.AnimStateTransitionNode_5.Transition
    src=Moving
    via=AnimStateTransitionNode_5
   }
  ]
 }
 Idle {
  path=AnimGraph.AnimGraphNode_StateMachine_1.Facial Additive.AnimStateNode_1.Angry.AnimGraphNode_StateMachine_0.AngryFacialAdditive.AnimStateNode_0.Idle
  nodes [Output Animation Pose(Result="Sequence Player (additive) 'ButterCake_Facial_Idle_Hostile'"() {sequence=&/ButterCake_Facial_Idle_Hostile})]
 }
 Moving {
  path=AnimGraph.AnimGraphNode_StateMachine_1.Facial Additive.AnimStateNode_1.Angry.AnimGraphNode_StateMachine_0.AngryFacialAdditive.AnimStateNode_1.Moving
  nodes [Output Animation Pose(Result="Sequence Player (additive) 'ButterCake_Facial_Walk_Hostile_N'"() {sequence=&/ButterCake_Facial_Walk_Hostile_N})]
 }
 Transition {
  path=AnimGraph.AnimGraphNode_StateMachine_1.Facial Additive.AnimStateNode_1.Angry.AnimGraphNode_StateMachine_0.AngryFacialAdditive.AnimStateTransitionNode_4.Transition
  nodes [Result(bCanEnterTransition=Get bIsMoving())]
 }
 Transition {
  path=AnimGraph.AnimGraphNode_StateMachine_1.Facial Additive.AnimStateNode_1.Angry.AnimGraphNode_StateMachine_0.AngryFacialAdditive.AnimStateTransitionNode_5.Transition
  nodes [Result(bCanEnterTransition=NOT Boolean(A=Get bIsMoving()))]
 }
 Happy {
  path=AnimGraph.AnimGraphNode_StateMachine_1.Facial Additive.AnimStateNode_5.Happy
  nodes [
   AnimGraphNode_StateMachine_0 = HappyFacialTransition() {is=title AnimGraphNode_StateMachine, bAllowConduitEntryStates=true}
   Output Animation Pose(Result=@AnimGraphNode_StateMachine_0)
  ]
 }
 HappyFacialTransition {
  entry=Conduit
  path=AnimGraph.AnimGraphNode_StateMachine_1.Facial Additive.AnimStateNode_5.Happy.AnimGraphNode_StateMachine_0.HappyFacialTransition
  ids=id
  nodes [
   AnimStateConduitNode_0 = Conduit()
   AnimStateNode_0 = Intro() {is=title AnimStateNode}
   AnimStateNode_1 = Loop() {is=title AnimStateNode}
   AnimStateTransitionNode_0 = AnimStateTransitionNode() {bAutomaticRuleBasedOnSequencePlayerInState=true, crossfadeDuration=0.5, logicType=TLT_Inertialization}
   AnimStateTransitionNode_1 = AnimStateTransitionNode() {crossfadeDuration=0.5, logicType=TLT_Inertialization}
   AnimStateTransitionNode_2 = AnimStateTransitionNode() {crossfadeDuration=0.5, logicType=TLT_Inertialization, priorityOrder=100}
  ]
  links [
   {
    dst=Intro
    rule=AnimGraph.AnimGraphNode_StateMachine_1.Facial Additive.AnimStateNode_5.Happy.AnimGraphNode_StateMachine_0.HappyFacialTransition.AnimStateTransitionNode_2.Transition
    src=Conduit
    via=AnimStateTransitionNode_2
   }
   {
    dst=Loop
    rule=AnimGraph.AnimGraphNode_StateMachine_1.Facial Additive.AnimStateNode_5.Happy.AnimGraphNode_StateMachine_0.HappyFacialTransition.AnimStateTransitionNode_1.Transition
    src=Conduit
    via=AnimStateTransitionNode_1
   }
   {
    dst=Loop
    rule=AnimGraph.AnimGraphNode_StateMachine_1.Facial Additive.AnimStateNode_5.Happy.AnimGraphNode_StateMachine_0.HappyFacialTransition.AnimStateTransitionNode_0.Transition
    src=Intro
    via=AnimStateTransitionNode_0
   }
  ]
 }
 Conduit {
  path=AnimGraph.AnimGraphNode_StateMachine_1.Facial Additive.AnimStateNode_5.Happy.AnimGraphNode_StateMachine_0.HappyFacialTransition.AnimStateConduitNode_0.Conduit
  nodes [Result(bCanEnterTransition=true)]
 }
 Intro {
  path=AnimGraph.AnimGraphNode_StateMachine_1.Facial Additive.AnimStateNode_5.Happy.AnimGraphNode_StateMachine_0.HappyFacialTransition.AnimStateNode_0.Intro
  nodes [Output Animation Pose(Result="Sequence Player (additive) 'ButterCake_Facial_Neutral_To_Happy'"() {bLoopAnimation=false, sequence=&/ButterCake_Facial_Neutral_To_Happy})]
 }
 Loop {
  path=AnimGraph.AnimGraphNode_StateMachine_1.Facial Additive.AnimStateNode_5.Happy.AnimGraphNode_StateMachine_0.HappyFacialTransition.AnimStateNode_1.Loop
  nodes [Output Animation Pose(Result="Sequence Player (additive) 'ButterCake_Facial_Happy_Idle'"() {sequence=&/ButterCake_Facial_Happy_Idle})]
 }
 Transition {
  path=AnimGraph.AnimGraphNode_StateMachine_1.Facial Additive.AnimStateNode_5.Happy.AnimGraphNode_StateMachine_0.HappyFacialTransition.AnimStateTransitionNode_0.Transition
  nodes [Result()]
 }
 Transition {
  path=AnimGraph.AnimGraphNode_StateMachine_1.Facial Additive.AnimStateNode_5.Happy.AnimGraphNode_StateMachine_0.HappyFacialTransition.AnimStateTransitionNode_1.Transition
  nodes [Result(bCanEnterTransition=Get bIsMouthOpen() -> (bIsMouthOpen:bool))]
 }
 Transition {
  path=AnimGraph.AnimGraphNode_StateMachine_1.Facial Additive.AnimStateNode_5.Happy.AnimGraphNode_StateMachine_0.HappyFacialTransition.AnimStateTransitionNode_2.Transition
  nodes [Result(bCanEnterTransition=true)]
 }
 Transition {
  path=AnimGraph.AnimGraphNode_StateMachine_1.Facial Additive.AnimStateTransitionNode_0.Transition
  nodes [Result(bCanEnterTransition=Get bIsHostile())]
 }
 Transition {
  path=AnimGraph.AnimGraphNode_StateMachine_1.Facial Additive.AnimStateTransitionNode_1.Transition
  nodes [Result(bCanEnterTransition=NOT Boolean(A=Get bIsHostile()))]
 }
 Transition {
  path=AnimGraph.AnimGraphNode_StateMachine_1.Facial Additive.AnimStateTransitionNode_3.Transition
  nodes [Result(bCanEnterTransition=Get bIsLured())]
 }
 Transition {path=AnimGraph.AnimGraphNode_StateMachine_1.Facial Additive.AnimStateTransitionNode_4.Transition, nodes [Result(bCanEnterTransition=true)]}
 Transition {
  path=AnimGraph.AnimGraphNode_StateMachine_1.Facial Additive.AnimStateTransitionNode_6.Transition
  nodes [Result(bCanEnterTransition=Get bIsHostile())]
 }
 Transition {
  path=AnimGraph.AnimGraphNode_StateMachine_1.Facial Additive.AnimStateTransitionNode_7.Transition
  nodes [Result(bCanEnterTransition=Get bIsLured())]
 }
 Transition {
  path=AnimGraph.AnimGraphNode_StateMachine_1.Facial Additive.AnimStateTransitionNode_9.Transition
  nodes [Result(bCanEnterTransition=NOT Boolean(A=Get bIsLured()))]
 }
 EventGraph {links [], nodes []}
]
variables []
