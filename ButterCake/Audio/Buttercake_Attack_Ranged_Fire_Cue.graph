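# Sound cue for the firing sound of Buttercake's ranged attack, on the creature sound class with Buttercake attack attenuation.
# A quality-level node sends the first two levels to a four-variant random pool of launch waves and the last two to a three-variant pool, each through a modulator.
# Both random pools are excluded from branch culling.
# Virtualization is off, and a concurrency override allows two simultaneous instances, stopping the oldest.
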SoundCue /ButterCake/Audio/Buttercake_Attack_Ranged_Fire_Cue
root=QualityLevel_0
nodes [
 {id=Modulator_0, children [Random_0]}
 {id=Modulator_1, children [Random_1]}
 {id=QualityLevel_0, children [Modulator_0,Modulator_0,Modulator_1,Modulator_1]}
 {
  id=Random_0
  children [WavePlayer_3,WavePlayer_4,WavePlayer_5,WavePlayer_6]
  properties {bShouldExcludeFromBranchCulling=true, weights [1,1,1,1]}
 }
 {id=Random_1, children [WavePlayer_0,WavePlayer_1,WavePlayer_2], properties {bShouldExcludeFromBranchCulling=true, weights [1,1,1]}}
 {id=WavePlayer_0, properties {soundWaveAssetPtr=&/Buttercake_Attack_Ranged_Launch_03}}
 {id=WavePlayer_1, properties {soundWaveAssetPtr=&/Buttercake_Attack_Ranged_Launch_04}}
 {id=WavePlayer_2, properties {soundWaveAssetPtr=&/Buttercake_Attack_Ranged_Launch_01}}
 {id=WavePlayer_3, properties {soundWaveAssetPtr=&Buttercake_Attack_Ranged_Launch_03}}
 {id=WavePlayer_4, properties {soundWaveAssetPtr=&Buttercake_Attack_Ranged_Launch_04}}
 {id=WavePlayer_5, properties {soundWaveAssetPtr=&Buttercake_Attack_Ranged_Launch_01}}
 {id=WavePlayer_6, properties {soundWaveAssetPtr=&/Buttercake_Attack_Ranged_Launch_02}}
]
properties {
 attenuationSettings=&/Buttercake_Attacks_Att
 bOverrideConcurrency=true
 maxDistance=35833.332
 priority=20
 soundClassObject=/Game/Sounds/FortSoundClassesAndModes/Fort_Creatures
 virtualizationMode=Disabled
 volumeMultiplier=1.5
 concurrencyOverrides {maxCount=2, resolutionRule=StopOldest}
}
