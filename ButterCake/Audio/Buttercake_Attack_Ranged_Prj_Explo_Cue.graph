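# Sound cue for the explosion of the Buttercake ranged attack projectile.
# A mixer layers close, medium and distant branches, each with its own attenuation and sound class override; the distant branch adds a short random delay.
# In each branch a quality-level switch picks a four- or two-variant random wave set through pitch and volume modulators.
# Concurrency caps instances at four, stopping the oldest.
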SoundCue /ButterCake/Audio/Buttercake_Attack_Ranged_Prj_Explo_Cue
root=Mixer_5
nodes [
 {
  id=Attenuation_0
  children [QualityLevel_2]
  properties {
   attenuationSettings=/Game/Sounds/Attenuation/Fort_Explosion_Close_Athena_Att
   attenuationOverrides {falloffDistance=900, attenuationShapeExtents {x=100}}
  }
 }
 {id=Attenuation_4, children [QualityLevel_0], properties {attenuationSettings=/Game/Sounds/Attenuation/Fort_Explosion_Medium_Athena_Att}}
 {
  id=Attenuation_8
  children [Delay_0]
  properties {
   attenuationSettings=/Game/Sounds/Attenuation/Fort_Explosion_Distant_Athena_Att
   attenuationOverrides {falloffDistance=3000, attenuationShapeExtents {x=1000}}
  }
 }
 {id=Delay_0, children [QualityLevel_1], properties {delayMax=0.1}}
 {id=Mixer_5, children [SoundClass_3,SoundClass_0,SoundClass_1], properties {inputVolume [1.2,1,1]}}
 {id=Modulator_0, children [Random_2]}
 {id=Modulator_1, children [Random_4]}
 {id=Modulator_2, children [Random_6], properties {pitchMax=0.9, pitchMin=0.8}}
 {id=Modulator_3, children [Random_1], properties {pitchMax=1.1, pitchMin=0.9, volumeMax=1, volumeMin=1}}
 {id=Modulator_4, children [Random_8], properties {pitchMax=1.1, pitchMin=0.9, volumeMax=1, volumeMin=1}}
 {id=Modulator_7, children [Random_0], properties {pitchMax=0.9, pitchMin=0.8}}
 {id=QualityLevel_0, children [Modulator_0,Modulator_0,Modulator_1,Modulator_1]}
 {id=QualityLevel_1, children [Modulator_7,Modulator_7,Modulator_2,Modulator_2]}
 {id=QualityLevel_2, children [Modulator_3,Modulator_3,Modulator_4,Modulator_4]}
 {
  id=Random_0
  children [WavePlayer_0,WavePlayer_1,WavePlayer_2,WavePlayer_3]
  properties {bShouldExcludeFromBranchCulling=true, weights [1,1,1,1]}
 }
 {id=Random_1, children [WavePlayer_4,WavePlayer_5,WavePlayer_6,WavePlayer_7], properties {weights [1,1,1,1]}}
 {
  id=Random_2
  children [WavePlayer_8, WavePlayer_9, WavePlayer_10, WavePlayer_11]
  properties {bShouldExcludeFromBranchCulling=true, weights [1,1,1,1]}
 }
 {id=Random_4, children [WavePlayer_12, WavePlayer_13], properties {bShouldExcludeFromBranchCulling=true, weights [1,1]}}
 {id=Random_6, children [WavePlayer_16, WavePlayer_17], properties {bShouldExcludeFromBranchCulling=true, weights [1,1]}}
 {id=Random_8, children [WavePlayer_20, WavePlayer_21], properties {bShouldExcludeFromBranchCulling=true, weights [1,1]}}
 {
  id=SoundClass_0
  children [Attenuation_4]
  properties {soundClassOverride=&Fort_WeaponRanged_Fire_Distant}
 }
 {
  id=SoundClass_1
  children [Attenuation_8]
  properties {soundClassOverride=&Fort_WeaponRanged_Fire_Distant}
 }
 {
  id=SoundClass_3
  children [Attenuation_0]
  properties {soundClassOverride=&Fort_WeaponRanged_Fire_3P_PSM}
 }
 {id=WavePlayer_0, properties {soundWaveAssetPtr=&/Buttercake_Attack_Ranged_Prj_Explo_Far_04}}
 {id=WavePlayer_1, properties {soundWaveAssetPtr=&/Buttercake_Attack_Ranged_Prj_Explo_Far_01}}
 {id=WavePlayer_10, properties {soundWaveAssetPtr=&/Buttercake_Attack_Ranged_Prj_Explo_Mid_03}}
 {id=WavePlayer_11, properties {soundWaveAssetPtr=&/Buttercake_Attack_Ranged_Prj_Explo_Mid_04}}
 {id=WavePlayer_12, properties {soundWaveAssetPtr=&/Buttercake_Attack_Ranged_Prj_Explo_Mid_01}}
 {id=WavePlayer_13, properties {soundWaveAssetPtr=&/Buttercake_Attack_Ranged_Prj_Explo_Mid_02}}
 {id=WavePlayer_16, properties {soundWaveAssetPtr=&Buttercake_Attack_Ranged_Prj_Explo_Far_04}}
 {id=WavePlayer_17, properties {soundWaveAssetPtr=&Buttercake_Attack_Ranged_Prj_Explo_Far_01}}
 {id=WavePlayer_2, properties {soundWaveAssetPtr=&/Buttercake_Attack_Ranged_Prj_Explo_Far_02}}
 {id=WavePlayer_20, properties {soundWaveAssetPtr=&/Buttercake_Attack_Ranged_Prj_Explo_Close_01}}
 {id=WavePlayer_21, properties {soundWaveAssetPtr=&/Buttercake_Attack_Ranged_Prj_Explo_Close_02}}
 {id=WavePlayer_3, properties {soundWaveAssetPtr=&/Buttercake_Attack_Ranged_Prj_Explo_Far_03}}
 {id=WavePlayer_4, properties {soundWaveAssetPtr=&Buttercake_Attack_Ranged_Prj_Explo_Close_01}}
 {id=WavePlayer_5, properties {soundWaveAssetPtr=&Buttercake_Attack_Ranged_Prj_Explo_Close_02}}
 {id=WavePlayer_6, properties {soundWaveAssetPtr=&/Buttercake_Attack_Ranged_Prj_Explo_Close_03}}
 {id=WavePlayer_7, properties {soundWaveAssetPtr=&/Buttercake_Attack_Ranged_Prj_Explo_Close_04}}
 {id=WavePlayer_8, properties {soundWaveAssetPtr=&Buttercake_Attack_Ranged_Prj_Explo_Mid_01}}
 {id=WavePlayer_9, properties {soundWaveAssetPtr=&Buttercake_Attack_Ranged_Prj_Explo_Mid_02}}
]
properties {
 bOverrideConcurrency=true
 maxDistance=27500
 priority=20
 soundClassObject=/Game/Sounds/FortSoundClassesAndModes/Fort_Explosives_Large_PSM
 volumeMultiplier=0.3715
 attenuationOverrides {reverbWetLevelMax=0.7, reverbWetLevelMin=0.2}
 concurrencyOverrides {maxCount=4, resolutionRule=StopOldest, volumeScale=0.5}
}
