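# Sound cue for the looping in-air sound of ButterCake's ranged attack projectile, played in the creatures sound class.
# A looping wave gets randomized pitch from a Modulator, then passes through an intensified Doppler node, an Enveloper and a dedicated attenuation asset.
# The Enveloper curves fade volume in while pitch drops from a high start.
# Concurrency is overridden to allow few simultaneous instances, and virtualization is disabled.
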SoundCue /ButterCake/Audio/Buttercake_Attack_Ranged_Prj_InAir_Loop_Cue
root=Attenuation_0
nodes [
 {id=Attenuation_0, children [Enveloper_0], properties {attenuationSettings=&/Buttercake_Attack_Ranged_Prj_InAir_Loop_Att}}
 {id=Doppler_0, children [Modulator_1], properties {dopplerIntensity=1.5}}
 {
  id=Enveloper_0
  children [Doppler_0]
  properties {
   pitchCurve {editorCurveData {keys [{arriveTangent=0.4043, interpMode=RCIM_Cubic, leaveTangent=0.4043, tangentMode=RCTM_User, time=0, value=3}, {time=1}]}}
   volumeCurve {editorCurveData {keys [{arriveTangent=1.671, interpMode=RCIM_Cubic, leaveTangent=1.671, tangentMode=RCTM_User, time=0, value=0}, <at 1 1>]}}
  }
 }
 {id=Modulator_1, children [WavePlayer_8], properties {pitchMax=1.5, pitchMin=0.8}}
 {id=WavePlayer_8, properties {bLooping=true, soundWaveAssetPtr=&/Buttercake_Attack_Ranged_Prj_In_Air_Loop}}
]
properties {
 bOverrideConcurrency=true
 maxDistance=13500
 pitchMultiplier=2
 priority=20
 soundClassObject=/Game/Sounds/FortSoundClassesAndModes/Fort_Creatures
 virtualizationMode=Disabled
 volumeMultiplier=1.125
 concurrencyOverrides {maxCount=2}
}
