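# Sound cue for the close-range whip of Buttercake's ranged attack projectile.
# A quality-level node sends its first two levels to a random pick of six waves and its last two to a pick of three.
# Each random branch is pitched up by its own modulator and is excluded from branch culling.
# It references the projectile's in-air loop attenuation asset and the ranged weapon projectile effects sound class.
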SoundCue /ButterCake/Audio/Buttercake_Attack_Ranged_Prj_Whip_Close_Cue
root=QualityLevel_0
nodes [
 {id=Modulator_0, children [Random_1], properties {pitchMax=1.5, pitchMin=1.25}}
 {id=Modulator_1, children [Random_2], properties {pitchMax=1.5, pitchMin=1.25}}
 {id=QualityLevel_0, children [Modulator_0,Modulator_0,Modulator_1,Modulator_1]}
 {
  id=Random_1
  children [WavePlayer_3,WavePlayer_4,WavePlayer_5,WavePlayer_6,WavePlayer_7,WavePlayer_8]
  properties {bShouldExcludeFromBranchCulling=true, weights [1,1,1,1,1,1]}
 }
 {id=Random_2, children [WavePlayer_0,WavePlayer_1,WavePlayer_2], properties {bShouldExcludeFromBranchCulling=true, weights [1,1,1]}}
 {id=WavePlayer_0, properties {soundWaveAssetPtr=&/Buttercake_Attack_Ranged_Prj_Whip_Close_01}}
 {id=WavePlayer_1, properties {soundWaveAssetPtr=&/Buttercake_Attack_Ranged_Prj_Whip_Close_02}}
 {id=WavePlayer_2, properties {soundWaveAssetPtr=&/Buttercake_Attack_Ranged_Prj_Whip_Close_03}}
 {id=WavePlayer_3, properties {soundWaveAssetPtr=&Buttercake_Attack_Ranged_Prj_Whip_Close_01}}
 {id=WavePlayer_4, properties {soundWaveAssetPtr=&Buttercake_Attack_Ranged_Prj_Whip_Close_02}}
 {id=WavePlayer_5, properties {soundWaveAssetPtr=&Buttercake_Attack_Ranged_Prj_Whip_Close_03}}
 {id=WavePlayer_6, properties {soundWaveAssetPtr=&/Buttercake_Attack_Ranged_Prj_Whip_Close_04}}
 {id=WavePlayer_7, properties {soundWaveAssetPtr=&/Buttercake_Attack_Ranged_Prj_Whip_Close_05}}
 {id=WavePlayer_8, properties {soundWaveAssetPtr=&/Buttercake_Attack_Ranged_Prj_Whip_Close_06}}
]
properties {
 attenuationSettings=&/Buttercake_Attack_Ranged_Prj_InAir_Loop_Att
 maxDistance=13500
 priority=20
 soundClassObject=&Fort_WeaponRanged_Projectile_Effects
 attenuationOverrides {
  bAttenuate=false
  bEnableReverbSend=false
  bSpatialize=false
  bUseComplexCollisionForOcclusion=true
  falloffDistance=1000
  occlusionVolumeAttenuation=0
  stereoSpread=0
  attenuationShapeExtents {x=512}
 }
}
