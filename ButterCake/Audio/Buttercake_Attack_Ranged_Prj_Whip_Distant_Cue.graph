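# Sound cue for the distant whip sound of Buttercake's ranged-attack projectile, routed to the ranged weapon projectile effects sound class.
# A quality-level node sends its first two levels to a random pool of four "Far" waves and its last two to a pool of two.
# Each pool has equal weights and passes through a modulator that randomly raises pitch.
# A shared whip attenuation asset handles spatialization.
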SoundCue /ButterCake/Audio/Buttercake_Attack_Ranged_Prj_Whip_Distant_Cue
root=QualityLevel_0
nodes [
 {id=Modulator_0, children [Random_1], properties {pitchMax=1.25, pitchMin=1}}
 {id=Modulator_1, children [Random_2], properties {pitchMax=1.25, pitchMin=1}}
 {id=QualityLevel_0, children [Modulator_0,Modulator_0,Modulator_1,Modulator_1]}
 {
  id=Random_1
  children [WavePlayer_3,WavePlayer_4,WavePlayer_5,WavePlayer_6]
  properties {bShouldExcludeFromBranchCulling=true, weights [1,1,1,1]}
 }
 {id=Random_2, children [WavePlayer_0,WavePlayer_1], properties {bShouldExcludeFromBranchCulling=true, weights [1,1]}}
 {id=WavePlayer_0, properties {soundWaveAssetPtr=&/Buttercake_Attack_Ranged_Prj_Whip_Far_01}}
 {id=WavePlayer_1, properties {soundWaveAssetPtr=&/Buttercake_Attack_Ranged_Prj_Whip_Far_02}}
 {id=WavePlayer_3, properties {soundWaveAssetPtr=&Buttercake_Attack_Ranged_Prj_Whip_Far_01}}
 {id=WavePlayer_4, properties {soundWaveAssetPtr=&Buttercake_Attack_Ranged_Prj_Whip_Far_02}}
 {id=WavePlayer_5, properties {soundWaveAssetPtr=&/Buttercake_Attack_Ranged_Prj_Whip_Far_03}}
 {id=WavePlayer_6, properties {soundWaveAssetPtr=&/Buttercake_Attack_Ranged_Prj_Whip_Far_04}}
]
properties {
 attenuationSettings=&/Buttercake_Attack_Ranged_Prj_Whip_Att
 maxDistance=8500
 priority=20
 soundClassObject=&Fort_WeaponRanged_Projectile_Effects
 volumeMultiplier=0.938
 attenuationOverrides {
  bEnableReverbSend=false
  bUseComplexCollisionForOcclusion=true
  falloffDistance=1000
  occlusionVolumeAttenuation=0
  stereoSpread=0
  attenuationShapeExtents {x=512}
 }
}
