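# Sound cue named for the Buttercake creature's ranged attack tell, currently empty with no root or sound nodes, so it has nothing to play.
# It keeps its playback setup: shared Buttercake attack attenuation, the Fort_Creatures sound class, raised volume and priority, and virtualization disabled.
# A concurrency override limits how many instances can play at once.
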SoundCue /ButterCake/Audio/Buttercake_Attack_Ranged_Tell_Cue
root=null
nodes []
properties {
 attenuationSettings=&/Buttercake_Attacks_Att
 bOverrideConcurrency=true
 maxDistance=32000
 priority=20
 soundClassObject=/Game/Sounds/FortSoundClassesAndModes/Fort_Creatures
 virtualizationMode=Disabled
 volumeMultiplier=1.5
 concurrencyOverrides {maxCount=2}
}
