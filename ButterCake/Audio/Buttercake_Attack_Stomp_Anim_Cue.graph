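# Sound cue for Buttercake's stomp attack animation, playing rear ground impacts as a creature sound limited to one concurrent instance.
# Each variant mixes a louder, pitch-modulated impact with a far layer delayed by a fixed interval, each using its own attenuation.
# A quality-level root sends its first two levels to randomized three-wave pools per layer and its last two to single-wave versions, reducing variation.
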SoundCue /ButterCake/Audio/Buttercake_Attack_Stomp_Anim_Cue
root=QualityLevel_0
nodes [
 {id=Attenuation_0, children [Modulator_0], properties {attenuationSettings=&/Buttercake_Attack_Stomp_Impact_Att}}
 {id=Attenuation_1, children [Delay_0], properties {attenuationSettings=&/Buttercake_Attacks_Far_Att}}
 {id=Attenuation_2, children [Modulator_2], properties {attenuationSettings=&Buttercake_Attack_Stomp_Impact_Att}}
 {id=Attenuation_3, children [Delay_1], properties {attenuationSettings=&Buttercake_Attacks_Far_Att}}
 {id=Delay_0, children [Modulator_1], properties {delayMax=0.2, delayMin=0.2}}
 {id=Delay_1, children [Modulator_3], properties {delayMax=0.2, delayMin=0.2}}
 {id=Mixer_0, children [Attenuation_0, Attenuation_1], properties {inputVolume [4,1.7]}}
 {id=Mixer_1, children [Attenuation_2, Attenuation_3], properties {inputVolume [4,1.7]}}
 {id=Modulator_0, children [Random_1], properties {pitchMax=1.2, volumeMax=1, volumeMin=1}}
 {id=Modulator_1, children [Random_0]}
 {id=Modulator_2, children [WavePlayer_6], properties {pitchMax=1.2, volumeMax=1, volumeMin=1}}
 {id=Modulator_3, children [WavePlayer_7], properties {pitchMax=1.1}}
 {id=QualityLevel_0, children [Mixer_0,Mixer_0,Mixer_1,Mixer_1]}
 {id=Random_0, children [WavePlayer_2,WavePlayer_3,WavePlayer_4], properties {bShouldExcludeFromBranchCulling=true, weights [1,1,1]}}
 {id=Random_1, children [WavePlayer_0,WavePlayer_1,WavePlayer_5], properties {bShouldExcludeFromBranchCulling=true, weights [1,1,1]}}
 {id=WavePlayer_0, properties {soundWaveAssetPtr=&/Buttercake_Rear_Ground_Impact_01}}
 {id=WavePlayer_1, properties {soundWaveAssetPtr=&/Buttercake_Rear_Ground_Impact_02}}
 {id=WavePlayer_2, properties {soundWaveAssetPtr=&/Buttercake_Rear_Ground_Impact_Far_01}}
 {id=WavePlayer_3, properties {soundWaveAssetPtr=&/Buttercake_Rear_Ground_Impact_Far_02}}
 {id=WavePlayer_4, properties {soundWaveAssetPtr=&/Buttercake_Rear_Ground_Impact_Far_03}}
 {id=WavePlayer_5, properties {soundWaveAssetPtr=&/Buttercake_Rear_Ground_Impact_03}}
 {id=WavePlayer_6, properties {soundWaveAssetPtr=&Buttercake_Rear_Ground_Impact_01}}
 {id=WavePlayer_7, properties {soundWaveAssetPtr=&Buttercake_Rear_Ground_Impact_Far_01}}
]
properties {
 bOverrideConcurrency=true
 maxDistance=35833.332
 priority=20
 soundClassObject=/Game/Sounds/FortSoundClassesAndModes/Fort_Creatures
 virtualizationMode=Disabled
 volumeMultiplier=1
 concurrencyOverrides {maxCount=1}
}
