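# Sound cue for Buttercake's stomp attack in water, on the creature sound class, limited to one concurrent instance with virtualization disabled.
# Each mix layers a briefly delayed far ground impact with a louder front-leg water footstep, each with its own attenuation and modulation.
# A quality-level switch gives the first two levels random picks from three impacts and eight footsteps.
# The last two levels use a lighter mix with one impact and four footsteps.
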SoundCue /ButterCake/Audio/Buttercake_Attack_Stomp_Anim_Water_Cue
root=QualityLevel_0
nodes [
 {id=Attenuation_1, children [Delay_0], properties {attenuationSettings=&/Buttercake_Attacks_Far_Att}}
 {id=Attenuation_2, children [Modulator_2], properties {attenuationSettings=&/Buttercake_Foley_Footsteps_Front_Att}}
 {id=Attenuation_3, children [Delay_1], properties {attenuationSettings=&Buttercake_Attacks_Far_Att}}
 {id=Attenuation_4, children [Modulator_4], properties {attenuationSettings=&Buttercake_Foley_Footsteps_Front_Att}}
 {id=Delay_0, children [Modulator_1], properties {delayMax=0.2, delayMin=0.2}}
 {id=Delay_1, children [Modulator_3], properties {delayMax=0.2, delayMin=0.2}}
 {id=Mixer_0, children [Attenuation_1, Attenuation_2], properties {inputVolume [1.7,4]}}
 {id=Mixer_1, children [Attenuation_3, Attenuation_4], properties {inputVolume [1.7,4]}}
 {id=Modulator_1, children [Random_0]}
 {id=Modulator_2, children [Random_2], properties {pitchMax=1.2, pitchMin=1, volumeMin=0.5}}
 {id=Modulator_3, children [WavePlayer_6], properties {pitchMax=1.1}}
 {id=Modulator_4, children [Random_4], properties {pitchMax=1.2, pitchMin=1, volumeMin=0.5}}
 {id=QualityLevel_0, children [Mixer_0,Mixer_0,Mixer_1,Mixer_1]}
 {id=Random_0, children [WavePlayer_2,WavePlayer_3,WavePlayer_4], properties {bShouldExcludeFromBranchCulling=true, weights [1,1,1]}}
 {
  id=Random_2
  children [WavePlayer_14, WavePlayer_15, WavePlayer_16, WavePlayer_17, WavePlayer_18, WavePlayer_19, WavePlayer_20, WavePlayer_21]
  properties {bShouldExcludeFromBranchCulling=true, weights [1,1,1,1,1,1,1,1]}
 }
 {
  id=Random_4
  children [WavePlayer_9, WavePlayer_10, WavePlayer_11, WavePlayer_12]
  properties {bShouldExcludeFromBranchCulling=true, weights [1,1,1,1]}
 }
 {id=WavePlayer_10, properties {soundWaveAssetPtr=&/Buttercake_Footstep_Walk_FrontLeg_Water_02}}
 {id=WavePlayer_11, properties {soundWaveAssetPtr=&/Buttercake_Footstep_Walk_FrontLeg_Water_03}}
 {id=WavePlayer_12, properties {soundWaveAssetPtr=&/Buttercake_Footstep_Walk_FrontLeg_Water_04}}
 {id=WavePlayer_14, properties {soundWaveAssetPtr=&/Buttercake_Footstep_Walk_FrontLeg_Water_01}}
 {id=WavePlayer_15, properties {soundWaveAssetPtr=&Buttercake_Footstep_Walk_FrontLeg_Water_02}}
 {id=WavePlayer_16, properties {soundWaveAssetPtr=&Buttercake_Footstep_Walk_FrontLeg_Water_03}}
 {id=WavePlayer_17, properties {soundWaveAssetPtr=&Buttercake_Footstep_Walk_FrontLeg_Water_04}}
 {id=WavePlayer_18, properties {soundWaveAssetPtr=&/Buttercake_Footstep_Walk_FrontLeg_Water_05}}
 {id=WavePlayer_19, properties {soundWaveAssetPtr=&/Buttercake_Footstep_Walk_FrontLeg_Water_06}}
 {id=WavePlayer_2, properties {soundWaveAssetPtr=&/Buttercake_Rear_Ground_Impact_Far_01}}
 {id=WavePlayer_20, properties {soundWaveAssetPtr=&/Buttercake_Footstep_Walk_FrontLeg_Water_07}}
 {id=WavePlayer_21, properties {soundWaveAssetPtr=&/Buttercake_Footstep_Walk_FrontLeg_Water_08}}
 {id=WavePlayer_3, properties {soundWaveAssetPtr=&/Buttercake_Rear_Ground_Impact_Far_02}}
 {id=WavePlayer_4, properties {soundWaveAssetPtr=&/Buttercake_Rear_Ground_Impact_Far_03}}
 {id=WavePlayer_6, properties {soundWaveAssetPtr=&Buttercake_Rear_Ground_Impact_Far_01}}
 {id=WavePlayer_9, properties {soundWaveAssetPtr=&Buttercake_Footstep_Walk_FrontLeg_Water_01}}
]
properties {
 bOverrideConcurrency=true
 maxDistance=34583.332
 priority=20
 soundClassObject=/Game/Sounds/FortSoundClassesAndModes/Fort_Creatures
 virtualizationMode=Disabled
 concurrencyOverrides {maxCount=1}
}
