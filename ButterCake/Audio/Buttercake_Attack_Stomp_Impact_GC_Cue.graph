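# Sound cue named for the impact of Buttercake's stomp attack, classed under Fort_Creatures.
# It has no nodes and a null root, so as stored it references no audio and only carries playback settings.
# It borrows the front footstep foley attenuation, overrides concurrency to a single instance, disables virtualization and sets its own priority.
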SoundCue /ButterCake/Audio/Buttercake_Attack_Stomp_Impact_GC_Cue
root=null
nodes []
properties {
 attenuationSettings=&/Buttercake_Foley_Footsteps_Front_Att
 bOverrideConcurrency=true
 maxDistance=48000
 priority=20
 soundClassObject=/Game/Sounds/FortSoundClassesAndModes/Fort_Creatures
 virtualizationMode=Disabled
 concurrencyOverrides {maxCount=1}
}
