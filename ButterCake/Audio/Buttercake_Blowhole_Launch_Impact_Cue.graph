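# Sound cue for Buttercake's blowhole launch impact, playing burst sounds with attack attenuation under the creature sound class.
# A quality-level node sends the first two levels to a random pick of two burst waves and the last two to a single burst wave.
# Each branch passes through a modulator that holds volume fixed and leaves default pitch variation.
# A concurrency override limits it to one instance, with virtualization disabled and boosted volume.
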SoundCue /ButterCake/Audio/Buttercake_Blowhole_Launch_Impact_Cue
root=QualityLevel_0
nodes [
 {id=Modulator_0, children [Random_0], properties {volumeMax=1, volumeMin=1}}
 {id=Modulator_1, children [WavePlayer_2], properties {volumeMax=1, volumeMin=1}}
 {id=QualityLevel_0, children [Modulator_0,Modulator_0,Modulator_1,Modulator_1]}
 {id=Random_0, children [WavePlayer_0,WavePlayer_1], properties {bShouldExcludeFromBranchCulling=true, weights [1,1]}}
 {id=WavePlayer_0, properties {soundWaveAssetPtr=&/Buttercake_Blowhole_Launch_Blowhole_Burst_02}}
 {id=WavePlayer_1, properties {soundWaveAssetPtr=&/Buttercake_Blowhole_Launch_Blowhole_Burst_01}}
 {id=WavePlayer_2, properties {soundWaveAssetPtr=&Buttercake_Blowhole_Launch_Blowhole_Burst_02}}
]
properties {
 attenuationSettings=&/Buttercake_Attacks_Att
 bOverrideConcurrency=true
 maxDistance=35833.332
 priority=20
 soundClassObject=/Game/Sounds/FortSoundClassesAndModes/Fort_Creatures
 virtualizationMode=Disabled
 volumeMultiplier=2.524
 concurrencyOverrides {maxCount=1}
}
