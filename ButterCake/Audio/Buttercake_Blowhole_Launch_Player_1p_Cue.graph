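# Sound cue for the first-person sound of Buttercake's blowhole launching the player, routed to the creatures sound class.
# A mixer layers a pitch-modulated "Yeet" launch sound with a much quieter random Buttercake lure vocal.
# Both layers branch on quality level, with the last two quality slots using fewer variations: one launch wave and six vocals.
# A concurrency override allows one instance at a time, with virtualization disabled and a very large max distance.
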SoundCue /ButterCake/Audio/Buttercake_Blowhole_Launch_Player_1p_Cue
root=Mixer_0
nodes [
 {id=Mixer_0, children [QualityLevel_0, QualityLevel_1], properties {inputVolume [1,0.3]}}
 {id=Modulator_0, children [Random_1], properties {volumeMax=1, volumeMin=1}}
 {id=Modulator_1, children [WavePlayer_2], properties {volumeMax=1, volumeMin=1}}
 {id=QualityLevel_0, children [Modulator_0,Modulator_0,Modulator_1,Modulator_1]}
 {id=QualityLevel_1, children [Random_4,Random_4,Random_0,Random_0]}
 {
  id=Random_0
  children [WavePlayer_6, WavePlayer_7, WavePlayer_8, WavePlayer_9, WavePlayer_10, WavePlayer_11]
  properties {bShouldExcludeFromBranchCulling=true, weights [1,1,1,1,1,1]}
 }
 {id=Random_1, children [WavePlayer_0,WavePlayer_1], properties {bShouldExcludeFromBranchCulling=true, weights [1,1]}}
 {
  id=Random_4
  children [
   WavePlayer_39
   WavePlayer_40
   WavePlayer_41
   WavePlayer_42
   WavePlayer_43
   WavePlayer_44
   WavePlayer_45
   WavePlayer_46
   WavePlayer_47
   WavePlayer_48
  ]
  properties {bShouldExcludeFromBranchCulling=true, weights [1,1,1,1,1,1,1,1,1,1]}
 }
 {id=WavePlayer_0, properties {soundWaveAssetPtr=&/Buttercake_Blowhole_Launch_Player_Yeet_01}}
 {id=WavePlayer_1, properties {soundWaveAssetPtr=&/Buttercake_Blowhole_Launch_Player_Yeet_02}}
 {id=WavePlayer_10, properties {soundWaveAssetPtr=&/Buttercake_Vox_Lure_05}}
 {id=WavePlayer_11, properties {soundWaveAssetPtr=&/Buttercake_Vox_Lure_06}}
 {id=WavePlayer_2, properties {soundWaveAssetPtr=&Buttercake_Blowhole_Launch_Player_Yeet_01}}
 {id=WavePlayer_39, properties {soundWaveAssetPtr=&/Buttercake_Vox_Lure_01}}
 {id=WavePlayer_40, properties {soundWaveAssetPtr=&/Buttercake_Vox_Lure_02}}
 {id=WavePlayer_41, properties {soundWaveAssetPtr=&/Buttercake_Vox_Lure_03}}
 {id=WavePlayer_42, properties {soundWaveAssetPtr=&/Buttercake_Vox_Lure_04}}
 {id=WavePlayer_43, properties {soundWaveAssetPtr=&Buttercake_Vox_Lure_05}}
 {id=WavePlayer_44, properties {soundWaveAssetPtr=&Buttercake_Vox_Lure_06}}
 {id=WavePlayer_45, properties {soundWaveAssetPtr=&/Buttercake_Vox_Lure_07}}
 {id=WavePlayer_46, properties {soundWaveAssetPtr=&/Buttercake_Vox_Lure_08}}
 {id=WavePlayer_47, properties {soundWaveAssetPtr=&/Buttercake_Vox_Lure_09}}
 {id=WavePlayer_48, properties {soundWaveAssetPtr=&/Buttercake_Vox_Lure_10}}
 {id=WavePlayer_6, properties {soundWaveAssetPtr=&Buttercake_Vox_Lure_01}}
 {id=WavePlayer_7, properties {soundWaveAssetPtr=&Buttercake_Vox_Lure_02}}
 {id=WavePlayer_8, properties {soundWaveAssetPtr=&Buttercake_Vox_Lure_03}}
 {id=WavePlayer_9, properties {soundWaveAssetPtr=&Buttercake_Vox_Lure_04}}
]
properties {
 bOverrideConcurrency=true
 maxDistance=2097152
 priority=20
 soundClassObject=/Game/Sounds/FortSoundClassesAndModes/Fort_Creatures
 virtualizationMode=Disabled
 volumeMultiplier=1.683
 concurrencyOverrides {maxCount=1}
}
