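# Sound cue for the third-person Buttercake blowhole player launch sound, in the creatures sound class with in-air attenuation.
# A quality-level node sends its first two slots to an equally weighted random pick of three launch waves, and its last two to a single wave.
# Both branches pass through modulators with fixed volume, and a concurrency override caps simultaneous instances.
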SoundCue /ButterCake/Audio/Buttercake_Blowhole_Launch_Player_3p_Cue
root=QualityLevel_0
nodes [
 {id=Modulator_0, children [WavePlayer_0], properties {volumeMax=1, volumeMin=1}}
 {id=Modulator_1, children [Random_0], properties {volumeMax=1, volumeMin=1}}
 {id=QualityLevel_0, children [Modulator_1,Modulator_1,Modulator_0,Modulator_0]}
 {id=Random_0, children [WavePlayer_2,WavePlayer_3,WavePlayer_4], properties {bShouldExcludeFromBranchCulling=true, weights [1,1,1]}}
 {id=WavePlayer_0, properties {soundWaveAssetPtr=&/Buttercake_Blowhole_Launch_3P_01}}
 {id=WavePlayer_2, properties {soundWaveAssetPtr=&Buttercake_Blowhole_Launch_3P_01}}
 {id=WavePlayer_3, properties {soundWaveAssetPtr=&/Buttercake_Blowhole_Launch_3P_02}}
 {id=WavePlayer_4, properties {soundWaveAssetPtr=&/Buttercake_Blowhole_Launch_3P_03}}
]
properties {
 attenuationSettings=&/Buttercake_Blowhole_Launch_InAir_Att
 bOverrideConcurrency=true
 maxDistance=25000
 priority=10
 soundClassObject=/Game/Sounds/FortSoundClassesAndModes/Fort_Creatures
 volumeMultiplier=2.103
 concurrencyOverrides {maxCount=3}
}
