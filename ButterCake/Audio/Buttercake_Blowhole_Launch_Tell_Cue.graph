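# Sound cue for the Buttercake creature's blowhole launch tell, played as spatialized creature audio with shared blowhole attenuation.
# A quality-level switch sends its first two levels to an equal-weight random pick of two tell waves, and its last two to the first wave alone.
# Pitch and volume are boosted at cue level, virtualization is disabled, and a concurrency override caps simultaneous instances at two.
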SoundCue /ButterCake/Audio/Buttercake_Blowhole_Launch_Tell_Cue
root=QualityLevel_0
nodes [
 {id=Modulator_0, children [Random_0], properties {volumeMax=1, volumeMin=1}}
 {id=Modulator_1, children [WavePlayer_2], properties {volumeMax=1, volumeMin=1}}
 {id=QualityLevel_0, children [Modulator_0,Modulator_0,Modulator_1,Modulator_1]}
 {id=Random_0, children [WavePlayer_0,WavePlayer_1], properties {bShouldExcludeFromBranchCulling=true, weights [1,1]}}
 {id=WavePlayer_0, properties {soundWaveAssetPtr=&/Buttercake_Blowhole_Launch_Tell_01}}
 {id=WavePlayer_1, properties {soundWaveAssetPtr=&/Buttercake_Blowhole_Launch_Tell_02}}
 {id=WavePlayer_2, properties {soundWaveAssetPtr=&Buttercake_Blowhole_Launch_Tell_01}}
]
properties {
 attenuationSettings=&/Buttercake_Blowhole_Att
 bOverrideConcurrency=true
 maxDistance=3300
 pitchMultiplier=1.5
 priority=5
 soundClassObject=/Game/Sounds/FortSoundClassesAndModes/Fort_Creatures
 virtualizationMode=Disabled
 volumeMultiplier=2.103
 concurrencyOverrides {maxCount=2}
}
