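# Sound cue for Buttercake's eyelid blink, played in the Fort_Creatures sound class with its own attenuation asset.
# A random node picks one of six equally weighted blink waves, and a modulator varies its volume.
# The chain feeds only the first slot of the quality-level root, leaving the others empty.
# A concurrency override allows one instance at a time and rejects new plays, with virtualization disabled and priority at zero.
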SoundCue /ButterCake/Audio/Buttercake_Eyelid_Blink_Cue
root=QualityLevel_0
nodes [
 {id=Modulator_0, children [Random_1], properties {volumeMax=1, volumeMin=0.6}}
 {id=QualityLevel_0, children [Modulator_0,null,null,null]}
 {
  id=Random_1
  children [WavePlayer_2,WavePlayer_3,WavePlayer_4,WavePlayer_5,WavePlayer_6,WavePlayer_7]
  properties {bShouldExcludeFromBranchCulling=true, weights [1,1,1,1,1,1]}
 }
 {id=WavePlayer_2, properties {soundWaveAssetPtr=&/Buttercake_Eye_Lid_Blink_01}}
 {id=WavePlayer_3, properties {soundWaveAssetPtr=&/Buttercake_Eye_Lid_Blink_02}}
 {id=WavePlayer_4, properties {soundWaveAssetPtr=&/Buttercake_Eye_Lid_Blink_03}}
 {id=WavePlayer_5, properties {soundWaveAssetPtr=&/Buttercake_Eye_Lid_Blink_04}}
 {id=WavePlayer_6, properties {soundWaveAssetPtr=&/Buttercake_Eye_Lid_Blink_05}}
 {id=WavePlayer_7, properties {soundWaveAssetPtr=&/Buttercake_Eye_Lid_Blink_06}}
]
properties {
 attenuationSettings=&/Buttercake_Eyelid_Blink_Att
 bOverrideConcurrency=true
 maxDistance=4000
 priority=0
 soundClassObject=/Game/Sounds/FortSoundClassesAndModes/Fort_Creatures
 virtualizationMode=Disabled
 concurrencyOverrides {maxCount=1, resolutionRule=PreventNew}
}
