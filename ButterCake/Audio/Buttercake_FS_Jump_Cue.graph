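# Footstep sound cue for Buttercake jumps, built from walk scuff samples and pitched up, in the creatures sound class.
# A quality-level node sends its first two levels to a modulated random pick of seven scuffs and its last two to a modulated pick of four.
# Both random nodes weight samples equally and are excluded from branch culling.
# Playback uses Buttercake footstep attenuation and concurrency, with virtualization disabled.
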SoundCue /ButterCake/Audio/Buttercake_FS_Jump_Cue
root=QualityLevel_0
nodes [
 {id=Modulator_0, children [Random_1]}
 {id=Modulator_1, children [Random_0]}
 {id=QualityLevel_0, children [Modulator_0,Modulator_0,Modulator_1,Modulator_1]}
 {
  id=Random_0
  children [WavePlayer_0,WavePlayer_1,WavePlayer_2,WavePlayer_3]
  properties {bShouldExcludeFromBranchCulling=true, weights [1,1,1,1]}
 }
 {
  id=Random_1
  children [WavePlayer_12, WavePlayer_13, WavePlayer_14, WavePlayer_15, WavePlayer_16, WavePlayer_19, WavePlayer_20]
  properties {bShouldExcludeFromBranchCulling=true, weights [1,1,1,1,1,1,1]}
 }
 {id=WavePlayer_0, properties {soundWaveAssetPtr=&/Buttercake_Footstep_Scuff_Walk_01}}
 {id=WavePlayer_1, properties {soundWaveAssetPtr=&/Buttercake_Footstep_Scuff_Walk_02}}
 {id=WavePlayer_12, properties {soundWaveAssetPtr=&/Buttercake_Footstep_Scuff_Walk_07}}
 {id=WavePlayer_13, properties {soundWaveAssetPtr=&Buttercake_Footstep_Scuff_Walk_01}}
 {id=WavePlayer_14, properties {soundWaveAssetPtr=&Buttercake_Footstep_Scuff_Walk_02}}
 {id=WavePlayer_15, properties {soundWaveAssetPtr=&/Buttercake_Footstep_Scuff_Walk_03}}
 {id=WavePlayer_16, properties {soundWaveAssetPtr=&/Buttercake_Footstep_Scuff_Walk_04}}
 {id=WavePlayer_19, properties {soundWaveAssetPtr=&/Buttercake_Footstep_Scuff_Walk_05}}
 {id=WavePlayer_2, properties {soundWaveAssetPtr=&Buttercake_Footstep_Scuff_Walk_03}}
 {id=WavePlayer_20, properties {soundWaveAssetPtr=&/Buttercake_Footstep_Scuff_Walk_06}}
 {id=WavePlayer_3, properties {soundWaveAssetPtr=&Buttercake_Footstep_Scuff_Walk_04}}
]
properties {
 attenuationSettings=&/Buttercake_Foley_Footsteps_Front_Att
 maxDistance=34583.332
 pitchMultiplier=1.5
 soundClassObject=/Game/Sounds/FortSoundClassesAndModes/Fort_Creatures
 virtualizationMode=Disabled
 volumeMultiplier=1.052
 concurrencyOverrides {maxCount=2}
 concurrencySet [&/Buttercake_FS_Concurrency]
}
