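# Creature sound cue for Buttercake's front-leg landing footstep, built from walk-step samples plus a landing sweetener.
# Each branch mixes a close layer and a far layer, each with its own attenuation and random sample choice.
# A root quality switch gives the last two levels fewer variations, and nested switches leave the sweetener only on the first two.
# Modulators vary pitch on the close layer and sweetener, and a concurrency set limits overlapping plays.
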SoundCue /ButterCake/Audio/Buttercake_FS_Land_FrontLeg_Cue
root=QualityLevel_1
nodes [
 {id=Attenuation_0, children [Mixer_1], properties {attenuationSettings=&/Buttercake_Foley_Footsteps_Front_Att}}
 {id=Attenuation_1, children [Modulator_2], properties {attenuationSettings=&/Buttercake_Foley_Footsteps_Front_Far_Att}}
 {id=Attenuation_2, children [Mixer_3], properties {attenuationSettings=&Buttercake_Foley_Footsteps_Front_Att}}
 {id=Attenuation_3, children [Modulator_1], properties {attenuationSettings=&Buttercake_Foley_Footsteps_Front_Far_Att}}
 {id=Mixer_0, children [Attenuation_0, Attenuation_1], properties {inputVolume [1.5,3]}}
 {id=Mixer_1, children [QualityLevel_0, Modulator_3], properties {inputVolume [2,2]}}
 {id=Mixer_2, children [Attenuation_2, Attenuation_3], properties {inputVolume [1.5,3]}}
 {id=Mixer_3, children [QualityLevel_2, Modulator_5], properties {inputVolume [2,2]}}
 {id=Modulator_0, children [WavePlayer_0], properties {pitchMax=1.25, pitchMin=0.9}}
 {id=Modulator_1, children [Random_0], properties {pitchMax=1, volumeMax=1, volumeMin=1}}
 {id=Modulator_2, children [Random_2], properties {pitchMax=1, volumeMax=1, volumeMin=1}}
 {id=Modulator_3, children [Random_3], properties {pitchMax=1.15, pitchMin=0.9, volumeMin=0.85}}
 {id=Modulator_4, children [WavePlayer_7], properties {pitchMax=1.25, pitchMin=0.9}}
 {id=Modulator_5, children [Random_4], properties {pitchMax=1.15, pitchMin=0.9, volumeMin=0.85}}
 {id=QualityLevel_0, children [Modulator_0,Modulator_0,null,null]}
 {id=QualityLevel_1, children [Mixer_0,Mixer_0,Mixer_2,Mixer_2]}
 {id=QualityLevel_2, children [Modulator_4,Modulator_4,null,null]}
 {
  id=Random_0
  children [WavePlayer_1,WavePlayer_2,WavePlayer_3,WavePlayer_4]
  properties {bShouldExcludeFromBranchCulling=true, weights [1,1,1,1]}
 }
 {
  id=Random_2
  children [WavePlayer_16, WavePlayer_17, WavePlayer_18, WavePlayer_19, WavePlayer_20, WavePlayer_21]
  properties {bShouldExcludeFromBranchCulling=true, weights [1,1,1,1,1,1]}
 }
 {
  id=Random_3
  children [
   WavePlayer_14
   WavePlayer_15
   WavePlayer_31
   WavePlayer_32
   WavePlayer_33
   WavePlayer_34
   WavePlayer_35
   WavePlayer_36
   WavePlayer_37
   WavePlayer_38
   WavePlayer_39
  ]
  properties {weights [1,1,1,1,1,1,1,1,1,1,1]}
 }
 {
  id=Random_4
  children [WavePlayer_8, WavePlayer_9, WavePlayer_10, WavePlayer_11]
  properties {bShouldExcludeFromBranchCulling=true, weights [1,1,1,1]}
 }
 {id=WavePlayer_0, properties {soundWaveAssetPtr=&/Buttercake_Footstep_Land_Sweetener_01}}
 {id=WavePlayer_1, properties {soundWaveAssetPtr=&/Buttercake_Footstep_Walk_FrontLeg_Far_01}}
 {id=WavePlayer_10, properties {soundWaveAssetPtr=&/Buttercake_Footstep_Walk_FrontLeg_03}}
 {id=WavePlayer_11, properties {soundWaveAssetPtr=&/Buttercake_Footstep_Walk_FrontLeg_04}}
 {id=WavePlayer_14, properties {soundWaveAssetPtr=&/Buttercake_Footstep_Walk_FrontLeg_01}}
 {id=WavePlayer_15, properties {soundWaveAssetPtr=&/Buttercake_Footstep_Walk_FrontLeg_02}}
 {id=WavePlayer_16, properties {soundWaveAssetPtr=&Buttercake_Footstep_Walk_FrontLeg_Far_01}}
 {id=WavePlayer_17, properties {soundWaveAssetPtr=&/Buttercake_Footstep_Walk_FrontLeg_Far_02}}
 {id=WavePlayer_18, properties {soundWaveAssetPtr=&/Buttercake_Footstep_Walk_FrontLeg_Far_03}}
 {id=WavePlayer_19, properties {soundWaveAssetPtr=&/Buttercake_Footstep_Walk_FrontLeg_Far_04}}
 {id=WavePlayer_2, properties {soundWaveAssetPtr=&Buttercake_Footstep_Walk_FrontLeg_Far_02}}
 {id=WavePlayer_20, properties {soundWaveAssetPtr=&/Buttercake_Footstep_Walk_FrontLeg_Far_05}}
 {id=WavePlayer_21, properties {soundWaveAssetPtr=&/Buttercake_Footstep_Walk_FrontLeg_Far_06}}
 {id=WavePlayer_3, properties {soundWaveAssetPtr=&Buttercake_Footstep_Walk_FrontLeg_Far_03}}
 {id=WavePlayer_31, properties {soundWaveAssetPtr=&Buttercake_Footstep_Walk_FrontLeg_03}}
 {id=WavePlayer_32, properties {soundWaveAssetPtr=&Buttercake_Footstep_Walk_FrontLeg_04}}
 {id=WavePlayer_33, properties {soundWaveAssetPtr=&/Buttercake_Footstep_Walk_FrontLeg_05}}
 {id=WavePlayer_34, properties {soundWaveAssetPtr=&/Buttercake_Footstep_Walk_FrontLeg_06}}
 {id=WavePlayer_35, properties {soundWaveAssetPtr=&/Buttercake_Footstep_Walk_FrontLeg_07}}
 {id=WavePlayer_36, properties {soundWaveAssetPtr=&/Buttercake_Footstep_Walk_FrontLeg_08}}
 {id=WavePlayer_37, properties {soundWaveAssetPtr=&/Buttercake_Footstep_Walk_FrontLeg_09}}
 {id=WavePlayer_38, properties {soundWaveAssetPtr=&/Buttercake_Footstep_Walk_FrontLeg_10}}
 {id=WavePlayer_39, properties {soundWaveAssetPtr=&/Buttercake_Footstep_Walk_FrontLeg_11}}
 {id=WavePlayer_4, properties {soundWaveAssetPtr=&Buttercake_Footstep_Walk_FrontLeg_Far_04}}
 {id=WavePlayer_7, properties {soundWaveAssetPtr=&Buttercake_Footstep_Land_Sweetener_01}}
 {id=WavePlayer_8, properties {soundWaveAssetPtr=&Buttercake_Footstep_Walk_FrontLeg_01}}
 {id=WavePlayer_9, properties {soundWaveAssetPtr=&Buttercake_Footstep_Walk_FrontLeg_02}}
]
properties {
 maxDistance=34583.332
 soundClassObject=/Game/Sounds/FortSoundClassesAndModes/Fort_Creatures
 virtualizationMode=Disabled
 volumeMultiplier=0.841
 concurrencyOverrides {maxCount=2}
 concurrencySet [&/Buttercake_FS_Concurrency]
}
