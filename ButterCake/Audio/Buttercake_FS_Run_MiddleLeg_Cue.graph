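# Sound cue for Buttercake creature middle-leg running footsteps, built from rear-leg walk footstep waves.
# A quality-level node routes its first two slots to a random pick of eleven waves and its last two to a pick of four.
# Each branch passes through a modulator that varies pitch and volume.
# It plays in the creature sound class with rear-footstep attenuation, shared footstep concurrency and virtualization disabled.
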SoundCue /ButterCake/Audio/Buttercake_FS_Run_MiddleLeg_Cue
root=QualityLevel_0
nodes [
 {id=Modulator_0, children [Random_2], properties {pitchMax=1.1, pitchMin=0.9, volumeMin=0.75}}
 {id=Modulator_1, children [Random_1], properties {pitchMax=1.1, pitchMin=0.9, volumeMin=0.75}}
 {id=QualityLevel_0, children [Modulator_1,Modulator_1,Modulator_0,Modulator_0]}
 {
  id=Random_1
  children [
   WavePlayer_6
   WavePlayer_7
   WavePlayer_8
   WavePlayer_9
   WavePlayer_10
   WavePlayer_11
   WavePlayer_12
   WavePlayer_13
   WavePlayer_14
   WavePlayer_15
   WavePlayer_16
  ]
  properties {bShouldExcludeFromBranchCulling=true, weights [1,1,1,1,1,1,1,1,1,1,1]}
 }
 {
  id=Random_2
  children [WavePlayer_0,WavePlayer_1,WavePlayer_2,WavePlayer_3]
  properties {bShouldExcludeFromBranchCulling=true, weights [1,1,1,1]}
 }
 {id=WavePlayer_0, properties {soundWaveAssetPtr=&/Buttercake_Footstep_Walk_RearLeg_01}}
 {id=WavePlayer_1, properties {soundWaveAssetPtr=&/Buttercake_Footstep_Walk_RearLeg_02}}
 {id=WavePlayer_10, properties {soundWaveAssetPtr=&/Buttercake_Footstep_Walk_RearLeg_05}}
 {id=WavePlayer_11, properties {soundWaveAssetPtr=&/Buttercake_Footstep_Walk_RearLeg_06}}
 {id=WavePlayer_12, properties {soundWaveAssetPtr=&/Buttercake_Footstep_Walk_RearLeg_07}}
 {id=WavePlayer_13, properties {soundWaveAssetPtr=&/Buttercake_Footstep_Walk_RearLeg_08}}
 {id=WavePlayer_14, properties {soundWaveAssetPtr=&/Buttercake_Footstep_Walk_RearLeg_09}}
 {id=WavePlayer_15, properties {soundWaveAssetPtr=&/Buttercake_Footstep_Walk_RearLeg_10}}
 {id=WavePlayer_16, properties {soundWaveAssetPtr=&/Buttercake_Footstep_Walk_RearLeg_11}}
 {id=WavePlayer_2, properties {soundWaveAssetPtr=&/Buttercake_Footstep_Walk_RearLeg_03}}
 {id=WavePlayer_3, properties {soundWaveAssetPtr=&/Buttercake_Footstep_Walk_RearLeg_04}}
 {id=WavePlayer_6, properties {soundWaveAssetPtr=&Buttercake_Footstep_Walk_RearLeg_01}}
 {id=WavePlayer_7, properties {soundWaveAssetPtr=&Buttercake_Footstep_Walk_RearLeg_02}}
 {id=WavePlayer_8, properties {soundWaveAssetPtr=&Buttercake_Footstep_Walk_RearLeg_03}}
 {id=WavePlayer_9, properties {soundWaveAssetPtr=&Buttercake_Footstep_Walk_RearLeg_04}}
]
properties {
 attenuationSettings=&/Buttercake_Foley_Footsteps_Rear_Att
 maxDistance=21000
 soundClassObject=/Game/Sounds/FortSoundClassesAndModes/Fort_Creatures
 virtualizationMode=Disabled
 volumeMultiplier=1.388
 concurrencyOverrides {maxCount=12}
 concurrencySet [&/Buttercake_FS_Concurrency]
}
