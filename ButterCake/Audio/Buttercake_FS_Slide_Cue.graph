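# Sound cue for the Buttercake creature's footstep slide, built from walk scuff samples and routed to the creatures sound class.
# A Modulator at the root feeds a quality-level switch, whose first two slots pick randomly among seven scuff waves and last two among three.
# Both random pools are equally weighted and excluded from branch culling.
# Footstep attenuation and concurrency assets govern playback, with virtualization disabled.
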SoundCue /ButterCake/Audio/Buttercake_FS_Slide_Cue
root=Modulator_0
nodes [
 {id=Modulator_0, children [QualityLevel_0]}
 {id=QualityLevel_0, children [Random_0,Random_0,Random_2,Random_2]}
 {
  id=Random_0
  children [WavePlayer_0,WavePlayer_1,WavePlayer_2,WavePlayer_3,WavePlayer_4,WavePlayer_5,WavePlayer_6]
  properties {bShouldExcludeFromBranchCulling=true, weights [1,1,1,1,1,1,1]}
 }
 {id=Random_2, children [WavePlayer_7,WavePlayer_8,WavePlayer_9], properties {bShouldExcludeFromBranchCulling=true, weights [1,1,1]}}
 {id=WavePlayer_0, properties {soundWaveAssetPtr=&/Buttercake_Footstep_Scuff_Walk_01}}
 {id=WavePlayer_1, properties {soundWaveAssetPtr=&/Buttercake_Footstep_Scuff_Walk_02}}
 {id=WavePlayer_2, properties {soundWaveAssetPtr=&/Buttercake_Footstep_Scuff_Walk_03}}
 {id=WavePlayer_3, properties {soundWaveAssetPtr=&/Buttercake_Footstep_Scuff_Walk_04}}
 {id=WavePlayer_4, properties {soundWaveAssetPtr=&/Buttercake_Footstep_Scuff_Walk_05}}
 {id=WavePlayer_5, properties {soundWaveAssetPtr=&/Buttercake_Footstep_Scuff_Walk_06}}
 {id=WavePlayer_6, properties {soundWaveAssetPtr=&/Buttercake_Footstep_Scuff_Walk_07}}
 {id=WavePlayer_7, properties {soundWaveAssetPtr=&Buttercake_Footstep_Scuff_Walk_01}}
 {id=WavePlayer_8, properties {soundWaveAssetPtr=&Buttercake_Footstep_Scuff_Walk_02}}
 {id=WavePlayer_9, properties {soundWaveAssetPtr=&Buttercake_Footstep_Scuff_Walk_03}}
]
properties {
 attenuationSettings=&/Buttercake_Foley_Footsteps_Front_Att
 maxDistance=34583.332
 soundClassObject=/Game/Sounds/FortSoundClassesAndModes/Fort_Creatures
 virtualizationMode=Disabled
 volumeMultiplier=2.371
 concurrencyOverrides {maxCount=2}
 concurrencySet [&/Buttercake_FS_Concurrency]
}
