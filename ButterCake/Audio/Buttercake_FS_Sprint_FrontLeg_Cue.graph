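# Sprint front-leg footstep cue for the Buttercake creature, in the creature sound class with limited concurrency.
# Each variant mixes a near layer of randomized run steps and sweeteners with a far layer of walk far recordings, each under its own attenuation.
# Modulators vary pitch and volume per layer.
# A quality-level node gives the first two slots larger sample pools and the last two four-sample pools.
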SoundCue /ButterCake/Audio/Buttercake_FS_Sprint_FrontLeg_Cue
root=Modulator_6
nodes [
 {id=Attenuation_0, children [Mixer_1], properties {attenuationSettings=&/Buttercake_Foley_Footsteps_Front_Att}}
 {id=Attenuation_1, children [Modulator_1], properties {attenuationSettings=&/Buttercake_Foley_Footsteps_Front_Far_Att}}
 {id=Attenuation_2, children [Modulator_4], properties {attenuationSettings=&Buttercake_Foley_Footsteps_Front_Far_Att}}
 {id=Attenuation_3, children [Mixer_3], properties {attenuationSettings=&Buttercake_Foley_Footsteps_Front_Att}}
 {id=Mixer_0, children [Attenuation_0, Attenuation_1], properties {inputVolume [3.5,3]}}
 {id=Mixer_1, children [Modulator_2,Modulator_0], properties {inputVolume [1,1]}}
 {id=Mixer_2, children [Attenuation_3, Attenuation_2], properties {inputVolume [3.5,3]}}
 {id=Mixer_3, children [Modulator_5,Modulator_3], properties {inputVolume [1,1]}}
 {id=Modulator_0, children [Random_1]}
 {id=Modulator_1, children [Random_0], properties {pitchMax=1.2, pitchMin=1, volumeMax=1}}
 {id=Modulator_2, children [Random_2], properties {pitchMax=1.2, volumeMax=1.5, volumeMin=0.5}}
 {id=Modulator_3, children [Random_3]}
 {id=Modulator_4, children [Random_4], properties {pitchMax=1.2, pitchMin=1, volumeMax=1}}
 {id=Modulator_5, children [Random_5], properties {pitchMax=1.2, volumeMax=1.5, volumeMin=0.5}}
 {id=Modulator_6, children [QualityLevel_0], properties {pitchMax=1, pitchMin=1, volumeMax=1, volumeMin=0.8}}
 {id=QualityLevel_0, children [Mixer_0,Mixer_0,Mixer_2,Mixer_2]}
 {
  id=Random_0
  children [WavePlayer_0,WavePlayer_1,WavePlayer_2,WavePlayer_3,WavePlayer_4,WavePlayer_5]
  properties {bShouldExcludeFromBranchCulling=true, weights [1,1,1,1,1,1]}
 }
 {
  id=Random_1
  children [WavePlayer_8, WavePlayer_9, WavePlayer_10, WavePlayer_11, WavePlayer_12, WavePlayer_13, WavePlayer_14, WavePlayer_15]
  properties {bShouldExcludeFromBranchCulling=true, weights [1,1,1,1,1,1,1,1]}
 }
 {
  id=Random_2
  children [WavePlayer_6, WavePlayer_7, WavePlayer_16, WavePlayer_17, WavePlayer_18, WavePlayer_19, WavePlayer_20, WavePlayer_21]
  properties {bShouldExcludeFromBranchCulling=true, weights [1,1,1,1,1,1,1,1]}
 }
 {
  id=Random_3
  children [WavePlayer_22, WavePlayer_23, WavePlayer_24, WavePlayer_25]
  properties {bShouldExcludeFromBranchCulling=true, weights [1,1,1,1]}
 }
 {
  id=Random_4
  children [WavePlayer_30, WavePlayer_31, WavePlayer_32, WavePlayer_33]
  properties {bShouldExcludeFromBranchCulling=true, weights [1,1,1,1]}
 }
 {
  id=Random_5
  children [WavePlayer_36, WavePlayer_37, WavePlayer_38, WavePlayer_39]
  properties {bShouldExcludeFromBranchCulling=true, weights [1,1,1,1]}
 }
 {id=WavePlayer_0, properties {soundWaveAssetPtr=&/Buttercake_Footstep_Walk_FrontLeg_Far_01}}
 {id=WavePlayer_1, properties {soundWaveAssetPtr=&/Buttercake_Footstep_Walk_FrontLeg_Far_02}}
 {id=WavePlayer_10, properties {soundWaveAssetPtr=&/Buttercake_Footstep_Run_FrontLeg_03}}
 {id=WavePlayer_11, properties {soundWaveAssetPtr=&/Buttercake_Footstep_Run_FrontLeg_04}}
 {id=WavePlayer_12, properties {soundWaveAssetPtr=&/Buttercake_Footstep_Run_FrontLeg_05}}
 {id=WavePlayer_13, properties {soundWaveAssetPtr=&/Buttercake_Footstep_Run_FrontLeg_06}}
 {id=WavePlayer_14, properties {soundWaveAssetPtr=&/Buttercake_Footstep_Run_FrontLeg_07}}
 {id=WavePlayer_15, properties {soundWaveAssetPtr=&/Buttercake_Footstep_Run_FrontLeg_08}}
 {id=WavePlayer_16, properties {soundWaveAssetPtr=&/Buttercake_Footstep_Run_Sweetener_03}}
 {id=WavePlayer_17, properties {soundWaveAssetPtr=&/Buttercake_Footstep_Run_Sweetener_04}}
 {id=WavePlayer_18, properties {soundWaveAssetPtr=&/Buttercake_Footstep_Run_Sweetener_05}}
 {id=WavePlayer_19, properties {soundWaveAssetPtr=&/Buttercake_Footstep_Run_Sweetener_06}}
 {id=WavePlayer_2, properties {soundWaveAssetPtr=&/Buttercake_Footstep_Walk_FrontLeg_Far_03}}
 {id=WavePlayer_20, properties {soundWaveAssetPtr=&/Buttercake_Footstep_Run_Sweetener_07}}
 {id=WavePlayer_21, properties {soundWaveAssetPtr=&/Buttercake_Footstep_Run_Sweetener_08}}
 {id=WavePlayer_22, properties {soundWaveAssetPtr=&/Buttercake_Footstep_Run_FrontLeg_01}}
 {id=WavePlayer_23, properties {soundWaveAssetPtr=&/Buttercake_Footstep_Run_FrontLeg_02}}
 {id=WavePlayer_24, properties {soundWaveAssetPtr=&Buttercake_Footstep_Run_FrontLeg_03}}
 {id=WavePlayer_25, properties {soundWaveAssetPtr=&Buttercake_Footstep_Run_FrontLeg_04}}
 {id=WavePlayer_3, properties {soundWaveAssetPtr=&/Buttercake_Footstep_Walk_FrontLeg_Far_04}}
 {id=WavePlayer_30, properties {soundWaveAssetPtr=&Buttercake_Footstep_Walk_FrontLeg_Far_01}}
 {id=WavePlayer_31, properties {soundWaveAssetPtr=&Buttercake_Footstep_Walk_FrontLeg_Far_02}}
 {id=WavePlayer_32, properties {soundWaveAssetPtr=&Buttercake_Footstep_Walk_FrontLeg_Far_03}}
 {id=WavePlayer_33, properties {soundWaveAssetPtr=&Buttercake_Footstep_Walk_FrontLeg_Far_04}}
 {id=WavePlayer_36, properties {soundWaveAssetPtr=&/Buttercake_Footstep_Run_Sweetener_01}}
 {id=WavePlayer_37, properties {soundWaveAssetPtr=&/Buttercake_Footstep_Run_Sweetener_02}}
 {id=WavePlayer_38, properties {soundWaveAssetPtr=&Buttercake_Footstep_Run_Sweetener_03}}
 {id=WavePlayer_39, properties {soundWaveAssetPtr=&Buttercake_Footstep_Run_Sweetener_04}}
 {id=WavePlayer_4, properties {soundWaveAssetPtr=&/Buttercake_Footstep_Walk_FrontLeg_Far_05}}
 {id=WavePlayer_5, properties {soundWaveAssetPtr=&/Buttercake_Footstep_Walk_FrontLeg_Far_06}}
 {id=WavePlayer_6, properties {soundWaveAssetPtr=&Buttercake_Footstep_Run_Sweetener_01}}
 {id=WavePlayer_7, properties {soundWaveAssetPtr=&Buttercake_Footstep_Run_Sweetener_02}}
 {id=WavePlayer_8, properties {soundWaveAssetPtr=&Buttercake_Footstep_Run_FrontLeg_01}}
 {id=WavePlayer_9, properties {soundWaveAssetPtr=&Buttercake_Footstep_Run_FrontLeg_02}}
]
properties {
 maxDistance=34583.332
 soundClassObject=/Game/Sounds/FortSoundClassesAndModes/Fort_Creatures
 virtualizationMode=Disabled
 volumeMultiplier=0.8
 concurrencyOverrides {maxCount=4}
 concurrencySet [&/Buttercake_FS_Concurrency]
}
