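# Sound cue for Buttercake's sprint rear-leg footsteps, built from Run FrontLeg and Run Sweetener waves under rear-footstep attenuation.
# Each mix layers a random footstep with a random sweetener that has wider volume and pitch modulation.
# A quality switch feeds the first two levels eight variants per layer and the last two four, excluded from branch culling.
# A root modulator adds slight volume variation, and a concurrency set governs overlapping plays.
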SoundCue /ButterCake/Audio/Buttercake_FS_Sprint_RearLeg_Cue
root=Modulator_4
nodes [
 {id=Mixer_0, children [Modulator_3,Modulator_0], properties {inputVolume [1.2,1]}}
 {id=Mixer_1, children [Modulator_2,Modulator_1], properties {inputVolume [1.2,1]}}
 {id=Modulator_0, children [Random_5]}
 {id=Modulator_1, children [Random_0]}
 {id=Modulator_2, children [Random_2], properties {pitchMax=1.2, volumeMax=1.2, volumeMin=0.35}}
 {id=Modulator_3, children [Random_4], properties {pitchMax=1.2, volumeMax=1.2, volumeMin=0.35}}
 {id=Modulator_4, children [QualityLevel_0], properties {pitchMax=1, pitchMin=1, volumeMax=1, volumeMin=0.8}}
 {id=QualityLevel_0, children [Mixer_1,Mixer_1,Mixer_0,Mixer_0]}
 {
  id=Random_0
  children [WavePlayer_8, WavePlayer_9, WavePlayer_10, WavePlayer_11, WavePlayer_0, WavePlayer_1, WavePlayer_2, WavePlayer_3]
  properties {bShouldExcludeFromBranchCulling=true, weights [1,1,1,1,1,1,1,1]}
 }
 {
  id=Random_2
  children [WavePlayer_6, WavePlayer_7, WavePlayer_4, WavePlayer_17, WavePlayer_18, WavePlayer_19, WavePlayer_20, WavePlayer_21]
  properties {bShouldExcludeFromBranchCulling=true, weights [1,1,1,1,1,1,1,1]}
 }
 {
  id=Random_4
  children [WavePlayer_24, WavePlayer_25, WavePlayer_26, WavePlayer_27]
  properties {bShouldExcludeFromBranchCulling=true, weights [1,1,1,1]}
 }
 {
  id=Random_5
  children [WavePlayer_5, WavePlayer_12, WavePlayer_13, WavePlayer_14]
  properties {bShouldExcludeFromBranchCulling=true, weights [1,1,1,1]}
 }
 {id=WavePlayer_0, properties {soundWaveAssetPtr=&/Buttercake_Footstep_Run_FrontLeg_05}}
 {id=WavePlayer_1, properties {soundWaveAssetPtr=&/Buttercake_Footstep_Run_FrontLeg_06}}
 {id=WavePlayer_10, properties {soundWaveAssetPtr=&/Buttercake_Footstep_Run_FrontLeg_03}}
 {id=WavePlayer_11, properties {soundWaveAssetPtr=&/Buttercake_Footstep_Run_FrontLeg_04}}
 {id=WavePlayer_12, properties {soundWaveAssetPtr=&/Buttercake_Footstep_Run_FrontLeg_02}}
 {id=WavePlayer_13, properties {soundWaveAssetPtr=&Buttercake_Footstep_Run_FrontLeg_03}}
 {id=WavePlayer_14, properties {soundWaveAssetPtr=&Buttercake_Footstep_Run_FrontLeg_04}}
 {id=WavePlayer_17, properties {soundWaveAssetPtr=&/Buttercake_Footstep_Run_Sweetener_04}}
 {id=WavePlayer_18, properties {soundWaveAssetPtr=&/Buttercake_Footstep_Run_Sweetener_05}}
 {id=WavePlayer_19, properties {soundWaveAssetPtr=&/Buttercake_Footstep_Run_Sweetener_06}}
 {id=WavePlayer_2, properties {soundWaveAssetPtr=&/Buttercake_Footstep_Run_FrontLeg_07}}
 {id=WavePlayer_20, properties {soundWaveAssetPtr=&/Buttercake_Footstep_Run_Sweetener_07}}
 {id=WavePlayer_21, properties {soundWaveAssetPtr=&/Buttercake_Footstep_Run_Sweetener_08}}
 {id=WavePlayer_24, properties {soundWaveAssetPtr=&/Buttercake_Footstep_Run_Sweetener_01}}
 {id=WavePlayer_25, properties {soundWaveAssetPtr=&/Buttercake_Footstep_Run_Sweetener_02}}
 {id=WavePlayer_26, properties {soundWaveAssetPtr=&/Buttercake_Footstep_Run_Sweetener_03}}
 {id=WavePlayer_27, properties {soundWaveAssetPtr=&Buttercake_Footstep_Run_Sweetener_04}}
 {id=WavePlayer_3, properties {soundWaveAssetPtr=&/Buttercake_Footstep_Run_FrontLeg_08}}
 {id=WavePlayer_4, properties {soundWaveAssetPtr=&Buttercake_Footstep_Run_Sweetener_03}}
 {id=WavePlayer_5, properties {soundWaveAssetPtr=&/Buttercake_Footstep_Run_FrontLeg_01}}
 {id=WavePlayer_6, properties {soundWaveAssetPtr=&Buttercake_Footstep_Run_Sweetener_01}}
 {id=WavePlayer_7, properties {soundWaveAssetPtr=&Buttercake_Footstep_Run_Sweetener_02}}
 {id=WavePlayer_8, properties {soundWaveAssetPtr=&Buttercake_Footstep_Run_FrontLeg_01}}
 {id=WavePlayer_9, properties {soundWaveAssetPtr=&Buttercake_Footstep_Run_FrontLeg_02}}
]
properties {
 attenuationSettings=&/Buttercake_Foley_Footsteps_Rear_Att
 maxDistance=21000
 soundClassObject=/Game/Sounds/FortSoundClassesAndModes/Fort_Creatures
 virtualizationMode=Disabled
 volumeMultiplier=1.5
 concurrencyOverrides {maxCount=12}
 concurrencySet [&/Buttercake_FS_Concurrency]
}
