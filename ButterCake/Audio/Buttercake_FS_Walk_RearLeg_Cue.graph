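# Sound cue for Buttercake's rear-leg walking footsteps, classed as a creature sound with rear-footstep attenuation and a footstep concurrency limit.
# A quality-level root sends the first two levels to an equally weighted random pool of eleven waves and the last two to a smaller pool of four.
# Each pool passes through a modulator that varies pitch and volume, and the random nodes are excluded from branch culling.
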SoundCue /ButterCake/Audio/Buttercake_FS_Walk_RearLeg_Cue
root=QualityLevel_0
nodes [
 {id=Modulator_0, children [Random_0], properties {pitchMax=1.1, pitchMin=0.9, volumeMin=0.75}}
 {id=Modulator_2, children [Random_2], properties {pitchMax=1.1, pitchMin=0.9, volumeMin=0.75}}
 {id=QualityLevel_0, children [Modulator_0,Modulator_0,Modulator_2,Modulator_2]}
 {
  id=Random_0
  children [
   WavePlayer_0
   WavePlayer_1
   WavePlayer_2
   WavePlayer_3
   WavePlayer_4
   WavePlayer_5
   WavePlayer_17
   WavePlayer_18
   WavePlayer_19
   WavePlayer_20
   WavePlayer_21
  ]
  properties {bShouldExcludeFromBranchCulling=true, weights [1,1,1,1,1,1,1,1,1,1,1]}
 }
 {
  id=Random_2
  children [WavePlayer_22, WavePlayer_23, WavePlayer_24, WavePlayer_25]
  properties {bShouldExcludeFromBranchCulling=true, weights [1,1,1,1]}
 }
 {id=WavePlayer_0, properties {soundWaveAssetPtr=&/Buttercake_Footstep_Walk_RearLeg_01}}
 {id=WavePlayer_1, properties {soundWaveAssetPtr=&/Buttercake_Footstep_Walk_RearLeg_02}}
 {id=WavePlayer_17, properties {soundWaveAssetPtr=&/Buttercake_Footstep_Walk_RearLeg_07}}
 {id=WavePlayer_18, properties {soundWaveAssetPtr=&/Buttercake_Footstep_Walk_RearLeg_08}}
 {id=WavePlayer_19, properties {soundWaveAssetPtr=&/Buttercake_Footstep_Walk_RearLeg_09}}
 {id=WavePlayer_2, properties {soundWaveAssetPtr=&/Buttercake_Footstep_Walk_RearLeg_03}}
 {id=WavePlayer_20, properties {soundWaveAssetPtr=&/Buttercake_Footstep_Walk_RearLeg_10}}
 {id=WavePlayer_21, properties {soundWaveAssetPtr=&/Buttercake_Footstep_Walk_RearLeg_11}}
 {id=WavePlayer_22, properties {soundWaveAssetPtr=&Buttercake_Footstep_Walk_RearLeg_01}}
 {id=WavePlayer_23, properties {soundWaveAssetPtr=&Buttercake_Footstep_Walk_RearLeg_02}}
 {id=WavePlayer_24, properties {soundWaveAssetPtr=&Buttercake_Footstep_Walk_RearLeg_03}}
 {id=WavePlayer_25, properties {soundWaveAssetPtr=&/Buttercake_Footstep_Walk_RearLeg_04}}
 {id=WavePlayer_3, properties {soundWaveAssetPtr=&Buttercake_Footstep_Walk_RearLeg_04}}
 {id=WavePlayer_4, properties {soundWaveAssetPtr=&/Buttercake_Footstep_Walk_RearLeg_05}}
 {id=WavePlayer_5, properties {soundWaveAssetPtr=&/Buttercake_Footstep_Walk_RearLeg_06}}
]
properties {
 attenuationSettings=&/Buttercake_Foley_Footsteps_Rear_Att
 maxDistance=21000
 soundClassObject=/Game/Sounds/FortSoundClassesAndModes/Fort_Creatures
 virtualizationMode=Disabled
 volumeMultiplier=1.683
 concurrencyOverrides {maxCount=12}
 concurrencySet [&/Buttercake_FS_Concurrency]
}
