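# Sound cue for Buttercake's walk-turn footstep sounds.
# A Random node picks one of five equally weighted wave players, then a Modulator varies pitch and volume per play.
# It routes to the Fort_Creatures sound class, using a rear footstep foley attenuation asset and a shared footstep concurrency set, with virtualization disabled.
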SoundCue /ButterCake/Audio/Buttercake_FS_Walk_Turn_Cue
root=Modulator_1
nodes [
 {id=Modulator_1, children [Random_0], properties {pitchMax=1.15, pitchMin=0.9, volumeMin=0.65}}
 {
  id=Random_0
  children [WavePlayer_0,WavePlayer_1,WavePlayer_2,WavePlayer_3,WavePlayer_4]
  properties {bShouldExcludeFromBranchCulling=true, weights [1,1,1,1,1]}
 }
 {id=WavePlayer_0, properties {soundWaveAssetPtr=&/Buttercake_Footstep_Walk_Turn_01}}
 {id=WavePlayer_1, properties {soundWaveAssetPtr=&/Buttercake_Footstep_Walk_Turn_02}}
 {id=WavePlayer_2, properties {soundWaveAssetPtr=&/Buttercake_Footstep_Walk_Turn_03}}
 {id=WavePlayer_3, properties {soundWaveAssetPtr=&/Buttercake_Footstep_Walk_Turn_04}}
 {id=WavePlayer_4, properties {soundWaveAssetPtr=&/Buttercake_Footstep_Walk_Turn_05}}
]
properties {
 attenuationSettings=&/Buttercake_Foley_Footsteps_Rear_Att
 maxDistance=21000
 soundClassObject=/Game/Sounds/FortSoundClassesAndModes/Fort_Creatures
 virtualizationMode=Disabled
 volumeMultiplier=1.052
 concurrencyOverrides {maxCount=12}
 concurrencySet [&/Buttercake_FS_Concurrency]
}
