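# Sound cue for Buttercake's water jump footstep, a creature sound built from front-leg walk samples.
# Each branch mixes a louder close water layer with a quieter, randomly delayed far layer, each with its own attenuation and random pitch and volume modulation.
# The quality switch gives its first two levels larger random sample pools and its last two a reduced subset.
# A concurrency set is assigned and virtualization is disabled.
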SoundCue /ButterCake/Audio/Buttercake_FS_Water_Jump_Cue
root=QualityLevel_0
nodes [
 {id=Attenuation_0, children [Modulator_0], properties {attenuationSettings=&/Buttercake_Foley_Footsteps_Front_Att}}
 {id=Attenuation_1, children [Delay_0], properties {attenuationSettings=&/Buttercake_Foley_Footsteps_Front_Far_Att}}
 {id=Attenuation_2, children [Modulator_3], properties {attenuationSettings=&Buttercake_Foley_Footsteps_Front_Att}}
 {id=Attenuation_3, children [Delay_1], properties {attenuationSettings=&Buttercake_Foley_Footsteps_Front_Far_Att}}
 {id=Delay_0, children [Modulator_1], properties {delayMax=0.35, delayMin=0.1}}
 {id=Delay_1, children [Modulator_4], properties {delayMax=0.35, delayMin=0.1}}
 {id=Mixer_0, children [Attenuation_0, Attenuation_1], properties {inputVolume [2,0.75]}}
 {id=Mixer_1, children [Attenuation_2, Attenuation_3], properties {inputVolume [2,0.75]}}
 {id=Modulator_0, children [Random_2], properties {pitchMax=1.2, pitchMin=1, volumeMin=0.5}}
 {id=Modulator_1, children [Random_3], properties {pitchMax=1, volumeMax=1, volumeMin=0.5}}
 {id=Modulator_3, children [Random_5], properties {pitchMax=1.2, pitchMin=1, volumeMin=0.5}}
 {id=Modulator_4, children [Random_4], properties {pitchMax=1, volumeMax=1, volumeMin=0.5}}
 {id=QualityLevel_0, children [Mixer_1,Mixer_1,Mixer_0,Mixer_0]}
 {
  id=Random_2
  children [WavePlayer_6,WavePlayer_7,WavePlayer_8,WavePlayer_9]
  properties {bShouldExcludeFromBranchCulling=true, weights [1,1,1,1]}
 }
 {id=Random_3, children [WavePlayer_0,WavePlayer_1], properties {bShouldExcludeFromBranchCulling=true, weights [1,1]}}
 {
  id=Random_4
  children [WavePlayer_35, WavePlayer_36, WavePlayer_37, WavePlayer_38, WavePlayer_39, WavePlayer_40]
  properties {bShouldExcludeFromBranchCulling=true, weights [1,1,1,1,1,1]}
 }
 {
  id=Random_5
  children [WavePlayer_41, WavePlayer_42, WavePlayer_43, WavePlayer_44, WavePlayer_45, WavePlayer_46, WavePlayer_47, WavePlayer_48]
  properties {bShouldExcludeFromBranchCulling=true, weights [1,1,1,1,1,1,1,1]}
 }
 {id=WavePlayer_0, properties {soundWaveAssetPtr=&/Buttercake_Footstep_Walk_FrontLeg_Far_01}}
 {id=WavePlayer_1, properties {soundWaveAssetPtr=&/Buttercake_Footstep_Walk_FrontLeg_Far_02}}
 {id=WavePlayer_35, properties {soundWaveAssetPtr=&Buttercake_Footstep_Walk_FrontLeg_Far_01}}
 {id=WavePlayer_36, properties {soundWaveAssetPtr=&Buttercake_Footstep_Walk_FrontLeg_Far_02}}
 {id=WavePlayer_37, properties {soundWaveAssetPtr=&/Buttercake_Footstep_Walk_FrontLeg_Far_03}}
 {id=WavePlayer_38, properties {soundWaveAssetPtr=&/Buttercake_Footstep_Walk_FrontLeg_Far_04}}
 {id=WavePlayer_39, properties {soundWaveAssetPtr=&/Buttercake_Footstep_Walk_FrontLeg_Far_05}}
 {id=WavePlayer_40, properties {soundWaveAssetPtr=&/Buttercake_Footstep_Walk_FrontLeg_Far_06}}
 {id=WavePlayer_41, properties {soundWaveAssetPtr=&/Buttercake_Footstep_Walk_FrontLeg_Water_01}}
 {id=WavePlayer_42, properties {soundWaveAssetPtr=&/Buttercake_Footstep_Walk_FrontLeg_Water_02}}
 {id=WavePlayer_43, properties {soundWaveAssetPtr=&/Buttercake_Footstep_Walk_FrontLeg_Water_03}}
 {id=WavePlayer_44, properties {soundWaveAssetPtr=&/Buttercake_Footstep_Walk_FrontLeg_Water_04}}
 {id=WavePlayer_45, properties {soundWaveAssetPtr=&/Buttercake_Footstep_Walk_FrontLeg_Water_05}}
 {id=WavePlayer_46, properties {soundWaveAssetPtr=&/Buttercake_Footstep_Walk_FrontLeg_Water_06}}
 {id=WavePlayer_47, properties {soundWaveAssetPtr=&/Buttercake_Footstep_Walk_FrontLeg_Water_07}}
 {id=WavePlayer_48, properties {soundWaveAssetPtr=&/Buttercake_Footstep_Walk_FrontLeg_Water_08}}
 {id=WavePlayer_6, properties {soundWaveAssetPtr=&Buttercake_Footstep_Walk_FrontLeg_Water_01}}
 {id=WavePlayer_7, properties {soundWaveAssetPtr=&Buttercake_Footstep_Walk_FrontLeg_Water_02}}
 {id=WavePlayer_8, properties {soundWaveAssetPtr=&Buttercake_Footstep_Walk_FrontLeg_Water_03}}
 {id=WavePlayer_9, properties {soundWaveAssetPtr=&Buttercake_Footstep_Walk_FrontLeg_Water_04}}
]
properties {
 maxDistance=34583.332
 soundClassObject=/Game/Sounds/FortSoundClassesAndModes/Fort_Creatures
 virtualizationMode=Disabled
 volumeMultiplier=0.473
 concurrencyOverrides {maxCount=2}
 concurrencySet [&/Buttercake_FS_Concurrency]
}
