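# Sound cue for the Buttercake creature's middle-leg running footstep in water, routed to the Fort_Creatures sound class.
# Its only node is a Modulator root with an empty child input, so no sound wave is connected in the graph.
# The cue references the rear-footstep attenuation and shared footstep concurrency assets, lowers the volume, and disables virtualization.
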SoundCue /ButterCake/Audio/Buttercake_FS_Water_Run_MiddleLeg_Cue
root=Modulator_0
nodes [{id=Modulator_0, children [null]}]
properties {
 attenuationSettings=&/Buttercake_Foley_Footsteps_Rear_Att
 maxDistance=21000
 soundClassObject=/Game/Sounds/FortSoundClassesAndModes/Fort_Creatures
 virtualizationMode=Disabled
 volumeMultiplier=0.473
 concurrencyOverrides {maxCount=2}
 concurrencySet [&/Buttercake_FS_Concurrency]
}
