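# Sound cue for Buttercake's rear-leg running footsteps in water, in the creatures sound class with footstep concurrency.
# A quality switch feeds the first two levels from eight random waves and the last two from four.
# Each branch randomizes pitch and volume, then passes through an attenuation node and a mixer with an empty second input.
# Despite the name, it plays front-leg walk waves with front-footstep attenuation nodes, while cue-level attenuation is rear.
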SoundCue /ButterCake/Audio/Buttercake_FS_Water_Run_RearLeg_Cue
root=QualityLevel_0
nodes [
 {id=Attenuation_0, children [Modulator_0], properties {attenuationSettings=&/Buttercake_Foley_Footsteps_Front_Att}}
 {id=Attenuation_2, children [Modulator_3], properties {attenuationSettings=&Buttercake_Foley_Footsteps_Front_Att}}
 {id=Mixer_0, children [Attenuation_0, null], properties {inputVolume [2,0.75]}}
 {id=Mixer_1, children [Attenuation_2, null], properties {inputVolume [2,0.75]}}
 {id=Modulator_0, children [Random_2], properties {pitchMax=1.2, pitchMin=1, volumeMin=0.5}}
 {id=Modulator_3, children [Random_5], properties {pitchMax=1.2, pitchMin=1, volumeMin=0.5}}
 {id=QualityLevel_0, children [Mixer_1,Mixer_1,Mixer_0,Mixer_0]}
 {
  id=Random_2
  children [WavePlayer_6,WavePlayer_7,WavePlayer_8,WavePlayer_9]
  properties {bShouldExcludeFromBranchCulling=true, weights [1,1,1,1]}
 }
 {
  id=Random_5
  children [WavePlayer_41, WavePlayer_42, WavePlayer_43, WavePlayer_44, WavePlayer_45, WavePlayer_46, WavePlayer_47, WavePlayer_48]
  properties {bShouldExcludeFromBranchCulling=true, weights [1,1,1,1,1,1,1,1]}
 }
 {id=WavePlayer_41, properties {soundWaveAssetPtr=&/Buttercake_Footstep_Walk_FrontLeg_Water_01}}
 {id=WavePlayer_42, properties {soundWaveAssetPtr=&/Buttercake_Footstep_Walk_FrontLeg_Water_02}}
 {id=WavePlayer_43, properties {soundWaveAssetPtr=&/Buttercake_Footstep_Walk_FrontLeg_Water_03}}
 {id=WavePlayer_44, properties {soundWaveAssetPtr=&/Buttercake_Footstep_Walk_FrontLeg_Water_04}}
 {id=WavePlayer_45, properties {soundWaveAssetPtr=&/Buttercake_Footstep_Walk_FrontLeg_Water_05}}
 {id=WavePlayer_46, properties {soundWaveAssetPtr=&/Buttercake_Footstep_Walk_FrontLeg_Water_06}}
 {id=WavePlayer_47, properties {soundWaveAssetPtr=&/Buttercake_Footstep_Walk_FrontLeg_Water_07}}
 {id=WavePlayer_48, properties {soundWaveAssetPtr=&/Buttercake_Footstep_Walk_FrontLeg_Water_08}}
 {id=WavePlayer_6, properties {soundWaveAssetPtr=&Buttercake_Footstep_Walk_FrontLeg_Water_01}}
 {id=WavePlayer_7, properties {soundWaveAssetPtr=&Buttercake_Footstep_Walk_FrontLeg_Water_02}}
 {id=WavePlayer_8, properties {soundWaveAssetPtr=&Buttercake_Footstep_Walk_FrontLeg_Water_03}}
 {id=WavePlayer_9, properties {soundWaveAssetPtr=&Buttercake_Footstep_Walk_FrontLeg_Water_04}}
]
properties {
 attenuationSettings=&/Buttercake_Foley_Footsteps_Rear_Att
 maxDistance=34583.332
 soundClassObject=/Game/Sounds/FortSoundClassesAndModes/Fort_Creatures
 virtualizationMode=Disabled
 volumeMultiplier=0.568
 concurrencyOverrides {maxCount=2}
 concurrencySet [&/Buttercake_FS_Concurrency]
}
