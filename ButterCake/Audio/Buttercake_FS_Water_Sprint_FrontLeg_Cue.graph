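# Sound cue for Buttercake's front-leg sprint footsteps in water, a creature-class sound with limited concurrency, built from walk-named recordings.
# Each mixer layers randomized water and far samples through separate pitch/volume modulators and near or far front-footstep attenuations.
# A quality node sends the first two levels to the larger sample pools and the last two to a reduced set of four water and two far waves.
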SoundCue /ButterCake/Audio/Buttercake_FS_Water_Sprint_FrontLeg_Cue
root=QualityLevel_0
nodes [
 {id=Attenuation_0, children [Modulator_0], properties {attenuationSettings=&/Buttercake_Foley_Footsteps_Front_Att}}
 {id=Attenuation_1, children [Modulator_1], properties {attenuationSettings=&/Buttercake_Foley_Footsteps_Front_Far_Att}}
 {id=Attenuation_2, children [Modulator_3], properties {attenuationSettings=&Buttercake_Foley_Footsteps_Front_Far_Att}}
 {id=Attenuation_3, children [Modulator_2], properties {attenuationSettings=&Buttercake_Foley_Footsteps_Front_Att}}
 {id=Mixer_0, children [Attenuation_0, Attenuation_1], properties {inputVolume [2,3]}}
 {id=Mixer_1, children [Attenuation_3, Attenuation_2], properties {inputVolume [2,3]}}
 {id=Modulator_0, children [Random_3], properties {pitchMax=1.2, pitchMin=0.9, volumeMax=1, volumeMin=0.9}}
 {id=Modulator_1, children [Random_0], properties {pitchMax=1.2, pitchMin=1, volumeMax=1, volumeMin=0.35}}
 {id=Modulator_2, children [Random_4], properties {pitchMax=1.2, pitchMin=0.9, volumeMax=1, volumeMin=0.9}}
 {id=Modulator_3, children [Random_5], properties {pitchMax=1.2, pitchMin=1, volumeMax=1, volumeMin=0.35}}
 {id=QualityLevel_0, children [Mixer_0,Mixer_0,Mixer_1,Mixer_1]}
 {
  id=Random_0
  children [WavePlayer_0,WavePlayer_1,WavePlayer_2,WavePlayer_3,WavePlayer_4,WavePlayer_5]
  properties {bShouldExcludeFromBranchCulling=true, weights [1,1,1,1,1,1]}
 }
 {
  id=Random_3
  children [WavePlayer_22, WavePlayer_23, WavePlayer_24, WavePlayer_25, WavePlayer_17, WavePlayer_18, WavePlayer_19, WavePlayer_20]
  properties {bShouldExcludeFromBranchCulling=true, weights [1,1,1,1,1,1,1,1]}
 }
 {
  id=Random_4
  children [WavePlayer_12, WavePlayer_13, WavePlayer_14, WavePlayer_15]
  properties {bShouldExcludeFromBranchCulling=true, weights [1,1,1,1]}
 }
 {id=Random_5, children [WavePlayer_6,WavePlayer_7], properties {bShouldExcludeFromBranchCulling=true, weights [1,1]}}
 {id=WavePlayer_0, properties {soundWaveAssetPtr=&/Buttercake_Footstep_Walk_FrontLeg_Far_01}}
 {id=WavePlayer_1, properties {soundWaveAssetPtr=&/Buttercake_Footstep_Walk_FrontLeg_Far_02}}
 {id=WavePlayer_12, properties {soundWaveAssetPtr=&/Buttercake_Footstep_Walk_FrontLeg_Water_01}}
 {id=WavePlayer_13, properties {soundWaveAssetPtr=&/Buttercake_Footstep_Walk_FrontLeg_Water_02}}
 {id=WavePlayer_14, properties {soundWaveAssetPtr=&/Buttercake_Footstep_Walk_FrontLeg_Water_03}}
 {id=WavePlayer_15, properties {soundWaveAssetPtr=&/Buttercake_Footstep_Walk_FrontLeg_Water_04}}
 {id=WavePlayer_17, properties {soundWaveAssetPtr=&/Buttercake_Footstep_Walk_FrontLeg_Water_05}}
 {id=WavePlayer_18, properties {soundWaveAssetPtr=&/Buttercake_Footstep_Walk_FrontLeg_Water_06}}
 {id=WavePlayer_19, properties {soundWaveAssetPtr=&/Buttercake_Footstep_Walk_FrontLeg_Water_07}}
 {id=WavePlayer_2, properties {soundWaveAssetPtr=&/Buttercake_Footstep_Walk_FrontLeg_Far_03}}
 {id=WavePlayer_20, properties {soundWaveAssetPtr=&/Buttercake_Footstep_Walk_FrontLeg_Water_08}}
 {id=WavePlayer_22, properties {soundWaveAssetPtr=&Buttercake_Footstep_Walk_FrontLeg_Water_01}}
 {id=WavePlayer_23, properties {soundWaveAssetPtr=&Buttercake_Footstep_Walk_FrontLeg_Water_02}}
 {id=WavePlayer_24, properties {soundWaveAssetPtr=&Buttercake_Footstep_Walk_FrontLeg_Water_03}}
 {id=WavePlayer_25, properties {soundWaveAssetPtr=&Buttercake_Footstep_Walk_FrontLeg_Water_04}}
 {id=WavePlayer_3, properties {soundWaveAssetPtr=&/Buttercake_Footstep_Walk_FrontLeg_Far_04}}
 {id=WavePlayer_4, properties {soundWaveAssetPtr=&/Buttercake_Footstep_Walk_FrontLeg_Far_05}}
 {id=WavePlayer_5, properties {soundWaveAssetPtr=&/Buttercake_Footstep_Walk_FrontLeg_Far_06}}
 {id=WavePlayer_6, properties {soundWaveAssetPtr=&Buttercake_Footstep_Walk_FrontLeg_Far_01}}
 {id=WavePlayer_7, properties {soundWaveAssetPtr=&Buttercake_Footstep_Walk_FrontLeg_Far_02}}
]
properties {
 maxDistance=34583.332
 soundClassObject=/Game/Sounds/FortSoundClassesAndModes/Fort_Creatures
 virtualizationMode=Disabled
 volumeMultiplier=0.668
 concurrencyOverrides {maxCount=2}
 concurrencySet [&/Buttercake_FS_Concurrency]
}
