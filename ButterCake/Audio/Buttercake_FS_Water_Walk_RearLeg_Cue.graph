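# Sound cue for Buttercake's rear-leg walking footsteps in water, built from front-leg water wave assets.
# A quality-level node sends the first two levels to a random pick of eight waves and the last two to four.
# Both branches add slight pitch and volume modulation.
# Playback uses rear footstep attenuation, the creatures sound class and footstep concurrency, with virtualization disabled.
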SoundCue /ButterCake/Audio/Buttercake_FS_Water_Walk_RearLeg_Cue
root=QualityLevel_0
nodes [
 {id=Modulator_0, children [Random_1], properties {pitchMax=1.2, pitchMin=0.9, volumeMax=1, volumeMin=0.9}}
 {id=Modulator_1, children [Random_3], properties {pitchMax=1.2, pitchMin=0.9, volumeMax=1, volumeMin=0.9}}
 {id=QualityLevel_0, children [Modulator_1,Modulator_1,Modulator_0,Modulator_0]}
 {
  id=Random_1
  children [WavePlayer_0,WavePlayer_1,WavePlayer_2,WavePlayer_3]
  properties {bShouldExcludeFromBranchCulling=true, weights [1,1,1,1]}
 }
 {
  id=Random_3
  children [WavePlayer_22, WavePlayer_23, WavePlayer_24, WavePlayer_25, WavePlayer_17, WavePlayer_18, WavePlayer_19, WavePlayer_20]
  properties {bShouldExcludeFromBranchCulling=true, weights [1,1,1,1,1,1,1,1]}
 }
 {id=WavePlayer_0, properties {soundWaveAssetPtr=&/Buttercake_Footstep_Walk_FrontLeg_Water_01}}
 {id=WavePlayer_1, properties {soundWaveAssetPtr=&/Buttercake_Footstep_Walk_FrontLeg_Water_02}}
 {id=WavePlayer_17, properties {soundWaveAssetPtr=&/Buttercake_Footstep_Walk_FrontLeg_Water_05}}
 {id=WavePlayer_18, properties {soundWaveAssetPtr=&/Buttercake_Footstep_Walk_FrontLeg_Water_06}}
 {id=WavePlayer_19, properties {soundWaveAssetPtr=&/Buttercake_Footstep_Walk_FrontLeg_Water_07}}
 {id=WavePlayer_2, properties {soundWaveAssetPtr=&/Buttercake_Footstep_Walk_FrontLeg_Water_03}}
 {id=WavePlayer_20, properties {soundWaveAssetPtr=&/Buttercake_Footstep_Walk_FrontLeg_Water_08}}
 {id=WavePlayer_22, properties {soundWaveAssetPtr=&Buttercake_Footstep_Walk_FrontLeg_Water_01}}
 {id=WavePlayer_23, properties {soundWaveAssetPtr=&Buttercake_Footstep_Walk_FrontLeg_Water_02}}
 {id=WavePlayer_24, properties {soundWaveAssetPtr=&Buttercake_Footstep_Walk_FrontLeg_Water_03}}
 {id=WavePlayer_25, properties {soundWaveAssetPtr=&/Buttercake_Footstep_Walk_FrontLeg_Water_04}}
 {id=WavePlayer_3, properties {soundWaveAssetPtr=&Buttercake_Footstep_Walk_FrontLeg_Water_04}}
]
properties {
 attenuationSettings=&/Buttercake_Foley_Footsteps_Rear_Att
 maxDistance=21000
 soundClassObject=/Game/Sounds/FortSoundClassesAndModes/Fort_Creatures
 virtualizationMode=Disabled
 volumeMultiplier=0.473
 concurrencyOverrides {maxCount=2}
 concurrencySet [&/Buttercake_FS_Concurrency]
}
