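# Sound cue named for the head-turn sound of the Buttercake creature, currently empty with no root or sound nodes, so it has nothing to play.
# It holds only playback settings: the Creatures sound class, reduced volume and the attenuation asset of the front footsteps foley.
# A concurrency override caps simultaneous instances at a small count, and virtualization is disabled.
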SoundCue /ButterCake/Audio/Buttercake_Head_Turn_Cue
root=null
nodes []
properties {
 attenuationSettings=&/Buttercake_Foley_Footsteps_Front_Att
 bOverrideConcurrency=true
 maxDistance=31500
 soundClassObject=/Game/Sounds/FortSoundClassesAndModes/Fort_Creatures
 virtualizationMode=Disabled
 volumeMultiplier=0.841
 concurrencyOverrides {maxCount=2}
}
