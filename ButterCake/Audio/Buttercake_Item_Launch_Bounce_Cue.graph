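# Sound cue for the ground impact when an item launched by the Buttercake creature bounces, routed to the creatures sound class.
# A Modulator varies playback over a quality-level switch: the first two slots share a five-wave random pool, the last two a two-wave pool.
# Item-drop attenuation shapes falloff, a concurrency override caps it at two simultaneous instances, and virtualization is disabled.
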SoundCue /ButterCake/Audio/Buttercake_Item_Launch_Bounce_Cue
root=Modulator_0
nodes [
 {id=Modulator_0, children [QualityLevel_0]}
 {id=QualityLevel_0, children [Random_0,Random_0,Random_1,Random_1]}
 {
  id=Random_0
  children [WavePlayer_0,WavePlayer_1,WavePlayer_2,WavePlayer_3,WavePlayer_4]
  properties {bShouldExcludeFromBranchCulling=true, weights [1,1,1,1,1]}
 }
 {id=Random_1, children [WavePlayer_5,WavePlayer_6], properties {bShouldExcludeFromBranchCulling=true, weights [1,1]}}
 {id=WavePlayer_0, properties {soundWaveAssetPtr=&/Buttercake_Item_Launch_Ground_Impact_01}}
 {id=WavePlayer_1, properties {soundWaveAssetPtr=&/Buttercake_Item_Launch_Ground_Impact_02}}
 {id=WavePlayer_2, properties {soundWaveAssetPtr=&/Buttercake_Item_Launch_Ground_Impact_03}}
 {id=WavePlayer_3, properties {soundWaveAssetPtr=&/Buttercake_Item_Launch_Ground_Impact_04}}
 {id=WavePlayer_4, properties {soundWaveAssetPtr=&/Buttercake_Item_Launch_Ground_Impact_05}}
 {id=WavePlayer_5, properties {soundWaveAssetPtr=&Buttercake_Item_Launch_Ground_Impact_01}}
 {id=WavePlayer_6, properties {soundWaveAssetPtr=&Buttercake_Item_Launch_Ground_Impact_02}}
]
properties {
 attenuationSettings=&/Buttercake_Item_Drop_Att
 bOverrideConcurrency=true
 maxDistance=28500
 soundClassObject=/Game/Sounds/FortSoundClassesAndModes/Fort_Creatures
 virtualizationMode=Disabled
 volumeMultiplier=0.631
 concurrencyOverrides {maxCount=2}
}
