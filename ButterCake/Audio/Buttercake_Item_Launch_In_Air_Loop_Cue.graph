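# Sound cue for the Buttercake item launch in-air sound, playing a launch whoosh under the creatures sound class.
# A modulator feeds a quality-level switch: the first two levels pick randomly among three equally weighted whoosh waves, while the last two use only the first variant.
# Launch attenuation shapes it in space, with a concurrency override capping simultaneous instances and virtualization disabled.
# Despite the Loop name, no looping node appears in the graph.
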SoundCue /ButterCake/Audio/Buttercake_Item_Launch_In_Air_Loop_Cue
root=Modulator_1
nodes [
 {id=Modulator_1, children [QualityLevel_0]}
 {id=QualityLevel_0, children [Random_0,Random_0,WavePlayer_4,WavePlayer_4]}
 {id=Random_0, children [WavePlayer_1,WavePlayer_2,WavePlayer_3], properties {bShouldExcludeFromBranchCulling=true, weights [1,1,1]}}
 {id=WavePlayer_1, properties {soundWaveAssetPtr=&/Buttercake_Item_Launch_Ground_Whoosh_01}}
 {id=WavePlayer_2, properties {soundWaveAssetPtr=&/Buttercake_Item_Launch_Ground_Whoosh_02}}
 {id=WavePlayer_3, properties {soundWaveAssetPtr=&/Buttercake_Item_Launch_Ground_Whoosh_03}}
 {id=WavePlayer_4, properties {soundWaveAssetPtr=&Buttercake_Item_Launch_Ground_Whoosh_01}}
]
properties {
 attenuationSettings=&/Buttercake_Item_Launch_Att
 bOverrideConcurrency=true
 maxDistance=27000
 soundClassObject=/Game/Sounds/FortSoundClassesAndModes/Fort_Creatures
 virtualizationMode=Disabled
 volumeMultiplier=1.262
 concurrencyOverrides {maxCount=2}
}
