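# Sound cue for the ground impact when a launched Buttercake item lands.
# A quality-level node picks an equal-weight random pool of five impact waves for the first two levels, or two for the last two.
# A root Modulator varies the chosen wave's pitch and volume.
# It uses the item-drop attenuation and creatures sound class, caps concurrency at two and disables virtualization.
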SoundCue /ButterCake/Audio/Buttercake_Item_Launch_Land_Cue
root=Modulator_0
nodes [
 {id=Modulator_0, children [QualityLevel_0]}
 {id=QualityLevel_0, children [Random_1,Random_1,Random_2,Random_2]}
 {
  id=Random_1
  children [WavePlayer_5,WavePlayer_6,WavePlayer_7,WavePlayer_8,WavePlayer_9]
  properties {bShouldExcludeFromBranchCulling=true, weights [1,1,1,1,1]}
 }
 {id=Random_2, children [WavePlayer_10, WavePlayer_11], properties {bShouldExcludeFromBranchCulling=true, weights [1,1]}}
 {id=WavePlayer_10, properties {soundWaveAssetPtr=&/Buttercake_Item_Launch_Ground_Impact_01}}
 {id=WavePlayer_11, properties {soundWaveAssetPtr=&/Buttercake_Item_Launch_Ground_Impact_02}}
 {id=WavePlayer_5, properties {soundWaveAssetPtr=&Buttercake_Item_Launch_Ground_Impact_01}}
 {id=WavePlayer_6, properties {soundWaveAssetPtr=&Buttercake_Item_Launch_Ground_Impact_02}}
 {id=WavePlayer_7, properties {soundWaveAssetPtr=&/Buttercake_Item_Launch_Ground_Impact_03}}
 {id=WavePlayer_8, properties {soundWaveAssetPtr=&/Buttercake_Item_Launch_Ground_Impact_04}}
 {id=WavePlayer_9, properties {soundWaveAssetPtr=&/Buttercake_Item_Launch_Ground_Impact_05}}
]
properties {
 attenuationSettings=&/Buttercake_Item_Drop_Att
 bOverrideConcurrency=true
 maxDistance=28500
 soundClassObject=/Game/Sounds/FortSoundClassesAndModes/Fort_Creatures
 virtualizationMode=Disabled
 volumeMultiplier=1.052
 concurrencyOverrides {maxCount=2}
}
