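# Sound cue for Buttercake's jowl sound as it shifts from happy to neutral, routed to the creature sound class.
# A random node picks one of five equally weighted wave variations, and a modulator varies its volume while keeping pitch fixed.
# Only the first quality-level slot is wired.
# Shared jowls attenuation and concurrency assets apply, with virtualization disabled and the lowest priority.
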SoundCue /ButterCake/Audio/Buttercake_Jowls_Happy_To_Neutral_Cue
root=QualityLevel_0
nodes [
 {id=Modulator_0, children [Random_0], properties {pitchMax=1, pitchMin=1, volumeMax=1, volumeMin=0.5}}
 {id=QualityLevel_0, children [Modulator_0,null,null,null]}
 {
  id=Random_0
  children [WavePlayer_0,WavePlayer_1,WavePlayer_2,WavePlayer_3,WavePlayer_4]
  properties {bShouldExcludeFromBranchCulling=true, weights [1,1,1,1,1]}
 }
 {id=WavePlayer_0, properties {soundWaveAssetPtr=&/Buttercake_Jowls_Happy_To_Neutral_01}}
 {id=WavePlayer_1, properties {soundWaveAssetPtr=&/Buttercake_Jowls_Happy_To_Neutral_02}}
 {id=WavePlayer_2, properties {soundWaveAssetPtr=&/Buttercake_Jowls_Happy_To_Neutral_03}}
 {id=WavePlayer_3, properties {soundWaveAssetPtr=&/Buttercake_Jowls_Happy_To_Neutral_04}}
 {id=WavePlayer_4, properties {soundWaveAssetPtr=&/Buttercake_Jowls_Happy_To_Neutral_05}}
]
properties {
 attenuationSettings=&/Buttercake_Jowls_Att
 maxDistance=6000
 priority=0
 soundClassObject=/Game/Sounds/FortSoundClassesAndModes/Fort_Creatures
 virtualizationMode=Disabled
 volumeMultiplier=1.262
 concurrencySet [&/Buttercake_Jowls_Concurrency]
}
