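# Sound cue for the Buttercake creature's neutral jowl movement sounds, routed to the Fort_Creatures sound class.
# A Random node picks among eight equally weighted wave variations and feeds a Modulator that randomizes volume while keeping pitch fixed.
# A quality-level root uses only its first slot for that chain.
# Playback uses dedicated jowl attenuation and concurrency assets, with virtualization disabled.
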SoundCue /ButterCake/Audio/Buttercake_Jowls_Neutral_Moving_Cue
root=QualityLevel_0
nodes [
 {id=Modulator_0, children [Random_1], properties {pitchMax=1, pitchMin=1, volumeMax=1, volumeMin=0.5}}
 {id=QualityLevel_0, children [Modulator_0,null,null,null]}
 {
  id=Random_1
  children [WavePlayer_5, WavePlayer_6, WavePlayer_7, WavePlayer_8, WavePlayer_9, WavePlayer_10, WavePlayer_11, WavePlayer_12]
  properties {bShouldExcludeFromBranchCulling=true, weights [1,1,1,1,1,1,1,1]}
 }
 {id=WavePlayer_10, properties {soundWaveAssetPtr=&/Buttercake_Jowls_Neutral_Moving_06}}
 {id=WavePlayer_11, properties {soundWaveAssetPtr=&/Buttercake_Jowls_Neutral_Moving_07}}
 {id=WavePlayer_12, properties {soundWaveAssetPtr=&/Buttercake_Jowls_Neutral_Moving_08}}
 {id=WavePlayer_5, properties {soundWaveAssetPtr=&/Buttercake_Jowls_Neutral_Moving_01}}
 {id=WavePlayer_6, properties {soundWaveAssetPtr=&/Buttercake_Jowls_Neutral_Moving_02}}
 {id=WavePlayer_7, properties {soundWaveAssetPtr=&/Buttercake_Jowls_Neutral_Moving_03}}
 {id=WavePlayer_8, properties {soundWaveAssetPtr=&/Buttercake_Jowls_Neutral_Moving_04}}
 {id=WavePlayer_9, properties {soundWaveAssetPtr=&/Buttercake_Jowls_Neutral_Moving_05}}
]
properties {
 attenuationSettings=&/Buttercake_Jowls_Att
 maxDistance=6000
 priority=0
 soundClassObject=/Game/Sounds/FortSoundClassesAndModes/Fort_Creatures
 virtualizationMode=Disabled
 volumeMultiplier=1.262
 concurrencySet [&/Buttercake_Jowls_Concurrency]
}
