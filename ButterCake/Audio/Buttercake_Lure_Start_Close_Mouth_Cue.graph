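# Sound cue for the Buttercake creature's closed-mouth vocal at the start of a lure.
# A Random node picks one of three equally weighted wave variations and is excluded from branch culling.
# The cue uses Buttercake vocal attenuation and the creatures sound class, with a concurrency override allowing a single instance and virtualization disabled.
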SoundCue /ButterCake/Audio/Buttercake_Lure_Start_Close_Mouth_Cue
root=Random_0
nodes [
 {id=Random_0, children [WavePlayer_0,WavePlayer_1,WavePlayer_2], properties {bShouldExcludeFromBranchCulling=true, weights [1,1,1]}}
 {id=WavePlayer_0, properties {soundWaveAssetPtr=&/Buttercake_Vox_Lure_Start_Close_Mouth_01}}
 {id=WavePlayer_1, properties {soundWaveAssetPtr=&/Buttercake_Vox_Lure_Start_Close_Mouth_02}}
 {id=WavePlayer_2, properties {soundWaveAssetPtr=&/Buttercake_Vox_Lure_Start_Close_Mouth_03}}
]
properties {
 attenuationSettings=&/Buttercake_Vox_Att
 bOverrideConcurrency=true
 maxDistance=25000
 priority=25
 soundClassObject=/Game/Sounds/FortSoundClassesAndModes/Fort_Creatures
 virtualizationMode=Disabled
 volumeMultiplier=0.496
 concurrencyOverrides {maxCount=1}
 concurrencySet [null]
}
