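# Sound cue named for the start of Buttercake's open-mouth lure, set up as a creature sound.
# Its node graph is empty with no root node, so the cue as stored references no audio to play.
# It carries only playback settings: the Buttercake vocal attenuation, the Fortnite creatures sound class and a raised volume.
# A concurrency override limits it to one instance, and virtualization is disabled.
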SoundCue /ButterCake/Audio/Buttercake_Lure_Start_Open_Mouth_Cue
root=null
nodes []
properties {
 attenuationSettings=&/Buttercake_Vox_Att
 bOverrideConcurrency=true
 maxDistance=16500
 priority=30
 soundClassObject=/Game/Sounds/FortSoundClassesAndModes/Fort_Creatures
 virtualizationMode=Disabled
 volumeMultiplier=1.125
 concurrencyOverrides {maxCount=1}
 concurrencySet [null]
}
