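# Sound cue for the exhale burst of Buttercake's sneeze vocal, played through the creatures sound class with Buttercake vox attenuation.
# A quality-level node sends its first two levels to a random pick of three exhale waves and its last two to a random pick of two.
# Each random set passes through a modulator with volume fixed at unity.
# A concurrency override allows one instance at a time, and virtualization is disabled.
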SoundCue /ButterCake/Audio/Buttercake_Sneeze_Burst_Cue
root=QualityLevel_0
nodes [
 {id=Modulator_0, children [Random_1], properties {volumeMax=1, volumeMin=1}}
 {id=Modulator_1, children [Random_0], properties {volumeMax=1, volumeMin=1}}
 {id=QualityLevel_0, children [Modulator_0,Modulator_0,Modulator_1,Modulator_1]}
 {id=Random_0, children [WavePlayer_3,WavePlayer_4], properties {bShouldExcludeFromBranchCulling=true, weights [1,1]}}
 {id=Random_1, children [WavePlayer_0,WavePlayer_1,WavePlayer_2], properties {bShouldExcludeFromBranchCulling=true, weights [1,1,1]}}
 {id=WavePlayer_0, properties {soundWaveAssetPtr=&/Buttercake_Vox_Sneeze_Exhale_01}}
 {id=WavePlayer_1, properties {soundWaveAssetPtr=&/Buttercake_Vox_Sneeze_Exhale_02}}
 {id=WavePlayer_2, properties {soundWaveAssetPtr=&/Buttercake_Vox_Sneeze_Exhale_03}}
 {id=WavePlayer_3, properties {soundWaveAssetPtr=&Buttercake_Vox_Sneeze_Exhale_01}}
 {id=WavePlayer_4, properties {soundWaveAssetPtr=&Buttercake_Vox_Sneeze_Exhale_02}}
]
properties {
 attenuationSettings=&/Buttercake_Vox_Att
 bOverrideConcurrency=true
 maxDistance=25000
 priority=20
 soundClassObject=/Game/Sounds/FortSoundClassesAndModes/Fort_Creatures
 virtualizationMode=Disabled
 volumeMultiplier=2.314
 concurrencyOverrides {maxCount=1}
}
