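# Sound cue for a long sneeze by the Buttercake creature, routed to the Fort_Creatures sound class and Buttercake vox attenuation.
# Its graph is a lone Modulator with fixed volume and an empty child input, so no wave player is connected.
# A concurrency override allows one instance at a time, with raised priority, slightly boosted volume and virtualization disabled.
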SoundCue /ButterCake/Audio/Buttercake_Sneeze_Long_Cue
root=Modulator_0
nodes [{id=Modulator_0, children [null], properties {volumeMax=1, volumeMin=1}}]
properties {
 attenuationSettings=&/Buttercake_Vox_Att
 bOverrideConcurrency=true
 maxDistance=25000
 priority=20
 soundClassObject=/Game/Sounds/FortSoundClassesAndModes/Fort_Creatures
 virtualizationMode=Disabled
 volumeMultiplier=1.052
 concurrencyOverrides {maxCount=1}
}
