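# Sound cue for Buttercake's short sneeze exhale vocal, routed to the creatures sound class.
# A quality-level node sends its first two levels to a random pick of three exhale waves and its last two to a random pick of two, both equally weighted and excluded from branch culling.
# It uses Buttercake vocal attenuation, boosts volume, disables virtualization and overrides concurrency to allow one instance at a time.
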SoundCue /ButterCake/Audio/Buttercake_Sneeze_Short_Exhale_Cue
root=QualityLevel_0
nodes [
 {id=QualityLevel_0, children [Random_2,Random_2,Random_0,Random_0]}
 {id=Random_0, children [WavePlayer_0,WavePlayer_1], properties {bShouldExcludeFromBranchCulling=true, weights [1,1]}}
 {id=Random_2, children [WavePlayer_6,WavePlayer_7,WavePlayer_8], properties {bShouldExcludeFromBranchCulling=true, weights [1,1,1]}}
 {id=WavePlayer_0, properties {soundWaveAssetPtr=&/Buttercake_Vox_Sneeze_Short_Exhale_01}}
 {id=WavePlayer_1, properties {soundWaveAssetPtr=&/Buttercake_Vox_Sneeze_Short_Exhale_02}}
 {id=WavePlayer_6, properties {soundWaveAssetPtr=&Buttercake_Vox_Sneeze_Short_Exhale_01}}
 {id=WavePlayer_7, properties {soundWaveAssetPtr=&Buttercake_Vox_Sneeze_Short_Exhale_02}}
 {id=WavePlayer_8, properties {soundWaveAssetPtr=&/Buttercake_Vox_Sneeze_Short_Exhale_03}}
]
properties {
 attenuationSettings=&/Buttercake_Vox_Att
 bOverrideConcurrency=true
 maxDistance=25000
 priority=20
 soundClassObject=/Game/Sounds/FortSoundClassesAndModes/Fort_Creatures
 virtualizationMode=Disabled
 volumeMultiplier=2.103
 concurrencyOverrides {maxCount=1}
}
