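# Sound cue for Buttercake's short sneeze inhale vocal, a creature-class sound with shared Buttercake vox attenuation.
# A quality-level node sends its first two tiers to a random pick of three inhale waves and its last two to a pick of two.
# Both random nodes use equal weights and are excluded from branch culling.
# A concurrency override allows one instance at a time, and virtualization is disabled.
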SoundCue /ButterCake/Audio/Buttercake_Sneeze_Short_Inhale_Cue
root=QualityLevel_0
nodes [
 {id=QualityLevel_0, children [Random_0,Random_0,Random_1,Random_1]}
 {id=Random_0, children [WavePlayer_3,WavePlayer_4,WavePlayer_5], properties {bShouldExcludeFromBranchCulling=true, weights [1,1,1]}}
 {id=Random_1, children [WavePlayer_0,WavePlayer_1], properties {bShouldExcludeFromBranchCulling=true, weights [1,1]}}
 {id=WavePlayer_0, properties {soundWaveAssetPtr=&/Buttercake_Vox_Sneeze_Short_Inhale_01}}
 {id=WavePlayer_1, properties {soundWaveAssetPtr=&/Buttercake_Vox_Sneeze_Short_Inhale_02}}
 {id=WavePlayer_3, properties {soundWaveAssetPtr=&Buttercake_Vox_Sneeze_Short_Inhale_01}}
 {id=WavePlayer_4, properties {soundWaveAssetPtr=&Buttercake_Vox_Sneeze_Short_Inhale_02}}
 {id=WavePlayer_5, properties {soundWaveAssetPtr=&/Buttercake_Vox_Sneeze_Short_Inhale_03}}
]
properties {
 attenuationSettings=&/Buttercake_Vox_Att
 bOverrideConcurrency=true
 maxDistance=25000
 priority=20
 soundClassObject=/Game/Sounds/FortSoundClassesAndModes/Fort_Creatures
 virtualizationMode=Disabled
 volumeMultiplier=2.103
 concurrencyOverrides {maxCount=1}
}
