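# Sound cue named for the Buttercake vacuum sound, classed under the Fort_Creatures sound class.
# The root Modulator node has an empty child input, so the cue references no sound wave.
# It uses the Buttercake vox attenuation asset and disables virtualization.
# A concurrency override limits it to a single instance per owner.
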SoundCue /ButterCake/Audio/Buttercake_Vacuum_Cue
root=Modulator_0
nodes [{id=Modulator_0, children [null]}]
properties {
 attenuationSettings=&/Buttercake_Vox_Att
 bOverrideConcurrency=true
 maxDistance=25000
 soundClassObject=/Game/Sounds/FortSoundClassesAndModes/Fort_Creatures
 virtualizationMode=Disabled
 volumeMultiplier=0.841
 concurrencyOverrides {bLimitToOwner=true, maxCount=1}
}
