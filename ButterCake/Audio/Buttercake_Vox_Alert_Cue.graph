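# Sound cue for Buttercake's alert vocalization in the creatures sound class, built from Buttercake walk vox waves.
# A quality-level root sends its first two slots to an equal-weight random pick of 18 waves and its last two to a 12-wave subset.
# Both random nodes are excluded from branch culling.
# The cue uses the Buttercake attacks attenuation and vox state concurrency, with virtualization disabled.
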SoundCue /ButterCake/Audio/Buttercake_Vox_Alert_Cue
root=QualityLevel_0
nodes [
 {id=QualityLevel_0, children [Random_1,Random_1,Random_0,Random_0]}
 {
  id=Random_0
  children [
   WavePlayer_0
   WavePlayer_1
   WavePlayer_2
   WavePlayer_3
   WavePlayer_4
   WavePlayer_5
   WavePlayer_6
   WavePlayer_7
   WavePlayer_8
   WavePlayer_9
   WavePlayer_10
   WavePlayer_11
  ]
  properties {bShouldExcludeFromBranchCulling=true, weights [1,1,1,1,1,1,1,1,1,1,1,1]}
 }
 {
  id=Random_1
  children [
   WavePlayer_12
   WavePlayer_13
   WavePlayer_14
   WavePlayer_15
   WavePlayer_16
   WavePlayer_17
   WavePlayer_18
   WavePlayer_19
   WavePlayer_20
   WavePlayer_21
   WavePlayer_22
   WavePlayer_23
   WavePlayer_24
   WavePlayer_25
   WavePlayer_26
   WavePlayer_27
   WavePlayer_28
   WavePlayer_29
  ]
  properties {bShouldExcludeFromBranchCulling=true, weights [1,1,1,1,1,1,1,1,1,1,1,1,1,1,1,1,1,1]}
 }
 {id=WavePlayer_0, properties {soundWaveAssetPtr=&/Buttercake_Vox_Walk_01}}
 {id=WavePlayer_1, properties {soundWaveAssetPtr=&/Buttercake_Vox_Walk_02}}
 {id=WavePlayer_10, properties {soundWaveAssetPtr=&/Buttercake_Vox_Walk_17}}
 {id=WavePlayer_11, properties {soundWaveAssetPtr=&/Buttercake_Vox_Walk_18}}
 {id=WavePlayer_12, properties {soundWaveAssetPtr=&Buttercake_Vox_Walk_01}}
 {id=WavePlayer_13, properties {soundWaveAssetPtr=&Buttercake_Vox_Walk_02}}
 {id=WavePlayer_14, properties {soundWaveAssetPtr=&/Buttercake_Vox_Walk_03}}
 {id=WavePlayer_15, properties {soundWaveAssetPtr=&/Buttercake_Vox_Walk_04}}
 {id=WavePlayer_16, properties {soundWaveAssetPtr=&/Buttercake_Vox_Walk_05}}
 {id=WavePlayer_17, properties {soundWaveAssetPtr=&/Buttercake_Vox_Walk_06}}
 {id=WavePlayer_18, properties {soundWaveAssetPtr=&/Buttercake_Vox_Walk_07}}
 {id=WavePlayer_19, properties {soundWaveAssetPtr=&/Buttercake_Vox_Walk_08}}
 {id=WavePlayer_2, properties {soundWaveAssetPtr=&Buttercake_Vox_Walk_03}}
 {id=WavePlayer_20, properties {soundWaveAssetPtr=&/Buttercake_Vox_Walk_09}}
 {id=WavePlayer_21, properties {soundWaveAssetPtr=&/Buttercake_Vox_Walk_10}}
 {id=WavePlayer_22, properties {soundWaveAssetPtr=&/Buttercake_Vox_Walk_11}}
 {id=WavePlayer_23, properties {soundWaveAssetPtr=&/Buttercake_Vox_Walk_12}}
 {id=WavePlayer_24, properties {soundWaveAssetPtr=&/Buttercake_Vox_Walk_13}}
 {id=WavePlayer_25, properties {soundWaveAssetPtr=&/Buttercake_Vox_Walk_14}}
 {id=WavePlayer_26, properties {soundWaveAssetPtr=&/Buttercake_Vox_Walk_15}}
 {id=WavePlayer_27, properties {soundWaveAssetPtr=&/Buttercake_Vox_Walk_16}}
 {id=WavePlayer_28, properties {soundWaveAssetPtr=&Buttercake_Vox_Walk_17}}
 {id=WavePlayer_29, properties {soundWaveAssetPtr=&Buttercake_Vox_Walk_18}}
 {id=WavePlayer_3, properties {soundWaveAssetPtr=&Buttercake_Vox_Walk_04}}
 {id=WavePlayer_4, properties {soundWaveAssetPtr=&Buttercake_Vox_Walk_05}}
 {id=WavePlayer_5, properties {soundWaveAssetPtr=&Buttercake_Vox_Walk_06}}
 {id=WavePlayer_6, properties {soundWaveAssetPtr=&Buttercake_Vox_Walk_07}}
 {id=WavePlayer_7, properties {soundWaveAssetPtr=&Buttercake_Vox_Walk_08}}
 {id=WavePlayer_8, properties {soundWaveAssetPtr=&Buttercake_Vox_Walk_09}}
 {id=WavePlayer_9, properties {soundWaveAssetPtr=&Buttercake_Vox_Walk_10}}
]
properties {
 attenuationSettings=&/Buttercake_Attacks_Att
 maxDistance=35833.332
 priority=20
 soundClassObject=/Game/Sounds/FortSoundClassesAndModes/Fort_Creatures
 virtualizationMode=Disabled
 volumeMultiplier=1
 concurrencySet [&/Buttercake_Vox_State_Concurrency]
}
