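# SoundCue for Buttercake's ranged attack vocalization, classed under the creatures sound class.
# A Random node picks between two equally weighted wave players and is excluded from branch culling.
# It uses the critical vox attenuation and the shared vox state concurrency, with boosted volume and virtualization disabled.
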SoundCue /ButterCake/Audio/Buttercake_Vox_Attack_Ranged_Cue
root=Random_1
nodes [
 {id=Random_1, children [WavePlayer_4,WavePlayer_5], properties {bShouldExcludeFromBranchCulling=true, weights [1,1]}}
 {id=WavePlayer_4, properties {soundWaveAssetPtr=&/Buttercake_Vox_Attack_Ranged_01}}
 {id=WavePlayer_5, properties {soundWaveAssetPtr=&/Buttercake_Vox_Attack_Ranged_02}}
]
properties {
 attenuationSettings=&/Buttercake_Vox_Critical_Att
 maxDistance=25000
 priority=30
 soundClassObject=/Game/Sounds/FortSoundClassesAndModes/Fort_Creatures
 virtualizationMode=Disabled
 volumeMultiplier=1.888
 concurrencySet [&/Buttercake_Vox_State_Concurrency]
}
