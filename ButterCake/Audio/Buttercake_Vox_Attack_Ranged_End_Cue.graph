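# Sound cue for Buttercake's vocalization at the end of its ranged attack.
# A Random node picks one of three equally weighted wave variants and is excluded from branch culling.
# It plays in the Fort_Creatures sound class with critical vox attenuation, boosted volume and long range.
# A shared vox state concurrency limits overlapping playback, and virtualization is disabled.
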SoundCue /ButterCake/Audio/Buttercake_Vox_Attack_Ranged_End_Cue
root=Random_0
nodes [
 {id=Random_0, children [WavePlayer_0,WavePlayer_1,WavePlayer_2], properties {bShouldExcludeFromBranchCulling=true, weights [1,1,1]}}
 {id=WavePlayer_0, properties {soundWaveAssetPtr=&/Buttercake_Attack_Ranged_End_01}}
 {id=WavePlayer_1, properties {soundWaveAssetPtr=&/Buttercake_Attack_Ranged_End_02}}
 {id=WavePlayer_2, properties {soundWaveAssetPtr=&/Buttercake_Attack_Ranged_End_03}}
]
properties {
 attenuationSettings=&/Buttercake_Vox_Critical_Att
 maxDistance=25000
 priority=35
 soundClassObject=/Game/Sounds/FortSoundClassesAndModes/Fort_Creatures
 virtualizationMode=Disabled
 volumeMultiplier=2.518
 concurrencySet [&/Buttercake_Vox_State_Concurrency]
}
