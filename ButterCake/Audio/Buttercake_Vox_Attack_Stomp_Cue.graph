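# Sound cue for Buttercake's stomp attack vocal, built from rear-whoosh-down voice takes.
# A Random node picks one of three equally weighted wave players and is excluded from branch culling.
# It plays in the creatures sound class with a critical vocal attenuation asset, boosted volume and a raised priority.
# A concurrency override limits it to a single instance, and virtualization is disabled.
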SoundCue /ButterCake/Audio/Buttercake_Vox_Attack_Stomp_Cue
root=Random_1
nodes [
 {id=Random_1, children [WavePlayer_3,WavePlayer_4,WavePlayer_5], properties {bShouldExcludeFromBranchCulling=true, weights [1,1,1]}}
 {id=WavePlayer_3, properties {soundWaveAssetPtr=&/Buttercake_Vox_Rear_Whoosh_Down_v2_03}}
 {id=WavePlayer_4, properties {soundWaveAssetPtr=&/Buttercake_Vox_Rear_Whoosh_Down_v2_01}}
 {id=WavePlayer_5, properties {soundWaveAssetPtr=&/Buttercake_Vox_Rear_Whoosh_Down_v2_02}}
]
properties {
 attenuationSettings=&/Buttercake_Vox_Critical_Att
 bOverrideConcurrency=true
 maxDistance=25000
 priority=30
 soundClassObject=/Game/Sounds/FortSoundClassesAndModes/Fort_Creatures
 virtualizationMode=Disabled
 volumeMultiplier=4
 concurrencyOverrides {maxCount=1}
}
