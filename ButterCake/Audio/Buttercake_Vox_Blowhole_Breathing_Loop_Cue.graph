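# Looping Sound Cue for Buttercake's blowhole breathing vocals, routed to the creatures sound class with its own attenuation.
# A loop repeats a short random delay into a pitch and volume modulator, which plays one of five equally weighted breath waves for varied repetition.
# Only the first quality-level branch is connected.
# Concurrency allows a single instance and prevents new ones, and the cue has the lowest priority.
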SoundCue /ButterCake/Audio/Buttercake_Vox_Blowhole_Breathing_Loop_Cue
root=QualityLevel_0
nodes [
 {id=Delay_0, children [Modulator_1], properties {delayMax=1}}
 {id=Looping_0, children [Delay_0]}
 {id=Modulator_1, children [Random_0], properties {pitchMax=1.2, pitchMin=0.9, volumeMax=1, volumeMin=0.3}}
 {id=QualityLevel_0, children [Looping_0,null,null,null]}
 {
  id=Random_0
  children [WavePlayer_0,WavePlayer_1,WavePlayer_2,WavePlayer_3,WavePlayer_4]
  properties {bShouldExcludeFromBranchCulling=true, weights [1,1,1,1,1]}
 }
 {id=WavePlayer_0, properties {soundWaveAssetPtr=&/Buttercake_Vox_Breathing_Blowhole_01}}
 {id=WavePlayer_1, properties {soundWaveAssetPtr=&/Buttercake_Vox_Breathing_Blowhole_02}}
 {id=WavePlayer_2, properties {soundWaveAssetPtr=&/Buttercake_Vox_Breathing_Blowhole_03}}
 {id=WavePlayer_3, properties {soundWaveAssetPtr=&/Buttercake_Vox_Breathing_Blowhole_04}}
 {id=WavePlayer_4, properties {soundWaveAssetPtr=&/Buttercake_Vox_Breathing_Blowhole_05}}
]
properties {
 attenuationSettings=&/Buttercake_Blowhole_Breathing_Att
 bOverrideConcurrency=true
 maxDistance=1350
 priority=0
 soundClassObject=/Game/Sounds/FortSoundClassesAndModes/Fort_Creatures
 volumeMultiplier=1.052
 concurrencyOverrides {maxCount=1, resolutionRule=PreventNew}
}
