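# Sound cue for Buttercake's blowhole launch vocal, routed through the Fort_Creatures sound class.
# The cue has no nodes and a null root, so this file defines playback settings without any audio source.
# It uses the critical vox attenuation asset, a long max distance, raised volume and priority, and disabled virtualization.
# A concurrency override limits it to a single instance at a time.
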SoundCue /ButterCake/Audio/Buttercake_Vox_Blowhole_Launch_Cue
root=null
nodes []
properties {
 attenuationSettings=&/Buttercake_Vox_Critical_Att
 bOverrideConcurrency=true
 maxDistance=35000
 priority=30
 soundClassObject=/Game/Sounds/FortSoundClassesAndModes/Fort_Creatures
 virtualizationMode=Disabled
 volumeMultiplier=2.383
 concurrencyOverrides {maxCount=1}
}
