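# Sound cue for Buttercake's running breath vocals, played as a creature sound with shared vox attenuation.
# A quality-level switch picks an equal-weight random pool: sixteen breath waves for the first two levels, a seven-wave subset for the last two.
# A modulator at the root varies volume and pitch on each play.
# A concurrency override allows one instance and rejects new ones while it plays, and virtualization is disabled.
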SoundCue /ButterCake/Audio/Buttercake_Vox_Breath_Run_Cue
root=Modulator_0
nodes [
 {id=Modulator_0, children [QualityLevel_0], properties {pitchMin=1, volumeMin=0.75}}
 {id=QualityLevel_0, children [Random_0,Random_0,Random_1,Random_1]}
 {
  id=Random_0
  children [
   WavePlayer_1
   WavePlayer_2
   WavePlayer_3
   WavePlayer_4
   WavePlayer_5
   WavePlayer_6
   WavePlayer_7
   WavePlayer_8
   WavePlayer_9
   WavePlayer_10
   WavePlayer_11
   WavePlayer_12
   WavePlayer_13
   WavePlayer_14
   WavePlayer_15
   WavePlayer_16
  ]
  properties {bShouldExcludeFromBranchCulling=true, weights [1,1,1,1,1,1,1,1,1,1,1,1,1,1,1,1]}
 }
 {
  id=Random_1
  children [WavePlayer_17, WavePlayer_18, WavePlayer_19, WavePlayer_20, WavePlayer_21, WavePlayer_22, WavePlayer_23]
  properties {bShouldExcludeFromBranchCulling=true, weights [1,1,1,1,1,1,1]}
 }
 {id=WavePlayer_1, properties {soundWaveAssetPtr=&/Buttercake_Vox_Breath_Run_01}}
 {id=WavePlayer_10, properties {soundWaveAssetPtr=&/Buttercake_Vox_Breath_Run_10}}
 {id=WavePlayer_11, properties {soundWaveAssetPtr=&/Buttercake_Vox_Breath_Run_11}}
 {id=WavePlayer_12, properties {soundWaveAssetPtr=&/Buttercake_Vox_Breath_Run_12}}
 {id=WavePlayer_13, properties {soundWaveAssetPtr=&/Buttercake_Vox_Breath_Run_13}}
 {id=WavePlayer_14, properties {soundWaveAssetPtr=&/Buttercake_Vox_Breath_Run_14}}
 {id=WavePlayer_15, properties {soundWaveAssetPtr=&/Buttercake_Vox_Breath_Run_15}}
 {id=WavePlayer_16, properties {soundWaveAssetPtr=&/Buttercake_Vox_Breath_Run_16}}
 {id=WavePlayer_17, properties {soundWaveAssetPtr=&Buttercake_Vox_Breath_Run_01}}
 {id=WavePlayer_18, properties {soundWaveAssetPtr=&/Buttercake_Vox_Breath_Run_02}}
 {id=WavePlayer_19, properties {soundWaveAssetPtr=&/Buttercake_Vox_Breath_Run_03}}
 {id=WavePlayer_2, properties {soundWaveAssetPtr=&Buttercake_Vox_Breath_Run_02}}
 {id=WavePlayer_20, properties {soundWaveAssetPtr=&/Buttercake_Vox_Breath_Run_04}}
 {id=WavePlayer_21, properties {soundWaveAssetPtr=&/Buttercake_Vox_Breath_Run_05}}
 {id=WavePlayer_22, properties {soundWaveAssetPtr=&/Buttercake_Vox_Breath_Run_06}}
 {id=WavePlayer_23, properties {soundWaveAssetPtr=&/Buttercake_Vox_Breath_Run_07}}
 {id=WavePlayer_3, properties {soundWaveAssetPtr=&Buttercake_Vox_Breath_Run_03}}
 {id=WavePlayer_4, properties {soundWaveAssetPtr=&Buttercake_Vox_Breath_Run_04}}
 {id=WavePlayer_5, properties {soundWaveAssetPtr=&Buttercake_Vox_Breath_Run_05}}
 {id=WavePlayer_6, properties {soundWaveAssetPtr=&Buttercake_Vox_Breath_Run_06}}
 {id=WavePlayer_7, properties {soundWaveAssetPtr=&Buttercake_Vox_Breath_Run_07}}
 {id=WavePlayer_8, properties {soundWaveAssetPtr=&/Buttercake_Vox_Breath_Run_08}}
 {id=WavePlayer_9, properties {soundWaveAssetPtr=&/Buttercake_Vox_Breath_Run_09}}
]
properties {
 attenuationSettings=&/Buttercake_Vox_Att
 bOverrideConcurrency=true
 maxDistance=25000
 priority=5
 soundClassObject=/Game/Sounds/FortSoundClassesAndModes/Fort_Creatures
 virtualizationMode=Disabled
 volumeMultiplier=2.75
 concurrencyOverrides {maxCount=1, resolutionRule=PreventNew}
 concurrencySet [null]
}
