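# Sound cue for Buttercake's walking breath vocalizations, played as a creature sound with its own attenuation and navigation concurrency.
# A quality-level root sends its first two levels to an equal-weight random pick of eighteen walk takes, and its last two to a twelve-take subset.
# Both random nodes are excluded from branch culling, and virtualization is disabled on the cue.
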SoundCue /ButterCake/Audio/Buttercake_Vox_Breath_Walk_Cue
root=QualityLevel_0
nodes [
 {id=QualityLevel_0, children [Random_0,Random_0,Random_2,Random_2]}
 {
  id=Random_0
  children [
   WavePlayer_0
   WavePlayer_1
   WavePlayer_2
   WavePlayer_3
   WavePlayer_4
   WavePlayer_5
   WavePlayer_6
   WavePlayer_7
   WavePlayer_8
   WavePlayer_9
   WavePlayer_10
   WavePlayer_11
   WavePlayer_30
   WavePlayer_31
   WavePlayer_32
   WavePlayer_33
   WavePlayer_34
   WavePlayer_35
  ]
  properties {bShouldExcludeFromBranchCulling=true, weights [1,1,1,1,1,1,1,1,1,1,1,1,1,1,1,1,1,1]}
 }
 {
  id=Random_2
  children [
   WavePlayer_36
   WavePlayer_37
   WavePlayer_38
   WavePlayer_39
   WavePlayer_40
   WavePlayer_41
   WavePlayer_42
   WavePlayer_43
   WavePlayer_44
   WavePlayer_45
   WavePlayer_46
   WavePlayer_47
  ]
  properties {bShouldExcludeFromBranchCulling=true, weights [1,1,1,1,1,1,1,1,1,1,1,1]}
 }
 {id=WavePlayer_0, properties {soundWaveAssetPtr=&/Buttercake_Vox_Walk_01}}
 {id=WavePlayer_1, properties {soundWaveAssetPtr=&/Buttercake_Vox_Walk_02}}
 {id=WavePlayer_10, properties {soundWaveAssetPtr=&/Buttercake_Vox_Walk_11}}
 {id=WavePlayer_11, properties {soundWaveAssetPtr=&/Buttercake_Vox_Walk_12}}
 {id=WavePlayer_2, properties {soundWaveAssetPtr=&/Buttercake_Vox_Walk_03}}
 {id=WavePlayer_3, properties {soundWaveAssetPtr=&/Buttercake_Vox_Walk_04}}
 {id=WavePlayer_30, properties {soundWaveAssetPtr=&/Buttercake_Vox_Walk_13}}
 {id=WavePlayer_31, properties {soundWaveAssetPtr=&/Buttercake_Vox_Walk_14}}
 {id=WavePlayer_32, properties {soundWaveAssetPtr=&/Buttercake_Vox_Walk_15}}
 {id=WavePlayer_33, properties {soundWaveAssetPtr=&/Buttercake_Vox_Walk_16}}
 {id=WavePlayer_34, properties {soundWaveAssetPtr=&/Buttercake_Vox_Walk_17}}
 {id=WavePlayer_35, properties {soundWaveAssetPtr=&/Buttercake_Vox_Walk_18}}
 {id=WavePlayer_36, properties {soundWaveAssetPtr=&Buttercake_Vox_Walk_01}}
 {id=WavePlayer_37, properties {soundWaveAssetPtr=&Buttercake_Vox_Walk_02}}
 {id=WavePlayer_38, properties {soundWaveAssetPtr=&Buttercake_Vox_Walk_03}}
 {id=WavePlayer_39, properties {soundWaveAssetPtr=&Buttercake_Vox_Walk_04}}
 {id=WavePlayer_4, properties {soundWaveAssetPtr=&/Buttercake_Vox_Walk_05}}
 {id=WavePlayer_40, properties {soundWaveAssetPtr=&Buttercake_Vox_Walk_05}}
 {id=WavePlayer_41, properties {soundWaveAssetPtr=&/Buttercake_Vox_Walk_06}}
 {id=WavePlayer_42, properties {soundWaveAssetPtr=&/Buttercake_Vox_Walk_07}}
 {id=WavePlayer_43, properties {soundWaveAssetPtr=&/Buttercake_Vox_Walk_08}}
 {id=WavePlayer_44, properties {soundWaveAssetPtr=&/Buttercake_Vox_Walk_09}}
 {id=WavePlayer_45, properties {soundWaveAssetPtr=&/Buttercake_Vox_Walk_10}}
 {id=WavePlayer_46, properties {soundWaveAssetPtr=&Buttercake_Vox_Walk_17}}
 {id=WavePlayer_47, properties {soundWaveAssetPtr=&Buttercake_Vox_Walk_18}}
 {id=WavePlayer_5, properties {soundWaveAssetPtr=&Buttercake_Vox_Walk_06}}
 {id=WavePlayer_6, properties {soundWaveAssetPtr=&Buttercake_Vox_Walk_07}}
 {id=WavePlayer_7, properties {soundWaveAssetPtr=&Buttercake_Vox_Walk_08}}
 {id=WavePlayer_8, properties {soundWaveAssetPtr=&Buttercake_Vox_Walk_09}}
 {id=WavePlayer_9, properties {soundWaveAssetPtr=&Buttercake_Vox_Walk_10}}
]
properties {
 attenuationSettings=&/Buttercake_Vox_Att
 maxDistance=25000
 priority=5
 soundClassObject=/Game/Sounds/FortSoundClassesAndModes/Fort_Creatures
 virtualizationMode=Disabled
 volumeMultiplier=1.259
 concurrencySet [&/Buttercake_Vox_Navigation_Concurrency]
}
