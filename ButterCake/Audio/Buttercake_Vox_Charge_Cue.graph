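# Sound cue for the Buttercake creature's vocalization at the start of a charge.
# A Random node picks one of six equally weighted charge-start waves, excluded from branch culling, and feeds a Modulator that varies volume and pitch.
# It plays in the creatures sound class with boosted volume, critical-vox attenuation and state-vox concurrency, with virtualization disabled.
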SoundCue /ButterCake/Audio/Buttercake_Vox_Charge_Cue
root=Modulator_0
nodes [
 {id=Modulator_0, children [Random_3], properties {pitchMin=1, volumeMin=0.75}}
 {
  id=Random_3
  children [WavePlayer_36, WavePlayer_37, WavePlayer_38, WavePlayer_39, WavePlayer_40, WavePlayer_41]
  properties {bShouldExcludeFromBranchCulling=true, weights [1,1,1,1,1,1]}
 }
 {id=WavePlayer_36, properties {soundWaveAssetPtr=&/Buttercake_Vox_Charge_Start_01}}
 {id=WavePlayer_37, properties {soundWaveAssetPtr=&/Buttercake_Vox_Charge_Start_02}}
 {id=WavePlayer_38, properties {soundWaveAssetPtr=&/Buttercake_Vox_Charge_Start_03}}
 {id=WavePlayer_39, properties {soundWaveAssetPtr=&/Buttercake_Vox_Charge_Start_04}}
 {id=WavePlayer_40, properties {soundWaveAssetPtr=&/Buttercake_Vox_Charge_Start_05}}
 {id=WavePlayer_41, properties {soundWaveAssetPtr=&/Buttercake_Vox_Charge_Start_06}}
]
properties {
 attenuationSettings=&/Buttercake_Vox_Critical_Att
 maxDistance=25000
 priority=45
 soundClassObject=/Game/Sounds/FortSoundClassesAndModes/Fort_Creatures
 virtualizationMode=Disabled
 volumeMultiplier=4.2
 concurrencySet [&/Buttercake_Vox_State_Concurrency]
}
