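# Sound cue for Buttercake's charge-end vocalization, classed under Fort_Creatures.
# A Random node picks one of three equally weighted wave variations and is excluded from branch culling.
# It uses the Critical vox attenuation and the State vox concurrency set, with boosted volume, a long max distance and virtualization disabled.
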SoundCue /ButterCake/Audio/Buttercake_Vox_Charge_End_Cue
root=Random_1
nodes [
 {id=Random_1, children [WavePlayer_12, WavePlayer_13, WavePlayer_14], properties {bShouldExcludeFromBranchCulling=true, weights [1,1,1]}}
 {id=WavePlayer_12, properties {soundWaveAssetPtr=&/Buttercake_Vox_Charge_End_01}}
 {id=WavePlayer_13, properties {soundWaveAssetPtr=&/Buttercake_Vox_Charge_End_02}}
 {id=WavePlayer_14, properties {soundWaveAssetPtr=&/Buttercake_Vox_Charge_End_03}}
]
properties {
 attenuationSettings=&/Buttercake_Vox_Critical_Att
 maxDistance=25000
 priority=35
 soundClassObject=/Game/Sounds/FortSoundClassesAndModes/Fort_Creatures
 virtualizationMode=Disabled
 volumeMultiplier=1.875
 concurrencySet [&/Buttercake_Vox_State_Concurrency]
}
