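# Sound cue for the Buttercake creature's chitter vocalization, routed to the Fortnite creatures sound class.
# A Random node picks one of four equally weighted chitter waves and is excluded from branch culling.
# A root Modulator adds slight pitch and volume variation to each play.
# The cue uses Buttercake vox attenuation, boosts volume, disables virtualization and overrides concurrency to allow a single instance at a time.
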SoundCue /ButterCake/Audio/Buttercake_Vox_Chitter_Cue
root=Modulator_0
nodes [
 {id=Modulator_0, children [Random_0], properties {pitchMax=1.1, pitchMin=0.9, volumeMax=1, volumeMin=0.9}}
 {
  id=Random_0
  children [WavePlayer_0,WavePlayer_1,WavePlayer_2,WavePlayer_3]
  properties {bShouldExcludeFromBranchCulling=true, weights [1,1,1,1]}
 }
 {id=WavePlayer_0, properties {soundWaveAssetPtr=&/Buttercake_Vox_Chitter_A_01}}
 {id=WavePlayer_1, properties {soundWaveAssetPtr=&/Buttercake_Vox_Chitter_A_02}}
 {id=WavePlayer_2, properties {soundWaveAssetPtr=&/Buttercake_Vox_Chitter_A_03}}
 {id=WavePlayer_3, properties {soundWaveAssetPtr=&/Buttercake_Vox_Chitter_A_04}}
]
properties {
 attenuationSettings=&/Buttercake_Vox_Att
 bOverrideConcurrency=true
 maxDistance=25000
 priority=5
 soundClassObject=/Game/Sounds/FortSoundClassesAndModes/Fort_Creatures
 virtualizationMode=Disabled
 volumeMultiplier=1.5
 concurrencyOverrides {maxCount=1}
}
