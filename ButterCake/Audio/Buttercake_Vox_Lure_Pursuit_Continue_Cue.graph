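# Sound cue for Buttercake creature lure vocals, named for continuing a pursuit, in the creature sound class.
# A quality-level node sends its first two levels to a random pick of ten Lure waves and its last two to a pick of six.
# Both random nodes weight variations equally and are excluded from branch culling.
# The cue uses Buttercake vocal attenuation and navigation concurrency, with boosted volume and virtualization disabled.
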SoundCue /ButterCake/Audio/Buttercake_Vox_Lure_Pursuit_Continue_Cue
root=QualityLevel_1
nodes [
 {id=QualityLevel_1, children [Random_4,Random_4,Random_1,Random_1]}
 {
  id=Random_1
  children [WavePlayer_6, WavePlayer_7, WavePlayer_8, WavePlayer_9, WavePlayer_10, WavePlayer_11]
  properties {bShouldExcludeFromBranchCulling=true, weights [1,1,1,1,1,1]}
 }
 {
  id=Random_4
  children [
   WavePlayer_39
   WavePlayer_40
   WavePlayer_41
   WavePlayer_42
   WavePlayer_43
   WavePlayer_44
   WavePlayer_45
   WavePlayer_46
   WavePlayer_47
   WavePlayer_48
  ]
  properties {bShouldExcludeFromBranchCulling=true, weights [1,1,1,1,1,1,1,1,1,1]}
 }
 {id=WavePlayer_10, properties {soundWaveAssetPtr=&/Buttercake_Vox_Lure_05}}
 {id=WavePlayer_11, properties {soundWaveAssetPtr=&/Buttercake_Vox_Lure_06}}
 {id=WavePlayer_39, properties {soundWaveAssetPtr=&/Buttercake_Vox_Lure_01}}
 {id=WavePlayer_40, properties {soundWaveAssetPtr=&/Buttercake_Vox_Lure_02}}
 {id=WavePlayer_41, properties {soundWaveAssetPtr=&/Buttercake_Vox_Lure_03}}
 {id=WavePlayer_42, properties {soundWaveAssetPtr=&/Buttercake_Vox_Lure_04}}
 {id=WavePlayer_43, properties {soundWaveAssetPtr=&Buttercake_Vox_Lure_05}}
 {id=WavePlayer_44, properties {soundWaveAssetPtr=&Buttercake_Vox_Lure_06}}
 {id=WavePlayer_45, properties {soundWaveAssetPtr=&/Buttercake_Vox_Lure_07}}
 {id=WavePlayer_46, properties {soundWaveAssetPtr=&/Buttercake_Vox_Lure_08}}
 {id=WavePlayer_47, properties {soundWaveAssetPtr=&/Buttercake_Vox_Lure_09}}
 {id=WavePlayer_48, properties {soundWaveAssetPtr=&/Buttercake_Vox_Lure_10}}
 {id=WavePlayer_6, properties {soundWaveAssetPtr=&Buttercake_Vox_Lure_01}}
 {id=WavePlayer_7, properties {soundWaveAssetPtr=&Buttercake_Vox_Lure_02}}
 {id=WavePlayer_8, properties {soundWaveAssetPtr=&Buttercake_Vox_Lure_03}}
 {id=WavePlayer_9, properties {soundWaveAssetPtr=&Buttercake_Vox_Lure_04}}
]
properties {
 attenuationSettings=&/Buttercake_Vox_Att
 maxDistance=25000
 priority=25
 soundClassObject=/Game/Sounds/FortSoundClassesAndModes/Fort_Creatures
 virtualizationMode=Disabled
 volumeMultiplier=2
 concurrencySet [&/Buttercake_Vox_Navigation_Concurrency]
}
