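# Sound cue for Buttercake's lure pursuit start vocalization, routed to the Fortnite creatures sound class.
# The node graph is empty with no root, so the cue holds no sound to play.
# It still carries playback settings: shared Buttercake vox attenuation, boosted volume, a set priority and disabled virtualization.
# A concurrency override limits it to a single instance at a time.
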SoundCue /ButterCake/Audio/Buttercake_Vox_Lure_Pursuit_Start_Cue
root=null
nodes []
properties {
 attenuationSettings=&/Buttercake_Vox_Att
 bOverrideConcurrency=true
 maxDistance=16500
 priority=25
 soundClassObject=/Game/Sounds/FortSoundClassesAndModes/Fort_Creatures
 virtualizationMode=Disabled
 volumeMultiplier=2
 concurrencyOverrides {maxCount=1}
}
