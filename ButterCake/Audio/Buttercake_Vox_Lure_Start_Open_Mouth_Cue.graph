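# Sound cue for Buttercake's vocalization at the start of its lure, named for the open-mouth variant.
# A Random node picks one of six lure-start wave variations with equal weight and is excluded from branch culling.
# It plays in the creatures sound class with Buttercake vox attenuation and a navigation vox concurrency group, with boosted volume and virtualization disabled.
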SoundCue /ButterCake/Audio/Buttercake_Vox_Lure_Start_Open_Mouth_Cue
root=Random_0
nodes [
 {
  id=Random_0
  children [WavePlayer_0,WavePlayer_1,WavePlayer_2,WavePlayer_3,WavePlayer_4,WavePlayer_5]
  properties {bShouldExcludeFromBranchCulling=true, weights [1,1,1,1,1,1]}
 }
 {id=WavePlayer_0, properties {soundWaveAssetPtr=&/Buttercake_Vox_Lure_Start_01}}
 {id=WavePlayer_1, properties {soundWaveAssetPtr=&/Buttercake_Vox_Lure_Start_02}}
 {id=WavePlayer_2, properties {soundWaveAssetPtr=&/Buttercake_Vox_Lure_Start_03}}
 {id=WavePlayer_3, properties {soundWaveAssetPtr=&/Buttercake_Vox_Lure_Start_04}}
 {id=WavePlayer_4, properties {soundWaveAssetPtr=&/Buttercake_Vox_Lure_Start_05}}
 {id=WavePlayer_5, properties {soundWaveAssetPtr=&/Buttercake_Vox_Lure_Start_06}}
]
properties {
 attenuationSettings=&/Buttercake_Vox_Att
 maxDistance=25000
 priority=30
 soundClassObject=/Game/Sounds/FortSoundClassesAndModes/Fort_Creatures
 virtualizationMode=Disabled
 volumeMultiplier=2.119
 concurrencySet [&/Buttercake_Vox_Navigation_Concurrency]
}
