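# Sound cue for Buttercake's sleep snore vocals, played in the creatures sound class with shared Buttercake vox attenuation and state concurrency.
# A quality level node sends its first two slots to an equal-weight random pick of nine snore waves and its last two to a random pick of four.
# Both random nodes are excluded from branch culling, and virtualization is disabled on the cue.
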SoundCue /ButterCake/Audio/Buttercake_Vox_Sleep_Snore_Cue
root=QualityLevel_0
nodes [
 {id=QualityLevel_0, children [Random_1,Random_1,Random_2,Random_2]}
 {
  id=Random_1
  children [WavePlayer_3, WavePlayer_4, WavePlayer_5, WavePlayer_6, WavePlayer_7, WavePlayer_8, WavePlayer_9, WavePlayer_10, WavePlayer_11]
  properties {bShouldExcludeFromBranchCulling=true, weights [1,1,1,1,1,1,1,1,1]}
 }
 {
  id=Random_2
  children [WavePlayer_12, WavePlayer_13, WavePlayer_14, WavePlayer_15]
  properties {bShouldExcludeFromBranchCulling=true, weights [1,1,1,1]}
 }
 {id=WavePlayer_10, properties {soundWaveAssetPtr=&/Buttercake_Vox_Sleep_Snore_08}}
 {id=WavePlayer_11, properties {soundWaveAssetPtr=&/Buttercake_Vox_Sleep_Snore_09}}
 {id=WavePlayer_12, properties {soundWaveAssetPtr=&/Buttercake_Vox_Sleep_Snore_01}}
 {id=WavePlayer_13, properties {soundWaveAssetPtr=&/Buttercake_Vox_Sleep_Snore_02}}
 {id=WavePlayer_14, properties {soundWaveAssetPtr=&/Buttercake_Vox_Sleep_Snore_03}}
 {id=WavePlayer_15, properties {soundWaveAssetPtr=&/Buttercake_Vox_Sleep_Snore_04}}
 {id=WavePlayer_3, properties {soundWaveAssetPtr=&Buttercake_Vox_Sleep_Snore_01}}
 {id=WavePlayer_4, properties {soundWaveAssetPtr=&Buttercake_Vox_Sleep_Snore_02}}
 {id=WavePlayer_5, properties {soundWaveAssetPtr=&Buttercake_Vox_Sleep_Snore_03}}
 {id=WavePlayer_6, properties {soundWaveAssetPtr=&Buttercake_Vox_Sleep_Snore_04}}
 {id=WavePlayer_7, properties {soundWaveAssetPtr=&/Buttercake_Vox_Sleep_Snore_05}}
 {id=WavePlayer_8, properties {soundWaveAssetPtr=&/Buttercake_Vox_Sleep_Snore_06}}
 {id=WavePlayer_9, properties {soundWaveAssetPtr=&/Buttercake_Vox_Sleep_Snore_07}}
]
properties {
 attenuationSettings=&/Buttercake_Vox_Att
 maxDistance=25000
 priority=37
 soundClassObject=/Game/Sounds/FortSoundClassesAndModes/Fort_Creatures
 virtualizationMode=Disabled
 volumeMultiplier=1.574
 concurrencySet [&/Buttercake_Vox_State_Concurrency]
}
