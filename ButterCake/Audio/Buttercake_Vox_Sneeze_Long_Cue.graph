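# Sound cue for Buttercake's long sneeze vocalization, playing one of three sneeze inhale recordings.
# A Random node picks among three equally weighted Wave Players and is excluded from branch culling.
# The cue routes to the Fort_Creatures sound class with Buttercake's vox attenuation and state concurrency, a volume boost and virtualization disabled.
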SoundCue /ButterCake/Audio/Buttercake_Vox_Sneeze_Long_Cue
root=Random_0
nodes [
 {id=Random_0, children [WavePlayer_0,WavePlayer_1,WavePlayer_2], properties {bShouldExcludeFromBranchCulling=true, weights [1,1,1]}}
 {id=WavePlayer_0, properties {soundWaveAssetPtr=&/Buttercake_Vox_Sneeze_Inhale_01}}
 {id=WavePlayer_1, properties {soundWaveAssetPtr=&/Buttercake_Vox_Sneeze_Inhale_02}}
 {id=WavePlayer_2, properties {soundWaveAssetPtr=&/Buttercake_Vox_Sneeze_Inhale_03}}
]
properties {
 attenuationSettings=&/Buttercake_Vox_Att
 maxDistance=25000
 priority=30
 soundClassObject=/Game/Sounds/FortSoundClassesAndModes/Fort_Creatures
 virtualizationMode=Disabled
 volumeMultiplier=1.888
 concurrencySet [&/Buttercake_Vox_State_Concurrency]
}
