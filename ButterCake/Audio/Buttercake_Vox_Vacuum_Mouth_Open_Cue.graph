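# Sound cue named for the Buttercake creature's vocal as its vacuum mouth opens, routed to the Fort_Creatures sound class.
# It has no root or sound nodes, so it produces no audio as it stands.
# It still carries playback settings: a boosted volume, shared Buttercake vox attenuation, a vox-state concurrency set and disabled virtualization.
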SoundCue /ButterCake/Audio/Buttercake_Vox_Vacuum_Mouth_Open_Cue
root=null
nodes []
properties {
 attenuationSettings=&/Buttercake_Vox_Att
 maxDistance=35000
 priority=30
 soundClassObject=/Game/Sounds/FortSoundClassesAndModes/Fort_Creatures
 virtualizationMode=Disabled
 volumeMultiplier=3.336
 concurrencySet [&/Buttercake_Vox_State_Concurrency]
}
