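# Sound cue for the Buttercake creature entering water, reusing heavy water-landing footstep samples from the Ostrich vehicle.
# A Random node picks between two equally weighted wave players, then a Modulator adds slight pitch and volume variation.
# It plays in the creatures sound class with the Buttercake blowhole attenuation settings.
# Concurrency limits it to one instance per owner, stopping the oldest.
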SoundCue /ButterCake/Audio/Buttercake_Water_Enter_Cue
root=Modulator_1
nodes [
 {id=Modulator_1, children [Random_1], properties {pitchMin=0.9, volumeMin=0.9}}
 {id=Random_1, children [WavePlayer_0,WavePlayer_5], properties {weights [1,1]}}
 {id=WavePlayer_0, properties {soundWaveAssetPtr=/FNGameplayCues/Athena/Sounds/Vehicles/Ostrich/Veh_Ostrich_FS_Water_Land_Hv_01}}
 {id=WavePlayer_5, properties {soundWaveAssetPtr=/FNGameplayCues/Athena/Sounds/Vehicles/Ostrich/Veh_Ostrich_FS_Water_Land_Hv_02}}
]
properties {
 attenuationSettings=&/Buttercake_Blowhole_Att
 maxDistance=3300
 soundClassObject=/Game/Sounds/FortSoundClassesAndModes/Fort_Creatures
 volumeMultiplier=0.841
 concurrencyOverrides {bLimitToOwner=true, maxCount=1, resolutionRule=StopOldest}
}
