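# Sound cue for the Buttercake water exit sound, classed under creature audio and built from borrowed Ostrich vehicle footstep water-exit waves.
# A random node picks one of four equally weighted variations and feeds a modulator that varies pitch and volume per play.
# It uses the Buttercake blowhole attenuation and limits concurrency to one instance per owner, stopping the oldest.
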SoundCue /ButterCake/Audio/Buttercake_Water_Exit_Cue
root=Modulator_0
nodes [
 {id=Modulator_0, children [Random_0], properties {pitchMin=0.9, volumeMin=0.9}}
 {id=Random_0, children [WavePlayer_1,WavePlayer_2,WavePlayer_3,WavePlayer_4], properties {weights [1,1,1,1]}}
 {id=WavePlayer_1, properties {soundWaveAssetPtr=/FNGameplayCues/Athena/Sounds/Vehicles/Ostrich/Veh_Ostrich_FS_Water_Exit_01}}
 {id=WavePlayer_2, properties {soundWaveAssetPtr=/FNGameplayCues/Athena/Sounds/Vehicles/Ostrich/Veh_Ostrich_FS_Water_Exit_02}}
 {id=WavePlayer_3, properties {soundWaveAssetPtr=/FNGameplayCues/Athena/Sounds/Vehicles/Ostrich/Veh_Ostrich_FS_Water_Exit_03}}
 {id=WavePlayer_4, properties {soundWaveAssetPtr=/FNGameplayCues/Athena/Sounds/Vehicles/Ostrich/Veh_Ostrich_FS_Water_Exit_04}}
]
properties {
 attenuationSettings=&/Buttercake_Blowhole_Att
 maxDistance=3300
 soundClassObject=/Game/Sounds/FortSoundClassesAndModes/Fort_Creatures
 volumeMultiplier=0.841
 concurrencyOverrides {bLimitToOwner=true, maxCount=1, resolutionRule=StopOldest}
}
