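# Sound cue that plays the hit-marker notification for the Buttercake charge, routed through an instigator 2D hit-notification sound class.
# A quality-level node sends its first two levels to an equally weighted random pick of four player-damaged push waves.
# Its other two levels play a single fixed variant.
# It references a charge hit-notify concurrency set, alongside override settings for one voice that stops the oldest.
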SoundCue /ButterCake/Audio/HitMarker_Buttercake_Charge_Cue
root=QualityLevel_0
nodes [
 {id=QualityLevel_0, children [Random_1, Random_1, WavePlayer_11, WavePlayer_11]}
 {
  id=Random_1
  children [WavePlayer_3,WavePlayer_4,WavePlayer_5,WavePlayer_6]
  properties {bShouldExcludeFromBranchCulling=true, weights [1,1,1,1]}
 }
 {id=WavePlayer_11, properties {soundWaveAssetPtr=&/Buttercake_Push_Player_Damaged_02}}
 {id=WavePlayer_3, properties {soundWaveAssetPtr=&/Buttercake_Push_Player_Damaged_01}}
 {id=WavePlayer_4, properties {soundWaveAssetPtr=&Buttercake_Push_Player_Damaged_02}}
 {id=WavePlayer_5, properties {soundWaveAssetPtr=&/Buttercake_Push_Player_Damaged_03}}
 {id=WavePlayer_6, properties {soundWaveAssetPtr=&/Buttercake_Push_Player_Damaged_04}}
]
properties {
 maxDistance=2097152
 soundClassObject=&Fort_2DHitNotifications_Instigate_PSM
 volumeMultiplier=0.7
 concurrencyOverrides {maxCount=1, resolutionRule=StopOldest}
 concurrencySet [&/Charge_HitNotify_Conc]
}
