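# HitMarker_Buttercake_Push_Cue is a Sound Cue for the hit marker sound of Buttercake's push-back hit.
# A Random node picks one of four equally weighted push-back hit waves on each play.
# It uses the creatures sound class at reduced volume and references a shared Charge hit-notify concurrency set.
# Concurrency overrides specify a single voice that stops the oldest instance, with a short release on voice steal.
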SoundCue /ButterCake/Audio/HitMarker_Buttercake_Push_Cue
root=Random_2
nodes [
 {
  id=Random_2
  children [WavePlayer_6,WavePlayer_7,WavePlayer_8,WavePlayer_9]
  properties {bShouldExcludeFromBranchCulling=true, weights [1,1,1,1]}
 }
 {id=WavePlayer_6, properties {soundWaveAssetPtr=&/Buttercake_Push_Back_Hit_01}}
 {id=WavePlayer_7, properties {soundWaveAssetPtr=&/Buttercake_Push_Back_Hit_02}}
 {id=WavePlayer_8, properties {soundWaveAssetPtr=&/Buttercake_Push_Back_Hit_03}}
 {id=WavePlayer_9, properties {soundWaveAssetPtr=&/Buttercake_Push_Back_Hit_04}}
]
properties {
 maxDistance=2097152
 priority=0.5
 soundClassObject=/Game/Sounds/FortSoundClassesAndModes/Fort_Creatures
 volumeMultiplier=0.421
 concurrencyOverrides {maxCount=1, resolutionRule=StopOldest, voiceStealReleaseTime=0.15, volumeScaleAttackTime=0.15}
 concurrencySet [&/Charge_HitNotify_Conc]
}
