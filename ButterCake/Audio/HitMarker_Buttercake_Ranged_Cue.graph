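# Sound cue for the hit marker played when a Buttercake ranged projectile hits a player.
# A random node picks evenly among four hit wave variants, then a modulator adds variation to the output.
# It uses the hit-notification attenuation and 2D instigator sound class at reduced volume, with concurrency limited to two voices that stop the oldest.
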SoundCue /ButterCake/Audio/HitMarker_Buttercake_Ranged_Cue
root=Modulator_0
nodes [
 {id=Modulator_0, children [Random_0]}
 {
  id=Random_0
  children [WavePlayer_0, WavePlayer_1, WavePlayer_2, WavePlayer_10]
  properties {bShouldExcludeFromBranchCulling=true, weights [1,1,1,1]}
 }
 {id=WavePlayer_0, properties {soundWaveAssetPtr=&/Buttercake_Attack_Ranged_Prj_Hit_Player_01}}
 {id=WavePlayer_1, properties {soundWaveAssetPtr=&/Buttercake_Attack_Ranged_Prj_Hit_Player_02}}
 {id=WavePlayer_10, properties {soundWaveAssetPtr=&/Buttercake_Attack_Ranged_Prj_Hit_Player_04}}
 {id=WavePlayer_2, properties {soundWaveAssetPtr=&/Buttercake_Attack_Ranged_Prj_Hit_Player_03}}
]
properties {
 attenuationSettings=/Game/Athena/Sounds/HitNotifications/Fort_HitMarker_Receive_Att
 maxDistance=2097152
 soundClassObject=&Fort_2DHitNotifications_Instigate_PSM
 volumeMultiplier=0.421
 concurrencyOverrides {maxCount=2, resolutionRule=StopOldest}
 concurrencySet [&HitNotifications_TwoMaximum_StopOldest_VolumeScale]
}
