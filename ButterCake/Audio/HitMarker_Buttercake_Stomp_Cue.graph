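# Sound cue for the Buttercake stomp hit marker, routed to a 2D hit-notification sound class for the instigator.
# A quality-level node sends its first two levels to a random pick among four equally weighted "Push Player Damaged" waves.
# Its last two levels play only the 02 wave, and the random node is excluded from branch culling.
# Two shared concurrency assets govern how overlapping hit notifications stack.
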SoundCue /ButterCake/Audio/HitMarker_Buttercake_Stomp_Cue
root=QualityLevel_0
nodes [
 {id=QualityLevel_0, children [Random_0,Random_0,WavePlayer_8,WavePlayer_8]}
 {
  id=Random_0
  children [WavePlayer_0, WavePlayer_1, WavePlayer_2, WavePlayer_10]
  properties {bShouldExcludeFromBranchCulling=true, weights [1,1,1,1]}
 }
 {id=WavePlayer_0, properties {soundWaveAssetPtr=&/Buttercake_Push_Player_Damaged_01}}
 {id=WavePlayer_1, properties {soundWaveAssetPtr=&/Buttercake_Push_Player_Damaged_02}}
 {id=WavePlayer_10, properties {soundWaveAssetPtr=&/Buttercake_Push_Player_Damaged_04}}
 {id=WavePlayer_2, properties {soundWaveAssetPtr=&/Buttercake_Push_Player_Damaged_03}}
 {id=WavePlayer_8, properties {soundWaveAssetPtr=&Buttercake_Push_Player_Damaged_02}}
]
properties {
 maxDistance=2097152
 soundClassObject=&Fort_2DHitNotifications_Instigate_PSM
 volumeMultiplier=0.673
 concurrencyOverrides {maxCount=2, resolutionRule=StopOldest}
 concurrencySet [&HitNotifications_TwoMaximum_StopOldest_VolumeScale, &/Charge_HitNotify_Conc]
}
