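# Material function that layers tiling micro-detail, with skin textures by default, onto incoming material attributes.
# Two textures share adjustable UV tiling: channel-selected, range-clamped mask values add to Roughness and multiply Specular, while a flattenable detail normal merges through angle-corrected blending.
# The SkinMask input blends between original and detailed attributes.
# A static switch, on by default, bypasses the effect when off.
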MaterialFunction /ButterCake/Materials/MF_BC_TilingDetail
graph {nodes [
  n1 = GetMaterialAttributes(None=FunctionInput() {inputName=Material, inputType=FunctionInput_MaterialAttributes}) {attributeGetTypes [#1, #2, #3]}
  n2 = Multiply(
   A=TextureCoordinate()
   B=AppendVector(
    A=ScalarParameter() {defaultValue=70, group=BC_Tiling_Detail, parameterName=MicroDetail_Tiling_X}
    B=ScalarParameter() {defaultValue=70, group=BC_Tiling_Detail, parameterName=MicroDetail_Tiling_Y}
   )
  )
  n3 = TextureSampleParameter2D(UVs=@n2) {
   group=BC_Tiling_Detail
   parameterName=TilingRoughness
   samplerType=SAMPLERTYPE_Masks
   texture=/ButterCake/Textures/Tiling/Buttercake_Skin_S_1
  }
  n4 = ChannelMaskParameter(None=@n3.RGB) {group=BC_Tiling_Detail, maskChannel=Green, parameterName=Detail_RoughnessChannel, defaultValue=<rgba 0 1 0 0>}
  n5 = Clamp(
   None=@n4.R
   Min=ScalarParameter() {group=BC_Tiling_Detail, parameterName=Detail_RoughnessMin}
   Max=ScalarParameter() {defaultValue=1, group=BC_Tiling_Detail, parameterName=Detail_RoughnessMax}
  )
  n6 = TextureSampleParameter2D(UVs=@n2) {
   group=BC_Tiling_Detail
   parameterName=TilingNormal
   samplerType=SAMPLERTYPE_Normal
   texture=/ButterCake/Textures/Tiling/Buttercake_Skin_Normal_1
  }
  n7 = ScalarParameter() {group=BC_Tiling_Detail, parameterName=DetailNormal_Flatness}
  n8 = Clamp(
   None=ChannelMaskParameter(None=@n3.RGB) {group=BC_Tiling_Detail, parameterName=Detail_SpecularChannel}.R
   Min=ScalarParameter() {group=BC_Tiling_Detail, parameterName=Detail_SpecularMin}
   Max=ScalarParameter() {defaultValue=1, group=BC_Tiling_Detail, parameterName=Detail_SpecularMax}
  )
  n9 = SetMaterialAttributes!(
   MaterialAttributes=@n1.MaterialAttributes
   Roughness=Clamp(None=Add(A=@n1.Roughness, B=@n5))
   Normal=MaterialFunctionCall(
    BaseNormal=@n1.Normal
    AdditionalNormal=MaterialFunctionCall(Normal=@n6.RGB, Flatness=@n7) {materialFunction=&FlattenNormal}.Result
   ) {materialFunction=&BlendAngleCorrectedNormals}.Result
   Specular=Clamp(None=Multiply(A=@n1.Specular, B=@n8))
  ) {attributeSetTypes [#1, #2, #3]}
  n10 = FunctionInput() {inputName=SkinMask, inputType=FunctionInput_Scalar}
  FunctionOutput(None=StaticSwitchParameter(True=BlendMaterialAttributes(A=@n1.MaterialAttributes, B=@n9, Alpha=@n10), False=@n1.MaterialAttributes) {defaultValue=true, group=BC_Tiling_Detail, parameterName=Use BC Tiling Detail})
 ]}
