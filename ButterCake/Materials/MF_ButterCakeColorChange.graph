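# Material function that recolors the ButterCake character between Neutral, Lured and Berzerk states, with optional emissive and ranged-attack glow.
# Pre-skinned position over local bounds, multiplied by a gradient-map channel, forms wipes driven by per-state start-time parameters or debug controls.
# Region masks steer tints built from hue shifts or from colors over detail extracted with a mip-biased diffuse sample.
# Static switches toggle glow and color change, falling back to variant colors.
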MaterialFunction /ButterCake/Materials/MF_ButterCakeColorChange
graph {nodes [
  n1 = PreSkinnedPosition()
  n2 = ShaderStageSwitch(PixelShader=VertexInterpolator(VS=@n1).PS, VertexShader=@n1)
  n3 = Divide(
   A=Add(
    A=FunctionInput(Preview=@n2) {bUsePreviewValueAsDefault=true, inputName=PreskinnedPos}
    B=VectorParameter() {parameterName=GradientPosition_Offset, defaultValue=<rgba 0 1400 0 1>}.RGB
   )
   B=MaterialFunctionCall() {materialFunction=&ObjectLocalBounds}.Local Bounds Size
  )
  n4 = FunctionInput(Preview=TextureSample() {texture=&T_BlacksRGB}.RGB) {bUsePreviewValueAsDefault=true, inputName=GMap}
  n5 = Add(A=Multiply(
    A=ChannelMaskParameter(None=@n3) {maskChannel=Green, parameterName=GradientDirection, defaultValue=<rgba 0 1 0 0>}.R
    B=ChannelMaskParameter(None=@n4) {parameterName=AO_Channel}.R
   ))
  n6 = Subtract(A=Time(), B=ScalarParameter() {defaultValue=9.999999843067494e+17, parameterName=GameTimeFromBP_Berzerk})
  n7 = ScalarParameter() {defaultValue=4, parameterName=BerzerkTransformDuration}
  n8 = OneMinus(None=Saturate(None=Divide(A=@n6, B=@n7)))
  n9 = ScalarParameter() {defaultValue=1, group=Debug, parameterName=Berzerk_Control}
  n10 = ScalarParameter() {defaultValue=1, group=Debug, parameterName=UseDebugControls}
  n11 = Subtract(A=@n5, B=Multiply(A=LinearInterpolate(A=@n8, B=@n9, Alpha=@n10)) {constB=2})
  Ceil(None=@n11)
  n12 = Subtract(A=Time(), B=ScalarParameter() {defaultValue=9.999999843067494e+17, parameterName=GameTimeFromBP_Neutral})
  n13 = ScalarParameter() {defaultValue=4, parameterName=NeutralTransformDuration}
  n14 = OneMinus(None=Saturate(None=Divide(A=@n12, B=@n13)))
  n15 = ScalarParameter() {defaultValue=1, group=Debug, parameterName=neutral_Control}
  n16 = Subtract(A=@n5, B=Multiply(A=LinearInterpolate(A=@n14, B=@n15, Alpha=@n10)) {constB=2})
  Ceil(None=@n16)
  ComponentMask(None=TextureCoordinate()) {r=true}
  Constant()
  n17 = NamedRerouteUsage()
  n18 = NamedRerouteUsage()
  n19 = ScalarParameter() {defaultValue=1, group=Debug, parameterName=Lured_Control}
  n20 = Clamp(None=Subtract(A=@n5, B=Multiply(A=@n19) {constB=2}))
  n21 = Clamp(None=@n11)
  n22 = LinearInterpolate(A=LinearInterpolate(A=@n17, B=@n18, Alpha=@n20), B=NamedRerouteUsage(), Alpha=@n21)
  n23 = ScalarParameter() {parameterName=StateOrderSwitchIsBerserk}
  n24 = LinearInterpolate(A=@n22, B=LinearInterpolate(A=@n22, B=@n18, Alpha=@n20), Alpha=@n23)
  n25 = Clamp(None=@n16)
  n26 = VectorParameter() {parameterName=BerzerkEmissiveColor, defaultValue=<rgba 1 0.0335 0 1>}
  n27 = TextureSampleParameter2D() {parameterName=Emissive, samplerType=SAMPLERTYPE_LinearColor, texture=/ButterCake/Textures/T_Buttercake_Head_E}
  n28 = ScalarParameter() {defaultValue=4, parameterName=EmissiveStrength}
  n29 = Multiply(
   A=Multiply(A=@n26.RGB, B=ChannelMaskParameter(None=@n27.RGB) {maskChannel=Green, parameterName=EmissiveChannel, defaultValue=<rgba 0 1 0 0>}.R)
   B=@n28
  )
  n30 = LinearInterpolate(A=LinearInterpolate(B=Power(Base=@n29), Alpha=@n21), Alpha=@n25) {constB=0}
  n31 = LinearInterpolate(A=@n30, B=LinearInterpolate(A=@n30, Alpha=@n20) {constB=0}, Alpha=@n23)
  n32 = Multiply(A=@n26.RGB, B=ChannelMaskParameter(None=@n27.RGB) {parameterName=RangedAttacKGlowChannel}.R)
  n33 = Multiply(A=Power(Base=Multiply(A=@n32, B=@n28)), B=ScalarParameter() {parameterName=BC_RangedAttack})
  n34 = Add(A=NamedRerouteUsage(), B=StaticSwitchParameter(True=Add(A=@n31, B=@n33), False=Constant()) {parameterName=UseEmissiveGlow})
  n35 = GetMaterialAttributes(None=FunctionInput() {inputName=Material, inputType=FunctionInput_MaterialAttributes}) {attributeGetTypes [#1, #2]}
  n36 = GetMaterialAttributes(None=FunctionInput() {inputName=VariantColors, inputType=FunctionInput_MaterialAttributes}) {attributeGetTypes [#1]}
  n37 = BlendMaterialAttributes(
   A=SetMaterialAttributes!(MaterialAttributes=@n35.MaterialAttributes, BaseColor=LinearInterpolate(A=@n24, B=@n17, Alpha=@n25), EmissiveColor=@n34) {attributeSetTypes [#1, #2]}
   B=@n36.MaterialAttributes
   Alpha=NamedRerouteUsage()
  )
  FunctionOutput(None=StaticSwitchParameter(
    True=SetMaterialAttributes!(MaterialAttributes=@n37, EmissiveColor=@n34) {attributeSetTypes [#2]}
    False=@n36.MaterialAttributes
   ) {defaultValue=true, parameterName=UseColorChange})
  n38 = Divide(
   A=NamedRerouteUsage()
   B=TextureSample!(
    UVs
    Tex=TextureObjectParameter() {parameterName=Diffuse_MipSample}
    Bias=ScalarParameter() {defaultValue=1, parameterName=MipBias}
    Apply View MipBias
   ) {mipValueMode=TMVM_MipBias}.RGB
  )
  n39 = VectorParameter() {parameterName=LuredColorRegionC, defaultValue=<rgba 0.0262 0.1094 0.0406 1>}
  n40 = ScalarParameter() {defaultValue=0.25, parameterName=BellyBlend}
  n41 = LinearInterpolate(
   A=VectorParameter() {parameterName=LuredColorRegionA, defaultValue=<rgba 0.0277 0.182 0.051 1>}.RGB
   B=VectorParameter() {parameterName=LuredColorRegionB, defaultValue=<rgba 0.0175 0.0729 0.027 1>}.RGB
   Alpha=NamedRerouteUsage()
  )
  n42 = NamedRerouteUsage()
  LinearInterpolate(
   A=LinearInterpolate(A=NamedRerouteUsage(), B=Multiply(A=@n38, B=@n39.RGB), Alpha=@n40)
   B=LinearInterpolate(A=NamedRerouteUsage(), B=Multiply(A=@n38, B=@n41), Alpha=ScalarParameter() {defaultValue=0.9, parameterName=TintBlendAmt})
   Alpha=@n42
  )
  n43 = Divide(
   A=Subtract(A=Time(), B=ScalarParameter() {defaultValue=9.999999843067494e+17, parameterName=GameTimeFromBP_Lured})
   B=ScalarParameter() {defaultValue=4, parameterName=LuredTransformDuration}
  )
  LinearInterpolate(A=OneMinus(None=Saturate(None=@n43)), B=@n19, Alpha=@n10)
  n44 = Constant()
  n45 = ScalarParameter() {parameterName=Berzerk_HueShift_RegionC}
  n46 = ScalarParameter() {defaultValue=0.25, parameterName=BellyBlend}
  n47 = LinearInterpolate(
   A=ScalarParameter() {defaultValue=0.102, parameterName=Berzerk_HueShift_RegionA}
   B=ScalarParameter() {defaultValue=0.09, parameterName=Berzerk_HueShift_RegionB}
   Alpha=NamedRerouteUsage()
  )
  n48 = ScalarParameter() {defaultValue=1, parameterName=TintBlendAmt}
  n49 = MaterialFunctionCall(
   Hue Shift Percentage=LinearInterpolate(A=LinearInterpolate(A=@n44, B=@n45, Alpha=@n46), B=LinearInterpolate(A=@n44, B=@n47, Alpha=@n48), Alpha=NamedRerouteUsage())
   Texture=NamedRerouteUsage()
  ) {materialFunction=&HueShift}
  n50 = Constant()
  n51 = ScalarParameter() {defaultValue=-2, parameterName=Berzerk_Saturation_RegionA}
  n52 = ScalarParameter() {parameterName=Berzerk_Saturation_RegionB}
  n53 = NamedRerouteUsage()
  n54 = LinearInterpolate(
   A=LinearInterpolate(
    A=@n50
    B=ScalarParameter() {parameterName=Berzerk_Saturation_RegionC}
    Alpha=ScalarParameter() {defaultValue=0.25, parameterName=BellyBlend}
   )
   B=LinearInterpolate(A=@n50, B=LinearInterpolate(A=@n51, B=@n52, Alpha=@n53), Alpha=ScalarParameter() {defaultValue=1, parameterName=TintBlendAmt})
   Alpha=NamedRerouteUsage()
  )
  n55 = Constant() {r=1}
  n56 = LinearInterpolate(
   A=@n55
   B=ScalarParameter() {defaultValue=0.5, parameterName=Berzerk_Level_RegionC}
   Alpha=ScalarParameter() {defaultValue=0.25, parameterName=BellyBlend}
  )
  n57 = LinearInterpolate(
   A=@n55
   B=LinearInterpolate(
    A=ScalarParameter() {defaultValue=1, parameterName=Berzerk_Level_RegionA}
    B=ScalarParameter() {defaultValue=0.5, parameterName=Berzerk_Level_RegionB}
    Alpha=NamedRerouteUsage()
   )
   Alpha=ScalarParameter() {defaultValue=1, parameterName=TintBlendAmt}
  )
  n58 = NamedRerouteUsage()
  Multiply(
   A=Multiply(A=Desaturation(None=@n49.Result, Fraction=@n54), B=LinearInterpolate(A=@n56, B=@n57, Alpha=@n58))
   B=Constant3Vector() {constant=<rgba 1 0.406 0.406 0>}
  )
  n59 = FunctionInput(Preview=TextureSampleParameter2D() {parameterName=BC_ColorTint_Mask, texture=&T_BlacksRGB}.RGBA) {bUsePreviewValueAsDefault=true, inputName=TintColorMask, inputType=FunctionInput_Vector4}
  NamedRerouteDeclaration(None=ChannelMaskParameter(None=@n59) {parameterName=MaskRegionA}.R) {name=MaskRegionA, nodeColor=<rgba 1 0 0.75 1>}
  NamedRerouteDeclaration(None=ChannelMaskParameter(None=@n59) {maskChannel=Green, parameterName=MaskRegionB, defaultValue=<rgba 0 1 0 0>}.R) {name=MaskRegionB, nodeColor=<rgba 0.3828 1 0 1>}
  n60 = Constant()
  n61 = ScalarParameter() {parameterName=Lured_HueShift_RegionC}
  n62 = ScalarParameter() {defaultValue=1, parameterName=BellyBlend}
  n63 = LinearInterpolate(
   A=ScalarParameter() {defaultValue=0.37, parameterName=Lured_HueShift_RegionA}
   B=ScalarParameter() {defaultValue=0.68, parameterName=Lured_HueShift_RegionB}
   Alpha=NamedRerouteUsage()
  )
  n64 = ScalarParameter() {defaultValue=1, parameterName=TintBlendAmt}
  n65 = MaterialFunctionCall(
   Hue Shift Percentage=LinearInterpolate(A=LinearInterpolate(A=@n60, B=@n61, Alpha=@n62), B=LinearInterpolate(A=@n60, B=@n63, Alpha=@n64), Alpha=NamedRerouteUsage())
   Texture=NamedRerouteUsage()
  ) {materialFunction=&HueShift}
  n66 = Constant()
  n67 = ScalarParameter() {defaultValue=-0.3, parameterName=Lured_Saturation_RegionA}
  n68 = ScalarParameter() {defaultValue=-0.5, parameterName=Lured_Saturation_RegionB}
  n69 = NamedRerouteUsage()
  n70 = LinearInterpolate(
   A=LinearInterpolate(
    A=@n66
    B=ScalarParameter() {defaultValue=-0.15, parameterName=Lured_Saturation_RegionC}
    Alpha=ScalarParameter() {defaultValue=1, parameterName=BellyBlend}
   )
   B=LinearInterpolate(A=@n66, B=LinearInterpolate(A=@n67, B=@n68, Alpha=@n69), Alpha=ScalarParameter() {defaultValue=1, parameterName=TintBlendAmt})
   Alpha=NamedRerouteUsage()
  )
  n71 = Constant() {r=1}
  n72 = LinearInterpolate(
   A=@n71
   B=ScalarParameter() {defaultValue=0.5, parameterName=Lured_Level_RegionC}
   Alpha=ScalarParameter() {defaultValue=1, parameterName=BellyBlend}
  )
  n73 = LinearInterpolate(
   A=@n71
   B=LinearInterpolate(
    A=ScalarParameter() {defaultValue=1, parameterName=Lured_Level_RegionA}
    B=ScalarParameter() {defaultValue=1, parameterName=Lured_Level_RegionB}
    Alpha=NamedRerouteUsage()
   )
   Alpha=ScalarParameter() {defaultValue=1, parameterName=TintBlendAmt}
  )
  n74 = NamedRerouteUsage()
  NamedRerouteDeclaration(None=Multiply(A=Desaturation(None=@n65.Result, Fraction=@n70), B=LinearInterpolate(A=@n72, B=@n73, Alpha=@n74))) {name=LuredFinalColor, nodeColor=<rgba 0.8281 1 0 1>}
  n75 = NamedRerouteUsage()
  n76 = Multiply(A=@n38, B=VectorParameter() {parameterName=BerzerkColorRegionC, defaultValue=<rgba 0.4375 0.0298 0.0298 1>}.RGB)
  n77 = NamedRerouteUsage()
  n78 = Multiply(
   A=@n38
   B=LinearInterpolate(
    A=VectorParameter() {parameterName=BerzerkColorRegionA, defaultValue=<rgba 0.115 0.0228 0.0228 1>}.RGB
    B=VectorParameter() {parameterName=BerzerkColorRegionB, defaultValue=<rgba 0.047 0.0055 0.0008 1>}.RGB
    Alpha=NamedRerouteUsage()
   )
  )
  n79 = ScalarParameter() {defaultValue=0.9, parameterName=TintBlendAmt}
  NamedRerouteDeclaration(None=LinearInterpolate(A=LinearInterpolate(A=@n75, B=@n76, Alpha=@n40), B=LinearInterpolate(A=@n77, B=@n78, Alpha=@n79), Alpha=@n42)) {name=BerzerkFinalColor, nodeColor=<rgba 0 1 0.4844 1>}
  NamedRerouteDeclaration(None=@n35.BaseColor) {name=BaseColor, nodeColor=<rgba 0.3359 1 0 1>}
  NamedRerouteDeclaration(None=@n35.EmissiveColor) {name=BaseEmissive, nodeColor=<rgba 0.875 0 1 1>}
  NamedRerouteDeclaration(None=@n36.BaseColor) {name=VariantBaseColor, nodeColor=<rgba 0.1953 1 0 1>}
  NamedRerouteDeclaration(None=ChannelMaskParameter(None=@n59) {maskChannel=Blue, parameterName=Mask, defaultValue=<rgba 0 0 1 0>}.R) {name=ColorMask, nodeColor=<rgba 0.7109 1 0 1>}
  NamedRerouteUsage()
  NamedRerouteUsage()
  NamedRerouteUsage()
  NamedRerouteUsage()
  NamedRerouteUsage()
  NamedRerouteUsage()
 ]}
