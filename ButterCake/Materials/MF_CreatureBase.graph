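# Material function that builds base material attributes for creatures from diffuse, normal, specular-mask and M textures on shared UVs.
# Specular masks drive specular, metallic and remapped roughness, with optional override range on low quality for mobile.
# M supplies AO, Gmap and skin masks, with remapped AO switchable into the attributes.
# Raw maps and pre-skinned local position are also exposed as outputs; a metal-mask power adjustment is left unconnected.
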MaterialFunction /ButterCake/Materials/MF_CreatureBase
graph {nodes [
  n1 = FunctionInput(Preview=TextureCoordinate()) {bUsePreviewValueAsDefault=true, inputName=TexCoords, inputType=FunctionInput_Vector2}
  n2 = TextureSampleParameter2D(UVs=@n1) {group=Diffuse, parameterName=Diffuse, texture=/Engine/EngineMaterials/DefaultDiffuse}
  n3 = ScalarParameter() {group=Roughness, parameterName=MobileRoughnessMin}
  n4 = ScalarParameter() {defaultValue=1, group=Roughness, parameterName=MobileRoughnessMax}
  n5 = TextureSampleParameter2D(UVs=@n1) {
   desc="R = SpecMask G = Metallic B = Roughness"
   group=Specular
   parameterName=SpecularMasks
   samplerType=SAMPLERTYPE_Masks
   texture=&T_Fortnite_Default_S
  }
  n6 = LinearInterpolate(
   A=ScalarParameter() {group=Roughness, parameterName=RoughnessMin}
   B=ScalarParameter() {defaultValue=1, group=Roughness, parameterName=RoughnessMax}
   Alpha=@n5.B
  )
  n7 = QualitySwitch(
   Default=@n6
   Low=StaticSwitchParameter(True=LinearInterpolate(A=@n3, B=@n4, Alpha=@n5.B), False=@n6) {parameterName=OverrideRoughnessOnMobile}
  )
  n8 = Constant()
  n9 = TextureSampleParameter2D(UVs=@n1) {group=Normal, parameterName=Normals, samplerType=SAMPLERTYPE_Normal, texture=/Engine/EngineMaterials/DefaultNormal}
  n10 = TextureSampleParameter2D(UVs=@n1) {
   desc="Masks R=AO G= Fresnel B = Tmission/Skin"
   group=Masks
   parameterName=M
   samplerType=SAMPLERTYPE_Masks
   texture=/Engine/EngineMaterials/DefaultDiffuse_TC_Masks
  }
  n11 = MaterialFunctionCall(Float3=@n10.RGB) {desc="R = AO\r\nG = GMap\r\nB = Skin", materialFunction=&BreakOutFloat3Components}
  n12 = LinearInterpolate(
   A=ScalarParameter() {defaultValue=0.1, parameterName=AO_MaxDarkness}
   B=ScalarParameter() {defaultValue=1, parameterName=AO_MinDarkness}
   Alpha=@n11.R
  )
  n13 = Constant() {r=1}
  FunctionOutput(None=MakeMaterialAttributes!(
    BaseColor=@n2.RGB
    Metallic=@n5.G
    Specular=@n5.R
    Roughness=@n7
    Anisotropy
    EmissiveColor=@n8
    Opacity=Constant()
    OpacityMask
    Normal=@n9.RGB
    Tangent
    WorldPositionOffset
    SubsurfaceColor=@n8
    ClearCoat=@n8
    ClearCoatRoughness=@n8
    AmbientOcclusion=StaticSwitchParameter(True=@n12, False=@n13) {parameterName=UseMatAOInput}
    Refraction
    CustomizedUVs_0
    CustomizedUVs_1
    CustomizedUVs_2
    CustomizedUVs_3
    CustomizedUVs_4
    CustomizedUVs_5
    CustomizedUVs_6
    CustomizedUVs_7
    PixelDepthOffset
    ShadingModel=ShadingModel()
    Displacement
   ))
  FunctionOutput(None=@n11.B) {outputName=Skin}
  FunctionOutput(None=@n9.RGB) {outputName=Normal Map, sortPriority=2}
  FunctionOutput(None=@n7) {outputName=Roughness, sortPriority=8}
  FunctionOutput(None=@n5.RGB) {outputName=SpecularMask, sortPriority=3}
  FunctionOutput(None=@n10.RGB) {outputName=M, sortPriority=4}
  FunctionOutput(None=@n2.RGB) {outputName=DIffuse, sortPriority=1}
  n14 = PreSkinnedPosition()
  n15 = ShaderStageSwitch(PixelShader=VertexInterpolator(VS=@n14).PS, VertexShader=@n14)
  FunctionOutput(None=@n15) {outputName=Pre skinned Local Position}
  FunctionOutput(None=@n11.G) {outputName=Gmap}
  FunctionOutput(None=@n11.R) {outputName=AO}
  NamedRerouteDeclaration(None=@n11.B) {name=PreFX_SkinMask, nodeColor=<rgba 1 0.5156 0 1>}
  NamedRerouteDeclaration(None=@n5.B) {name=PreFX_RoughnessRawPaint, nodeColor=<rgba 1 0.5156 0 1>}
  NamedRerouteDeclaration(None=@n15) {name=Pre Skinned Local Postition, nodeColor=<rgba 1 0.5156 0 1>}
  NamedRerouteDeclaration(None=@n11.R) {name=PreFX_AO, nodeColor=<rgba 1 0.5156 0 1>}
  NamedRerouteDeclaration(None=@n11.G) {name=PreFX_Gmap, nodeColor=<rgba 1 0.5156 0 1>}
  n16 = ScalarParameter() {defaultValue=0.4, group=Masks, parameterName=MetalMaskPower}
  StaticSwitchParameter(True=Saturate(None=Power(Base=@n5.G, Exp=@n16)), False=@n5.G) {group=Masks, parameterName=useMetalMaskPowerAdjustment}
 ]}
