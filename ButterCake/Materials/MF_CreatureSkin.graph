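# Material function that applies a creature skin look to incoming material attributes wherever the SkinMask input is set.
# Base color is tinted by a boost color, brightened toward camera-facing areas by inverse Fresnel, globally scaled and saturated.
# The Subsurface Profile result, with mask-driven opacity between SSS brightness limits, blends over the original by SkinMask.
# An alternative Subsurface branch with tinted color and Fresnel, distance-based opacity is not connected to the output.
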MaterialFunction /ButterCake/Materials/MF_CreatureSkin
graph {nodes [
  Clamp(None=Power(Base=Fresnel() {exponent=1}))
  n1 = GetMaterialAttributes(None=FunctionInput() {inputName=Material, inputType=FunctionInput_MaterialAttributes}) {attributeGetTypes [#1]}
  n2 = ScalarParameter() {defaultValue=0.75, parameterName=Global Skin Boost EX}
  n3 = VectorParameter() {group=SkinColor, parameterName=Skin Boost Color And Exponent, defaultValue=<rgba 0.87 0.7794 0.7911 2>}
  n4 = Fresnel(ExponentIn=@n3.A) {exponent=2}
  n5 = Multiply(A=@n1.BaseColor, B=@n3.RGB)
  n6 = Add(A=Multiply(A=OneMinus(None=@n4), B=@n5), B=@n5)
  n7 = FunctionInput() {inputName=SkinMask, inputType=FunctionInput_Scalar}
  n8 = SetMaterialAttributes!(
   MaterialAttributes=@n1.MaterialAttributes
   ShadingModel=ShadingModel() {shadingModel=MSM_SubsurfaceProfile}
   BaseColor=Saturate(None=Multiply(A=@n2, B=@n6))
   Opacity=LinearInterpolate(
    A=ScalarParameter() {group=SkinColor, parameterName=SSS Brightness_Min}
    B=ScalarParameter() {defaultValue=1, group=SkinColor, parameterName=SSS Brightness_Max}
    Alpha=@n7
   )
   Metallic
  ) {attributeSetTypes [
    #2
    #1
    #3
    #4
   ]}
  n9 = BlendMaterialAttributes(A=@n1.MaterialAttributes, B=@n8, Alpha=@n7)
  FunctionOutput(None=@n9)
  n10 = GetMaterialAttributes(None=@n9) {attributeGetTypes [#1, #5]}
  n11 = VectorParameter() {parameterName=SS_ColorTint, defaultValue=<rgba 1 0.0828 0.1126 1>}
  n12 = PixelNormalWS()
  n13 = OneMinus(None=Power(Base=Fresnel(Normal=@n12) {exponent=1}))
  n14 = ObjectPositionWS()
  n15 = CameraPositionWS()
  n16 = Divide(A=Subtract(A=Distance(A=@n14, B=@n15)) {constB=512}, B=ScalarParameter() {defaultValue=5000, parameterName=ClampDistance})
  SetMaterialAttributes!(
   MaterialAttributes=@n10.MaterialAttributes
   ShadingModel=ShadingModel() {shadingModel=MSM_Subsurface}
   BaseColor
   SubsurfaceColor=Multiply(A=Multiply(A=@n10.BaseColor, B=@n11.RGB), B=ScalarParameter() {defaultValue=1, parameterName=SS_Strength})
   Opacity=LinearInterpolate(A=Clamp(None=@n13), Alpha=Clamp(None=@n16))
  ) {attributeSetTypes [
    #2
    #1
    #6
    #3
   ]}
  SetMaterialAttributes!(MaterialAttributes=@n10.MaterialAttributes)
 ]}
