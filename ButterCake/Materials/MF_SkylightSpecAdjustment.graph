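# Material function that scales a material's Specular where the surface turns away from the sky atmosphere light, passing other attributes through.
# An inverted, powered light-to-vertex-normal dot product fades the adjustment in.
# Its strength blends between min and max parameters, driven by the AO input times a sphere mask on camera-to-normal facing.
# The DebugSpecAdjustment blend is not connected to the output.
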MaterialFunction /ButterCake/Materials/MF_SkylightSpecAdjustment
graph {nodes [
  n1 = GetMaterialAttributes(None=FunctionInput() {inputName=Material, inputType=FunctionInput_MaterialAttributes}) {attributeGetTypes [#1]}
  n2 = ScalarParameter() {defaultValue=0.25, parameterName=MinSpecAdjustment}
  n3 = ScalarParameter() {defaultValue=1, parameterName=MaxSpecADjustment}
  n4 = DotProduct(A=CameraVectorWS(), B=VertexNormalWS())
  n5 = Multiply(
   A=FunctionInput(Preview=Constant() {r=1}) {bUsePreviewValueAsDefault=true, inputName=AO, inputType=FunctionInput_Scalar}
   B=SphereMask(
    A=Clamp(None=@n4)
    B=Constant() {r=1}
    Radius=ScalarParameter() {defaultValue=1, parameterName=SampleTRadius}
    Hardness=ScalarParameter() {parameterName=SampleHardness}
   )
  )
  n6 = SkyAtmosphereLightDirection()
  n7 = VertexNormalWS()
  n8 = Power(Base=OneMinus(None=DotProduct(A=@n6, B=@n7))) {constExponent=4}
  n9 = Multiply(A=@n1.Specular, B=LinearInterpolate(A=Constant() {r=1}, B=LinearInterpolate(A=@n2, B=@n3, Alpha=@n5), Alpha=Clamp(None=@n8)))
  FunctionOutput(None=SetMaterialAttributes!(MaterialAttributes=@n1.MaterialAttributes, Specular=@n9) {attributeSetTypes [#1]})
  LinearInterpolate(A=@n1.Specular, B=@n9, Alpha=ScalarParameter() {defaultValue=1, parameterName=DebugSpecAdjustment})
 ]}
