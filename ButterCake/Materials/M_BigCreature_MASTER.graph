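# Master material for the ButterCake big creature, a skeletal-mesh surface supporting mesh deformers, with shading model chosen per expression among default lit, subsurface and subsurface profile.
# It uses material attributes to chain material functions from a creature base through tiling detail, skylight specular adjustment and character brightness balance.
# A tint-mask texture parameter drives color variation, which feeds a color-change function alongside the balanced material.
# Creature skin and rim functions finish the chain, reusing base outputs such as AO and skin mask.
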Material /ButterCake/Materials/M_BigCreature_MASTER
graph {
 properties {
  bUseMaterialAttributes=true
  bUsedWithMeshDeformer=true
  bUsedWithSkeletalMesh=true
  pixelDepthOffsetMode=PDOM_Legacy
  shadingModel=MSM_FromMaterialExpression
  subsurfaceProfile=&/ButterCakeSSProfile
  usedShadingModels=Default Lit | Subsurface | Subsurface Profile
 }
 nodes [
  n1 = MaterialFunctionCall!(TexCoords) {materialFunction=&/MF_CreatureBase}
  n2 = MaterialFunctionCall!(Material=MaterialFunctionCall!(
    AO=@n1.AO
    Material=MaterialFunctionCall!(Material=@n1.Result, SkinMask=@n1.Result) {materialFunction=&/MF_BC_TilingDetail}.Result
   ) {materialFunction=&/MF_SkylightSpecAdjustment}.Result) {materialFunction=/Game/Packages/Fortress_SharedMaterials/Base_Material_Functions/CharacterBrightnessandScale_Balance}
  n3 = TextureSampleParameter2D() {parameterName=BC_ColorTint_Mask, texture=&T_BlacksRGB}
  n4 = MaterialFunctionCall!(BC_TintMask=@n3.RGBA, MAterial=@n2.Result) {materialFunction=&/MF_ColorVariation}
  MaterialFunctionCall!(
   AO=@n1.AO
   Distance Blend=@n2.Distance Blend
   Gmap=@n1.AO
   Pre Skinned Local Position=@n1.AO
   Material=MaterialFunctionCall!(
    SkinMask=@n1.Skin
    Material=MaterialFunctionCall!(
     PreskinnedPos=@n1.Pre skinned Local Position
     GMap=@n1.Pre skinned Local Position
     VariantColors=@n4.Result
     TintColorMask=@n3.RGBA
     Material=@n2.Result
    ) {materialFunction=&/MF_ButterCakeColorChange}.Result
   ) {materialFunction=&/MF_CreatureSkin}.Result
  ) {materialFunction=/BRCosmetics/Packages/Fortress_SharedMaterials/Base_Material_Functions/MF_Rimv2}
 ]
}
