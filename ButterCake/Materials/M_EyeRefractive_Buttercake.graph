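# Refractive eye material for the ButterCake character, using the Eye shading model with a Default Lit fallback on forward and mobile paths.
# ML_EyeRefraction_BC supplies refracted UVs and iris masks for pupil scaling, color, roughness and concavity.
# An optional switch rotates normals and refraction direction, blending curve-driven idle motion toward target or player tracking.
# An optional hue-shifted rage glow adds emissive, gated by state control parameters.
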Material /ButterCake/Materials/M_EyeRefractive_Buttercake
graph {
 properties {
  bUseMaterialAttributes=true
  bUsedWithMeshDeformer=true
  bUsedWithMorphTargets=true
  bUsedWithSkeletalMesh=true
  bUsedWithStaticLighting=true
  pixelDepthOffsetMode=PDOM_Legacy
  shadingModel=MSM_FromMaterialExpression
  subsurfaceProfile=/Game/Characters/Shared/EyeMaterials/SSProfile/SSP_Eye_Inner
  usedShadingModels=Default Lit | Eye
 }
 nodes [
  n1 = Constant() {r=0.5}
  n2 = ScalarParameter() {defaultValue=4.8653, parameterName=ScaleByCenter}
  n3 = MaterialFunctionCall(Texture Scale=@n2) {materialFunction=&ScaleUVsByCenter}
  n4 = ScalarParameter() {defaultValue=0.11, parameterName=Iris UV Radius, sliderMax=0.5}
  n5 = Add(A=Multiply(A=@n4, B=@n2)) {constB=0.5}
  n6 = Constant() {r=0.5}
  n7 = TextureSample() {
   desc="UVs not scaled by center as the base UVs\r\nare a ortho projection and this depth needs\r\nto stay in that space"
   texture=/ButterCake/Textures/Eye/T_ButterCakeReferencePlane
  }
  n8 = Normalize(VectorInput=Transform(None=Constant3Vector() {constant=<rgba 1 0 0 0>}) {transformSourceType=TRANSFORMSOURCE_Local})
  n9 = Time()
  n10 = ScalarParameter() {defaultValue=0.1, parameterName=EyeMovementSpeed}
  n11 = CurveAtlasRowParameter(CurveTime=Frac(None=Multiply(A=@n9, B=@n10))) {
   atlas=&AthenaCurveAtlas01
   curve=/Game/Characters/Enemies/Cattus_Head/Materials/Cattus_EyeMovement_Curve
   defaultValue=92
  }
  n12 = MaterialFunctionCall(Float3=Subtract(A=@n11, B=Constant3Vector() {constant=<rgba 0.5 0.5 0.5 0>})) {materialFunction=&BreakOutFloat3Components}
  n13 = VectorParameter() {parameterName=Target, defaultValue=<rgba 200 0 0 1>}
  n14 = CollectionParameter() {collection=&FortniteMaterialParameters, parameterName=PlayerPosition}
  n15 = ComponentMask(None=@n14) {b=true, g=true, r=true}
  n16 = ScalarParameter() {parameterName=TrackLocalPlayer}
  n17 = Transform(None=Subtract(A=LinearInterpolate(A=@n13.RGB, B=@n15, Alpha=@n16), B=ActorPositionWS())) {transformSourceType=TRANSFORMSOURCE_World, transformType=TRANSFORM_Local}
  n18 = Normalize(VectorInput=Multiply(A=Constant3Vector() {constant=<rgba 0 1 1 0>}, B=@n17))
  n19 = Constant3Vector() {constant=<rgba 0 1 0 0>}
  n20 = Sign(None=ComponentMask(None=@n17) {b=true})
  n21 = ScalarParameter() {parameterName=PlayerTrackingAngleMin}
  n22 = ScalarParameter() {defaultValue=1.2, parameterName=PlayerTrackingAngleMax}
  n23 = Constant3Vector() {constant=<rgba 0 1 -0.2 0>}
  n24 = ActorPositionWS()
  n25 = DotProduct(
   A=Normalize(VectorInput=Transform(None=@n23) {transformSourceType=TRANSFORMSOURCE_Local})
   B=Normalize(VectorInput=Subtract(A=@n13.RGB, B=@n24))
  )
  n26 = LinearInterpolate(
   A=Constant()
   B=ScalarParameter() {defaultValue=1, parameterName=TrackPlayer}
   Alpha=Saturate(None=LinearInterpolate(A=@n21, B=@n22, Alpha=@n25))
  )
  n27 = LinearInterpolate(
   A=Multiply(A=ScalarParameter() {defaultValue=0.0476, parameterName=MaxRotationAmountX}, B=@n12.R)
   B=Multiply(
    A=MaterialFunctionCall!(Normalized V1=@n18, Normalized V2=@n19) {materialFunction=/Engine/Functions/Engine_MaterialFunctions02/Math/AngleBetweenVectors}.0-1
    B=Multiply(A=@n20) {constB=0.5}
   )
   Alpha=@n26
  ) {constB=0}
  n28 = Normalize(VectorInput=Transform(None=Constant3Vector() {constant=<rgba 0 0 1 0>}) {transformSourceType=TRANSFORMSOURCE_Local})
  n29 = Normalize(VectorInput=Multiply(A=Constant3Vector() {constant=<rgba 1 1 0 0>}, B=@n17))
  n30 = Constant3Vector() {constant=<rgba 0 1 0 0>}
  n31 = Sign(None=ComponentMask(None=@n17) {r=true})
  n32 = LinearInterpolate(
   A=Multiply(A=ScalarParameter() {defaultValue=0.1, parameterName=MaxRotationAmountY}, B=@n12.G)
   B=Multiply(
    A=MaterialFunctionCall(Normalized V1=@n29, Normalized V2=@n30) {materialFunction=&AngleBetweenVectors}.0-1
    B=Multiply(A=@n31) {constB=-0.5}
   )
   Alpha=@n26
  ) {constB=0}
  n33 = MaterialFunctionCall(
   BaseNormal=VertexInterpolator(VS=MaterialFunctionCall(Rotation Axis=@n8, Rotation Angle=@n27) {materialFunction=&FixRotateAboutAxisNormals}.New Tangent Space Vertex Normal).PS
   AdditionalNormal=VertexInterpolator(VS=MaterialFunctionCall(Rotation Axis=@n28, Rotation Angle=@n32) {materialFunction=&FixRotateAboutAxisNormals}.New Tangent Space Vertex Normal).PS
  ) {materialFunction=&BlendAngleCorrectedNormals}
  n34 = StaticSwitchParameter(
   True=TextureSample() {samplerType=SAMPLERTYPE_Normal, texture=/Game/Characters/Shared/EyeMaterials/Textures/T_Eye_N}.RGB
   False=Constant3Vector() {constant=<rgba 0 0 1 0>}
  ) {parameterName=UseEyeBuldge}
  n35 = MaterialFunctionCall(BaseNormal=@n33.Result, AdditionalNormal=@n34) {materialFunction=&BlendAngleCorrectedNormals}
  n36 = Normalize(VectorInput=Transform(None=StaticSwitchParameter(True=@n35.Result, False=@n34) {parameterName=UseWPOeyeMovement}))
  n37 = TextureCoordinate()
  n38 = MaterialFunctionCall!(
   DepthPlaneOffset=TextureSample(UVs=MaterialFunctionCall(X=@n5, Y=@n6) {materialFunction=&MakeFloat2}.Result) {desc="Sample Displacement map as above, sampled at the Iris Radius", texture=&T_ButterCakeReferencePlane}.R
   InternalIoR=ScalarParameter() {defaultValue=1.33, desc=Internal IoR, parameterName=IoR, sliderMax=10, sliderMin=1.0003}
   ScaleByCenter=@n2
   LimbusUVWidth=MaterialFunctionCall(
    X=ScalarParameter() {defaultValue=0.24, desc=Color transition range width, parameterName=LimbusUVWidthColor}
    Y=ScalarParameter() {defaultValue=0.1, desc=Shading transition range width, parameterName=LimbusUVWidthShading}
   ) {materialFunction=&MakeFloat2}.Result
   DepthScale=ScalarParameter() {defaultValue=2, parameterName=DepthScale}
   IrisUVRadius=@n4
   MidPlaneDisplacement=@n7.RGB
   EyeDirectionWorld=@n36
   UVS=@n37
   PupilScaleref=NamedRerouteUsage()
  ) {materialFunction=&/ML_EyeRefraction_BC}
  n39 = LinearInterpolate(A=@n3.UVs, B=@n38.RefractedUV, Alpha=ScalarParameter() {defaultValue=1, parameterName=Refraction On/Off, sliderMax=1}) {constA=1, desc=UVs with Refraction or not}
  n40 = Constant() {r=1}
  n41 = Multiply(A=@n4, B=Constant() {r=2}) {desc=Radius to diameter}
  MaterialExpressionAdd_2 = Add(
   A=Multiply(A=Subtract(A=@n39, B=@n1), B=Divide(A=@n40, B=@n41) {desc="Scale Iris texture coordinates up\r\nby this amount before sampling iris"}) {desc=Centered Scale}
   B=@n1
  )
  n42 = ScalarParameter() {defaultValue=1.2, parameterName=PupilScale, sliderMax=1}
  n43 = MaterialFunctionCall(
   UVs=MaterialFunctionCall(UVs=@MaterialExpressionAdd_2, Texture Scale=@n42) {materialFunction=&ScaleUVsByCenter}.UVs
   Texture Scale=NamedRerouteUsage()
  ) {materialFunction=&ScaleUVsByCenter}
  n44 = Constant() {r=0.5}
  n45 = ScalarParameter() {defaultValue=0.25, parameterName=PupilMaskScale}
  n46 = Constant() {r=1}
  MaterialExpressionReroute_54 = Reroute(None=@n42)
  n47 = Custom!(UV=@MaterialExpressionAdd_2, PupilScale=@MaterialExpressionReroute_54) {
   code="// Scale UVs from from unit circle in or out from center\r\n// float2 UV, float PupilScale\r\n\r\nfloat2 UVcentered = UV - float2(0.5f, 0.5f);\r\nfloat UVlength = length(UVcentered);\r\n// UV on circle at distance 0.5 from the center, in direction of original UV\r\nfloat2 UVmax = normalize(UVcentered)*0.5f;\r\n\r\nfloat2 UVscaled = lerp(UVmax, float2(0.f, 0.f), saturate((1.f - UVlength*2.f)*PupilScale));\r\nreturn UVscaled + float2(0.5f, 0.5f);"
   desc=Scale Pupils
   outputType=CMOT_Float2
   inputs [
    {inputName=UV, input {expression=&self:MaterialExpressionAdd_2}}
    {inputName=PupilScale, input {expression=&self:MaterialExpressionReroute_54}}
   ]
  }
  n48 = TextureSampleParameter2D(UVs=@n47) {parameterName=IrisColor, samplerSource=SSM_Clamp_WorldGroupSettings, texture=/ButterCake/Textures/Eye/T_ButterCake_Scalera_copy}
  n49 = SphereMask(
   A=TextureCoordinate()
   B=Constant2Vector() {g=0.5, r=0.5}
   Radius=ScalarParameter() {defaultValue=0.5, parameterName=ScaleraDarkenning_Radius}
   Hardness=ScalarParameter() {defaultValue=0.5, parameterName=ScaleraDarkening_Hardness}
  )
  n50 = ScalarParameter() {defaultValue=0.1, parameterName=Scalera_Darken_Min}
  n51 = Multiply(A=Multiply(A=@n48.RGB, B=Clamp(None=@n49, Min=@n50)), B=ScalarParameter() {defaultValue=1, parameterName=IrisBRightness})
  n52 = Multiply(A=@n51) {desc=Composited Base Color}
  Add(
   A=Multiply(
    A=SphereMask(A=@n43.UVs, B=@n44, Radius=@n45, Hardness=@n46)
    B=VectorParameter() {parameterName=CloudyIris, defaultValue=<rgba 0 0 0 1>}.RGB
   )
   B=@n52
  )
  MaterialExpressionScalarParameter_70 = ScalarParameter() {defaultValue=0.9, parameterName=IrisDispScaleUV, sliderMax=1}
  n53 = Custom(UV=@MaterialExpressionAdd_2, PupilScale=@MaterialExpressionScalarParameter_70) {
   code="// Scale UVs from from unit circle in or out from center\r\n// float2 UV, float PupilScale\r\n\r\nfloat2 UVcentered = UV - float2(0.5f, 0.5f);\r\nfloat UVlength = length(UVcentered);\r\n// UV on circle at distance 0.5 from the center, in direction of original UV\r\nfloat2 UVmax = normalize(UVcentered)*0.5f;\r\n\r\nfloat2 UVscaled = lerp(UVmax, float2(0.f, 0.f), saturate((1.f - UVlength*2.f)*PupilScale));\r\nreturn UVscaled + float2(0.5f, 0.5f);"
   desc=Scale Iris Disp
   outputType=CMOT_Float2
   inputs [
    {inputName=UV, input {expression=&self:MaterialExpressionAdd_2}}
    {inputName=PupilScale, input {expression=&self:MaterialExpressionScalarParameter_70}}
   ]
  }
  n54 = ScalarParameter() {defaultValue=0.5, parameterName=IrisDispStrength}
  n55 = Constant() {r=1}
  n56 = MaterialFunctionCall(
   BaseNormal=Multiply(A=@n34, B=Constant3Vector() {constant=<rgba 1 1 1 0>}) {constB=-1}
   AdditionalNormal=Multiply(
    A=TextureSample(UVs=@n53) {samplerType=SAMPLERTYPE_Normal, texture=/Game/Characters/Shared/EyeMaterials/Textures/iris08_leftEye_nml}.RGB
    B=MaterialFunctionCall(X=@n54, Y=@n54, Z=@n55) {materialFunction=&MakeFloat3}.Result
   )
  ) {materialFunction=&BlendAngleCorrectedNormals}
  ClearCoatNormalCustomOutput!(None=StaticSwitchParameter(
    True=MaterialFunctionCall(BaseNormal=@n33.Result, AdditionalNormal=@n56.Result) {materialFunction=&BlendAngleCorrectedNormals}.Result
    False=@n56.Result
   ) {parameterName=UseWPOeyeMovement})
  Constant() {r=1}
  Frac(None=Time())
  n57 = TextureSampleParameter2D(UVs=@n3.UVs) {
   parameterName=ScleraColor
   samplerSource=SSM_Clamp_WorldGroupSettings
   texture=/Game/Characters/Shared/EyeMaterials/Textures/T_EyeScleraBaseColor
  }
  n58 = Add(A=@n47) {constB=-0.5}
  n59 = ScalarParameter() {defaultValue=2.15, parameterName=LimbusDarkScale}
  n60 = Power(
   Base=MaterialFunctionCall(Vector 2=Multiply(A=@n58, B=@n59) {constB=2}) {materialFunction=/Engine/Functions/Engine_MaterialFunctions02/Utility/VectorLength}.V2 Length
   Exp=ScalarParameter() {defaultValue=8, parameterName=LimbusPow}
  )
  n61 = MaterialFunctionCall(Float2=@n38.IrisMask) {desc="R = Color, G = Shading", materialFunction=&BreakOutFloat2Components}
  LinearInterpolate(
   A=Multiply(A=@n57.RGB, B=ScalarParameter() {defaultValue=0.9, parameterName=ScleraBrightness})
   B=Multiply(A=@n51, B=OneMinus(None=@n60))
   Alpha=@n61.R
  )
  LinearInterpolate(A=@n4, B=ScalarParameter() {parameterName=Iris_UVRadius_Min}, Alpha=ScalarParameter() {parameterName=IrisScale})
  MaterialFunctionCall(UVs=@MaterialExpressionAdd_2, Texture Scale=NamedRerouteUsage()) {materialFunction=&ScaleUVsByCenter}
  MaterialFunctionCall(UVs=@n3.UVs, Texture Scale=NamedRerouteUsage()) {materialFunction=&ScaleUVsByCenter}
  n62 = NamedRerouteUsage()
  MaterialFunctionCall(UVs=@n37, Texture Scale=@n62) {materialFunction=&ScaleUVsByCenter}
  MaterialFunctionCall(UVs=@n37, Texture Scale=@n62) {materialFunction=&ScaleUVsByCenter}
  n63 = VectorParameter() {parameterName=EyeSlitScaleMax, defaultValue=<rgba 1 1 0 1>}
  n64 = VectorParameter() {parameterName=EyeSlitScaleMin, defaultValue=<rgba 0.1 1 0 1>}
  n65 = ScalarParameter() {parameterName=PupilBlend}
  NamedRerouteDeclaration(None=ComponentMask(None=LinearInterpolate(A=@n63.RGB, B=@n64.RGB, Alpha=@n65)) {g=true, r=true}) {name=PupilScale, nodeColor=<rgba 0.2422 0 1 1>}
  n66 = ActorPositionWS()
  n67 = WorldPosition()
  n68 = RotateAboutAxis(NormalizedRotationAxis=@n8, RotationAngle=@n27, PivotPoint=@n66, Position=@n67.XYZ)
  n69 = RotateAboutAxis(NormalizedRotationAxis=@n28, RotationAngle=@n32, PivotPoint=@n66, Position=Add(A=@n68, B=@n67.XYZ))
  Reroute(None=StaticSwitchParameter(True=Add(A=@n68, B=@n69), False=Constant3Vector()) {parameterName=UseWPOeyeMovement})
  n70 = Time()
  n71 = ScalarParameter() {defaultValue=9.999999843067494e+17, parameterName=GameTimeFromBP_Berzerk}
  Saturate(None=Divide(A=Subtract(A=@n70, B=@n71), B=ScalarParameter() {defaultValue=1, parameterName=BerzerkTransformDuration}))
  n72 = Time()
  n73 = ScalarParameter() {defaultValue=9.999999843067494e+17, parameterName=GameTimeFromBP_Neutral}
  Saturate(None=Divide(A=Subtract(A=@n72, B=@n73), B=ScalarParameter() {defaultValue=1, parameterName=NeutralTransformDuration}))
  ScalarParameter() {parameterName=StateOrderSwitchIsBerserk}
  n74 = TextureSampleParameter2D(UVs=MaterialFunctionCall(Texture Scale=ScalarParameter() {defaultValue=2.4, parameterName=NormalUVScale}) {materialFunction=&ScaleUVsByCenter}.UVs) {
   parameterName=NormalMap
   samplerSource=SSM_Clamp_WorldGroupSettings
   samplerType=SAMPLERTYPE_Normal
   texture=/Game/Characters/Shared/EyeMaterials/Textures/T_Eye_Wet_N
  }
  n75 = ScalarParameter() {defaultValue=1, parameterName=FlattenNormal}
  n76 = Constant() {r=1}
  n77 = Multiply(
   A=MaterialFunctionCall(
    Normal=@n74.RGB
    Flatness=LinearInterpolate(A=@n75, B=@n76, Alpha=@n61.R) {desc=Flatten completely over Iris as this normal will be used for the cornea shading}
   ) {materialFunction=&FlattenNormal}.Result
   B=Constant() {r=10}
  )
  n78 = VectorParameter() {parameterName=EyeCornerDarknessColor, defaultValue=<rgba 0.23 0.083 0.0319 1>}
  n79 = SphereMask(
   A=@n3.UVs
   B=Constant() {r=0.5}
   Radius=ScalarParameter() {defaultValue=0.65, parameterName=ShadowRadius}
   Hardness=ScalarParameter() {defaultValue=0.3451, parameterName=ShadowHardness}
  )
  n80 = ScalarParameter() {defaultValue=1, parameterName=Pupil_maxScale}
  n81 = ScalarParameter() {defaultValue=0.1, parameterName=Pupil_MinScale}
  n82 = ScalarParameter() {parameterName=IrisScale}
  n83 = SphereMask(
   A=SphereMask(A=@n43.UVs, B=Constant2Vector() {g=0.5, r=0.5}, Radius=Constant() {r=0.25}, Hardness=Constant())
   B=Constant() {r=1}
   Radius=Clamp(None=LinearInterpolate(A=@n80, B=@n81, Alpha=@n82))
   Hardness=ScalarParameter() {defaultValue=0.8, parameterName=PupilHardness}
  )
  n84 = LinearInterpolate(A=@n52, B=Constant3Vector(), Alpha=Multiply(A=@n83, B=@n43.0-1 mask))
  n85 = ScalarParameter() {defaultValue=0.1, parameterName=BlendPupil}
  n86 = Constant() {r=0.5}
  n87 = Divide(
   A=MaterialFunctionCall(Vector 2=Subtract(A=@n38.RefractedUV, B=@n86) {desc=Center Refracted UVs}) {materialFunction=&VectorLength}.V2 Length
   B=@n4
  ) {desc=Scale UV distance to match Iris size in UV space}
  n88 = ScalarParameter() {defaultValue=0.1116, parameterName=Iris Concavity Scale, sliderMax=4}
  n89 = SetMaterialAttributes!(
   MaterialAttributes
   Base Color=Multiply(A=LinearInterpolate(A=@n78.RGB, Alpha=@n79), B=LinearInterpolate(A=@n84, B=@n52, Alpha=@n85)) {desc=b}
   Specular=LinearInterpolate(
    A=ScalarParameter() {defaultValue=0.3, parameterName=SpecularityScelera}
    B=ScalarParameter() {defaultValue=0.3, parameterName=SpecularityIris}
    Alpha=@n61.R
   )
   Roughness=LinearInterpolate(
    A=ScalarParameter() {defaultValue=0.1, parameterName=ScleraRougness}
    B=ScalarParameter() {defaultValue=0.1, parameterName=IrisRoughness}
    Alpha=@n61.R
   ) {desc=r}
   Iris Mask or Custom Data 0=@n61.G
   Iris Distance or Custom Data 1=Power(Base=Multiply(A=@n87, B=@n88), Exp=ScalarParameter() {defaultValue=0.2767, parameterName=Iris Concavity Power, sliderMax=2, sliderMin=0.1}) {desc=Adjust distance - Drives concavity}
   World Position Offset
   Normal=StaticSwitchParameter(
    True=MaterialFunctionCall(BaseNormal=@n33.Result, AdditionalNormal=@n77) {materialFunction=&BlendAngleCorrectedNormals}.Result
    False=@n77
   ) {parameterName=UseWPOeyeMovement}
   Shading Model=ShadingModel() {shadingModel=MSM_Eye}
  ) {attributeSetTypes [
    #1
    #2
    #3
    #4
    #5
    #6
    #7
    #8
   ]}
  n90 = Multiply(
   A=OneMinus(None=@n83)
   B=MaterialFunctionCall(Hue Shift Percentage=ScalarParameter() {defaultValue=0.5, parameterName=RageGlow HueShift}, Texture=@n52) {materialFunction=&HueShift}.Result
  )
  n91 = Distance(A=TextureCoordinate(), B=Constant2Vector() {g=0.5, r=0.5})
  n92 = Constant()
  n93 = ScalarParameter() {defaultValue=1, parameterName=RageGlowDistance2}
  n94 = OneMinus(None=ScalarParameter() {defaultValue=1, group=Debug, parameterName=Berzerk_Control})
  n95 = ScalarParameter() {defaultValue=1, group=Debug, parameterName=Lured_Control}
  n96 = OneMinus(None=ScalarParameter() {defaultValue=1, group=Debug, parameterName=neutral_Control})
  n97 = LinearInterpolate(A=LinearInterpolate(A=LinearInterpolate(A=@n92, B=@n93, Alpha=@n94), B=@n92, Alpha=OneMinus(None=@n95)), B=@n92, Alpha=@n96)
  n98 = Clamp(None=OneMinus(None=Divide(A=@n91, B=@n97)))
  n99 = Multiply(
   A=Power(Base=Multiply(A=@n90, B=@n98), Exp=ScalarParameter() {defaultValue=2, parameterName=RageGlowPower})
   B=ScalarParameter() {defaultValue=50, parameterName=RageGlow}
  )
  n100 = Constant()
  n101 = LinearInterpolate(A=LinearInterpolate(A=@n100, B=ScalarParameter() {defaultValue=1, parameterName=Rage}, Alpha=@n94), B=@n100, Alpha=@n96)
  n102 = StaticSwitchParameter(
   True=SetMaterialAttributes!(
    MaterialAttributes=GetMaterialAttributes(None=@n89) {attributeGetTypes [#1]}.MaterialAttributes
    Emissive Color=Multiply(A=@n99, B=@n101)
   ) {attributeSetTypes [#9]}
   False=@n89
  ) {parameterName=UseRageGlow}
  n103 = SetMaterialAttributes!(MaterialAttributes=@n102, Shading Model=ShadingModel()) {attributeSetTypes [#8]}
  ShadingPathSwitch(Default=@n102, Forward=@n103, Mobile=@n103)
  n104 = Time()
  n105 = ScalarParameter() {defaultValue=0.1, parameterName=EyeIrisadjustSpeed}
  n106 = CurveAtlasRowParameter(CurveTime=Frac(None=Multiply(A=@n104, B=@n105))) {atlas=&AthenaCurveAtlas01, curve=&Cattus_EyeMovement_Curve}
  n107 = LinearInterpolate(
   A=ScalarParameter() {defaultValue=0.73, parameterName=PupilScaleMin, sliderMax=1}
   B=ScalarParameter() {defaultValue=0.94, parameterName=PupilScaleMax, sliderMax=1}
   Alpha=@n106.B
  )
  StaticSwitchParameter(
   True=LinearInterpolate(
    A=LinearInterpolate(A=@n107, B=@n42, Alpha=@n26) {constB=0}
    B=ScalarParameter() {defaultValue=0.619, parameterName=PupilScaleRage, sliderMax=1}
    Alpha=ScalarParameter() {defaultValue=1, parameterName=Rage}
   )
   False=@n42
  ) {parameterName=UseWPOeyeMovement}
  TangentOutput!(None=Multiply(A=@n36))
 ]
}
