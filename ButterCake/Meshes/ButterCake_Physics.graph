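# Physics asset for ButterCake, a six-limbed creature with a tail and fins, defining simplified collision bodies on its skeleton.
# Boxes cover head and spine, while capsules cover neck, face, jaws and mirrored limbs; tail and fin bodies have no shapes.
# Only five constraints exist, linking jaws to face and face and FinA pair to head, all with free swing and twist, leaving the rest unjointed.
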PhysicsAsset /ButterCake/Meshes/ButterCake_Physics
bodies {
 C_BipedHeadNeck_A_Head_Jnt [{shapeType=Box, extentX=200, extentY=600, extentZ=550, shapeName=C_BipedHeadNeck_A_Head_Jnt_box, center=<36 -90 0>, rotation=<rot 0 90 0>}]
 C_BipedHeadNeck_A_NeckMid_Jnt [{
   shapeType=Capsule
   length=440
   radius=175
   shapeName=C_BipedHeadNeck_A_NeckMid_Jnt_capsule
   center=<85 80.0776 0.0004>
   rotation=<rot 0 -89.9999 0>
  }]
 C_Leaf_FaceAttach_Leaf_Jnt [{
   shapeType=Capsule
   length=400
   radius=220
   shapeName=C_Leaf_FaceAttach_Leaf_Jnt_capsule
   center=<211.4603 0.0069 -50.2015>
   rotation=<rot 0 0 89.9999>
  }]
 C_Mouth_Lower_Jnt [{
   shapeType=Capsule
   length=350
   radius=165
   shapeName=C_Leaf_FaceAttach_Leaf_Jnt_capsule
   center=<-150.8461 -0.001 3.9463>
   rotation=<rot 0 0 89.9999>
  }]
 C_Mouth_Upper_Jnt [{
   shapeType=Capsule
   length=280
   radius=220
   shapeName=C_Leaf_FaceAttach_Leaf_Jnt_capsule
   center=<-172.1793 -0.0054 30>
   rotation=<rot 0 0 89.9999>
  }]
 C_QuadSpine_A_Chest_Jnt [{shapeType=Box, extentX=500, extentY=275, extentZ=640, shapeName=C_QuadSpine_A_Spine2_Jnt_box, center=<25 40 0>, rotation=<rot 0 90 0>}]
 C_QuadSpine_A_Pelvis_Jnt [{shapeType=Box, extentX=400, extentY=360, extentZ=660, shapeName=C_QuadSpine_A_Spine2_Jnt_box, center=<20 -30 0>, rotation=<rot 0 90 0>}]
 C_QuadSpine_A_Spine1_Jnt [{shapeType=Box, extentX=500, extentY=275, extentZ=710, shapeName=C_QuadSpine_A_Spine2_Jnt_box, center=<25 40 0>, rotation=<rot 0 90 0>}]
 C_QuadSpine_A_Spine2_Jnt [{shapeType=Box, extentX=560, extentY=275, extentZ=700, shapeName=C_QuadSpine_A_Spine2_Jnt_box, center=<25 70 0>, rotation=<rot 0 90 0>}]
 C_QuadSpine_A_Spine3_Jnt [{shapeType=Box, extentX=500, extentY=275, extentZ=700, shapeName=C_QuadSpine_A_Spine2_Jnt_box, center=<25 40 0>, rotation=<rot 0 90 0>}]
 C_Tail_A_Tail1_Jnt []
 C_Tail_A_Tail2_Jnt []
 C_Tail_A_Tail3_Jnt []
 C_Tail_A_Tail4_Jnt []
 C_Tail_A_Tail6_Jnt []
 L_Leaf_FinA_Leaf_Jnt []
 L_Leaf_FinD_Leaf_Jnt []
 L_Leaf_FinE_Leaf_Jnt []
 L_Leaf_FinF_Leaf_Jnt []
 L_Leaf_FinG_Leaf_Jnt []
 L_Leaf_FinH_Leaf_Jnt []
 L_PawedArm_A_Elbow_Jnt [{shapeType=Capsule, length=150, radius=110, shapeName=L_PawedArm_B_Elbow_Jnt_capsule, center=<150 0 0>, rotation=<rot 90 0 0>}]
 L_PawedArm_A_Shoulder_Jnt [{
   shapeType=Capsule
   length=110
   radius=170
   shapeName=L_PawedArm_A_Shoulder_Jnt_capsule
   center=<85.724 0.0007 37.8712>
   rotation=<rot 90 -7.125 12.8756>
  }]
 L_PawedArm_B_Elbow_Jnt [{shapeType=Capsule, length=150, radius=110, shapeName=L_PawedArm_B_Elbow_Jnt_capsule, center=<175 -15 0>, rotation=<rot 90 0 0>}]
 L_PawedArm_B_Shoulder_Jnt [{
   shapeType=Capsule
   length=60
   radius=180
   shapeName=L_PawedArm_B_Shoulder_Jnt_capsule
   center=<49.3238 29.6605 67.9214>
   rotation=<rot 12 -109.9999 -89.9999>
  }]
 L_PawedLeg_C_Knee_Jnt [{shapeType=Capsule, length=220, radius=120, shapeName=L_PawedArm_B_Elbow_Jnt_capsule, center=<60 0 0>, rotation=<rot 90 -15 -30>}]
 L_PawedLeg_C_Thigh_Jnt [{shapeType=Capsule, length=90, radius=190, shapeName=R_PawedLeg_C_Thigh_Jnt_capsule, center=<0 0 -53.4158>, rotation=<rot 90 0 -10>}]
 R_Leaf_FinA_Leaf_Jnt []
 R_Leaf_FinD_Leaf_Jnt []
 R_Leaf_FinE_Leaf_Jnt []
 R_Leaf_FinF_Leaf_Jnt []
 R_Leaf_FinG_Leaf_Jnt []
 R_Leaf_FinH_Leaf_Jnt []
 R_PawedArm_A_Elbow_Jnt [{shapeType=Capsule, length=150, radius=110, shapeName=L_PawedArm_B_Elbow_Jnt_capsule, center=<-150 0 0>, rotation=<rot 90 0 0>}]
 R_PawedArm_A_Shoulder_Jnt [{
   shapeType=Capsule
   length=110
   radius=170
   shapeName=L_PawedArm_A_Shoulder_Jnt_capsule
   center=<-85.724 0 -37.8712>
   rotation=<rot 90 -7.125 12.8756>
  }]
 R_PawedArm_B_Elbow_Jnt [{shapeType=Capsule, length=150, radius=110, shapeName=L_PawedArm_B_Elbow_Jnt_capsule, center=<-175 15 0>, rotation=<rot 90 0 0>}]
 R_PawedArm_B_Shoulder_Jnt [{
   shapeType=Capsule
   length=60
   radius=180
   shapeName=L_PawedArm_B_Shoulder_Jnt_capsule
   center=<-49.3238 -29.6605 -67.9214>
   rotation=<rot 12 -109.9999 -89.9999>
  }]
 R_PawedLeg_C_Knee_Jnt [{shapeType=Capsule, length=220, radius=120, shapeName=L_PawedArm_B_Elbow_Jnt_capsule, center=<-60 0 0>, rotation=<rot 90 -15 -30>}]
 R_PawedLeg_C_Thigh_Jnt [{shapeType=Capsule, length=90, radius=190, shapeName=R_PawedLeg_C_Thigh_Jnt_capsule, center=<0 0 53.4158>, rotation=<rot 90 0 -10>}]
}
constraints [
 {
  bone1Name=C_Mouth_Upper_Jnt
  bone2Name=C_Leaf_FaceAttach_Leaf_Jnt
  swing1LimitDegrees=45
  swing1Motion=Free
  swing2LimitDegrees=45
  swing2Motion=Free
  twistLimitDegrees=45
  twistMotion=Free
 }
 {
  bone1Name=C_Leaf_FaceAttach_Leaf_Jnt
  bone2Name=C_BipedHeadNeck_A_Head_Jnt
  swing1LimitDegrees=45
  swing1Motion=Free
  swing2LimitDegrees=45
  swing2Motion=Free
  twistLimitDegrees=45
  twistMotion=Free
 }
 {
  bone1Name=C_Mouth_Lower_Jnt
  bone2Name=C_Leaf_FaceAttach_Leaf_Jnt
  swing1LimitDegrees=45
  swing1Motion=Free
  swing2LimitDegrees=45
  swing2Motion=Free
  twistLimitDegrees=45
  twistMotion=Free
 }
 {
  bone1Name=R_Leaf_FinA_Leaf_Jnt
  bone2Name=C_BipedHeadNeck_A_Head_Jnt
  swing1LimitDegrees=45
  swing1Motion=Free
  swing2LimitDegrees=45
  swing2Motion=Free
  twistLimitDegrees=45
  twistMotion=Free
 }
 {
  bone1Name=L_Leaf_FinA_Leaf_Jnt
  bone2Name=C_BipedHeadNeck_A_Head_Jnt
  swing1LimitDegrees=45
  swing1Motion=Free
  swing2LimitDegrees=45
  swing2Motion=Free
  twistLimitDegrees=45
  twistMotion=Free
 }
]
