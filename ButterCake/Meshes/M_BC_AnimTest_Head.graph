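# M_BC_AnimTest_Head is a Default Lit surface material for skeletal meshes with mesh deformer support, apparently for an animation test head in ButterCake.
# Its graph holds one texture sample of T_BC_AnimTextures_Body_D_1, a body texture despite the Head name, and a 0.5 scalar constant.
# The file shows no parameters and no links from either node to material outputs.
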Material /ButterCake/Meshes/M_BC_AnimTest_Head
graph {
 properties {bUsedWithMeshDeformer=true, bUsedWithSkeletalMesh=true, pixelDepthOffsetMode=PDOM_Legacy, usedShadingModels=Default Lit}
 nodes [
  Constant() {r=0.5}
  TextureSample() {texture=&/T_BC_AnimTextures_Body_D_1}
 ]
}
