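# A Default Lit test material in the ButterCake meshes folder.
# It contains a constant pinkish-red color vector and a normal-map sample of TextureTest_N.
# The graph shows no links from either node to a material output.
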Material /ButterCake/Meshes/materialTest
graph {
 properties {pixelDepthOffsetMode=PDOM_Legacy, usedShadingModels=Default Lit}
 nodes [
  Constant3Vector() {constant=<rgba 0.75 0.1175 0.315 0>}
  TextureSample() {samplerType=SAMPLERTYPE_Normal, texture=&/TextureTest_N}
 ]
}
