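# Unlit translucent Niagara sprite material for a ButterCake spike effect that erodes a spike texture with noise.
# Dynamic parameters scale the tiled UVs of a blurred low-res noise texture and set its erosion strength.
# That weighted noise is subtracted from the spike texture's red channel, saturated, then multiplied by particle alpha, so each particle system can drive the breakup.
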Material /ButterCake/VFX/M_ButtercakeSpikeElemErode
graph {
 properties {bUsedWithNiagaraSprites=true, blendMode=BLEND_Translucent, pixelDepthOffsetMode=PDOM_Legacy, shadingModel=MSM_Unlit, usedShadingModels=Unlit}
 nodes [
  n1 = TextureSample() {texture=/Game/Effects/Fort_Effects/Textures/Explo/T_Spike_01}
  n2 = TextureCoordinate() {uTiling=2, vTiling=2}
  n3 = DynamicParameter() {paramNames [NoiseScale, NoiseErodeStrength, Param3, Param4]}
  n4 = Multiply(
   A=TextureSample(UVs=Multiply(A=@n2, B=@n3.NoiseScale)) {samplerType=SAMPLERTYPE_Grayscale, texture=&LowResBlurredNoise}.R
   B=@n3.NoiseErodeStrength
  )
  Multiply(A=ParticleColor().A, B=Saturate(None=Subtract(A=@n1.R, B=@n4)))
 ]
}
