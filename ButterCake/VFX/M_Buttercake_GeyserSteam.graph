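# Translucent, two-sided material for Buttercake geyser steam on mesh particles, tinted from pale cyan to sulfur green along V.
# Two opposing waterfall-texture pans distort scrolling smoke and normal samples; the smoke modulates color, opacity and vertex-normal displacement.
# A V-weighted local-X offset slants the mesh, while opacity combines inverted Fresnel, V gradients, world-aligned panning puffs and depth fade.
# Static switches enable dynamic-parameter slant and erosion, plus particle alpha.
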Material /ButterCake/VFX/M_Buttercake_GeyserSteam
graph {
 properties {
  bUsedWithMeshParticles=true
  bUsedWithNiagaraMeshParticles=true
  blendMode=BLEND_Translucent
  pixelDepthOffsetMode=PDOM_Legacy
  translucencyLightingMode=TLM_VolumetricPerVertexDirectional
  twoSided=true
  usedShadingModels=Default Lit
 }
 nodes [
  n1 = Multiply(
   A=TextureCoordinate()
   B=AppendVector(
    A=ScalarParameter() {defaultValue=3, parameterName=Distortion_UScale}
    B=ScalarParameter() {defaultValue=0.5, parameterName=Distortion_VScale}
   )
  )
  n2 = Multiply(
   A=TextureSample(UVs=Panner(Coordinate=@n1) {speedY=0.625}) {texture=/Game/Packages/Fortnite_UniqueMaterials_1/Textures/T_WaterFall_Tile}.B
   B=ScalarParameter() {defaultValue=0.0163, group=WaterFall_UVDistort, parameterName=UVDistortAmountA}
  )
  n3 = Multiply(
   A=TextureSample(UVs=Panner(Coordinate=@n1) {speedY=-0.625}) {texture=&T_WaterFall_Tile}.R
   B=ScalarParameter() {defaultValue=0.0163, group=WaterFall_UVDistort, parameterName=UVDistortAmountB}
  )
  n4 = TextureCoordinate()
  n5 = AppendVector(
   A=ScalarParameter() {defaultValue=3, parameterName=Waterfall_UScale}
   B=Multiply(A=ScalarParameter() {defaultValue=0.5, parameterName=Waterfall_VScale}, B=ScalarParameter() {defaultValue=1, parameterName=UVScale})
  )
  n6 = Panner(
   Coordinate=Add(A=Multiply(A=@n4, B=@n5), B=Add(A=@n2, B=@n3))
   Speed=ComponentMask(None=VectorParameter() {parameterName=WaterfallSpeed, defaultValue=<rgba 0 0 0 1>}.RGB) {g=true, r=true}
  ) {speedY=1}
  n7 = TextureSample(UVs=@n6) {samplerType=SAMPLERTYPE_LinearColor, texture=/Game/Global/Textures/Default/Tileable_Patterns/T_Tile_Smoke_Small}
  n8 = LinearInterpolate(Alpha=@n7.G) {constA=-1}
  n9 = TextureCoordinate()
  n10 = Power(Base=OneMinus(None=ComponentMask(None=@n9) {g=true}), Exp=ScalarParameter() {defaultValue=1, parameterName=WPO_Contrast})
  n11 = VertexNormalWS()
  n12 = ScalarParameter() {defaultValue=750, parameterName=WPO_Amount}
  n13 = DynamicParameter() {paramNames [SlantAmount, SlantContrast, Erode Contrast, Param4]}
  n14 = ComponentMask(None=TextureCoordinate()) {g=true}
  n15 = ScalarParameter() {defaultValue=3, parameterName=Contrast}
  n16 = Multiply(
   A=StaticSwitchParameter(True=@n13.SlantAmount, False=ScalarParameter() {defaultValue=750, parameterName=SlantAmount}) {parameterName=Use Dynamic Parameter}
   B=Power(Base=OneMinus(None=@n14), Exp=StaticSwitchParameter(True=@n13.SlantContrast, False=@n15) {parameterName=Use Dynamic Parameter})
  )
  n17 = Constant3Vector() {constant=<rgba 1 0 0 0>}
  Add(
   A=Multiply(A=Multiply(A=@n8, B=@n10), B=Multiply(A=@n11, B=@n12))
   B=Transform(None=Multiply(A=@n16, B=@n17)) {transformSourceType=TRANSFORMSOURCE_Local}
  )
  Constant() {r=1}
  n18 = VectorParameter() {group=Base Color, parameterName=BaseColor, defaultValue=<rgba 0.7126 0.9761 1 1>}
  n19 = VectorParameter() {group=Base Color, parameterName=SulfurBaseColor, defaultValue=<rgba 0.5 1 0.675 1>}
  n20 = Power(Base=ComponentMask(None=TextureCoordinate()) {g=true}, Exp=ScalarParameter() {defaultValue=3, parameterName=ContribColor_Contrast})
  n21 = Saturate(None=Add(A=@n7.RGB) {constB=0.25})
  Multiply(
   A=Multiply(A=LinearInterpolate(A=@n18.RGB, B=@n19.RGB, Alpha=@n20), B=@n21)
   B=ScalarParameter() {defaultValue=0.1, parameterName=EmissiveContribution}
  ) {constB=0.2}
  n22 = ScalarParameter() {defaultValue=1, parameterName=Alpha_Top}
  n23 = ScalarParameter() {defaultValue=2.5, parameterName=Alpha_Bottom}
  n24 = TextureCoordinate()
  n25 = ScalarParameter() {defaultValue=0.1, parameterName=FresnelEdgeContrast}
  n26 = Multiply(
   A=Power(Base=ComponentMask(None=@n24) {g=true}, Exp=ScalarParameter() {defaultValue=1, parameterName=Top Edge Contrast})
   B=OneMinus(None=Fresnel(ExponentIn=@n25))
  )
  n27 = MaterialFunctionCall!(
   TextureObject=TextureObject() {texture=/Game/Global/Textures/Default/Tileable_Patterns/T_Smoke_Puff_Tile_Up}
   TextureSize=Multiply(
    A=VectorParameter() {parameterName=Texture Size, defaultValue=<rgba 500 500 1500 0>}.RGB
    B=ScalarParameter() {defaultValue=1, parameterName=WorldTexUVScale}
   )
   WorldPosition
   Export Float 4
   In=ComponentMask(None=VectorParameter() {parameterName=WorldPannerSpeed, defaultValue=<rgba 0 1 0 0>}.RGB) {g=true, r=true}
   World Space Normal
   ProjectionTransitionContrast
  ) {materialFunction=/Game/Athena/SafeZone/WorldAlignedTexture_ScalePan}
  n28 = Multiply(
   A=Multiply(A=LinearInterpolate(A=@n22, B=@n23, Alpha=@n20), B=@n21)
   B=DepthFade(Opacity=Multiply(A=@n26, B=@n27.XYZ Texture), FadeDistance=ScalarParameter() {defaultValue=500, parameterName=DepthFade Distance})
  )
  n29 = Power(
   Base=@n28
   Exp=StaticSwitchParameter(
    True=DynamicParameter() {paramNames [SlantAmount, SlantContrast, Erode Contrast, Param4]}.Erode Contrast
    False=ScalarParameter() {defaultValue=1, parameterName=Erode}
   ) {parameterName=Use Dynamic Parameter}
  )
  StaticSwitchParameter(True=Multiply(A=StaticSwitchParameter(True=@n29, False=@n28) {parameterName=Allow Erode}, B=ParticleColor().A), False=@n28) {parameterName=Use with ParticleSystem}
  TextureSample(UVs=@n6) {samplerType=SAMPLERTYPE_Normal, texture=/Game/Packages/Fortnite_UniqueMaterials_1/Textures/T_WaterFall_Tile_N}
 ]
}
