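# Niagara system for the ButterCake crown effect: a crown mesh that grows in with rays, rising sparkles, beam elements and a light-casting halo.
# Enabled emitters follow sockets on the user-supplied ButterCake skeletal mesh, with beams and halo bursting after a short delay.
# System brightness and flare values drive crown and ray color, while sparkle alpha scales with time of day.
# Center sparkle, gnomon and duplicate beam emitters are disabled.
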NiagaraSystem /ButterCake/VFX/NS_ButterCake_Crown
dependencies {
 usedDataInterfaces [&NiagaraDataInterfaceColorCurve, &NiagaraDataInterfaceCurve, &NiagaraDataInterfaceVector2DCurve]
 usedDynamicInputs [
  &VectorLength
  &ClampFloat
  &Lerp_Float
  &ModuloFloat
 ]
 usedModules [
  &self:SystemSpawnScript.SystemScriptSource.SystemScriptGraph.NiagaraNodeAssignment_5.SetVariables_#1
  &SystemState
  &Timeline
  &self:SystemSpawnScript.SystemScriptSource.SystemScriptGraph.NiagaraNodeAssignment_4.SetVariables_#2
  &self:SampleSingleBoneSpeed
  &self:SystemSpawnScript.SystemScriptSource.SystemScriptGraph.NiagaraNodeAssignment_0.SetVariables_#3
  &self:SystemSpawnScript.SystemScriptSource.SystemScriptGraph.NiagaraNodeAssignment_1.SetVariables_#4
  &self:SystemSpawnScript.SystemScriptSource.SystemScriptGraph.NiagaraNodeAssignment_3.SetVariables_#5
  &self:SystemSpawnScript.SystemScriptSource.SystemScriptGraph.NiagaraNodeAssignment_2.SetVariables_#6
 ]
}
emitters {
 Body_Rays_Mesh {
  emitterUpdateScript [
   &EmitterState(
    Inactive Response=::NewEnumerator1
    Life Cycle Mode=::NewEnumerator1
    Loop Behavior=::NewEnumerator1
    Loop Duration=7
    Scalability Mode=::NewEnumerator1
    UseLoopDelay=false
   )
   &SpawnBurst_Instantaneous(Use Spawn Probability=false, UseLoopCountLimit=false)
  ]
  particleSpawnScript [&InitializeParticle(
    Color=<rgba 1 0.4 0 0.3>
    Lifetime=7
    Sprite Size Mode=::NewEnumerator3
    Uniform Sprite Size=75
    UsePositionOffset=false
    Write Material Random=true
   )]
  particleUpdateScript [
   &SkeletalMeshLocation(
    MaskSpawnGroup=false
    Mesh Sampling Type=::NewEnumerator5
    Orientation Sampling=::NewEnumerator1
    Sampled Position Offset=<0 0 30>
    Sampled Position Offset Coordinate Space:ENiagaraSkelMeshTransforms=::NewEnumerator1
    Skeletal Mesh=$User.SkelMesh_ButterCake
    Socket Sampling Mode=::NewEnumerator5
   ) {module=SkeletalMeshLocation001}
   &CameraOffset(Camera Offset Amount=-5)
   /SetVariables(Particles.Scale=<4 4 8>) {isSetParameters=true, module=SetVariables_#7}
   /SetVariables(Particles.SpriteSize=<50 60>) {isSetParameters=true, module=SetVariables_#8}
   &DynamicMaterialParameters(
    Index 0 Param 1=&Multiply_Float(A=&MakeFloatFromInt(Int=&ReturnExecIndex()), B=22)
    Index 0 Param 2=1
    Index 0 Param 3=5
    Index 0 Param 4=5
    Module.Param0WriteEnabled=true
    Module.Param1WriteEnabled=false
    Module.Param2WriteEnabled=true
    Module.Param3WriteEnabled=false
   )
   /SetVariables(Particles.ColorB:LinearColor=<rgba 1 0.5 0 1>) {isSetParameters=true, module=SetVariables_#9}
   &ScaleColor(
    Scale RGB=&VectorFromFloat(Value=&Add_Float(A=$System.FXBrightnessRamped, B=$System.FlareAll))
    ScaleA=false
    ScaleRGB=true
   )
   &ScaleColor(
    Scale Alpha=$System.FXBrightnessRamped
    Scale RGB=&VectorFromFloat(Value=$System.FXBrightnessRamped)
    ScaleA=true
    ScaleRGB=true
   ) {module=ScaleColor003}
  ]
  rendererClasses [&NiagaraMeshRendererProperties]
  renderers [{rendererClass=&NiagaraMeshRendererProperties, rendererIndex=0}]
 }
 Body_Sparkles {
  emitterUpdateScript [
   &EmitterState(Inactive Response=::NewEnumerator1, Life Cycle Mode=::NewEnumerator1, Loop Behavior=::NewEnumerator1, Loop Duration=7, UseLoopDelay=false)
   &SpawnRate(
    SpawnRate=&MakeCustomFloatFromBool(Bool=&SetBoolByFloatComparison(A=1, B=0.2), True Float=50, False Float=0)
    Use Spawn Probability=false
   )
  ]
  particleSpawnScript [
   &InitializeParticle(
    Color Maximum=<rgba 0.9039 1 0.6539 1>
    Color Minimum=<rgba 1 0.3801 0 1>
    Color Mode=::NewEnumerator2
    Lifetime=&RandomRangeFloat(Minimum=0.5, Maximum=1)
    Sprite Rotation Mode=::NewEnumerator2
    Sprite Size Mode=::NewEnumerator1
    Uniform Sprite Size Max=5
    Uniform Sprite Size Min=3
    UsePositionOffset=false
    Write Material Random=true
   )
   /SetVariables(Particles.RandomCore:NiagaraFloat=&FloatFromCurve(
     FloatCurve=DataInterface() {propertyValues {Curve {keys [
         {arriveTangent=0.0559, interpMode=RCIM_Cubic, leaveTangent=0.0559, tangentMode=RCTM_User, time=0, value=0.15}
         {arriveTangent=0.1895, interpMode=RCIM_Cubic, leaveTangent=0.1895, tangentMode=RCTM_User, time=0.9, value=0.3}
         {arriveTangent=1.1808, interpMode=RCIM_Cubic, leaveTangent=1.1808, tangentMode=RCTM_User, time=1, value=0.4}
        ]}}}
     CurveIndex=&RandomRangeFloat(Minimum=0, Maximum=1)
     Scale Curve=2
    )) {isSetParameters=true, module=SetVariables_#10}
   &SkeletalMeshLocation(MaskSpawnGroup=false, Mesh Sampling Type=::NewEnumerator5, Skeletal Mesh=$User.SkelMesh_ButterCake, Socket Sampling Mode=::NewEnumerator5)
   &SphereLocation(
    MaskSpawnGroup=false
    Override Seed=false
    Sphere Radius=35
    Surface Only=true
    UseNonUniformScale=false
    UseOffset=false
    UseSphereOrientation=false
   ) {module=SphereLocation001}
   &AddVelocityFromPoint(Module.Use Falloff Distance=false, Origin Offset=<0 0 -50>, Velocity Strength=5)
   &AddVelocity(Velocity=&Multiply_VectorByFloat(Vector=$Output.SphereLocation001.SphereNormal, Float=15))
   &AddVelocity(Velocity=<0 -100 0>) {enabled=false, module=AddVelocity002}
   &ApplyInitialForces()
  ]
  particleUpdateScript [
   &ParticleState()
   &CurlNoiseForce(Noise Frequency=4, Noise Strength=75, Pan Noise=true, Pan Noise Field=<0.5 0.5 0>)
   &GravityForce(Gravity=<0 0 255>)
   &ScaleSpriteSize(Scale Factor=&Vector2DFromFloat(Value=&FloatFromCurve(
      FloatCurve=DataInterface() {propertyValues {Curve {keys [<at 0 1 RCIM_Cubic>, {arriveTangent=-1.6315, interpMode=RCIM_Cubic, leaveTangent=-1.6315, tangentMode=RCTM_User, time=1, value=0}]}}}
      Scale Curve=$Particles.RandomCore
     )))
   &ScaleSpriteSize(
    Scale Factor=&Multiply_Vector2D(A=$Particles.SpriteSize, B=&Vector2DFromFloat(Value=0.6))
    Scale Sprite Size Mode=::NewEnumerator2
   ) {module=ScaleSpriteSize002}
   &ScaleColor(
    Scale Alpha=&Multiply_Float(
     A=&FloatFromCurve(FloatCurve=DataInterface() {propertyValues {Curve {keys [<at 0.8 1>,<at 1 0>]}}})
     B=&FNTimeOfDayScalar(Night Scale=0.3, Morning Scale=0.1, Evening Scale=0.2)
    )
    Scale RGB=&VectorFromFloat(Value=&FloatFromCurve(FloatCurve=DataInterface() {propertyValues {Curve {keys [<at 0 1>,<at 0.1 3>,<at 0.2 1>]}}}, Scale Curve=0.2))
    ScaleA=true
    ScaleRGB=true
   )
   &Drag(Drag=2, Use Linear Drag=true, Use Rotational Drag=false)
   &SolveForcesAndVelocity(Clamp Velocity=false, Limit Acceleration=false)
   &CameraOffset(Camera Offset Amount=-55)
  ]
  rendererClasses [&NiagaraSpriteRendererProperties]
  renderers [{rendererClass=&NiagaraSpriteRendererProperties}]
 }
 CrownMesh {
  emitterUpdateScript [
   &EmitterState(
    Inactive Response=::NewEnumerator2
    Life Cycle Mode=::NewEnumerator1
    Loop Behavior=::NewEnumerator1
    Loop Duration=1
    Scalability Mode=::NewEnumerator1
    UseLoopDelay=false
   )
   &SpawnBurst_Instantaneous(Use Spawn Probability=false, UseLoopCountLimit=false)
  ]
  particleSpawnScript [
   &InitializeParticle(
    Color=<rgba 1 0.4 0 0.3>
    Lifetime=125
    Mesh Scale Mode=::NewEnumerator3
    Mesh Uniform Scale=150
    Sprite Size Mode=::NewEnumerator3
    Uniform Sprite Size=55
    UsePositionOffset=false
    Write Material Random=true
   )
   &InitialMeshOrientation(Mesh Orientation Mode=::NewEnumerator2, Orientation Vector=<0 -1 0>, Rotation=<0 0 0.05>, Use Rotation Vector=true)
  ]
  particleUpdateScript [
   &ParticleState()
   &SkeletalMeshLocation(
    MaskSpawnGroup=false
    Mesh Sampling Type=::NewEnumerator5
    Sampled Position Offset=<0 0 -15>
    Skeletal Mesh=$User.SkelMesh_ButterCake
    Socket Sampling Mode=::NewEnumerator5
   ) {module=SkeletalMeshLocation001}
   &CameraOffset(Camera Offset Amount=-5) {enabled=false}
   /SetVariables() {isSetParameters=true, module=SetVariables_#7}
   /SetVariables(Particles.SpriteSize=<50 60>) {isSetParameters=true, module=SetVariables_#8}
   &DynamicMaterialParameters(
    Index 0 Param 1=&Multiply_Float(A=&MakeFloatFromInt(Int=&ReturnExecIndex()), B=22)
    Index 0 Param 2=1
    Index 0 Param 3=5
    Index 0 Param 4=5
    Module.Param0WriteEnabled=true
    Module.Param1WriteEnabled=false
    Module.Param2WriteEnabled=true
    Module.Param3WriteEnabled=false
   )
   /SetVariables(Particles.ColorB:LinearColor=<rgba 1 0.5 0 1>) {isSetParameters=true, module=SetVariables_#9}
   &ScaleColor(
    Scale RGB=&VectorFromFloat(Value=&Add_Float(A=$System.FXBrightnessRamped, B=$System.FlareAll))
    ScaleA=false
    ScaleRGB=true
   )
   &ScaleColor(
    Scale Alpha=$System.FXBrightnessRamped
    Scale RGB=&VectorFromFloat(Value=$System.FXBrightnessRamped)
    ScaleA=true
    ScaleRGB=true
   ) {module=ScaleColor003}
   &ScaleMeshSize(Scale Factor=&Multiply_VectorByFloat(
     Vector=<1 1 1>
     Float=&MakeFloatFromVector2d(Vector 2D=&Multiply_Vector2D_ByFloat(
       Float=&FloatFromCurve(
        FloatCurve=DataInterface() {propertyValues {Curve {keys [
            {arriveTangent=-0.0094, interpMode=RCIM_Cubic, leaveTangent=-0.0094, tangentMode=RCTM_User, time=0, value=0}
            {arriveTangent=3.1585, interpMode=RCIM_Cubic, leaveTangent=3.1585, tangentMode=RCTM_User, time=0.2, value=0.625}
            {arriveTangent=-0.0405, interpMode=RCIM_Cubic, leaveTangent=-0.0405, tangentMode=RCTM_User, time=0.5, value=1}
           ]}}}
        CurveIndex=$Emitter.Age
        Scale Curve=0.015
       )
       Vector2D=<1 1>
      ))
    ))
  ]
  rendererClasses [&NiagaraMeshRendererProperties]
  renderers [{rendererClass=&NiagaraMeshRendererProperties, rendererIndex=0}]
 }
 Gnomon {
  bEnabled=false
  emitterUpdateScript [
   &EmitterState()
   &SpawnBurst_Instantaneous(Use Spawn Probability=false, UseLoopCountLimit=false)
  ]
  particleSpawnScript [&InitializeParticle(
    Color Mode=::NewEnumerator0
    Lifetime=5
    Mesh Scale Mode=::NewEnumerator3
    Mesh Uniform Scale=5
    UsePositionOffset=false
    Write Material Random=true
   )]
  particleUpdateScript [
   &ParticleState()
   &SkeletalMeshLocation(
    MaskSpawnGroup=false
    Mesh Sampling Type=::NewEnumerator5
    Sampled Position Offset=<0 5 30>
    Skeletal Mesh=$User.SkelMesh_ButterCake
    Socket Sampling Mode=::NewEnumerator5
   ) {module=SkeletalMeshLocation001}
  ]
  rendererClasses [&NiagaraMeshRendererProperties]
  renderers [{rendererClass=&NiagaraMeshRendererProperties, rendererIndex=0}]
 }
 beam_elements {
  simTarget=GPUComputeSim
  emitterUpdateScript [
   &EmitterState(
    Inactive Response=::NewEnumerator1
    Life Cycle Mode=::NewEnumerator1
    Loop Behavior=::NewEnumerator1
    Loop Duration=7
    Scalability Mode=::NewEnumerator1
    UseLoopDelay=false
   )
   &SpawnBurst_Instantaneous(Spawn Time=0.55, Use Spawn Probability=false, UseLoopCountLimit=false)
  ]
  particleSpawnScript [&InitializeParticle(
    Color=<rgba 4 2 0 0.2>
    Lifetime=10
    Mesh Scale=<0.75 0.75 0.75>
    Mesh Scale Mode=::NewEnumerator4
    Sprite Size Mode=::NewEnumerator3
    Uniform Sprite Size=75
    UsePositionOffset=false
    Write Material Random=true
   )]
  particleUpdateScript [
   &ParticleState()
   &SkeletalMeshLocation(
    MaskSpawnGroup=false
    Mesh Sampling Type=::NewEnumerator5
    Orientation Sampling=::NewEnumerator1
    Skeletal Mesh=$User.SkelMesh_ButterCake
    Socket Sampling Mode=::NewEnumerator5
   ) {module=SkeletalMeshLocation001}
   &CameraOffset(Camera Offset Amount=-15)
   /SetVariables(Particles.SpriteSize=&Multiply_Vector2D_ByFloat(
     Float=&FloatFromCurve(
      FloatCurve=DataInterface() {propertyValues {Curve {keys [
          {arriveTangent=-0.0094, interpMode=RCIM_Cubic, leaveTangent=-0.0094, tangentMode=RCTM_User, time=0, value=0}
          {arriveTangent=7.6747, interpMode=RCIM_Cubic, leaveTangent=7.6747, tangentMode=RCTM_User, time=0.3, value=2}
          {arriveTangent=-0.0405, interpMode=RCIM_Cubic, leaveTangent=-0.0405, tangentMode=RCTM_User, time=1, value=4.375}
         ]}}}
      Scale Curve=33
     )
     Vector2D=<1 1>
    )) {isSetParameters=true, module=SetVariables_#11}
   &Color(Color=&LerpLinearColors(
     StartColor=<rgba 5 4 0 0.2>
     EndColor=<rgba 3.0447 0 5 0.2>
     LerpFactor=&FloatFromCurve(FloatCurve=DataInterface() {propertyValues {Curve {keys [
          {arriveTangent=9.0725, interpMode=RCIM_Cubic, leaveTangent=9.0725, tangentMode=RCTM_User, time=0, value=0}
          {arriveTangent=-2.5249, interpMode=RCIM_Cubic, leaveTangent=-2.5249, tangentMode=RCTM_User, time=0.1408, value=0.9847}
          {arriveTangent=-0.0405, interpMode=RCIM_Cubic, leaveTangent=-0.0405, tangentMode=RCTM_User, time=1, value=0}
         ]}}})
    ))
   &ScaleColor(
    Scale Alpha=&FloatFromCurve(FloatCurve=DataInterface() {propertyValues {Curve {keys [
         {arriveTangent=-0.2933, interpMode=RCIM_Cubic, leaveTangent=-0.2933, tangentMode=RCTM_User, time=0, value=0}
         {arriveTangent=3.5997, interpMode=RCIM_Cubic, leaveTangent=3.5997, tangentMode=RCTM_User, time=0.1, value=0.95}
         {arriveTangent=-1.6622, interpMode=RCIM_Cubic, leaveTangent=-1.6622, tangentMode=RCTM_User, time=0.5, value=0.35}
         {arriveTangent=-0.0383, interpMode=RCIM_Cubic, leaveTangent=-0.0383, tangentMode=RCTM_User, time=0.9, value=0}
        ]}}})
    ScaleA=true
    ScaleRGB=false
   ) {module=ScaleColor001}
   &ScaleSpriteSize(Uniform Curve Sprite Scale=DataInterface() {propertyValues {Curve {
       postInfinityExtrap=RCCE_Cycle
       preInfinityExtrap=RCCE_Cycle
       keys [
        {interpMode=RCIM_Cubic, leaveTangentWeight=0.7084, tangentMode=RCTM_User, time=0, value=1}
        {
         arriveTangent=-0.0106
         arriveTangentWeight=0.3333
         interpMode=RCIM_Cubic
         leaveTangent=-0.0106
         leaveTangentWeight=0.3333
         time=2.1218
         value=0.879
        }
        {arriveTangent=0.01, interpMode=RCIM_Cubic, leaveTangent=0.01, time=5.2552, value=0.9441}
        {arriveTangent=0.0118, interpMode=RCIM_Cubic, leaveTangent=0.0118, time=7.1328, value=0.929}
        {arriveTangentWeight=0.956, interpMode=RCIM_Cubic, tangentMode=RCTM_User, time=10, value=1}
       ]
      }}})
   /SetVariables(Particles.SpriteSize_B:Vector2f=&Multiply_Vector2D_ByFloat(
     Float=&FloatFromCurve(
      FloatCurve=DataInterface() {propertyValues {Curve {
         postInfinityExtrap=RCCE_Cycle
         preInfinityExtrap=RCCE_Cycle
         keys [
          {interpMode=RCIM_Cubic, leaveTangentWeight=0.7374, tangentMode=RCTM_User, time=0, value=1}
          {
           arriveTangent=-0.0106
           arriveTangentWeight=0.5406
           interpMode=RCIM_Cubic
           leaveTangent=-0.0106
           tangentMode=RCTM_User
           time=2.1218
           value=0.879
          }
          {
           arriveTangent=0.01
           arriveTangentWeight=1.0445
           interpMode=RCIM_Cubic
           leaveTangent=0.01
           leaveTangentWeight=1.5816
           tangentMode=RCTM_User
           time=5.2552
           value=0.9441
          }
          {
           arriveTangent=0.0118
           arriveTangentWeight=0.6259
           interpMode=RCIM_Cubic
           leaveTangent=0.0118
           leaveTangentWeight=0.9557
           tangentMode=RCTM_User
           time=7.1328
           value=0.929
          }
          {arriveTangentWeight=2.6261, interpMode=RCIM_Cubic, tangentMode=RCTM_User, time=10, value=1}
         ]
        }}}
      Scale Curve=3
     )
     Vector2D=<15 123.75>
    )) {isSetParameters=true, module=SetVariables_#12}
   &CameraOffset(Camera Offset Amount=15) {module=CameraOffset001}
   &self:SetAlignment(AimPosition:Vector3f=<0 0 1>) {module=SetAlignment}
  ]
  rendererClasses [&NiagaraSpriteRendererProperties, &NiagaraMeshRendererProperties]
  renderers [
   {rendererClass=&NiagaraSpriteRendererProperties}
   {rendererClass=&NiagaraMeshRendererProperties, rendererIndex=1}
  ]
 }
 beam_elements001 {
  bEnabled=false
  simTarget=GPUComputeSim
  emitterUpdateScript [
   &EmitterState(
    Life Cycle Mode=::NewEnumerator1
    Loop Behavior=::NewEnumerator1
    Loop Duration Mode=::NewEnumerator1
    Scalability Mode=::NewEnumerator1
    UseLoopDelay=false
   )
   &SpawnBurst_Instantaneous(Spawn Time=0.55, Use Spawn Probability=false, UseLoopCountLimit=false)
  ]
  particleSpawnScript [&InitializeParticle(
    Color=<rgba 4 2 0 0.2>
    Lifetime=1.3
    Mesh Scale Mode=::NewEnumerator3
    Mesh Uniform Scale=55
    Sprite Size Mode=::NewEnumerator3
    Uniform Sprite Size=75
    UsePositionOffset=false
    Write Material Random=true
   )]
  particleUpdateScript [
   &ParticleState(Kill Particles When Lifetime Has Elapsed=false)
   &SkeletalMeshLocation(
    MaskSpawnGroup=false
    Mesh Sampling Type=::NewEnumerator5
    Orientation Sampling=::NewEnumerator1
    Skeletal Mesh=$User.SkelMesh_ButterCake
    Socket Sampling Mode=::NewEnumerator5
   ) {module=SkeletalMeshLocation001}
   &CameraOffset(Camera Offset Amount=-15)
   /SetVariables(Particles.SpriteSize=&Multiply_Vector2D_ByFloat(
     Float=&FloatFromCurve(
      FloatCurve=DataInterface() {propertyValues {Curve {keys [
          {arriveTangent=-0.0094, interpMode=RCIM_Cubic, leaveTangent=-0.0094, tangentMode=RCTM_User, time=0, value=0}
          {arriveTangent=7.6747, interpMode=RCIM_Cubic, leaveTangent=7.6747, tangentMode=RCTM_User, time=0.3, value=2}
          {arriveTangent=-0.0405, interpMode=RCIM_Cubic, leaveTangent=-0.0405, tangentMode=RCTM_User, time=1, value=4.375}
         ]}}}
      Scale Curve=33
     )
     Vector2D=<1 1>
    )) {isSetParameters=true, module=SetVariables_#11}
   &Color(Color=&LerpLinearColors(
     StartColor=<rgba 5 4 0 0.2>
     EndColor=<rgba 3.0447 0 5 0.2>
     LerpFactor=&FloatFromCurve(FloatCurve=DataInterface() {propertyValues {Curve {keys [
          {arriveTangent=9.0725, interpMode=RCIM_Cubic, leaveTangent=9.0725, tangentMode=RCTM_User, time=0, value=0}
          {arriveTangent=-2.5249, interpMode=RCIM_Cubic, leaveTangent=-2.5249, tangentMode=RCTM_User, time=0.1408, value=0.9847}
          {arriveTangent=-0.0405, interpMode=RCIM_Cubic, leaveTangent=-0.0405, tangentMode=RCTM_User, time=1, value=0}
         ]}}})
    ))
   &ScaleColor(
    Scale Alpha=&FloatFromCurve(FloatCurve=DataInterface() {propertyValues {Curve {keys [
         {arriveTangent=-0.2933, interpMode=RCIM_Cubic, leaveTangent=-0.2933, tangentMode=RCTM_User, time=0, value=0}
         {arriveTangent=3.5997, interpMode=RCIM_Cubic, leaveTangent=3.5997, tangentMode=RCTM_User, time=0.1, value=0.95}
         {arriveTangent=-1.6622, interpMode=RCIM_Cubic, leaveTangent=-1.6622, tangentMode=RCTM_User, time=0.5, value=0.35}
         {arriveTangent=-0.0383, interpMode=RCIM_Cubic, leaveTangent=-0.0383, tangentMode=RCTM_User, time=0.9, value=0}
        ]}}})
    ScaleA=true
    ScaleRGB=false
   ) {module=ScaleColor001}
   &ScaleSpriteSize(Uniform Curve Sprite Scale=DataInterface() {propertyValues {Curve {
       postInfinityExtrap=RCCE_Cycle
       preInfinityExtrap=RCCE_Cycle
       keys [
        {interpMode=RCIM_Cubic, leaveTangentWeight=0.7084, tangentMode=RCTM_User, time=0, value=1}
        {
         arriveTangent=-0.0106
         arriveTangentWeight=0.3333
         interpMode=RCIM_Cubic
         leaveTangent=-0.0106
         leaveTangentWeight=0.3333
         time=2.1218
         value=0.879
        }
        {arriveTangent=0.01, interpMode=RCIM_Cubic, leaveTangent=0.01, time=5.2552, value=0.9441}
        {arriveTangent=0.0118, interpMode=RCIM_Cubic, leaveTangent=0.0118, time=7.1328, value=0.929}
        {arriveTangentWeight=0.956, interpMode=RCIM_Cubic, tangentMode=RCTM_User, time=10, value=1}
       ]
      }}})
   /SetVariables(Particles.SpriteSize_B:Vector2f=&Multiply_Vector2D_ByFloat(
     Float=&FloatFromCurve(
      FloatCurve=DataInterface() {propertyValues {Curve {
         postInfinityExtrap=RCCE_Cycle
         preInfinityExtrap=RCCE_Cycle
         keys [
          {interpMode=RCIM_Cubic, leaveTangentWeight=0.7374, tangentMode=RCTM_User, time=0, value=1}
          {
           arriveTangent=-0.0106
           arriveTangentWeight=0.5406
           interpMode=RCIM_Cubic
           leaveTangent=-0.0106
           tangentMode=RCTM_User
           time=2.1218
           value=0.879
          }
          {
           arriveTangent=0.01
           arriveTangentWeight=1.0445
           interpMode=RCIM_Cubic
           leaveTangent=0.01
           leaveTangentWeight=1.5816
           tangentMode=RCTM_User
           time=5.2552
           value=0.9441
          }
          {
           arriveTangent=0.0118
           arriveTangentWeight=0.6259
           interpMode=RCIM_Cubic
           leaveTangent=0.0118
           leaveTangentWeight=0.9557
           tangentMode=RCTM_User
           time=7.1328
           value=0.929
          }
          {arriveTangentWeight=2.6261, interpMode=RCIM_Cubic, tangentMode=RCTM_User, time=10, value=1}
         ]
        }}}
      Scale Curve=3
     )
     Vector2D=<15 123.75>
    )) {isSetParameters=true, module=SetVariables_#12}
   &CameraOffset(Camera Offset Amount=15) {module=CameraOffset001}
   &self:SetAlignment(AimPosition:Vector3f=<0 0 1>) {module=SetAlignment}
   &ScaleMeshSize(Scale Factor=&Multiply_VectorByFloat(
     Vector=<1 1 1>
     Float=&MakeFloatFromVector2d(Vector 2D=&Multiply_Vector2D_ByFloat(
       Float=&FloatFromCurve(
        FloatCurve=DataInterface() {propertyValues {Curve {keys [
            {arriveTangent=-0.0094, interpMode=RCIM_Cubic, leaveTangent=-0.0094, tangentMode=RCTM_User, time=0, value=0}
            {arriveTangent=7.6747, interpMode=RCIM_Cubic, leaveTangent=7.6747, tangentMode=RCTM_User, time=0.3, value=2}
            {arriveTangent=-0.0405, interpMode=RCIM_Cubic, leaveTangent=-0.0405, tangentMode=RCTM_User, time=1, value=4.375}
           ]}}}
        Scale Curve=0.0827
       )
       Vector2D=<1 1>
      ))
    ))
  ]
  rendererClasses [&NiagaraSpriteRendererProperties, &NiagaraMeshRendererProperties]
  renderers [
   {rendererClass=&NiagaraSpriteRendererProperties}
   {rendererClass=&NiagaraMeshRendererProperties, rendererIndex=1}
  ]
 }
 single_halo001 {
  emitterUpdateScript [
   &EmitterState(Inactive Response=::NewEnumerator1, Life Cycle Mode=::NewEnumerator1, Loop Behavior=::NewEnumerator1, Loop Duration=7, UseLoopDelay=false)
   &SpawnBurst_Instantaneous(Spawn Time=0.65, Use Spawn Probability=false, UseLoopCountLimit=false)
  ]
  particleSpawnScript [&InitializeParticle(
    Color=<rgba 5 2 0 1>
    Lifetime=10
    Sprite Size Mode=::NewEnumerator3
    Uniform Sprite Size=75
    UsePositionOffset=false
    Write Material Random=true
   )]
  particleUpdateScript [
   &ParticleState()
   &SkeletalMeshLocation(
    MaskSpawnGroup=false
    Mesh Sampling Type=::NewEnumerator5
    Orientation Sampling=::NewEnumerator1
    Skeletal Mesh=$User.SkelMesh_ButterCake
    Socket Sampling Mode=::NewEnumerator5
   ) {module=SkeletalMeshLocation001}
   &CameraOffset(Camera Offset Amount=15)
   /SetVariables(Particles.SpriteSize=&Multiply_Vector2D_ByFloat(
     Float=&FloatFromCurve(
      FloatCurve=DataInterface() {propertyValues {Curve {keys [
          {arriveTangent=16.2037, interpMode=RCIM_Cubic, leaveTangent=16.2037, tangentMode=RCTM_User, time=0, value=0}
          {arriveTangent=4.6475, interpMode=RCIM_Cubic, leaveTangent=4.6475, tangentMode=RCTM_User, time=0.3, value=3.125}
          {arriveTangent=-0.0405, interpMode=RCIM_Cubic, leaveTangent=-0.0405, tangentMode=RCTM_User, time=1, value=4.375}
         ]}}}
      Scale Curve=43
     )
     Vector2D=<0.65 0.65>
    )) {enabled=false, isSetParameters=true, module=SetVariables_#11}
   /SetVariables(Particles.SpriteSize=&Multiply_Vector2D_ByFloat(
     Float=&FloatFromCurve(
      FloatCurve=DataInterface() {propertyValues {Curve {keys [
          {arriveTangent=34.6115, interpMode=RCIM_Cubic, leaveTangent=34.6115, tangentMode=RCTM_User, time=0, value=0}
          {arriveTangent=4.6475, interpMode=RCIM_Cubic, leaveTangent=4.6475, tangentMode=RCTM_User, time=0.2, value=3.75}
          <at 0.7 4.625>
          {arriveTangent=-0.0405, interpMode=RCIM_Cubic, leaveTangent=-0.0405, tangentMode=RCTM_User, time=1, value=4.375}
         ]}}}
      Scale Curve=43
     )
     Vector2D=<0.65 0.65>
    )) {isSetParameters=true}
   &ScaleColor(
    Scale Alpha=&FloatFromCurve(
     FloatCurve=DataInterface() {propertyValues {Curve {keys [
         {arriveTangent=-0.2933, interpMode=RCIM_Cubic, leaveTangent=-0.2933, tangentMode=RCTM_User, time=0, value=0}
         {arriveTangent=-2.8313, interpMode=RCIM_Cubic, leaveTangent=-2.8313, tangentMode=RCTM_User, time=0.1, value=0.95}
         {arriveTangent=-1.6622, interpMode=RCIM_Cubic, leaveTangent=-1.6622, tangentMode=RCTM_User, time=0.2, value=0.45}
         {arriveTangent=-0.0383, interpMode=RCIM_Cubic, leaveTangent=-0.0383, tangentMode=RCTM_User, time=1, value=0.2}
        ]}}}
     Scale Curve=0.4
    )
    ScaleA=true
    ScaleRGB=false
   ) {module=ScaleColor001}
  ]
  rendererClasses [&NiagaraSpriteRendererProperties, &NiagaraLightRendererProperties]
  renderers [
   {rendererClass=&NiagaraSpriteRendererProperties}
   {rendererClass=&NiagaraSpriteRendererProperties, rendererIndex=1}
   {rendererClass=&NiagaraLightRendererProperties, rendererIndex=2}
  ]
 }
 sparkle_center {
  bEnabled=false
  emitterUpdateScript [
   &EmitterState(Life Cycle Mode=::NewEnumerator1, Loop Behavior=::NewEnumerator1, Loop Duration Mode=::NewEnumerator1, UseLoopDelay=false)
   &SpawnBurst_Instantaneous(Spawn Count=44, Spawn Time=0.55, Use Spawn Probability=false, UseLoopCountLimit=false)
  ]
  particleSpawnScript [
   &InitializeParticle(
    Color Maximum=<rgba 1 0.5692 0.3873 1>
    Color Minimum=<rgba 1 0.6949 0.0328 1>
    Color Mode=::NewEnumerator2
    Lifetime=&RandomRangeFloat(Minimum=0.3, Maximum=1.2)
    Sprite Rotation Mode=::NewEnumerator2
    Sprite Size Mode=::NewEnumerator1
    Uniform Sprite Size Max=6
    Uniform Sprite Size Min=3.5
    UsePositionOffset=false
    Write Material Random=true
   )
   /SetVariables(Particles.RandomCore:NiagaraFloat=&FloatFromCurve(
     FloatCurve=DataInterface() {propertyValues {Curve {keys [
         {arriveTangent=0.0559, interpMode=RCIM_Cubic, leaveTangent=0.0559, tangentMode=RCTM_User, time=0, value=0.2}
         {arriveTangent=0.1895, interpMode=RCIM_Cubic, leaveTangent=0.1895, tangentMode=RCTM_User, time=0.9, value=0.25}
         {arriveTangent=1.1808, interpMode=RCIM_Cubic, leaveTangent=1.1808, tangentMode=RCTM_User, time=1, value=0.3}
        ]}}}
     CurveIndex=&RandomRangeFloat(Minimum=0, Maximum=1)
     Scale Curve=3
    )) {isSetParameters=true, module=SetVariables_#10}
   &SkeletalMeshLocation(
    MaskSpawnGroup=false
    Mesh Sampling Type=::NewEnumerator5
    Orientation Sampling=::NewEnumerator1
    Sampled Position Offset Coordinate Space:ENiagaraSkelMeshTransforms=::NewEnumerator1
    Skeletal Mesh=$User.SkelMesh_ButterCake
    Socket Sampling Mode=::NewEnumerator5
   )
   &SphereLocation(
    MaskSpawnGroup=false
    Override Seed=false
    Sphere Coordinate Space=::World
    Sphere Origin:NiagaraPosition=$Particles.SkeletalMeshLocation.SampledPosition
    Sphere Radius=15
    Surface Only=false
    UseNonUniformScale=false
    UseOffset=true
    UseSphereOrientation=false
   )
   &AddVelocityFromPoint(
    Module.Use Falloff Distance=false
    Origin Offset:NiagaraPosition=$Output.SphereLocation.Position
    Origin Offset Coordinate Space=::World
    Velocity Strength=222
   )
  ]
  particleUpdateScript [
   &ParticleState(Kill Particles When Lifetime Has Elapsed=false)
   &CurlNoiseForce(Noise Frequency=125, Noise Strength=&Multiply_Float(A=125, B=1), Pan Noise=true, Pan Noise Field=<0.5 0.5 0>)
   &ScaleSpriteSize(Scale Factor=&Vector2DFromFloat(Value=&FloatFromCurve(
      FloatCurve=DataInterface() {propertyValues {Curve {keys [
          <at 0 1 RCIM_Cubic>
          {arriveTangent=-1, interpMode=RCIM_Cubic, leaveTangent=-1, time=0.6, value=1.25}
          {arriveTangent=-0.1423, interpMode=RCIM_Cubic, leaveTangent=-0.1423, tangentMode=RCTM_User, time=1, value=0}
         ]}}}
      Scale Curve=$Particles.RandomCore
     )))
   &ScaleSpriteSize(
    Scale Factor=&Multiply_Vector2D(A=$Particles.SpriteSize, B=<0.3 0.3>)
    Scale Sprite Size Mode=::NewEnumerator2
   ) {module=ScaleSpriteSize001}
   &ScaleColor(
    Scale Alpha=&Multiply_Float(
     A=&FloatFromCurve(FloatCurve=DataInterface() {propertyValues {Curve {keys [
          {tangentMode=RCTM_User, time=0, value=1}
          {arriveTangent=-0.0405, interpMode=RCIM_Cubic, leaveTangent=-0.0405, tangentMode=RCTM_User, time=0.7009, value=1}
          {arriveTangent=-8.0089, interpMode=RCIM_Cubic, leaveTangent=-8.0089, tangentMode=RCTM_User, time=1, value=0}
         ]}}})
     B=&FNTimeOfDayScalar(Night Scale=0.3, Morning Scale=0.1, Evening Scale=0.2)
    )
    Scale RGB=&VectorFromFloat(Value=&FloatFromCurve(FloatCurve=DataInterface() {propertyValues {Curve {keys [
          {arriveTangent=2.9857, interpMode=RCIM_Cubic, leaveTangent=2.9857, tangentMode=RCTM_User, time=0, value=0}
          <at 0.1 2 RCIM_Cubic>
          {arriveTangent=-30.1063, interpMode=RCIM_Cubic, leaveTangent=-30.1063, tangentMode=RCTM_User, time=0.2, value=0.5}
         ]}}}))
    ScaleA=true
    ScaleRGB=true
   )
   &self:ScratchModule_03(AimPosition:NiagaraPosition=$Particles.SkeletalMeshLocation.SampledPosition) {module=ScratchModule_03}
   &Drag(Drag=2, Use Linear Drag=true, Use Rotational Drag=false)
   &SolveForcesAndVelocity(Clamp Velocity=false, Limit Acceleration=false)
  ]
  rendererClasses [&NiagaraSpriteRendererProperties]
  renderers [{rendererClass=&NiagaraSpriteRendererProperties}]
 }
}
summary {
 systemName=NS_ButterCake_Crown
 emitters [
  {emitterName=Body_Rays_Mesh, rendererClasses [&NiagaraMeshRendererProperties]}
  {emitterName=Body_Sparkles, rendererClasses [&NiagaraSpriteRendererProperties]}
  {
   emitterName=single_halo001
   rendererClasses [&NiagaraSpriteRendererProperties, &NiagaraLightRendererProperties]
  }
  {
   emitterName=beam_elements
   simTarget=GPUComputeSim
   rendererClasses [&NiagaraSpriteRendererProperties, &NiagaraMeshRendererProperties]
  }
  {bEnabled=false, emitterName=sparkle_center, rendererClasses [&NiagaraSpriteRendererProperties]}
  {bEnabled=false, emitterName=Gnomon, rendererClasses [&NiagaraMeshRendererProperties]}
  {emitterName=CrownMesh, rendererClasses [&NiagaraMeshRendererProperties]}
  {
   bEnabled=false
   emitterName=beam_elements001
   simTarget=GPUComputeSim
   rendererClasses [&NiagaraSpriteRendererProperties, &NiagaraMeshRendererProperties]
  }
 ]
 userVariables [
  User.bIsFrontEndPreview {type=NiagaraBool}
  User.Flare {type=NiagaraFloat, defaultValue=0, description=""}
  User.FXBrightnessScale {type=NiagaraFloat, defaultValue=0.86, description=""}
  User.IsClient {type=NiagaraBool}
  User.RampIn {type=NiagaraFloat, defaultValue=1, description=""}
  User.RampOut {type=NiagaraFloat, defaultValue=1, description=""}
  User.RibbonIDOffset {type=NiagaraInt32}
  User.SkelMesh_ButterCake {
   type=NiagaraDataInterfaceSkeletalMesh
   defaultValue {struct=NiagaraExt_VariableValue_DataInterface, value {dataInterface=&self:NiagaraDataInterfaceSkeletalMesh_3}}
  }
 ]
}
