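# Niagara system for a Buttercake goo projectile, built from a burst mesh goop cluster, pulsing energy-wave meshes, flung droplets and wind-blown leaf sprites.
# The Projectile emitter throws off tumbling, velocity-inheriting mesh chunks that fall under gravity.
# A ribbon emitter trails each chunk through particle attribute reads, tapering by parent age.
# Leaf drag responds to music snare and kick parameters; GooTrail and an alternate ribbon emitter are disabled.
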NiagaraSystem /ButterCake/VFX/NS_Buttercake_Prj
dependencies {
 usedDataInterfaces [
  &NiagaraDataInterfaceColorCurve
  &NiagaraDataInterfaceCurve
  &NiagaraDataInterfaceCurlNoise
  &NiagaraDataInterfaceParticleRead
 ]
 usedDynamicInputs []
 usedModules [&SystemState]
}
emitters {
 EnergyWave {
  emitterUpdateScript [
   &EmitterState(Inactive Response=::NewEnumerator1, Life Cycle Mode=::NewEnumerator1, Loop Behavior=::NewEnumerator1, UseLoopDelay=true)
   &SpawnRate(SpawnRate=10, Use Spawn Probability=false)
   /SetVariables(
    Emitter.DepthFade=10
    Emitter.EmissiveTilingMethod:Vector4f=<0 1 0 1>
    Emitter.EmissiveUV:Vector4f=<2 1 0 0>
    Emitter.NoiseUV:Vector4f=<1 0.5 0 0>
    Emitter.RefractionIntensity=10
   ) {isSetParameters=true}
  ]
  particleSpawnScript [
   /SetVariables(
    Particles.ColorEnd:LinearColor=&MultiplyLinearColor(
     Linear Color=&RandomRangeLinearColor(Minimum=<rgba 0.4635 0.1461 0 1>, Maximum=<rgba 0 0.8854 0.2272 1>)
     Scale Factor=&MakeFloat4FromVectorAndFloat("Vector (RGB)"=&VectorFromFloat(Value=5), "Float (Alpha)"=1)
    )
    Particles.ColorStart:LinearColor=&MultiplyLinearColor(
     Linear Color=&RandomRangeLinearColor(Minimum=<rgba 0.9688 0.0572 0 1>, Maximum=<rgba 1 0.4913 0 1>)
     Scale Factor=&MakeFloat4FromVectorAndFloat("Vector (RGB)"=&VectorFromFloat(Value=20), "Float (Alpha)"=1)
    )
   ) {isSetParameters=true}
   &InitializeParticle(
    Color Mode=::NewEnumerator0
    Lifetime Max=0.5
    Lifetime Min=0.2
    Lifetime Mode=::NewEnumerator1
    Mesh Scale Mode=::NewEnumerator3
    Mesh Uniform Scale=2.2
    Sprite Rotation Mode=::NewEnumerator2
    UsePositionOffset=false
    Write Material Random=true
   )
   &InitialMeshOrientation(Orientation Axis=<0 0 1>, Orientation Vector=<0 0 1>, Rotation=<0 0.25 0>, Use Orientation Vector=true, Use Rotation Vector=true)
   &InitialMeshOrientation(
    Orientation Axis=<0 0 0>
    Orientation Vector=<0 0 0>
    Rotation=&RandomRangeVector(Minimum=<0 0 0>, Maximum=<1 0 0>)
    Use Orientation Vector=false
    Use Rotation Vector=true
   ) {module=InitialMeshOrientation001}
   &CameraOffset(Camera Offset Amount=30)
  ]
  particleUpdateScript [
   &ParticleState()
   &ScaleColor(
    Scale Alpha=&FloatFromCurve(FloatCurve=DataInterface() {propertyValues {Curve {keys [
         {interpMode=RCIM_Cubic, tangentMode=RCTM_Break, time=0, value=0}
         {interpMode=RCIM_Cubic, tangentMode=RCTM_Break, time=0.4, value=1}
         {interpMode=RCIM_Cubic, tangentMode=RCTM_Break, time=1, value=0}
        ]}}})
    Scale RGB=&MakeVectorFromLinearColorRGB(Color=&LerpLinearColors(
      StartColor=$Particles.ColorEnd
      EndColor=$Particles.ColorEnd
      LerpFactor=&FloatFromCurve(FloatCurve=DataInterface() {propertyValues {Curve {keys [
           {interpMode=RCIM_Cubic, leaveTangent=-0.0693, tangentMode=RCTM_Break, time=0.2, value=0}
           {arriveTangent=1.0727, interpMode=RCIM_Cubic, leaveTangent=1.0727, tangentMode=RCTM_User, time=0.6, value=0.25}
           {interpMode=RCIM_Cubic, tangentMode=RCTM_Break, time=0.9, value=1}
          ]}}})
     ))
    ScaleA=true
    ScaleRGB=true
   )
   &ScaleColor(
    Scale Alpha=&FNTimeOfDayScalar(Night Scale=0.7, Morning Scale=0.5, Evening Scale=0.5)
    Scale RGB=&VectorFromFloat(Value=&FloatFromCurve(
      FloatCurve=DataInterface() {propertyValues {Curve {keys [{interpMode=RCIM_Cubic, tangentMode=RCTM_Break, time=0, value=1}, {interpMode=RCIM_Cubic, tangentMode=RCTM_Break, time=1, value=6}]}}}
      Scale Curve=15
     ))
    ScaleA=false
    ScaleRGB=true
   ) {module=ScaleColor001}
   &DynamicMaterialParameters(
    Index 0 Param 1=&FloatFromCurve(FloatCurve=DataInterface() {propertyValues {Curve {keys [
         {arriveTangent=-1.5888, interpMode=RCIM_Cubic, leaveTangent=-1.5888, tangentMode=RCTM_User, time=0, value=0.5}
         {arriveTangent=-1.2614, interpMode=RCIM_Cubic, leaveTangent=-1.2614, tangentMode=RCTM_User, time=0.15, value=0.0979}
         <at 1 -0.3>
        ]}}})
    Index 0 Param 2=&FloatFromCurve(FloatCurve=DataInterface() {propertyValues {Curve {keys [
         {arriveTangent=-0.2969, interpMode=RCIM_Cubic, leaveTangent=-0.2969, tangentMode=RCTM_User, time=0, value=1}
         {arriveTangent=-0.0229, interpMode=RCIM_Cubic, leaveTangent=-0.0229, tangentMode=RCTM_User, time=0.15, value=0.945}
         <at 1 0.9>
        ]}}})
    Index 0 Param 3=&FloatFromCurve(FloatCurve=DataInterface() {propertyValues {Curve {keys [<at 0 0>,<at 0.12 0.8>,<at 1 1.2>]}}})
    Module.Param0WriteEnabled=true
    Module.Param1WriteEnabled=true
    Module.Param2WriteEnabled=true
    Module.Param3WriteEnabled=false
   )
   &ScaleMeshSize(Scale Factor=&VectorFromFloat(Value=&FloatFromCurve(FloatCurve=DataInterface() {propertyValues {Curve {keys [
          {arriveTangent=1.4295, interpMode=RCIM_Cubic, leaveTangent=1.4295, tangentMode=RCTM_User, time=0, value=0.4}
          {arriveTangent=0.034, interpMode=RCIM_Cubic, tangentMode=RCTM_Break, time=1, value=1}
         ]}}})))
  ]
  rendererClasses [&NiagaraMeshRendererProperties]
  renderers [{rendererClass=&NiagaraMeshRendererProperties, rendererIndex=0}]
 }
 FN_SourceParticleRibbon {
  bEnabled=false
  emitterSpawnScript [/Game/Effects/Niagara/Modules/Attributes/FN_Source_v2(AttributeReader=DataInterface() {propertyValues {EmitterBinding {bindingMode=Other, emitterName=Projectile}}}) {enabled=false}]
  emitterUpdateScript [
   &EmitterState(Life Cycle Mode=::NewEnumerator1, Loop Duration=1, UseLoopDelay=false)
   &SpawnRate(
    SpawnRate=/Game/Effects/Niagara/DynamicInputs/Attributes/FN_GetEmitterParticleCount(AttributeReader:NiagaraDataInterfaceParticleRead=$Emitter.AttributeReader, Multiplier:NiagaraFloat=30)
    Use Spawn Probability=false
   )
   &FN_Source_v2(AttributeReader=DataInterface() {propertyValues {EmitterBinding {bindingMode=Other, emitterName=Projectile}}}) {module=FN_Source_v2001}
  ]
  particleSpawnScript [
   /Game/Effects/Niagara/Modules/Attributes/FN_SourceParentID(Ribbon ID from Parent ID=true)
   /Game/Effects/Niagara/Modules/Attributes/FN_SourceLocation(Position:Vector3f=/Game/Effects/Niagara/DynamicInputs/Attributes/FN_GetVectorByID(Attrib name:ENiagaraCommonVectors=::NewEnumerator0, Default on Invalid001:Vector3f=<0 0 0>))
   &InitializeRibbon(
    Lifetime=$Particles.ParentAge
    Module.Write Color=true
    Module.Write Lifetime=true
    Module.Write Mass=true
    Module.Write Material Random=true
    Module.Write Position=false
    Module.Write RibbonTwist=false
    Module.Write RibbonWidth=true
    Ribbon Width=150
   )
  ]
  particleUpdateScript [
   /Game/Effects/Niagara/Modules/Attributes/FN_UpdateSourcedRibbbons(
    Inherit Parent Age=false
    Locked Particles Position:Vector3f=&FN_GetVectorByID(Attrib name:ENiagaraCommonVectors=::NewEnumerator0, Default on Invalid001:Vector3f=<0 0 0>)
    Ribbon Width:NiagaraFloat=$Particles.Initial.RibbonWidth
   ) {module=FN_UpdateLockedParticles}
   /SetVariables(Particles.Lifetime=&FloatFromCurve(FloatCurve=DataInterface() {propertyValues {Curve {keys [
         <at 0 1>
         {interpMode=RCIM_Cubic, tangentMode=RCTM_Break, time=0.5, value=1}
         {arriveTangent=-4.6273, interpMode=RCIM_Cubic, tangentMode=RCTM_Break, time=1, value=0}
        ]}}})) {isSetParameters=true, module=SetVariables_#1}
   &AccelerationForce(Acceleration=<0 0 -300>)
   &SolveForcesAndVelocity(Clamp Velocity=false, Limit Acceleration=false)
   /SetVariables(Particles.RenderRibbonWidth:NiagaraFloat=&Multiply_Float(
     A=$Particles.Initial.RibbonWidth
     B=&FloatFromCurve(
      FloatCurve=DataInterface() {propertyValues {Curve {keys [
          {interpMode=RCIM_Cubic, leaveTangent=-0.0854, tangentMode=RCTM_Break, time=0.8, value=1}
          {arriveTangent=-5.8726, arriveTangentWeight=0.2821, interpMode=RCIM_Cubic, tangentMode=RCTM_Break, time=1, value=0}
         ]}}}
      CurveIndex=$Particles.ParentAge
     )
    )) {isSetParameters=true, module=SetVariables_#2}
   &ScaleColor(
    Scale Alpha=&FloatFromCurve(
     FloatCurve=DataInterface() {propertyValues {Curve {keys [
         {interpMode=RCIM_Cubic, leaveTangent=-0.0854, tangentMode=RCTM_Break, time=0.2, value=0}
         {arriveTangent=3.6189, arriveTangentWeight=0.3334, interpMode=RCIM_Cubic, tangentMode=RCTM_Break, time=1, value=1}
        ]}}}
     CurveIndex=$Particles.ParentAge
    )
    Scale RGB=&VectorFromFloat(Value=1)
    ScaleA=true
    ScaleRGB=true
   )
  ]
  rendererClasses [&NiagaraRibbonRendererProperties]
  renderers [{rendererClass=&NiagaraRibbonRendererProperties}]
 }
 FN_SourceParticleRibbon001 {
  emitterSpawnScript [&FN_Source_v2(AttributeReader=DataInterface() {propertyValues {EmitterBinding {bindingMode=Other, emitterName=Projectile}}}) {enabled=false}]
  emitterUpdateScript [
   &EmitterState()
   &SpawnRate(
    SpawnRate=&FN_GetEmitterParticleCount(AttributeReader:NiagaraDataInterfaceParticleRead=$Emitter.AttributeReader, Multiplier:NiagaraFloat=30)
    Use Spawn Probability=false
   )
   &FN_Source_v2(AttributeReader=DataInterface() {propertyValues {EmitterBinding {bindingMode=Other, emitterName=Projectile}}}) {module=FN_Source_v2001}
  ]
  particleSpawnScript [
   &FN_SourceParentID(Ribbon ID from Parent ID=true)
   &FN_SourceLocation(Position:Vector3f=&FN_GetVectorByID(Attrib name:ENiagaraCommonVectors=::NewEnumerator0, Default on Invalid001:Vector3f=<0 0 0>))
   &InitializeRibbon(
    Module.Write Color=true
    Module.Write Lifetime=true
    Module.Write Mass=true
    Module.Write Material Random=true
    Module.Write Position=false
    Module.Write RibbonTwist=false
    Module.Write RibbonWidth=true
    Ribbon Width=120
   )
  ]
  particleUpdateScript [
   &FN_UpdateSourcedRibbbons(
    Locked Particles Position:Vector3f=&FN_GetVectorByID(Attrib name:ENiagaraCommonVectors=::NewEnumerator0, Default on Invalid001:Vector3f=<0 0 0>)
    Ribbon Width:NiagaraFloat=$Particles.Initial.RibbonWidth
   ) {module=FN_UpdateLockedParticles}
   /SetVariables(Particles.Lifetime=&FloatFromCurve(
     FloatCurve=DataInterface() {propertyValues {Curve {keys [
         <at 0 1>
         {interpMode=RCIM_Cubic, tangentMode=RCTM_Break, time=0.5, value=1}
         {arriveTangent=-4.6273, interpMode=RCIM_Cubic, tangentMode=RCTM_Break, time=1, value=0}
        ]}}}
     Scale Curve=&FloatFromCurve(
      FloatCurve=DataInterface() {propertyValues {Curve {keys [
          {interpMode=RCIM_Cubic, tangentMode=RCTM_Break, time=0, value=1}
          {arriveTangent=-2.249, interpMode=RCIM_Cubic, tangentMode=RCTM_Break, time=1, value=0.2}
         ]}}}
      CurveIndex=&self:SampleParentNormalizedAge(
       Emitter:NiagaraDataInterfaceParticleRead=DataInterface() {propertyValues {EmitterBinding {bindingMode=Other, emitterName=Projectile}}}
       SampleID:NiagaraInt32=$Particles.SourceID
      )
     )
    )) {isSetParameters=true, module=SetVariables_#1}
   &AccelerationForce(Acceleration=<0 0 -300>) {enabled=false}
   &SolveForcesAndVelocity(Clamp Velocity=false, Limit Acceleration=false)
   /SetVariables(Particles.RenderRibbonWidth:NiagaraFloat=&Multiply_Float(
     A=$Particles.Initial.RibbonWidth
     B=&FloatFromCurve(
      FloatCurve=DataInterface() {propertyValues {Curve {keys [
          {interpMode=RCIM_Cubic, leaveTangent=-0.0854, tangentMode=RCTM_Break, time=0, value=1}
          {arriveTangent=-3.6886, arriveTangentWeight=0.2821, interpMode=RCIM_Cubic, tangentMode=RCTM_Break, time=0.5, value=0}
         ]}}}
      Scale Curve=&FloatFromCurve(
       FloatCurve=DataInterface() {propertyValues {Curve {keys [
           {interpMode=RCIM_Cubic, tangentMode=RCTM_Break, time=0, value=1}
           {arriveTangent=-2.016, interpMode=RCIM_Cubic, tangentMode=RCTM_Break, time=1, value=0.2}
          ]}}}
       CurveIndex=$Particles.ParentAge
      )
     )
    )) {enabled=false, isSetParameters=true, module=SetVariables_#2}
   /SetVariables(Particles.RenderRibbonWidth:NiagaraFloat=&Multiply_Float(
     A=$Particles.Initial.RibbonWidth
     B=&FloatFromCurve(
      FloatCurve=DataInterface() {propertyValues {Curve {keys [
          {interpMode=RCIM_Cubic, leaveTangent=-0.0854, tangentMode=RCTM_Break, time=0, value=1}
          {arriveTangent=-3.6886, arriveTangentWeight=0.2821, interpMode=RCIM_Cubic, tangentMode=RCTM_Break, time=0.5, value=0}
         ]}}}
      Scale Curve=&FloatFromCurve(
       FloatCurve=DataInterface() {propertyValues {Curve {keys [
           {interpMode=RCIM_Cubic, tangentMode=RCTM_Break, time=0, value=1}
           {
            arriveTangent=-0.74
            arriveTangentWeight=0.1
            interpMode=RCIM_Cubic
            leaveTangent=-0.74
            leaveTangentWeight=0.2658
            tangentMode=RCTM_User
            time=0.4
            value=0.9
           }
           {arriveTangent=-0.0479, interpMode=RCIM_Cubic, tangentMode=RCTM_Break, time=1, value=0.2}
          ]}}}
       CurveIndex=&self:SampleParentNormalizedAge(
        Emitter:NiagaraDataInterfaceParticleRead=DataInterface() {propertyValues {EmitterBinding {bindingMode=Other, emitterName=Projectile}}}
        SampleID:NiagaraInt32=$Particles.SourceID
       )
      )
     )
    )) {isSetParameters=true}
   &ScaleColor(
    Scale Alpha=&FloatFromCurve(
     FloatCurve=DataInterface() {propertyValues {Curve {keys [
         {interpMode=RCIM_Cubic, tangentMode=RCTM_Break, time=0.2, value=1}
         {arriveTangent=0.0059, arriveTangentWeight=0.3334, interpMode=RCIM_Cubic, tangentMode=RCTM_Break, time=1, value=0}
        ]}}}
     CurveIndex=$Particles.ParentAge
    )
    Scale RGB=&VectorFromFloat(Value=1)
    ScaleA=true
    ScaleRGB=true
   )
  ]
  rendererClasses [&NiagaraRibbonRendererProperties]
  renderers [{rendererClass=&NiagaraRibbonRendererProperties}]
 }
 GooDrops {
  emitterSpawnScript [&FN_PreserveSkydiveUserVariables() {enabled=false, module=PreserveUserVariables}]
  emitterUpdateScript [
   &EmitterState()
   &SpawnRate(SpawnRate=60, Use Spawn Probability=false)
  ]
  particleSpawnScript [
   &InitializeParticle(
    Lifetime Max=1
    Lifetime Min=0.5
    Lifetime Mode=::NewEnumerator1
    Mesh Scale Mode=::NewEnumerator3
    Mesh Uniform Scale=&Lerp_Float(
     A=5
     B=12
     Alpha=&FloatFromCurve(
      FloatCurve=DataInterface() {propertyValues {Curve {keys [
          {interpMode=RCIM_Cubic, tangentMode=RCTM_Break, time=0, value=0}
          {arriveTangent=2.3492, interpMode=RCIM_Cubic, tangentMode=RCTM_Break, time=1, value=1}
         ]}}}
      CurveIndex=&RandomRangeFloat(Minimum=0, Maximum=1)
     )
    )
    UsePositionOffset=false
    Write Material Random=true
   )
   &SphereLocation(
    MaskSpawnGroup=false
    Override Seed=false
    Sphere Radius=60
    Surface Only=false
    UseNonUniformScale=false
    UseOffset=false
    UseSphereOrientation=false
   )
   &AddVelocityFromPoint(
    Module.Use Falloff Distance=false
    Velocity Origin:Vector3f=&SimulationPosition()
    Velocity Strength=&UniformRangedFloat(Minimum=50, Maximum=400)
   )
   &DynamicMaterialParameters(
    Index 0 Param 1=1
    Index 0 Param 2=1
    Module.Param0WriteEnabled=true
    Module.Param1WriteEnabled=true
    Module.Param2WriteEnabled=false
    Module.Param3WriteEnabled=false
   )
  ]
  particleUpdateScript [
   &ParticleState()
   &GravityForce(Gravity=<0 0 -500>)
   &SolveForcesAndVelocity(Clamp Velocity=false, Limit Acceleration=false)
   &ScaleMeshSize(Scale Factor=&VectorFromFloat(Value=&FloatFromCurve(FloatCurve=DataInterface() {propertyValues {Curve {keys [
          {arriveTangent=-1, interpMode=RCIM_Cubic, leaveTangent=0.0193, tangentMode=RCTM_Break, time=0, value=1}
          {arriveTangent=-0.4706, interpMode=RCIM_Cubic, leaveTangent=-1, tangentMode=RCTM_Break, time=1, value=0}
         ]}}})))
   &ScaleColor(
    Scale Alpha=&FloatFromCurve(
     FloatCurve=DataInterface() {propertyValues {Curve {keys [
         {interpMode=RCIM_Cubic, tangentMode=RCTM_User, time=0, value=1}
         {arriveTangent=-0.3941, interpMode=RCIM_Cubic, leaveTangent=-0.3941, tangentMode=RCTM_User, time=1, value=0}
        ]}}}
     Scale Curve=0.85
    )
    ScaleA=true
    ScaleRGB=false
   )
   &ScaleMeshSizeBySpeed(
    Max Scale Factor=<2 1.5 1>
    Speed Threshold:NiagaraFloat=&FloatFromCurve(FloatCurve=DataInterface() {propertyValues {Curve {keys [
         {arriveTangent=2294.4238, interpMode=RCIM_Cubic, leaveTangent=2294.4238, tangentMode=RCTM_User, time=0, value=150}
         {arriveTangent=78.1625, interpMode=RCIM_Cubic, leaveTangent=78.3263, tangentMode=RCTM_User, time=1, value=2000}
        ]}}})
   )
  ]
  rendererClasses [&NiagaraMeshRendererProperties]
  renderers [{rendererClass=&NiagaraMeshRendererProperties, rendererIndex=0}]
 }
 GooTrail {
  bEnabled=false
  emitterUpdateScript [
   &EmitterState()
   /Game/Effects/Niagara/Modules/Ribbons/FN_SpawnRibbon_v3(Skeleton Mesh:NiagaraDataInterfaceSkeletalMesh=$User.SkeletalMesh, Spawn Mode:ENiagaraSkeletalSpawning=::NewEnumerator1, SpawnRate=40)
  ]
  particleSpawnScript [
   /SetVariables(Particles.ColorStart:LinearColor=<rgba 1 0.0459 0 1>) {isSetParameters=true}
   /Game/Effects/Niagara/Modules/Ribbons/FN_InitializeRibbon_v5(
    Enable Skeletal Position=true
    Facing Axis:ENiagaraOrientationAxis=::YAxis
    Lifetime=0.2
    Offset Amount=25
    Offset Axis:ENiagaraOrientationAxis=::XAxis
    Ribbon Width=325
    SkeletalMesh:NiagaraDataInterfaceSkeletalMesh=$User.SkeletalMesh
   )
  ]
  particleUpdateScript [
   /Game/Effects/Niagara/Modules/Ribbons/FN_UpdateRibbon_v5()
   /SetVariables(Particles.RibbonFacing=<1 0 0>) {isSetParameters=true}
   &ScaleColor(
    Scale Alpha=&FloatFromCurve(FloatCurve=DataInterface() {propertyValues {Curve {keys [
         {interpMode=RCIM_Cubic, tangentMode=RCTM_Break, time=0, value=1}
         {
          arriveTangent=-0.7676
          arriveTangentWeight=0.0667
          interpMode=RCIM_Cubic
          leaveTangent=-0.7676
          leaveTangentWeight=0.4021
          tangentMode=RCTM_User
          time=0.1
          value=0.5
         }
         {interpMode=RCIM_Cubic, tangentMode=RCTM_Break, time=1, value=0.3}
        ]}}})
    Scale RGB=&MakeVectorFromLinearColorRGB(Color=&LerpLinearColors(
      StartColor=$Particles.ColorStart
      EndColor=$Particles.ColorEnd
      LerpFactor=&FloatFromCurve(
       FloatCurve=DataInterface() {propertyValues {Curve {keys [
           {interpMode=RCIM_Cubic, leaveTangent=-0.0081, tangentMode=RCTM_Break, time=0, value=0}
           {arriveTangent=2.8653, interpMode=RCIM_Cubic, leaveTangent=2.8653, tangentMode=RCTM_User, time=0.1, value=0.2}
           {interpMode=RCIM_Cubic, tangentMode=RCTM_Break, time=0.7, value=1}
          ]}}}
       CurveIndex=$Particles.Age
      )
     ))
    ScaleA=true
    ScaleRGB=false
   )
  ]
  rendererClasses [&NiagaraRibbonRendererProperties]
  renderers [
   {rendererClass=&NiagaraRibbonRendererProperties}
   {rendererClass=&NiagaraRibbonRendererProperties, rendererIndex=1}
  ]
 }
 GoopGroup001 {
  emitterUpdateScript [
   &EmitterState(Inactive Response=::NewEnumerator1, Life Cycle Mode=::NewEnumerator1, Loop Behavior=::NewEnumerator1, UseLoopDelay=false)
   &SpawnBurst_Instantaneous(Spawn Count=7, Use Spawn Probability=false)
  ]
  particleSpawnScript [
   &InitializeParticle(
    Color=<rgba 0.3906 0.0217 0 1>
    Lifetime=5
    Mesh Scale Mode=::NewEnumerator1
    Mesh Uniform Scale Max=0.9
    Mesh Uniform Scale Min=0.5
    UsePositionOffset=false
    Write Material Random=true
   )
   /SetVariables(Particles.SpriteCamOffset=80, Particles.SpriteParams:Vector4f=<0.5 1 1 1>) {isSetParameters=true}
   /SetVariables(Particles.CamOffset2=-5, Particles.DynParams2:Vector4f=<0.1 2 0 0>) {isSetParameters=true}
   &SphereLocation(
    MaskSpawnGroup=false
    Non Uniform Scale=<1 1 2>
    Override Seed=false
    Sphere Radius=25
    Surface Only=false
    UseNonUniformScale=true
    UseOffset=false
    UseSphereOrientation=false
   )
  ]
  particleUpdateScript [
   &ParticleState()
   &DynamicMaterialParameters(
    Index 0 Param 1=&RandomRangeFloat(Minimum=30, Maximum=70)
    Index 0 Param 2=1
    Index 0 Param 3=50
    Index 0 Param 4=&Multiply_Float(
     A=$Engine.System.Age
     B=&Lerp_Float(
      A=-100
      B=100
      Alpha=&FloatFromCurve(
       FloatCurve=DataInterface() {propertyValues {Curve {keys [<at 0 0>,<at 0.48 0.1>,<at 0.52 0.9>,<at 1 1>]}}}
       CurveIndex=&RandomRangeFloat(Minimum=0, Maximum=1)
      )
     )
    )
    Index 1 Param 1=&RandomRangeFloat(Minimum=20, Maximum=50)
    Index 1 Param 3=1
    Module.Param0WriteEnabled=true
    Module.Param0WriteEnabled1=true
    Module.Param1WriteEnabled=true
    Module.Param1WriteEnabled1=true
    Module.Param2WriteEnabled=true
    Module.Param2WriteEnabled1=true
    Module.Param3WriteEnabled=true
    Module.Param3WriteEnabled1=false
    Write Parameter Index 1=true
   )
   &ScaleColor(Scale RGB=&VectorFromFloat(Value=1), ScaleA=false, ScaleRGB=true)
  ]
  rendererClasses [&NiagaraMeshRendererProperties]
  renderers [{rendererClass=&NiagaraMeshRendererProperties, rendererIndex=0}]
 }
 Leaves {
  emitterUpdateScript [
   &EmitterState()
   &SpawnRate(SpawnRate=12, Use Spawn Probability=false)
   /SetVariables(Emitter.SpriteRotRate:NiagaraFloat=&FloatFromCurve(
     FloatCurve=DataInterface() {propertyValues {Curve {keys [
         <at 0.0041 0.1315>
         <at 0.1088 1.9664>
         {arriveTangent=-3.3757, interpMode=RCIM_Cubic, leaveTangent=-3.3757, tangentMode=RCTM_User, time=0.3609, value=0.6391}
         {arriveTangent=-1, interpMode=RCIM_Cubic, leaveTangent=-1, tangentMode=RCTM_User, time=0.5614, value=0.4386}
         <at 0.7086 1.6472>
         <at 1 0>
        ]}}}
     CurveIndex=$Emitter.NormalizedLoopAge
     Scale Curve=&RandomRangeFloat(Minimum=-300, Maximum=300)
    )) {isSetParameters=true}
  ]
  particleSpawnScript [
   &InitializeParticle(
    Color=&ScaleLinearColorByCurve(
     InitialValue=&ColorFromCurve(
      ColorCurve=DataInterface() {propertyValues {
        AlphaCurve {keys [<at 0 1 RCIM_Cubic>,<at 1 1 RCIM_Cubic>]}
        BlueCurve {keys [<at 0 0 RCIM_Cubic>,<at 0.5009 0 RCIM_Cubic>]}
        GreenCurve {keys [
          {arriveTangent=-0.7055, interpMode=RCIM_Cubic, leaveTangent=-0.7055, time=0, value=0.7604}
          {arriveTangent=-0.7055, interpMode=RCIM_Cubic, leaveTangent=-0.7055, time=0.5009, value=0.3073}
         ]}
        RedCurve {keys [
          {arriveTangent=-0.4701, interpMode=RCIM_Cubic, leaveTangent=-0.4701, time=0, value=0.4613}
          {arriveTangent=-0.4701, interpMode=RCIM_Cubic, leaveTangent=-0.4701, time=0.5009, value=0.1594}
         ]}
       }}
      CurveIndex=&RandomRangeFloat(Minimum=0, Maximum=1)
     )
     ScaleCurve=DataInterface() {propertyValues {
       AlphaCurve {keys [<at 0 1>,<at 1 1>]}
       BlueCurve {keys [{arriveTangent=0.9494, interpMode=RCIM_Cubic, leaveTangent=0.9494, tangentMode=RCTM_User, time=0, value=0.4}, <at 1 0.349>]}
       GreenCurve {keys [{arriveTangent=0.9494, interpMode=RCIM_Cubic, leaveTangent=0.9494, tangentMode=RCTM_User, time=0, value=0.4}, <at 1 0.349>]}
       RedCurve {keys [{arriveTangent=0.9494, interpMode=RCIM_Cubic, leaveTangent=0.9494, tangentMode=RCTM_User, time=0, value=0.4}, <at 1 0.349>]}
      }}
     CurveIndex=&RandomRangeFloat(Minimum=0, Maximum=1)
    )
    Lifetime Max=1.1
    Lifetime Min=0.3
    Lifetime Mode=::NewEnumerator1
    Sprite Rotation Mode=::NewEnumerator2
    Sprite Size Mode=::NewEnumerator1
    Uniform Sprite Size Max=75
    Uniform Sprite Size Min=40
    UsePositionOffset=false
    Write Material Random=true
   )
   &SphereLocation(MaskSpawnGroup=false, Override Seed=false, Surface Only=false, UseNonUniformScale=false, UseOffset=false, UseSphereOrientation=false)
   &AddVelocityFromPoint(Module.Use Falloff Distance=false, Velocity Strength=&RandomRangeFloat(Minimum=200, Maximum=1000))
  ]
  particleUpdateScript [
   &Drag(
    Drag=&Multiply_Float(
     A=13
     B=&FloatFromCurve(
      FloatCurve=DataInterface() {propertyValues {Curve {keys [{arriveTangent=15.573, interpMode=RCIM_Cubic, time=0, value=1}, <at 0.9 0 RCIM_Cubic>]}}}
      CurveIndex=&Max_Float(A:NiagaraFloat=$NPC.NPC_Audio_Reactive_Parameters.MusicStem_Snare, B:NiagaraFloat=$NPC.NPC_Audio_Reactive_Parameters.MusicStem_Kick)
     )
    )
    Use Linear Drag=true
    Use Rotational Drag=false
   )
   &GravityForce(Gravity=&RandomRangeVector(Minimum=<0 0 -850>, Maximum=<0 0 -1000>))
   &GravityForce(Gravity=&VectorFromFloat(Value=&RandomRangeFloat(Minimum=-550, Maximum=-750))) {enabled=false, module=GravityForce001}
   &WindForce(Wind Speed=&RandomRangeVector(Minimum=<300 0 0>, Maximum=<450 1 1>))
   &CurlNoiseForce(
    Curl Noise Cone Mask Angle=120
    Mask Curl Noise=true
    Noise Frequency=90
    Noise Strength=&RandomRangeFloat(Minimum=500, Maximum=1000)
    Pan Noise=true
    Pan Noise Field=<0 0 1>
   )
   &SolveForcesAndVelocity(Clamp Velocity=false, Limit Acceleration=false)
   &ParticleState()
   &DynamicMaterialParameters(
    Index 0 Param 1=0.25
    Index 0 Param 2=&FloatFromCurve(
     FloatCurve=DataInterface() {propertyValues {Curve {keys [
         {interpMode=RCIM_Cubic, tangentMode=RCTM_User, time=0, value=0.1}
         {arriveTangent=1.3207, interpMode=RCIM_Cubic, leaveTangent=1.3207, tangentMode=RCTM_User, time=0.6999, value=0.2782}
         {arriveTangent=3.3434, interpMode=RCIM_Cubic, leaveTangent=3.3434, tangentMode=RCTM_User, time=1, value=1}
        ]}}}
     Scale Curve=0.1
    )
    Module.Param0WriteEnabled=true
    Module.Param1WriteEnabled=true
    Module.Param2WriteEnabled=true
    Module.Param3WriteEnabled=true
   )
   &ScaleColor(
    Scale Alpha=&FloatFromCurve(FloatCurve=DataInterface() {propertyValues {Curve {keys [
         <at -0.0007 -0.0138 RCIM_Cubic>
         {arriveTangent=2.0745, interpMode=RCIM_Cubic, leaveTangent=2.0745, tangentMode=RCTM_User, time=0.0966, value=0.7205}
         {arriveTangent=-0.4258, interpMode=RCIM_Cubic, leaveTangent=-0.4258, tangentMode=RCTM_User, time=0.8392, value=0.8196}
         <at 1.0011 0.002 RCIM_Cubic>
        ]}}})
    ScaleA=true
    ScaleRGB=false
   )
   &SubUVAnimation(SubUV Animation Mode:ENiagara_SubUVLookupMode=::NewEnumerator1)
   &ScaleSpriteSize(
    Uniform Curve Scale=0.7
    Uniform Curve Sprite Scale=DataInterface() {propertyValues {Curve {keys [<at 0 1>, {arriveTangent=-1.25, interpMode=RCIM_Cubic, leaveTangent=-1.25, tangentMode=RCTM_User, time=0.6505, value=0.9983}, <at 1 0>]}}}
   )
   &SpriteRotationRate(Rotation Rate=$Emitter.SpriteRotRate)
  ]
  rendererClasses [&NiagaraSpriteRendererProperties]
  renderers [{rendererClass=&NiagaraSpriteRendererProperties}]
 }
 Projectile {
  emitterUpdateScript [
   &EmitterState()
   &SpawnBurst_Instantaneous(Spawn Count=3, Use Spawn Probability=false, UseLoopCountLimit=false) {enabled=false}
   &SpawnRate(SpawnRate=5, Use Spawn Probability=false)
  ]
  particleSpawnScript [
   &InitializeParticle(
    Color=&ColorFromCurve(
     ColorCurve=DataInterface() {propertyValues {
       AlphaCurve {keys [<at 0 1 RCIM_Cubic>,<at 1 1 RCIM_Cubic>]}
       BlueCurve {keys [<at 0 0 RCIM_Cubic>,<at 0.1517 0>,<at 0.3805 0>,<at 0.7658 0 RCIM_Cubic>]}
       GreenCurve {keys [
         {arriveTangent=-1.5741, interpMode=RCIM_Cubic, leaveTangent=-1.5741, time=0, value=0.3511}
         <at 0.1517 0.1123>
         <at 0.3805 0.0542>
         {arriveTangent=-0.0964, interpMode=RCIM_Cubic, leaveTangent=-0.0964, time=0.7658, value=0.0171}
        ]}
       RedCurve {keys [
         <at 0 1 RCIM_Cubic>
         <at 0.1517 1>
         <at 0.3805 0.8177>
         {arriveTangent=-1.3382, interpMode=RCIM_Cubic, leaveTangent=-1.3382, time=0.7658, value=0.3021}
        ]}
      }}
     CurveIndex=&RandomRangeFloat(Minimum=0, Maximum=1)
    )
    Lifetime=&Multiply_Float(
     A=&RandomRangeFloat(Minimum=1, Maximum=1.25)
     B=&FloatFromCurve(
      FloatCurve=DataInterface() {propertyValues {Curve {keys [{interpMode=RCIM_Cubic, tangentMode=RCTM_Break, time=0, value=0.4}, {interpMode=RCIM_Cubic, tangentMode=RCTM_Break, time=1, value=0.7}]}}}
      CurveIndex=$Emitter.Age
     )
    )
    Mesh Scale Max=<1 1 1>
    Mesh Scale Min=<0.5 0.5 0.5>
    Mesh Scale Mode=::NewEnumerator2
    UsePositionOffset=false
    Write Material Random=true
   )
   &DynamicMaterialParameters(
    Index 0 Param 1=50
    Index 0 Param 2=1
    Index 0 Param 4=&RandomRangeFloat(Minimum=-100, Maximum=100)
    Index 1 Param 1=&RandomRangeFloat(Minimum=-10, Maximum=5)
    Index 1 Param 2=1
    Index 1 Param 3=1
    Module.Param0WriteEnabled=true
    Module.Param0WriteEnabled1=true
    Module.Param1WriteEnabled=true
    Module.Param1WriteEnabled1=true
    Module.Param2WriteEnabled=true
    Module.Param2WriteEnabled1=true
    Module.Param3WriteEnabled=true
    Module.Param3WriteEnabled1=false
    Write Parameter Index 1=true
   ) {module=DynamicMaterialParameters001}
   &SphereLocation(
    MaskSpawnGroup=false
    Override Seed=false
    Sphere Radius=60
    Surface Only=true
    UseNonUniformScale=true
    UseOffset=false
    UseSphereOrientation=false
   )
   &InitialMeshOrientation(Rotation=&RandomRangeVector(Minimum=<0 0 0>, Maximum=<1 1 1>), Use Rotation Vector=true)
   &InheritVelocity(Inherited Velocity Amount Scale=<1.5 1.5 1.5>, Inherited Velocity Speed Limit=2000, Speed Limit=true, UseSpeedThreshold=false)
   &AddVelocityFromPoint(Module.Use Falloff Distance=false, Velocity Strength=400)
  ]
  particleUpdateScript [
   &ParticleState()
   &Drag(Drag=2, Use Linear Drag=true, Use Rotational Drag=false)
   &AccelerationForce(Acceleration=<0 0 -1600>)
   &SolveForcesAndVelocity(
    Clamp Velocity=true
    Limit Acceleration=false
    Speed Limit:NiagaraFloat=&FloatFromCurve(FloatCurve=DataInterface() {propertyValues {Curve {keys [<at 4 20000>,<at 10 5000>]}}}, CurveIndex=$Particles.Age)
   )
   &ScaleMeshSize(Scale Factor=&VectorFromFloat(Value=&FloatFromCurve(
      FloatCurve=DataInterface() {propertyValues {Curve {keys [
          {arriveTangent=10, interpMode=RCIM_Cubic, leaveTangent=10, time=0, value=0}
          {interpMode=RCIM_Cubic, tangentMode=RCTM_User, time=0.1, value=1}
          {interpMode=RCIM_Cubic, tangentMode=RCTM_User, time=0.7, value=1}
          {arriveTangent=-4.0437, interpMode=RCIM_Cubic, leaveTangent=-4.0437, tangentMode=RCTM_User, time=1, value=0}
         ]}}}
      Scale Curve=0.7
     )))
   &UpdateMeshOrientation(
    Rotation Rate=&RandomRangeFloat(Minimum=-1, Maximum=1)
    Rotation Vector=&RandomRangeVector(Minimum=<0 0 0>, Maximum=<1 1 1>)
   )
  ]
  rendererClasses [&NiagaraMeshRendererProperties]
  renderers [{rendererClass=&NiagaraMeshRendererProperties, rendererIndex=0}]
 }
}
summary {
 systemName=NS_Buttercake_Prj
 emitters [
  {emitterName=EnergyWave, rendererClasses [&NiagaraMeshRendererProperties]}
  {bEnabled=false, emitterName=GooTrail, rendererClasses [&NiagaraRibbonRendererProperties]}
  {emitterName=GooDrops, rendererClasses [&NiagaraMeshRendererProperties]}
  {emitterName=Leaves, rendererClasses [&NiagaraSpriteRendererProperties]}
  {emitterName=Projectile, rendererClasses [&NiagaraMeshRendererProperties]}
  {bEnabled=false, emitterName=FN_SourceParticleRibbon, rendererClasses [&NiagaraRibbonRendererProperties]}
  {emitterName=GoopGroup001, rendererClasses [&NiagaraMeshRendererProperties]}
  {emitterName=FN_SourceParticleRibbon001, rendererClasses [&NiagaraRibbonRendererProperties]}
 ]
 userVariables [
  User.bIsFrontEndPreview {type=NiagaraBool}
  User.PlayerSpeed {type=NiagaraFloat, defaultValue=0, description=""}
  User.RibbonIDOffset {type=NiagaraInt32}
  User.SkeletalMesh {
   type=NiagaraDataInterfaceSkeletalMesh
   defaultValue {struct=NiagaraExt_VariableValue_DataInterface, value {dataInterface=&self:NiagaraDataInterfaceSkeletalMesh_0}}
  }
 ]
}
